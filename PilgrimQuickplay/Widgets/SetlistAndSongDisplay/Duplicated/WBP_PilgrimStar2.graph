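# WBP_PilgrimStar2 is a small fixed-size star widget in the Pilgrim Quickplay setlist and song display, showing a faint, centered results-star image.
# SetComplete and SetStarProgress are empty stubs, and the event graph has no nodes.
# UpdateContrastBackground only re-applies the StarBG image's current visibility to itself, so it changes nothing.
# Star progress and dynamic material variables are declared but unused in the graphs.
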WidgetBlueprint /PilgrimQuickplay/Widgets/SetlistAndSongDisplay/Duplicated/WBP_PilgrimStar2
graphs [
 EventGraph {links [], nodes []}
 SetComplete {nodes [SetComplete() {is=title K2Node_FunctionEntry}]}
 SetStarProgress {nodes [SetStarProgress!() -> (then, Set Star Progress:double=~) {is=title K2Node_FunctionEntry}]}
 UpdateContrastBackground {nodes [
   UpdateContrastBackground() {is=title K2Node_FunctionEntry}
   n1 = Get StarBG() -> (StarBG:Image Object Reference)
   > SetVisibility(self=@n1, InVisibility=GetVisibility!(self:Widget Object Reference=@n1) -> (ReturnValue:ESlateVisibility Enum=Visible))
  ]}
]
variables [
 StarBGMat {type=MaterialInstanceDynamic}
 StarCompleteMat {type=MaterialInstanceDynamic}
 StarIncompleteMat {type=MaterialInstanceDynamic}
 StarProgress {type=boolean}
 StarProgressMat {type=MaterialInstanceDynamic}
]
widgets {
 info {inheritedWidgetCount=0, parentClass=&UserWidget, rootWidgetClass=&SizeBox, widgetCount=3}
 widgets [
  SizeBox_0 {
   widgetClassPath=&SizeBox
   widget=SizeBox_0
   properties {bOverride_HeightOverride=true, bOverride_WidthOverride=true, heightOverride=70, widthOverride=70}
  }
  Overlay_0 {widgetClassPath=&Overlay, parent=SizeBox_0, widget=Overlay_0, slot {class=SizeBoxSlot}}
  StarBG {
   widgetClassPath=&Image
   bIsVariable=true
   parent=Overlay_0
   widget=StarBG
   properties {
    visibility=HitTestInvisible
    brush {
     bIsDynamicallyLoaded=false
     drawAs=Image
     imageType=NoImage
     mirroring=NoMirror
     resourceName=None
     resourceObject=/PilgrimQuickplay/Widgets/SetlistAndSongDisplay/Duplicated/Textures/T_PilgrimQuickPlay_ResultsStar2
     tiling=NoTile
     imageSize=<64 64>
     margin=<ltrb 0 0 0 0>
     outlineSettings {
      bUseBrushTransparency=false
      roundingType=HalfHeightRadius
      width=0
      color {colorUseRule=UseColor_Specified, specifiedColor=<rgba 0 0 0 0>}
      cornerRadii=<0 0 0 0>
     }
     tintColor {colorUseRule=UseColor_Specified, specifiedColor=<rgba 1 1 1 0.12>}
     uVRegion {bIsValid=false, max=<0 0>, min=<0 0>}
    }
   }
   slot {class=OverlaySlot, properties {horizontalAlignment=HAlign_Center, verticalAlignment=VAlign_Center}}
  }
 ]
}
