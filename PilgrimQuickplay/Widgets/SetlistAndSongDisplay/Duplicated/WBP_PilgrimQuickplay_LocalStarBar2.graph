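# Five-star score bar widget for the local player in Pilgrim Quickplay.
# Each score update passes star index and progress to Handle Star Update, which refreshes completed stars, the current star's fill, or a gold indicator past the maximum.
# The same event compares hit, miss and pass counts with previous values and broadcasts note reaction events through the Event Router, with hits requiring a streak of at least four.
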WidgetBlueprint /PilgrimQuickplay/Widgets/SetlistAndSongDisplay/Duplicated/WBP_PilgrimQuickplay_LocalStarBar2
graphs [
 Check Is Star Changed {nodes [
   n1 = Check Is Star Changed!() -> (then, CurrentStarIndex:int, CurrentStarProgress:double=~) {is=title K2Node_FunctionEntry}
   n2 = Get LastStarProgress() -> (LastStarProgress:double)
   n3 = Get LastStarIndex() -> (LastStarIndex:int=0)
   > n4 = Set HasChanged(HasChanged:bool=OR Boolean(A="Not Equal (Integer)"(A=@n1, B=@n3), B="float < float"(A=@n2, B=@n1.CurrentStarProgress))) -> (Output_Get:bool)
   > Set LastStarProgress(LastStarProgress:double=@n1.CurrentStarProgress) -> (Output_Get:double)
   > Return Node!(execute, IsChanged:bool=@n4)
  ]}
 CheckStarCompletion {nodes [
   n1 = Get LastStarIndex() -> (LastStarIndex=0)
   n2 = CheckStarCompletion!() -> (then, CurrentStar:int) {is=title K2Node_FunctionEntry}
   n3 = For Loop(execute=@n9."<", FirstIndex=@n2, LastIndex=@n1)
   n4 = Get Star Widget!(execute, self, Index:int=@n3) -> (then, StarWidget:WBP Pilgrim Star 2 Object Reference) {is=title K2Node_CallFunction}
   > SetVisibility(self=Get StarProgress!(self:WBP Pilgrim Star Object Reference=@n4) -> (StarProgress:Radial Slider Object Reference), InVisibility=Hidden)
   > SetVisibility(self=Get StarComplete!(self:WBP Pilgrim Star Object Reference=@n4) -> (StarComplete:Image Object Reference), InVisibility=Hidden)
   n5 = Get Star Widget(execute=@n3.LoopBody, Index=@n3)
   > SetComplete(self:WBP Pilgrim Star 2 Object Reference=@n5)
   n6 = For Loop(execute=@n8.then, FirstIndex=Get LastStarIndex() -> (LastStarIndex=0), LastIndex=int - int(A=@n2))
   > n7 = Get Star Widget(Index=@n6)
   > SetComplete(self=@n7)
   Set LastStarIndex(execute=@n6.Completed, LastStarIndex:int=@n2) -> (Output_Get:int=0)
   n8 = Branch(execute=@n9.">")
   n9 = Compare Int!(Exec:Exec=@n2.then, Input:int=@n2, CompareWith:int=@n1) -> (">":Exec, "==":Exec, "<":Exec) {is=title K2Node_MacroInstance}
  ]}
 EventGraph {nodes [
   Event PreConstruct()
   > SetStretch!(
    execute
    self:Scale Box Object Reference=Get ScaleBox() -> (ScaleBox:Scale Box Object Reference)
    InStretch:EStretch Enum=Get Scale Stretch() -> (Scale Stretch:EStretch Enum=None)
   ) -> (then)
   > SetUserSpecifiedScale!(
    execute
    self:Scale Box Object Reference=Get ScaleBox()
    InUserSpecifiedScale:float=Get In User Specified Scale() -> (In User Specified Scale:float)
   ) -> (then)
   n1 = Event OnScoreDataReceivedLocalPlayer!() -> (OutputDelegate, then, GameProgress:"Pilgrim Game Progress Structure (by ref)") {is=title K2Node_Event}
   > n2 = Sequence!(execute) -> (then_0, then_1, then_2:Exec)
   n3 = Break Pilgrim Game Progress!(PilgrimGameProgress:"Pilgrim Game Progress Structure (by ref)"=@n1.GameProgress) -> (
    NumStreakBreakingMisses:int
    bHasPerfectCombo:bool=~
    RawCurrentScore:float=~
    CurrentModifierBonus:float=~
    CurrentStar:int
    NextStarProgress:float=~
    BandScoreContribution:float=~
    BandModifierBonusContribution:float=~
    RawMaxScore:float=~
    MaxStreakLength:int
    AverageStreakMultiplierPerHitOrPass:float=~
    CurrentSectionCompletionPercentage:float=~
   ) {is=title K2Node_BreakStruct}
   > Handle Star Update!(execute, self, CurrentStar:int=@n3.CurrentStar, StarProgress:double=@n3.NextStarProgress) -> (then)
   Branch(
    execute=@n2.then_1
    Condition="integer > integer"(A=@n3.CurrentStar, B=Get LocalPlayerPreviousStarValue() -> (LocalPlayerPreviousStarValue:int=0))
   )
   > Set LocalPlayerPreviousStarValue(LocalPlayerPreviousStarValue:int=@n3.CurrentStar) -> (Output_Get:int=0)
   n4 = Break Pilgrim Game Progress!(PilgrimGameProgress=@n1.GameProgress) -> (
    NumHits:int
    NumMisses:int
    NumStreakBreakingMisses
    NumPasses:int
    bHasFullCombo:bool=~
    bHasPerfectCombo=~
    CurrentTick:int
    NoteStreak:int
    CurrentScore:float=~
    RawCurrentScore=~
    CurrentModifierBonus=~
    CurrentStar
    NextStarProgress=~
    StreakScoreMultiplier:int
    NotesPastLastMultiplier:int
    NotesToNextMultiplier:int
    CurrentHealth:float=~
    IsOverdriveActive:bool=~
    CurrentOverdrive:float=~
    CurrentOverdriveSectionNum:int
    BandScoreContribution=~
    BandModifierBonusContribution=~
    OverdriveScoringState:EOverdriveScoringState Enum
    MaxScore:float=~
    RawMaxScore=~
    MaxStreakLength
    AverageStreakMultiplierPerHitOrPass=~
    bIsAtMaxStreak:bool=~
    CurrentSectionCompletionPercentage=~
   )
   n5 = Get PrintDebugInfo() -> (PrintDebugInfo:bool)
   PrintString!(
    execute=@n2.then_2
    InString=Append!(
     A=Score Event...Num Hits:
     B="To String (Integer)"(InInt=@n4)
     C:String="   Num Misses: "
     D:String="To String (Integer)"(InInt=@n4.NumMisses)
     E:String="  Num Passes:"
     F:String="To String (Integer)"(InInt=@n4.NumPasses)
    ) -> (ReturnValue)
    bPrintToScreen=@n5
    bPrintToLog=@n5
    TextColor
    Duration
    Key
   ) -> (then)
   > n6 = Sequence!(execute) -> (then_0, then_1, then_2, then_3:Exec, then_4:Exec, then_5:Exec)
   n7 = Get Prev Hits() -> (Prev Hits:int=0)
   > Branch(Condition=AND Boolean(A="integer > integer"(A=@n4, B=@n7), B="integer >= integer"(A=@n4.NoteStreak, B=4)))
   n8 = Get Event Router() -> (Event Router:Gameplay Event Router Component Object Reference)
   > Is Valid(InputObject=@n8)
   > n9 = Cast To FortPlayerController(Object=GetOwningPlayer())
   > Broadcast Event!(
    execute
    self:Gameplay Event Router Component Object Reference=@n8
    EventData:Pilgrim Event Gameplay Note Reaction Structure=Make Pilgrim Event Gameplay Note Reaction!(
     FortPlayerController_3_#1:Fort Player Controller Object Reference=@n9
     PlayerNoteReaction_6_#2:PlayerNoteReactionEnum Enum=NewEnumerator2
    ) -> (PilgrimEvent_Gameplay_NoteReaction:Pilgrim Event Gameplay Note Reaction Structure) {is=title K2Node_MakeStruct}
    EventContext:Object Reference
   ) -> (then, ReturnValue:bool)
   Set Prev Hits(execute=@n6.then_1, Prev Hits:int=@n4) -> (Output_Get:int=0)
   Set Prev Misses(execute=@n6.then_3, Prev Misses:int=@n4.NumMisses) -> (Output_Get:int=0)
   Set Prev Passes(execute=@n6.then_5, Prev Passes:int=@n4.NumPasses) -> (Output_Get:int=0)
   Branch(execute=@n6.then_2, Condition="integer > integer"(A=@n4.NumMisses, B=Get Prev Misses() -> (Prev Misses:int=0)))
   n10 = Get Event Router()
   > Is Valid(InputObject=@n10)
   > n11 = Cast To FortPlayerController(Object=GetOwningPlayer())
   > Broadcast Event(
    self=@n10
    EventData=Make Pilgrim Event Gameplay Note Reaction(FortPlayerController_3_#1=@n11, PlayerNoteReaction_6_#2=NewEnumerator0)
   )
   Branch(execute=@n6.then_4, Condition="integer > integer"(A=@n4.NumPasses, B=Get Prev Passes() -> (Prev Passes:int=0)))
   n12 = Get Event Router()
   > Is Valid(InputObject=@n12)
   > n13 = Cast To FortPlayerController(Object=GetOwningPlayer())
   > Broadcast Event(
    self=@n12
    EventData=Make Pilgrim Event Gameplay Note Reaction(FortPlayerController_3_#1=@n13, PlayerNoteReaction_6_#2=NewEnumerator1)
   )
  ]}
 Get Star Widget {nodes [
   n1 = Get Star Widget!() -> (then, Index:int) {is=title K2Node_FunctionEntry}
   > Return Node!(
    execute
    StarWidget:WBP Pilgrim Star 2 Object Reference=Select!(
     Option 0:WBP Pilgrim Star 2 Object Reference=Get WBP_PilgrimStar1() -> (WBP_PilgrimStar1:WBP Pilgrim Star 2 Object Reference)
     Option 1:WBP Pilgrim Star 2 Object Reference=Get WBP_PilgrimStar2() -> (WBP_PilgrimStar2:WBP Pilgrim Star 2 Object Reference)
     Option 2:WBP Pilgrim Star 2 Object Reference=Get WBP_PilgrimStar3() -> (WBP_PilgrimStar3:WBP Pilgrim Star 2 Object Reference)
     Option 3:WBP Pilgrim Star 2 Object Reference=Get WBP_PilgrimStar4() -> (WBP_PilgrimStar4:WBP Pilgrim Star 2 Object Reference)
     Option 4:WBP Pilgrim Star 2 Object Reference=Get WBP_PilgrimStar5() -> (WBP_PilgrimStar5:WBP Pilgrim Star 2 Object Reference)
     Index:int=@n1
    ) -> (ReturnValue:WBP Pilgrim Star 2 Object Reference)
   )
  ]}
 Handle Star Update {nodes [
   n1 = Handle Star Update!() -> (then, CurrentStar:int, StarProgress:double=~) {is=title K2Node_FunctionEntry}
   > n2 = Check Is Star Changed!(execute, self, CurrentStarIndex:int=@n1, CurrentStarProgress:double=@n1.StarProgress) -> (then, IsChanged:bool) {is=title K2Node_CallFunction}
   > Branch(Condition=@n2)
   > CheckStarCompletion!(execute, self, CurrentStar:int=@n1) -> (then) {is=title K2Node_CallFunction}
   > n3 = Branch(Condition="integer <= integer"(A=Get MaxStars() -> (MaxStars:int=0), B=@n1))
   > Update Gold Star Progress Widgets!(execute, self, CurrentStar:int=@n1, GoldStarProgress:double=@n1.StarProgress) -> (then)
   Update Current Star!(execute=@n3.else, self, CurrentStar:int=@n1, NextStarProgress:double=@n1.StarProgress) -> (then)
  ]}
 SetupStarsResults {nodes [
   n1 = SetupStarsResults!() -> (then, Instrument:EPilgrimTrackType Enum=TrackGuitar, StarCount:int) {is=title K2Node_FunctionEntry}
   > n2 = For Loop(LastIndex=@n1.StarCount)
   > n3 = Get Star Widget(Index=@n2) {is=title K2Node_CallFunction}
   Branch(Condition="integer > integer"(A=@n1.StarCount, B=@n2))
   > SetComplete(self=@n3)
   "integer > integer"(A=@n1.StarCount, B=5)
  ]}
 ShowGoldStarIndicator {nodes [
   ShowGoldStarIndicator() {is=title K2Node_FunctionEntry}
   > SetVisibility(self=Get GoldStarIndicator() -> (GoldStarIndicator:Border Object Reference), InVisibility=Visible)
  ]}
 Update Current Star {nodes [
   n1 = Update Current Star!() -> (then, CurrentStar:int, NextStarProgress:double=~) {is=title K2Node_FunctionEntry}
   > Branch(Condition="float > float"(A=@n1.NextStarProgress))
   > n2 = Get Star Widget(Index=@n1)
   > SetStarProgress!(execute, self:WBP Pilgrim Star 2 Object Reference=@n2, Set Star Progress:double=@n1.NextStarProgress) -> (then)
  ]}
 Update Gold Star Progress Widgets {nodes [
   n1 = Update Gold Star Progress Widgets!() -> (then, CurrentStar:int, GoldStarProgress:double=~) {is=title K2Node_FunctionEntry}
   > n2 = For Loop(LastIndex=4)
   Branch(execute=@n2.Completed, Condition="integer > integer"(A=@n1, B=Get MaxStars() -> (MaxStars=0)))
   > ShowGoldStarIndicator() {is=title K2Node_CallFunction}
  ]}
 Update Icons Individualy {nodes [
   n1 = Update Icons Individualy!() -> (then, InstrumentsPlayed:Array of EPilgrimTrackType Enums=TrackGuitar) {is=title K2Node_FunctionEntry}
   > n2 = For Loop(LastIndex=3)
   > n3 = Get Star Widget(Index=@n2)
   n4 = Length(TargetArray:Array of EPilgrimTrackType Enums=@n1)
   > SetVisibility(
    self=@n3
    InVisibility=Select!(Option 0:ESlateVisibility Enum=Collapsed, Option 1:ESlateVisibility Enum=SelfHitTestInvisible, Index:bool="integer < integer"(A=@n2, B=@n4)) -> (ReturnValue:ESlateVisibility Enum)
   )
   > Branch(Condition=Is Valid Index!(TargetArray:Array of EPilgrimTrackType Enums=@n1, IndexToTest:int=@n2) -> (ReturnValue:bool))
   > SetStarProgress(self=@n3, Set Star Progress=0)
   > SetVisibility(self=Get WBP_PilgrimStar5(), InVisibility=Collapsed)
   Select!(
    TrackGuitar:float=0
    TrackBass:float=1
    TrackVocals:float=2
    TrackDrum:float=3
    TrackPlasticGuitar:float=0
    TrackPlasticBass:float=0
    TrackPlasticDrum:float=0
    TrackProVocals:float=0
    TrackProCymbals:float=0
    TrackEvents:float=0
    TrackNone:float=0
    Index:EPilgrimTrackType Enum="Get (a copy)"(Array:Array of EPilgrimTrackType Enums=@n1, Dimension 1=@n2) -> (Output:EPilgrimTrackType Enum)
   ) -> (ReturnValue:float)
  ]}
]
variables [
 Current Star {type=integer, default=1}
 Event Router {type=GameplayEventRouterComponent, default=null}
 In User Specified Scale {type=number, default=1}
 LastStarIndex {type=integer}
 LastStarProgress {type=number}
 LocalPlayerPreviousStarValue {type=integer}
 MaxStars {type=integer, default=5}
 OD Deploy Buzz {type=AudioComponent, default=null}
 Prev Hits {type=integer}
 Prev Misses {type=integer}
 Prev Passes {type=integer}
 PrintDebugInfo {type=boolean}
 Scale Stretch {type=EStretch, default=ScaleToFit}
 Star Progress {type=number}
]
widgets {
 info {inheritedWidgetCount=0, parentClass=/Script/PilgrimCoreUI.PilgrimGameWidgetBase, rootWidgetClass=&Overlay, widgetCount=9}
 widgets [
  StarsPanel {widgetClassPath=&Overlay, widget=StarsPanel}
  GoldStarIndicator {
   widgetClassPath=&Border
   bIsVariable=true
   parent=StarsPanel
   widget=GoldStarIndicator
   properties {
    visibility=Hidden
    background {
     bIsDynamicallyLoaded=false
     drawAs=Box
     imageType=NoImage
     mirroring=NoMirror
     resourceName=None
     resourceObject=null
     tiling=NoTile
     imageSize=<128 128>
     margin=<ltrb 1 1 1 1>
     outlineSettings {
      bUseBrushTransparency=false
      roundingType=HalfHeightRadius
      width=0
      color {colorUseRule=UseColor_Specified, specifiedColor=<rgba 0 0 0 0>}
      cornerRadii=<0 0 0 0>
     }
     tintColor {colorUseRule=UseColor_Specified, specifiedColor=<rgba 1 1 0 0>}
     uVRegion {bIsValid=false, max=<0 0>, min=<0 0>}
    }
    desiredSizeScale=<2 2>
    padding=<ltrb 0 0 0 0>
   }
   slot {class=OverlaySlot, properties {horizontalAlignment=HAlign_Fill, verticalAlignment=VAlign_Center}}
  }
  ScaleBox {
   widgetClassPath=&ScaleBox
   bIsVariable=true
   parent=StarsPanel
   widget=ScaleBox
   slot {class=OverlaySlot, properties {horizontalAlignment=HAlign_Fill, verticalAlignment=VAlign_Fill}}
  }
  StarContainer {
   widgetClassPath=&HorizontalBox
   bIsVariable=true
   parent=ScaleBox
   widget=StarContainer
   slot {class=ScaleBoxSlot, properties {horizontalAlignment=HAlign_Right}}
  }
  WBP_PilgrimStar1 {
   widgetClassPath=&/WBP_PilgrimStar2
   bIsVariable=true
   parent=StarContainer
   widget=WBP_PilgrimStar1
   slot {class=HorizontalBoxSlot}
  }
  WBP_PilgrimStar2 {
   widgetClassPath=&WBP_PilgrimStar2
   bIsVariable=true
   parent=StarContainer
   widget=WBP_PilgrimStar2
   slot {class=HorizontalBoxSlot}
  }
  WBP_PilgrimStar3 {
   widgetClassPath=&WBP_PilgrimStar2
   bIsVariable=true
   parent=StarContainer
   widget=WBP_PilgrimStar3
   slot {class=HorizontalBoxSlot}
  }
  WBP_PilgrimStar4 {
   widgetClassPath=&WBP_PilgrimStar2
   bIsVariable=true
   parent=StarContainer
   widget=WBP_PilgrimStar4
   slot {class=HorizontalBoxSlot}
  }
  WBP_PilgrimStar5 {
   widgetClassPath=&WBP_PilgrimStar2
   bIsVariable=true
   parent=StarContainer
   widget=WBP_PilgrimStar5
   slot {class=HorizontalBoxSlot}
  }
 ]
}
