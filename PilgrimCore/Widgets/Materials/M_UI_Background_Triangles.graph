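# Translucent UI material that draws a procedural background of scattered triangles over a two-colour gradient.
# Three rotated, tiled UV layers use per-cell noise to rotate, offset, scale and weight SDF triangles, faded by a vertical gradient.
# The triangle mask tints the rotated background gradient, and a top-right radial glow is added on top.
# The SquareUV aspect term divides pixel height by itself, so it applies no aspect correction.
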Material /PilgrimCore/Widgets/Materials/M_UI_Background_Triangles
graph {
 properties {
  bUsedWithStaticMesh=false
  blendMode=BLEND_Translucent
  materialDomain=MD_UI
  pixelDepthOffsetMode=PDOM_Legacy
  usedShadingModels=Default Lit
 }
 nodes [
  n1 = NamedRerouteUsage()
  n2 = ScalarParameter() {defaultValue=0.65, parameterName=TriangleMultiply}
  n3 = NamedRerouteUsage()
  n4 = Multiply(
   A=VectorParameter() {parameterName=RGB3, defaultValue=<rgba 0 0.2465 0.8438 1>}.RGB
   B=MaterialFunctionCall(
    UVs=NamedRerouteUsage()
    CenterPosition=Constant2Vector() {r=1}
    Radius=ScalarParameter() {defaultValue=1.25, parameterName=Param_1}
    Density=ScalarParameter() {defaultValue=1}
   ) {materialFunction=&RadialGradientExponential}.RadialGradientExponential
  )
  Add(
   A=LinearInterpolate(
    A=NamedRerouteUsage()
    B=VectorParameter() {parameterName=RGB1, defaultValue=<rgba 0.1421 0.0129 0.3542 1>}.RGB
    Alpha=Multiply(A=@n1, B=@n2)
   )
   B=Add(A=Multiply(A=@n3, B=@n4), B=@n4)
  )
  n5 = Rotator(Coordinate=TextureCoordinate(), Time=ScalarParameter() {defaultValue=-0.125, parameterName=Angle}) {speed=6.2832}
  n6 = ScalarParameter() {defaultValue=0.2125, parameterName=GradientOffSet}
  n7 = Power(Base=ComponentMask(None=Add(A=@n5, B=@n6)) {r=true}, Exp=ScalarParameter() {defaultValue=0.5, parameterName=Gradient Power})
  NamedRerouteDeclaration(None=LinearInterpolate(
    A=VectorParameter() {parameterName=RGB1, defaultValue=<rgba 0.1421 0.0129 0.3542 1>}.RGB
    B=VectorParameter() {parameterName=RGB2, defaultValue=<rgba 0.0232 0.0051 0.1406 1>}.RGB
    Alpha=Saturate(None=@n7)
   )) {name=BackgroundColor, nodeColor=<rgba 0 0 0 1>}
  n8 = TextureCoordinate()
  n9 = Subtract(A=ComponentMask(None=OneMinus(None=@n8)) {g=true}, B=ScalarParameter() {defaultValue=0.35, parameterName=Gradient Offset})
  n10 = Constant() {r=0.125}
  n11 = Constant() {r=1}
  n12 = MaterialFunctionCall() {materialFunction=&GetUserInterfaceUV}
  n13 = MaterialFunctionCall(Float2=@n12.Pixel Size) {materialFunction=&BreakOutFloat2Components}
  n14 = Divide(A=@n13.G, B=@n13.G)
  n15 = NamedRerouteDeclaration(None=Multiply(A=AppendVector(A=@n11, B=@n14), B=@n12.Normalized UV)) {name=SquareUV, nodeColor=<rgba 0 0 0 1>}
  n16 = Multiply(A=Add(A=Rotator(Coordinate=@n15, Time=@n10) {speed=6.2832}, B=Constant2Vector() {g=0.25, r=0.35}) {constB=0.25}) {constB=3}
  n17 = MaterialFunctionCall!(
   PrimaryTextureCoordinates=@n16
   2ndOrderTextureCoordinates
   RandomSeed=Constant() {r=1234}
   NoiseDensityBias
   Bypass Noise Influence Crosstalk
  ) {materialFunction=/Game/Effects/Fort_Effects/Materials/_Functions/MF_RandomUVSample}
  n18 = Frac(None=@n16)
  n19 = LinearInterpolate(Alpha=@n17.NoiseWithRange) {constA=0.5, constB=1.25}
  n20 = MaterialFunctionCall(
   UVs=Add(A=Rotator(Coordinate=@n18, Time=@n19) {speed=6.2832}, B=LinearInterpolate(Alpha=@n17.NoiseWithRange) {constB=0})
   Texture Scale=LinearInterpolate(Alpha=@n17.Noise) {constA=0.85}
  ) {materialFunction=&ScaleUVsByCenter}
  n21 = LinearInterpolate(Alpha=SmoothStep(Value=MaterialFunctionCall!(UVs=@n20.UVs, Corner 01, Corner 02, Corner 03, Glow Min, Glow Max, Outline Thickness) {materialFunction=/FortUILibrary/MaterialFunctions/Shapes/MF_UI_WIP_Triangle}.SDF) {constMax=0.02}) {constA=1, constB=0}
  n22 = Constant() {r=-0.125}
  n23 = Multiply(A=Add(A=Rotator(Coordinate=@n15, Time=@n22) {speed=6.2832}, B=Constant2Vector() {g=0.35, r=0.25})) {constB=3}
  n24 = MaterialFunctionCall(PrimaryTextureCoordinates=@n23) {materialFunction=&MF_RandomUVSample}
  n25 = Frac(None=@n23)
  n26 = LinearInterpolate(Alpha=@n24.NoiseWithRange) {constA=0.25}
  n27 = MaterialFunctionCall(
   UVs=Add(A=Rotator(Coordinate=@n25, Time=@n26) {speed=6.2832}, B=LinearInterpolate(Alpha=@n24.NoiseWithRange) {constA=-0.35, constB=0})
   Texture Scale=LinearInterpolate(Alpha=@n24.Noise) {constA=0.85, constB=0.6}
  ) {materialFunction=&ScaleUVsByCenter}
  n28 = LinearInterpolate(Alpha=SmoothStep(Value=MaterialFunctionCall(UVs=@n27.UVs) {materialFunction=&MF_UI_WIP_Triangle}.SDF) {constMax=0.02}) {constA=1, constB=0}
  n29 = Constant() {r=-0.25}
  n30 = Multiply(A=Add(A=Rotator(Coordinate=@n15, Time=@n29) {speed=6.2832}, B=Constant2Vector() {g=0.5, r=0.125})) {constB=2.5}
  n31 = MaterialFunctionCall(PrimaryTextureCoordinates=@n30) {materialFunction=&MF_RandomUVSample}
  n32 = Add(
   A=Rotator(Coordinate=Frac(None=@n30), Time=LinearInterpolate(Alpha=@n31.NoiseWithRange) {constA=0.25, constB=0.85}) {speed=6.2832}
   B=LinearInterpolate(Alpha=@n31.NoiseWithRange) {constA=-0.35, constB=0.5}
  )
  n33 = LinearInterpolate(Alpha=@n31.Noise) {constA=0.8, constB=0.75}
  n34 = SmoothStep(Value=MaterialFunctionCall(UVs=MaterialFunctionCall(UVs=@n32, Texture Scale=@n33) {materialFunction=&ScaleUVsByCenter}.UVs) {materialFunction=&MF_UI_WIP_Triangle}.SDF) {constMax=0.02}
  n35 = Add(
   A=Add(A=Multiply(A=@n17.NoiseWithRange, B=@n21), B=Multiply(A=@n24.NoiseWithRange, B=@n28))
   B=Multiply(A=@n31.NoiseWithRange, B=LinearInterpolate(Alpha=@n34) {constA=1, constB=0})
  )
  NamedRerouteDeclaration(None=Multiply(A=Saturate(None=@n9), B=Saturate(None=@n35))) {name=Triagles, nodeColor=<rgba 0 0 0 1>}
 ]
}
