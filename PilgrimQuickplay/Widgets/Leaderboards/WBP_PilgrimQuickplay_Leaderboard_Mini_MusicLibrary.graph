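# Mini leaderboard widget for the Pilgrim Quickplay music library, showing solo rankings for the selected song and the player's instrument.
# Selection starts a debounce timer, then checks the song's track part and sends a cancellable async player-centered friends request.
# Thin or failed results fall back to global centered or top requests.
# A clamped window around the player fills the list, while a switcher shows loading, unavailable or missing-part states.
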WidgetBlueprint /PilgrimQuickplay/Widgets/Leaderboards/WBP_PilgrimQuickplay_Leaderboard_Mini_MusicLibrary
graphs [
 CalculateIndicesAroundPlayerToDisplay {nodes [
   n1 = CalculateIndicesAroundPlayerToDisplay!() -> (then, TotalEntries:int) {is=title K2Node_FunctionEntry}
   > n2 = Sequence()
   n3 = Get EntriesPlusMinusAroundPlayer() -> (EntriesPlusMinusAroundPlayer:int=0)
   n4 = Get PlayerEntryIndex() -> (PlayerEntryIndex:int=0)
   > Set StartingDisplayIndex(StartingDisplayIndex:int=int - int(A=@n4, B=@n3)) -> (Output_Get:int=0)
   > Set LastDisplayIndex(LastDisplayIndex:int=int + int(A=@n4, B=@n3)) -> (Output_Get:int=0)
   n5 = Branch(execute=[@n6."  ", @n2.then_1], Condition="integer < integer"(A=Get StartingDisplayIndex() -> (StartingDisplayIndex:int=0)))
   > Increment Int(Value=Get StartingDisplayIndex() -> (StartingDisplayIndex=0))
   > n6 = Increment Int(Value=Get LastDisplayIndex() -> (LastDisplayIndex:int=0))
   n7 = int - int(A=@n1)
   n8 = Branch(execute=[@n5.else, @n9."  "], Condition="integer > integer"(A=Get LastDisplayIndex() -> (LastDisplayIndex=0), B=@n7))
   > Decrement Int(Value=Get LastDisplayIndex() -> (LastDisplayIndex=0))
   > n9 = Decrement Int(Value=Get StartingDisplayIndex() -> (StartingDisplayIndex=0))
   Set StartingDisplayIndex(
    execute=@n8.else
    StartingDisplayIndex="Clamp (Integer)"(Value=Get StartingDisplayIndex() -> (StartingDisplayIndex=0), Min=0, Max=@n7) -> (ReturnValue=0)
   ) -> (Output_Get=0)
   > Set LastDisplayIndex(LastDisplayIndex="Clamp (Integer)"(Value=Get LastDisplayIndex() -> (LastDisplayIndex=0), Min=0, Max=@n7) -> (ReturnValue=0)) -> (Output_Get=0)
  ]}
 CancelTimerIfSet {nodes [
   n1 = CancelTimerIfSet!() -> (then, Handle:Timer Handle Structure) {is=title K2Node_FunctionEntry}
   > Branch(Condition=Is Valid Timer Handle!(Handle:Timer Handle Structure=@n1) -> (ReturnValue:bool))
   > Clear and Invalidate Timer by Handle(Handle=@n1)
  ]}
 EventGraph {nodes [
   n1 = DelayedLBR_event() {is=title K2Node_CustomEvent}
   > SetLeaderboardRequestData()
   > Branch(Condition=Get IsAllLeaderboardDataGathered() -> (IsAllLeaderboardDataGathered:bool))
   > RequestCenteredFriendsLeaderboardData()
   SetLeaderboardRequestData() {is=title K2Node_CustomEvent}
   > n2 = Cast To FortPlayerController(Object=GetOwningPlayer())
   > Set RequestingPlayerController(RequestingPlayerController:Fort Player Controller Object Reference=@n2) -> (Output_Get:Fort Player Controller Object Reference)
   > n3 = GetPilgrimQuickplayPlayerStateComponent!(
    execute
    PlayerState:Player State Object Reference=Get PlayerState!(self:Controller Object Reference=@n2) -> (PlayerState:Player State Object Reference)
   ) -> (then, ReturnValue:Pilgrim Quickplay Player State Component Object Reference)
   > n4 = Is Valid(InputObject=@n3)
   ShowFailedState(execute=@n4.Is Not Valid)
   Set PlayersInBand(
    execute=@n4.Is Valid
    PlayersInBand:Array of Player State Object References=GetPlayersInBandByController!(
     self:Pilgrim Quickplay Band State Component Object Reference=GetBandState!(self:Pilgrim Quickplay Player State Component Object Reference=@n3) -> (ReturnValue:Pilgrim Quickplay Band State Component Object Reference)
     Controller:Player Controller Object Reference=Get RequestingPlayerController() -> (RequestingPlayerController:Fort Player Controller Object Reference)
    ) -> (ReturnValue:Array of Player State Object References)
   ) -> (Output_Get:Array of Player State Object References)
   > n5 = Branch(Condition="Not Equal (Name)"(A=Get SongShortName() -> (SongShortName:Name=None)) -> (ReturnValue=~))
   n6 = GetInstrumentChoice!(self:Pilgrim Core Player State Component Object Reference=@n3) -> (ReturnValue:EPilgrimTrackType Enum=TrackGuitar)
   n7 = Select!(
    TrackGuitar:EPilgrimLeaderboardType Enum=SoloGuitar
    TrackBass:EPilgrimLeaderboardType Enum=SoloBass
    TrackVocals:EPilgrimLeaderboardType Enum=SoloVocals
    TrackDrum:EPilgrimLeaderboardType Enum=SoloDrum
    TrackPlasticGuitar:EPilgrimLeaderboardType Enum=SoloPeripheralGuitar
    TrackPlasticBass:EPilgrimLeaderboardType Enum=SoloPeripheralBass
    TrackPlasticDrum:EPilgrimLeaderboardType Enum=SoloPeripheralDrum
    TrackProVocals:EPilgrimLeaderboardType Enum=SoloPeripheralVocals
    TrackProCymbals:EPilgrimLeaderboardType Enum=SoloPeripheralCymbal
    TrackEvents:EPilgrimLeaderboardType Enum=Invalid
    TrackNone:EPilgrimLeaderboardType Enum=Invalid
    Index:EPilgrimTrackType Enum=@n6
   ) -> (ReturnValue:EPilgrimLeaderboardType Enum)
   > n8 = Branch(Condition="Not Equal (Enum)"!(A:EPilgrimLeaderboardType Enum=@n7, B:EPilgrimLeaderboardType Enum=Invalid) -> (ReturnValue:bool))
   > Set LeaderboardType(LeaderboardType:EPilgrimLeaderboardType Enum=@n7) -> (Output_Get:EPilgrimLeaderboardType Enum=SoloGuitar)
   > n9 = Branch(Condition=HasTrackPart!(
     SongMetadata:Pilgrim Song Metadata Object Reference=Get SongMetadata() -> (SongMetadata:Pilgrim Song Metadata Object Reference)
     TrackType:EPilgrimTrackType Enum=@n6
    ) -> (HasTrackPart:bool))
   > Set IsAllLeaderboardDataGathered(IsAllLeaderboardDataGathered:bool=true) -> (Output_Get:bool)
   SetActiveWidget(
    execute=@n9.else
    self=Get ContentSwitcher() -> (ContentSwitcher:Common Visibility Switcher Object Reference)
    Widget=Get Overlay_NoLeaderboardEntries_MissingParts() -> (Overlay_NoLeaderboardEntries_MissingParts:Overlay Object Reference)
   )
   > Set IsAllLeaderboardDataGathered(IsAllLeaderboardDataGathered=false)
   ShowFailedState(execute=@n8.else)
   ShowFailedState(execute=@n5.else)
   ShowFailedState(execute=@n2.CastFailed)
   ShowFailedState() {is=title K2Node_CustomEvent}
   > SetActiveWidget(self=Get ContentSwitcher(), Widget=Get Overlay_NoLeaderboardEntries() -> (Overlay_NoLeaderboardEntries:Overlay Object Reference))
   RequestCenteredFriendsLeaderboardData() {is=title K2Node_CustomEvent}
   > Set RequestShouldCenterOnPlayer(RequestShouldCenterOnPlayer:bool=true) -> (Output_Get:bool)
   > Set ShouldRequestFriendsLeaderboard(ShouldRequestFriendsLeaderboard:bool=true) -> (Output_Get:bool)
   > MakeAsyncLeaderboardRequest()
   n10 = OnMusicLibraryEntrySelected!() -> (OutputDelegate, then, Song:Pilgrim Song Metadata Object Reference) {is=title K2Node_CustomEvent}
   > n11 = Set SongMetadata(SongMetadata:Pilgrim Song Metadata Object Reference=@n10.Song) -> (Output_Get:Pilgrim Song Metadata Object Reference)
   > Set SongShortName(SongShortName:Name=GetShortName!(self:Sparks Song Data Object Reference=@n11) -> (ReturnValue:Name=None)) -> (Output_Get:Name=None)
   > SetActiveWidget(self=Get ContentSwitcher(), Widget=Get Content_LoadingEntries() -> (Content_LoadingEntries:Horizontal Box Object Reference))
   > ClearListItems(self=Get LeaderboardEntries() -> (LeaderboardEntries:Common List View Object Reference))
   > CancelTimerIfSet!(execute, self, Handle:Timer Handle Structure=Get ItemSelectedLeaderboardTimer() -> (ItemSelectedLeaderboardTimer:Timer Handle Structure)) -> (then) {is=title K2Node_CallFunction}
   > n12 = Set Timer by Event(Delegate=@n1, Time=GetSongSelectTimerDelaySeconds!() -> (ReturnValue:float), bLooping=false)
   > Set ItemSelectedLeaderboardTimer(ItemSelectedLeaderboardTimer:Timer Handle Structure=@n12) -> (Output_Get:Timer Handle Structure)
   MakeAsyncLeaderboardRequest() {is=title K2Node_CustomEvent}
   > SetLeaderboardTitle!(
    execute
    self
    LeaderboardType:EPilgrimLeaderboardType Enum=Get LeaderboardType() -> (LeaderboardType:EPilgrimLeaderboardType Enum=SoloGuitar)
   ) -> (then)
   n13 = Get AsyncLeaderboardRequest() -> (AsyncLeaderboardRequest:Pilgrim Async Action Request Leaderboard Object Reference)
   > n14 = Is Valid(InputObject=@n13)
   > n15 = Cancel(self:Cancellable Async Action Object Reference=@n13)
   Set RemainingCenteredRetries(execute=[@n15.then, @n14.Is Not Valid], RemainingCenteredRetries:int=Get MaxRetries() -> (MaxRetries:int=0)) -> (Output_Get:int=0)
   > n16 = PilgrimRequestLeaderboard!(
    execute:Exec
    SongShortName:Name=Get SongShortName() -> (SongShortName=None)
    SpotlightGroup:String
    LeaderboardType:EPilgrimLeaderboardType Enum=Get LeaderboardType() -> (LeaderboardType=SoloGuitar)
    RequestingPlayer:Fort Player Controller Object Reference=Get RequestingPlayerController()
    bFriendsOnly:bool=Get ShouldRequestFriendsLeaderboard() -> (ShouldRequestFriendsLeaderboard:bool)
    bCenteredOnRequestingPlayer:bool=Get RequestShouldCenterOnPlayer() -> (RequestShouldCenterOnPlayer:bool)
    Page:int=0
    bIsSpotlightCompetition:bool=false
    bIsAllTimeHighScore:bool=Get AllTimeHighScoreEnabled() -> (AllTimeHighScoreEnabled:bool)
    bUseCache:bool=true
    FreshnessTime:int=-1
    RequestPriority:EPilgrimLeaderboardRequestPriority Enum=Normal
   ) -> (
    then:Exec
    AsyncTaskProxy:Pilgrim Async Action Request Leaderboard Object Reference
    OnSuccess:Exec
    OnFailure:Exec
    ResultPage:Pilgrim Leaderboard Page Structure
    RequestParams:Pilgrim Leaderboard Request Params Structure
   ) {is=title K2Node_AsyncAction}
   > Set AsyncLeaderboardRequest(AsyncLeaderboardRequest:Pilgrim Async Action Request Leaderboard Object Reference=@n16) -> (Output_Get:Pilgrim Async Action Request Leaderboard Object Reference)
   n17 = Set LeaderboardPageToDisplay(execute=@n16.OnSuccess, LeaderboardPageToDisplay:Pilgrim Leaderboard Page Structure=@n16.ResultPage) -> (Output_Get:Pilgrim Leaderboard Page Structure)
   > n18 = ValidateLeaderboardPageAgainstCurrentRequest!(
    execute
    self
    PilgrimLeaderboardPage:Pilgrim Leaderboard Page Structure=@n17
    RequestParams:Pilgrim Leaderboard Request Params Structure=@n16.RequestParams
   ) -> (then, Show Failed:bool, Success:bool, Retry Global Player Request:bool, Retry Top Global Request:bool)
   > n19 = Branch(Condition=@n18)
   > ShowFailedState() {is=title K2Node_CallFunction}
   n20 = Branch(execute=@n19.else, Condition=@n18.Success)
   > DisplayCachedLeaderboardData()
   n21 = Branch(execute=@n20.else, Condition=@n18.Retry Global Player Request)
   > RequestGlobalLeaderboardData()
   n22 = Branch(execute=@n21.else, Condition=@n18.Retry Top Global Request)
   > RequestTopLeaderboardData()
   ShowFailedState(execute=@n22.else)
   n23 = Break Pilgrim Leaderboard Request Params!(PilgrimLeaderboardRequestParams:"Pilgrim Leaderboard Request Params Structure (by ref)"=@n16.RequestParams) -> (
    SongShortName:Name
    LeaderboardType:EPilgrimLeaderboardType Enum
    RequestingPlayerAccountId:Unique Net Id Structure
    bFriendsOnly:bool=~
    bCenteredOnRequestingPlayer:bool=~
    Page:int
    ControllerId:int
    bIsSpotlightCompetition:bool=~
    SpotlightGroup:String
    bIsAllTimeHighScore:bool=~
    bUseCache:bool=~
    FreshnessTime:int
    Priority:EPilgrimLeaderboardRequestPriority Enum
   ) {is=title K2Node_BreakStruct}
   n24 = Branch(execute=@n16.OnFailure, Condition=@n23.bFriendsOnly)
   > RequestTopLeaderboardData()
   n25 = Branch(execute=@n24.else, Condition=@n23.bCenteredOnRequestingPlayer)
   > RequestTopLeaderboardData()
   ShowFailedState(execute=@n25.else)
   RequestTopLeaderboardData() {is=title K2Node_CustomEvent}
   > Set RequestShouldCenterOnPlayer(RequestShouldCenterOnPlayer=false)
   > Set ShouldRequestFriendsLeaderboard(ShouldRequestFriendsLeaderboard=false)
   > MakeAsyncLeaderboardRequest() {is=title K2Node_CallFunction}
   RequestGlobalLeaderboardData() {is=title K2Node_CustomEvent}
   > Set RequestShouldCenterOnPlayer(RequestShouldCenterOnPlayer=true)
   > Set ShouldRequestFriendsLeaderboard(ShouldRequestFriendsLeaderboard=false)
   > MakeAsyncLeaderboardRequest()
   Event Destruct()
   > Parent: Destruct()
   n26 = Get AsyncLeaderboardRequest()
   > Is Valid(InputObject=@n26)
   > Cancel(self=@n26)
   > Set AsyncLeaderboardRequest()
   Event Construct()
   > Parent: Construct()
   > "SetText (Text)"(self=Get LeaderboardTitle() -> (LeaderboardTitle:Common Text Object Reference))
   K2Node_Event_6 = DisplayCachedLeaderboardData() {is=title K2Node_CustomEvent}
   n27 = Break Pilgrim Leaderboard Page!(PilgrimLeaderboardPage:"Pilgrim Leaderboard Page Structure (by ref)"=Get LeaderboardPageToDisplay() -> (LeaderboardPageToDisplay:Pilgrim Leaderboard Page Structure)) -> (
    SongShortName:Name
    LeaderboardType:EPilgrimLeaderboardType Enum
    Page:int
    TotalPages:int
    Entries:Array of Pilgrim Leaderboard Entry Structures
    bIsSpotlightCompetition:bool=~
    SpotlightGroup:String
    bIsAllTimeHighScore:bool=~
   ) {is=title K2Node_BreakStruct}
   > Branch(Condition="Not Equal (Name)"(A=@n27) -> (ReturnValue=~))
   > Set PlayerPageNumber(PlayerPageNumber:int=@n27.Page) -> (Output_Get:int=0)
   > Set Total Pages(Total Pages:int=@n27.TotalPages) -> (Output_Get:int=0)
   > SetLeaderboardTitle(LeaderboardType=@n27.LeaderboardType)
   > n28 = Branch(Condition="integer < integer"(A=Get PlayerEntryIndex() -> (PlayerEntryIndex=0)))
   > Set StartingDisplayIndex(StartingDisplayIndex=0) -> (Output_Get=0)
   > n29 = Set LastDisplayIndex(LastDisplayIndex=int * int(A=Get EntriesPlusMinusAroundPlayer() -> (EntriesPlusMinusAroundPlayer=0), B=2)) -> (Output_Get=0)
   SetActiveWidget(
    execute=[@n33.then, @n29.then]
    self=Get ContentSwitcher()
    Widget=Get LeaderboardContainer() -> (LeaderboardContainer:Vertical Box Object Reference)
   )
   > ClearListItems(self=Get LeaderboardEntries())
   > n30 = For Loop(FirstIndex=Get StartingDisplayIndex() -> (StartingDisplayIndex=0), LastIndex=Get LastDisplayIndex() -> (LastDisplayIndex=0))
   > Branch(Condition=Is Valid Index!(TargetArray:Array of Pilgrim Leaderboard Entry Structures=@n27.Entries, IndexToTest:int=@n30) -> (ReturnValue:bool))
   n31 = "Get (a copy)"(Array:Array of Pilgrim Leaderboard Entry Structures=@n27.Entries, Dimension 1=@n30) -> (Output:Pilgrim Leaderboard Entry Structure)
   > n32 = Construct Object from Class!(
    execute
    Class=&/BP_PilgrimQuickplay_LeaderboardEntry_ListObject.BP_PilgrimQuickplay_LeaderboardEntry_ListObject_C
    self
    LeaderboardEntryData:Pilgrim Leaderboard Entry Structure=@n31
    ListIndex:int=@n30
    IsSoloEntry:bool=True
    HasLocalPlayer:bool=Break Pilgrim Leaderboard Entry!(PilgrimLeaderboardEntry:"Pilgrim Leaderboard Entry Structure (by ref)"=@n31) -> (
     bHasLocalPlayer:bool=~
     PlayerAccountIdToDisplayNameMap:Map of Strings to Strings
     Rank:int
     Percentile:float=~
     ScoreData:Pilgrim Leaderboard Score Data Structure
     DateTime:Date Time Structure
     FriendRank:int
    ) {is=title K2Node_BreakStruct}
    IsAllTimeHighScore:bool=Get AllTimeHighScoreEnabled()
   ) -> (then, ReturnValue:BP Pilgrim Quickplay Leaderboard Entry List Object Object Reference)
   > AddItem!(execute, self:List View Object Reference=Get LeaderboardEntries(), Item:Object Reference=@n32) -> (then)
   n33 = CalculateIndicesAroundPlayerToDisplay!(execute=@n28.else, self, TotalEntries:int=Length(TargetArray:Array of Pilgrim Leaderboard Entry Structures=@n27.Entries)) -> (then) {is=title K2Node_CallFunction}
  ]}
 Get Song Short Name For Leaderboard Request {nodes [
   Get Song Short Name For Leaderboard Request() {is=title K2Node_FunctionEntry}
   > Return Node!(execute, SongShortName:Name=Get SongShortName() -> (SongShortName=None))
  ]}
 GetSeasonNumberText {nodes [
   GetSeasonNumberText() {is=title K2Node_FunctionEntry}
   > n1 = EvaluateCurveTableRow!(
    execute
    CurveTableRowHandle:Curve Table Row Handle Structure=Make Curve Table Row Handle!(CurveTable:Curve Table Object Reference=/PilgrimQuickplay/DataTables/PilgrimQuickplayGameData, RowName:Name=SeasonData.SeasonNumber) -> (CurveTableRowHandle:Curve Table Row Handle Structure) {is=title K2Node_MakeStruct}
    InXY:float=0
    ContextString:String
   ) -> (then, OutXY:float, ReturnValue:bool)
   > n2 = Branch(Condition=@n1.ReturnValue)
   > Return Node!(
    execute
    SeasonNumberText:Text="To Text (Float)"(Value=@n1, MinimumIntegralDigits=1, MaximumIntegralDigits=324, MinimumFractionalDigits=0, MaximumFractionalDigits=3)
   )
   Return Node(execute=@n2.else, SeasonNumberText="LOCTEXT(\"?\")")
  ]}
 IsCurrentBandEntry {nodes [
   n1 = IsCurrentBandEntry!() -> (then, LeaderboardEntry:Pilgrim Leaderboard Entry Structure) {is=title K2Node_FunctionEntry}
   > n2 = Branch(Condition=Is Not Empty!(TargetArray:Array of Player State Object References=Get PlayersInBand() -> (PlayersInBand:Array of Player State Object References)) -> (ReturnValue:bool))
   n3 = Get PlayersInBand()
   n4 = Break Pilgrim Leaderboard Entry(PilgrimLeaderboardEntry=@n1) -> (bHasLocalPlayer=~, Percentile=~)
   > n5 = Branch(Condition="Equal (Integer)"(
     A=Length(TargetArray:Array of Player State Object References=@n3)
     B=Length!(TargetMap:Map of Strings to Strings=@n4.PlayerAccountIdToDisplayNameMap) -> (ReturnValue:int=0) {is=title K2Node_CallFunction}
    ))
   > n6 = For Each Loop with Break!(Exec:Exec, Array:Array of Player State Object References=@n3, Break:Exec=@n9.else) -> (LoopBody:Exec, Array Element:Player State Object Reference, Array Index:int, Completed:Exec)
   Return Node!(execute=[@n6.Completed, @n5.else, @n2.else], AllBandMembersMatch:bool=Get AllBandMembersMatchEntry() -> (AllBandMembersMatchEntry:bool))
   > n7 = GetDebugStringForUniqueId!(
    execute
    UniqueID:"Unique Net Id Structure (by ref)"=GetGameAccountId!(self:Fort Player Controller Object Reference=GetFortPlayerControllerFromActor!(Actor:Actor Object Reference=@n6) -> (ReturnValue:Fort Player Controller Object Reference)) -> (ReturnValue:Unique Net Id Structure)
   ) -> (then, ReturnValue:String)
   > n8 = Set AllBandMembersMatchEntry(AllBandMembersMatchEntry:bool=Contains!(
     TargetMap:Map of Strings to Strings=Break Pilgrim Leaderboard Entry(PilgrimLeaderboardEntry=@n1) -> (bHasLocalPlayer=~, Percentile=~).PlayerAccountIdToDisplayNameMap
     Key:"String (by ref)"=@n7
    ) -> (ReturnValue)) -> (Output_Get:bool)
   > n9 = Branch(Condition=@n8)
  ]}
 SetLeaderboardTitle {nodes [
   n1 = SetLeaderboardTitle!() -> (then, LeaderboardType:EPilgrimLeaderboardType Enum=SoloGuitar) {is=title K2Node_FunctionEntry}
   n2 = Select!(
    Option 0:Text=Format Text!(
     Format="LOCTEXT(\"Season {SeasonNumber}\")"
     SeasonNumber:Text=GetSeasonNumberText!() -> (SeasonNumberText:Text) {is=title K2Node_CallFunction}
    ) -> (Result)
    Option 1:Text=Get All-TimeText() -> (All-TimeText:Text)
    Index:bool=Get AllTimeHighScoreEnabled()
   ) -> (ReturnValue:Text)
   n3 = Select!(
    SoloGuitar:Text="LOCTABLE(\"/SparksSettings/UI/ST_Sparks_InstrumentNames\", \"Lead\")"
    SoloBass:Text="LOCTABLE(\"&ST_Sparks_InstrumentNames\", \"Bass\")"
    SoloVocals:Text="LOCTABLE(\"&ST_Sparks_InstrumentNames\", \"Vocals\")"
    SoloDrum:Text="LOCTABLE(\"&ST_Sparks_InstrumentNames\", \"Drums\")"
    BandDuo:Text=MakeLiteralText(Value="LOCTEXT(\"Band Duo\")")
    BandTrio:Text=MakeLiteralText(Value="LOCTEXT(\"Band Trio\")")
    BandQuad:Text=MakeLiteralText(Value="LOCTEXT(\"Band Squad\")")
    SoloPeripheralGuitar:Text="LOCTABLE(\"&ST_Sparks_InstrumentNames\", \"ProLead\")"
    SoloPeripheralBass:Text="LOCTABLE(\"&ST_Sparks_InstrumentNames\", \"ProBass\")"
    SoloPeripheralDrum:Text="LOCTABLE(\"&ST_Sparks_InstrumentNames\", \"ProDrums\")"
    SoloPeripheralCymbal:Text="LOCTABLE(\"&ST_Sparks_InstrumentNames\", \"ProDrumsCymbals\")"
    SoloPeripheralVocals:Text="LOCTABLE(\"&ST_Sparks_InstrumentNames\", \"ProVocals\")"
    Invalid:Text=MakeLiteralText(Value="INVTEXT(\"[PLACEHOLDER]\")")
    Index:EPilgrimLeaderboardType Enum=@n1
   ) -> (ReturnValue)
   > "SetText (Text)"(
    self=Get LeaderboardTitle()
    InText=Select!(
     Option 0:Text=Format Text!(Format="LOCTEXT(\"{AllTimeOrSeasonNumber} Global Leaderboard ({Instrument}) \")", AllTimeOrSeasonNumber:Text=@n2, Instrument:Text=@n3) -> (Result)
     Option 1:Text=Format Text(Format="LOCTEXT(\"{AllTimeOrSeasonNumber} Friends Leaderboard ({Instrument})\")", AllTimeOrSeasonNumber=@n2, Instrument=@n3)
     Index:bool=Get ShouldRequestFriendsLeaderboard()
    ) -> (ReturnValue)
   )
  ]}
 ValidateLeaderboardPageAgainstCurrentRequest {nodes [
   n1 = ValidateLeaderboardPageAgainstCurrentRequest!() -> (then, PilgrimLeaderboardPage:Pilgrim Leaderboard Page Structure, RequestParams:Pilgrim Leaderboard Request Params Structure) {is=title K2Node_FunctionEntry}
   n2 = Break Pilgrim Leaderboard Request Params(PilgrimLeaderboardRequestParams=@n1.RequestParams) -> (bFriendsOnly=~, bCenteredOnRequestingPlayer=~, bIsSpotlightCompetition=~, bIsAllTimeHighScore=~, bUseCache=~)
   > Set MatchingSongShortName(MatchingSongShortName:bool="Equal (Name)"(A=Get SongShortName() -> (SongShortName=None), B=@n2)) -> (Output_Get:bool)
   n3 = Break Pilgrim Leaderboard Page(PilgrimLeaderboardPage=@n1) -> (bIsSpotlightCompetition=~, bIsAllTimeHighScore=~)
   n4 = Get MinimumFriendEntryCountForDisplay() -> (MinimumFriendEntryCountForDisplay:int=0)
   n5 = Get MinimumGlobalEntryCountForDisplay() -> (MinimumGlobalEntryCountForDisplay:int=0)
   > Set EntryCountHighEnough(EntryCountHighEnough:bool="integer >= integer"(
     A=Length(TargetArray:Array of Pilgrim Leaderboard Entry Structures=@n3.Entries) {is=title K2Node_CallArrayFunction}
     B=Select(Option 0:int=@n5, Option 1:int=@n4, Index=@n2.bFriendsOnly) -> (ReturnValue:int)
    )) -> (Output_Get:bool)
   > Set PlayerEntryIndex(PlayerEntryIndex:int=-1) -> (Output_Get:int=0)
   > n6 = For Each Loop with Break(Array:Array of Pilgrim Leaderboard Entry Structures=@n3.Entries, Break=@n7.then) -> (Array Element:Pilgrim Leaderboard Entry Structure)
   > Branch(Condition=Break Pilgrim Leaderboard Entry(PilgrimLeaderboardEntry=@n6) -> (bHasLocalPlayer=~, Percentile=~))
   > Set LocalPlayerFound(LocalPlayerFound:bool=true) -> (Output_Get:bool)
   > n7 = Set PlayerEntryIndex(PlayerEntryIndex=@n6.Array Index) -> (Output_Get=0)
   n8 = Branch(
    execute=@n6.Completed
    Condition=Break Pilgrim Leaderboard Request Params(PilgrimLeaderboardRequestParams=@n1.RequestParams) -> (bFriendsOnly=~, bCenteredOnRequestingPlayer=~, bIsSpotlightCompetition=~, bIsAllTimeHighScore=~, bUseCache=~).bFriendsOnly
   )
   > n9 = Branch(Condition=Get EntryCountHighEnough() -> (EntryCountHighEnough:bool))
   > Return Node!(execute, Show Failed:bool=false, Success:bool=true, Retry Global Player Request:bool=false, Retry Top Global Request:bool=false)
   n10 = Branch(execute=@n9.else, Condition=Get LocalPlayerFound() -> (LocalPlayerFound:bool))
   > Return Node(Show Failed=false, Success=false, Retry Global Player Request=true, Retry Top Global Request=false)
   Return Node(execute=@n10.else, Show Failed=false, Success=false, Retry Global Player Request=false, Retry Top Global Request=true)
   n11 = Branch(execute=@n8.else, Condition=Get EntryCountHighEnough())
   > Return Node(Show Failed=false, Success=true, Retry Global Player Request=false, Retry Top Global Request=false)
   Return Node(execute=@n11.else, Show Failed=true, Success=false, Retry Global Player Request=false, Retry Top Global Request=false)
  ]}
]
variables [
 All-TimeText {type=string, default=All-Time}
 AllTimeHighScoreEnabled {type=boolean}
 AsyncLeaderboardRequest {type=PilgrimAsyncAction_RequestLeaderboard, default=null}
 CurrentRequestParameters {
  type=PilgrimLeaderboardRequestParams
  default {
   bCenteredOnRequestingPlayer=false
   bFriendsOnly=false
   bIsAllTimeHighScore=false
   bIsSpotlightCompetition=false
   bUseCache=true
   controllerId=0
   freshnessTime=-1
   leaderboardType=Invalid
   localPlayer=null
   page=0
   paramKey=0
   priority=Normal
   requestId=0
   requestingPlayerAccountId=INVALID
   songShortName=None
   spotlightGroup=""
   worldContextObject=null
  }
 }
 EntriesPlusMinusAroundPlayer {type=integer, default=1}
 IsAllLeaderboardDataGathered {type=boolean}
 ItemSelectedLeaderboardTimer {type=TimerHandle}
 LastDisplayIndex {type=integer}
 LeaderboardPageToDisplay {
  type=PilgrimLeaderboardPage
  default {
   bIsAllTimeHighScore=false
   bIsSpotlightCompetition=false
   leaderboardType=SoloGuitar
   page=0
   songShortName=None
   spotlightGroup=""
   totalPages=0
   entries []
  }
 }
 LeaderboardType {type=EPilgrimLeaderboardType, default=SoloGuitar}
 MaxRetries {type=integer, default=3}
 MinimumFriendEntryCountForDisplay {type=integer, default=2}
 MinimumGlobalEntryCountForDisplay {type=integer, default=1}
 PlayerEntryIndex {type=integer, default=-6}
 PlayerPageNumber {type=integer}
 PlayersInBand {type="array<PlayerState>", default []}
 RemainingCenteredRetries {type=integer}
 RequestingPlayerController {type=FortPlayerController, default=null}
 RequestShouldCenterOnPlayer {type=boolean, default=true}
 RequestToDisplayCenteredDataHasBeenMade {type=boolean}
 ShouldRequestFriendsLeaderboard {type=boolean, default=true}
 SongMetadata {type=PilgrimSongMetadata, default=null}
 SongShortName {type=string, default=None}
 StartingDisplayIndex {type=integer}
 Total Pages {type=integer}
]
widgets {
 info {inheritedWidgetCount=0, parentClass=&CommonUserWidget, rootWidgetClass=&VerticalBox, widgetCount=32}
 widgets [
  VerticalBox_0 {widgetClassPath=&VerticalBox, widget=VerticalBox_0}
  LeaderboardTitle {
   widgetClassPath=&CommonTextBlock
   bIsVariable=true
   parent=VerticalBox_0
   widget=LeaderboardTitle
   properties {
    justification=Center
    text=Friends Leaderboard
    textTransformPolicy=ToUpper
    colorAndOpacity {colorUseRule=UseColor_Specified, specifiedColor=<rgba 0.4125 0.5089 0.9647 1>}
    font {
     bForceMonospaced=false
     bMaterialIsStencil=false
     fontMaterial=null
     fontObject=&BurbankSmall
     letterSpacing=20
     monospacedWidth=1
     size=15
     skewAmount=0
     typefaceFontName=Black
     outlineSettings {
      bApplyOutlineToDropShadows=false
      bMiteredCorners=false
      bSeparateFillAlpha=false
      outlineMaterial=null
      outlineSize=0
      outlineColor=<rgba 0 0 0 1>
     }
    }
    shadowOffset=<0 0>
    strikeBrush {
     bIsDynamicallyLoaded=false
     drawAs=NoDrawType
     imageType=NoImage
     mirroring=NoMirror
     resourceName=None
     resourceObject=null
     tiling=NoTile
     imageSize=<32 32>
     margin=<ltrb 0 0 0 0>
     outlineSettings {
      bUseBrushTransparency=false
      roundingType=HalfHeightRadius
      width=0
      color {colorUseRule=UseColor_Specified, specifiedColor=<rgba 0 0 0 0>}
      cornerRadii=<0 0 0 0>
     }
     tintColor {colorUseRule=UseColor_Specified, specifiedColor=<rgba 1 1 1 1>}
     uVRegion {bIsValid=false, max=<0 0>, min=<0 0>}
    }
   }
   slot {class=VerticalBoxSlot, properties {horizontalAlignment=HAlign_Left, verticalAlignment=VAlign_Center, padding=<ltrb 0 0 0 12>}}
  }
  ContentSwitcher {
   widgetClassPath=&CommonVisibilitySwitcher
   bIsVariable=true
   parent=VerticalBox_0
   widget=ContentSwitcher
   properties {bAutoActivateSlot=false, shownVisibility=HitTestInvisible}
   slot {class=VerticalBoxSlot}
  }
  Content_LoadingEntries {
   widgetClassPath=&HorizontalBox
   bIsVariable=true
   parent=ContentSwitcher
   widget=Content_LoadingEntries
   slot {class=CommonVisibilitySwitcherSlot}
  }
  Overlay_0 {
   widgetClassPath=&Overlay
   parent=Content_LoadingEntries
   widget=Overlay_0
   slot {class=HorizontalBoxSlot, properties {size {sizeRule=Fill, value=1}}}
  }
  VerticalBox_75 {
   widgetClassPath=&VerticalBox
   parent=Overlay_0
   widget=VerticalBox_75
   properties {renderOpacity=0.25}
   slot {class=OverlaySlot, properties {horizontalAlignment=HAlign_Fill, verticalAlignment=VAlign_Fill}}
  }
  Image_113 {
   widgetClassPath=&Image
   parent=VerticalBox_75
   widget=Image_113
   properties {brush {
     bIsDynamicallyLoaded=false
     drawAs=Image
     imageType=NoImage
     mirroring=NoMirror
     resourceName=None
     resourceObject=/PilgrimCore/Widgets/Materials/M_UI_Pilgrim_LeaderBoard_Backing
     tiling=NoTile
     imageSize=<32 64>
     margin=<ltrb 0 0 0 0>
     outlineSettings {
      bUseBrushTransparency=false
      roundingType=HalfHeightRadius
      width=0
      color {colorUseRule=UseColor_Specified, specifiedColor=<rgba 0 0 0 0>}
      cornerRadii=<0 0 0 0>
     }
     tintColor {colorUseRule=UseColor_Specified, specifiedColor=<rgba 1 1 1 1>}
     uVRegion {bIsValid=false, max=<0 0>, min=<0 0>}
    }}
   slot {class=VerticalBoxSlot}
  }
  Image {
   widgetClassPath=&Image
   parent=VerticalBox_75
   widget=Image
   properties {brush {
     bIsDynamicallyLoaded=false
     drawAs=Image
     imageType=NoImage
     mirroring=NoMirror
     resourceName=None
     resourceObject=&M_UI_Pilgrim_LeaderBoard_Backing
     tiling=NoTile
     imageSize=<32 64>
     margin=<ltrb 0 0 0 0>
     outlineSettings {
      bUseBrushTransparency=false
      roundingType=HalfHeightRadius
      width=0
      color {colorUseRule=UseColor_Specified, specifiedColor=<rgba 0 0 0 0>}
      cornerRadii=<0 0 0 0>
     }
     tintColor {colorUseRule=UseColor_Specified, specifiedColor=<rgba 1 1 1 1>}
     uVRegion {bIsValid=false, max=<0 0>, min=<0 0>}
    }}
   slot {class=VerticalBoxSlot, properties {padding=<ltrb 0 4 0 0>}}
  }
  Image_1 {
   widgetClassPath=&Image
   parent=VerticalBox_75
   widget=Image_1
   properties {brush {
     bIsDynamicallyLoaded=false
     drawAs=Image
     imageType=NoImage
     mirroring=NoMirror
     resourceName=None
     resourceObject=&M_UI_Pilgrim_LeaderBoard_Backing
     tiling=NoTile
     imageSize=<32 64>
     margin=<ltrb 0 0 0 0>
     outlineSettings {
      bUseBrushTransparency=false
      roundingType=HalfHeightRadius
      width=0
      color {colorUseRule=UseColor_Specified, specifiedColor=<rgba 0 0 0 0>}
      cornerRadii=<0 0 0 0>
     }
     tintColor {colorUseRule=UseColor_Specified, specifiedColor=<rgba 1 1 1 1>}
     uVRegion {bIsValid=false, max=<0 0>, min=<0 0>}
    }}
   slot {class=VerticalBoxSlot, properties {padding=<ltrb 0 4 0 0>}}
  }
  Border_0 {
   widgetClassPath=&Border
   parent=Overlay_0
   widget=Border_0
   properties {
    background {
     bIsDynamicallyLoaded=false
     drawAs=Image
     imageType=NoImage
     mirroring=NoMirror
     resourceName=None
     resourceObject=/PilgrimCore/Widgets/Materials/M_UI_Overdrive_ProgressMeter_Backing_Empty
     tiling=NoTile
     imageSize=<32 32>
     margin=<ltrb 0.5 0.5 0.5 0.5>
     outlineSettings {
      bUseBrushTransparency=false
      roundingType=HalfHeightRadius
      width=0
      color {colorUseRule=UseColor_Specified, specifiedColor=<rgba 0 0 0 0>}
      cornerRadii=<0 0 0 0>
     }
     tintColor {colorUseRule=UseColor_Specified, specifiedColor=<rgba 0 0 0 0.5>}
     uVRegion {bIsValid=false, max=<0 0>, min=<0 0>}
    }
    padding=<ltrb 48 24 48 24>
   }
   slot {class=OverlaySlot, properties {horizontalAlignment=HAlign_Center, verticalAlignment=VAlign_Center}}
  }
  HorizontalBox_0 {
   widgetClassPath=&HorizontalBox
   parent=Border_0
   widget=HorizontalBox_0
   slot {class=BorderSlot, properties {padding=<ltrb 48 24 48 24>}}
  }
  CommonTextBlock_0 {
   widgetClassPath=&CommonTextBlock
   parent=HorizontalBox_0
   widget=CommonTextBlock_0
   properties {
    justification=Center
    text=Loading
    textTransformPolicy=ToUpper
    font {
     bForceMonospaced=false
     bMaterialIsStencil=false
     fontMaterial=null
     fontObject=&BurbankSmall
     letterSpacing=0
     monospacedWidth=1
     size=18
     skewAmount=0
     typefaceFontName=Bold
     outlineSettings {
      bApplyOutlineToDropShadows=false
      bMiteredCorners=false
      bSeparateFillAlpha=false
      outlineMaterial=null
      outlineSize=0
      outlineColor=<rgba 0 0 0 1>
     }
    }
    shadowOffset=<0 0>
   }
   slot {class=HorizontalBoxSlot, properties {horizontalAlignment=HAlign_Center, verticalAlignment=VAlign_Center}}
  }
  Spacer_126 {widgetClassPath=&Spacer, parent=HorizontalBox_0, widget=Spacer_126, properties {size=<24 1>}, slot {class=HorizontalBoxSlot}}
  Image_0 {
   widgetClassPath=&Image
   parent=HorizontalBox_0
   widget=Image_0
   properties {
    visibility=SelfHitTestInvisible
    brush {
     bIsDynamicallyLoaded=false
     drawAs=Image
     imageType=NoImage
     mirroring=NoMirror
     resourceName=None
     resourceObject=/PilgrimQuickplay/Widgets/Materials/MI_UI_Throbber_Sparks
     tiling=NoTile
     imageSize=<64 64>
     margin=<ltrb 0 0 0 0>
     outlineSettings {
      bUseBrushTransparency=false
      roundingType=HalfHeightRadius
      width=0
      color {colorUseRule=UseColor_Specified, specifiedColor=<rgba 0 0 0 0>}
      cornerRadii=<0 0 0 0>
     }
     tintColor {colorUseRule=UseColor_Specified, specifiedColor=<rgba 1 1 1 1>}
     uVRegion {bIsValid=false, max=<0 0>, min=<0 0>}
    }
   }
   slot {class=HorizontalBoxSlot, properties {horizontalAlignment=HAlign_Center, verticalAlignment=VAlign_Center}}
  }
  LeaderboardContainer {
   widgetClassPath=&VerticalBox
   bIsVariable=true
   parent=ContentSwitcher
   widget=LeaderboardContainer
   slot {class=CommonVisibilitySwitcherSlot, properties {verticalAlignment=VAlign_Top}}
  }
  LeaderboardEntries {
   widgetClassPath=&CommonListView
   bIsVariable=true
   parent=LeaderboardContainer
   widget=LeaderboardEntries
   properties {
    bIsFocusable=false
    entryWidgetClass=&/WBP_PilgrimQuickplay_LeaderboardEntry_Mini.WBP_PilgrimQuickplay_LeaderboardEntry_Mini_C
    numDesignerPreviewEntries=3
    verticalEntrySpacing=4
   }
   slot {class=VerticalBoxSlot}
  }
  ScoreBeatenTextBorder {
   widgetClassPath=&CommonBorder
   parent=LeaderboardContainer
   widget=ScoreBeatenTextBorder
   properties {
    verticalAlignment=VAlign_Center
    visibility=Collapsed
    background {
     bIsDynamicallyLoaded=false
     drawAs=Box
     imageType=NoImage
     mirroring=NoMirror
     resourceName=None
     resourceObject=null
     tiling=NoTile
     imageSize=<32 32>
     margin=<ltrb 0 0 0 0>
     outlineSettings {
      bUseBrushTransparency=false
      roundingType=HalfHeightRadius
      width=0
      color {colorUseRule=UseColor_Specified, specifiedColor=<rgba 0 0 0 0>}
      cornerRadii=<0 0 0 0>
     }
     tintColor {colorUseRule=UseColor_Specified, specifiedColor=<rgba 1 1 1 1>}
     uVRegion {bIsValid=false, max=<0 0>, min=<0 0>}
    }
    brushColor=<rgba 1 0.9866 0 1>
    desiredSizeScale=<1 0>
   }
   slot {class=VerticalBoxSlot, properties {padding=<ltrb 0 12 0 0>}}
  }
  ScoreBeatenText {
   widgetClassPath=&CommonTextBlock
   parent=ScoreBeatenTextBorder
   widget=ScoreBeatenText
   properties {
    justification=Center
    text="You beat {PLAYER} and {NUM} others"
    colorAndOpacity {colorUseRule=UseColor_Specified, specifiedColor=<rgba 1 1 1 0.6>}
    font {
     bForceMonospaced=false
     bMaterialIsStencil=false
     fontMaterial=null
     fontObject=&BurbankSmall
     letterSpacing=0
     monospacedWidth=1
     size=13
     skewAmount=0
     typefaceFontName=Black
     outlineSettings {
      bApplyOutlineToDropShadows=false
      bMiteredCorners=false
      bSeparateFillAlpha=false
      outlineMaterial=null
      outlineSize=0
      outlineColor=<rgba 0 0 0 1>
     }
    }
    shadowOffset=<0 0>
   }
   slot {class=BorderSlot, properties {verticalAlignment=VAlign_Center}}
  }
  Overlay_NoLeaderboardEntries {
   widgetClassPath=&Overlay
   bIsVariable=true
   parent=ContentSwitcher
   widget=Overlay_NoLeaderboardEntries
   properties {visibility=HitTestInvisible}
   slot {class=CommonVisibilitySwitcherSlot}
  }
  VerticalBox {
   widgetClassPath=&VerticalBox
   parent=Overlay_NoLeaderboardEntries
   widget=VerticalBox
   properties {renderOpacity=0.25}
   slot {class=OverlaySlot, properties {horizontalAlignment=HAlign_Fill, verticalAlignment=VAlign_Fill}}
  }
  Image_2 {
   widgetClassPath=&Image
   parent=VerticalBox
   widget=Image_2
   properties {brush {
     bIsDynamicallyLoaded=false
     drawAs=Image
     imageType=NoImage
     mirroring=NoMirror
     resourceName=None
     resourceObject=&M_UI_Pilgrim_LeaderBoard_Backing
     tiling=NoTile
     imageSize=<32 64>
     margin=<ltrb 0 0 0 0>
     outlineSettings {
      bUseBrushTransparency=false
      roundingType=HalfHeightRadius
      width=0
      color {colorUseRule=UseColor_Specified, specifiedColor=<rgba 0 0 0 0>}
      cornerRadii=<0 0 0 0>
     }
     tintColor {colorUseRule=UseColor_Specified, specifiedColor=<rgba 1 1 1 1>}
     uVRegion {bIsValid=false, max=<0 0>, min=<0 0>}
    }}
   slot {class=VerticalBoxSlot}
  }
  Image_3 {
   widgetClassPath=&Image
   parent=VerticalBox
   widget=Image_3
   properties {brush {
     bIsDynamicallyLoaded=false
     drawAs=Image
     imageType=NoImage
     mirroring=NoMirror
     resourceName=None
     resourceObject=&M_UI_Pilgrim_LeaderBoard_Backing
     tiling=NoTile
     imageSize=<32 64>
     margin=<ltrb 0 0 0 0>
     outlineSettings {
      bUseBrushTransparency=false
      roundingType=HalfHeightRadius
      width=0
      color {colorUseRule=UseColor_Specified, specifiedColor=<rgba 0 0 0 0>}
      cornerRadii=<0 0 0 0>
     }
     tintColor {colorUseRule=UseColor_Specified, specifiedColor=<rgba 1 1 1 1>}
     uVRegion {bIsValid=false, max=<0 0>, min=<0 0>}
    }}
   slot {class=VerticalBoxSlot, properties {padding=<ltrb 0 4 0 0>}}
  }
  Image_4 {
   widgetClassPath=&Image
   parent=VerticalBox
   widget=Image_4
   properties {brush {
     bIsDynamicallyLoaded=false
     drawAs=Image
     imageType=NoImage
     mirroring=NoMirror
     resourceName=None
     resourceObject=&M_UI_Pilgrim_LeaderBoard_Backing
     tiling=NoTile
     imageSize=<32 64>
     margin=<ltrb 0 0 0 0>
     outlineSettings {
      bUseBrushTransparency=false
      roundingType=HalfHeightRadius
      width=0
      color {colorUseRule=UseColor_Specified, specifiedColor=<rgba 0 0 0 0>}
      cornerRadii=<0 0 0 0>
     }
     tintColor {colorUseRule=UseColor_Specified, specifiedColor=<rgba 1 1 1 1>}
     uVRegion {bIsValid=false, max=<0 0>, min=<0 0>}
    }}
   slot {class=VerticalBoxSlot, properties {padding=<ltrb 0 4 0 0>}}
  }
  Border_1 {
   widgetClassPath=&Border
   parent=Overlay_NoLeaderboardEntries
   widget=Border_1
   properties {
    verticalAlignment=VAlign_Center
    background {
     bIsDynamicallyLoaded=false
     drawAs=Image
     imageType=NoImage
     mirroring=NoMirror
     resourceName=None
     resourceObject=/PilgrimCore/Widgets/Materials/M_UI_Pilgrim_Difficulty_Backing
     tiling=NoTile
     imageSize=<32 32>
     margin=<ltrb 0.5 0.5 0.5 0.5>
     outlineSettings {
      bUseBrushTransparency=false
      roundingType=HalfHeightRadius
      width=0
      color {colorUseRule=UseColor_Specified, specifiedColor=<rgba 0 0 0 0>}
      cornerRadii=<0 0 0 0>
     }
     tintColor {colorUseRule=UseColor_Specified, specifiedColor=<rgba 0.0039 0.0118 0.098 0>}
     uVRegion {bIsValid=false, max=<0 0>, min=<0 0>}
    }
    padding=<ltrb 64 32 64 32>
   }
   slot {class=OverlaySlot, properties {horizontalAlignment=HAlign_Center, verticalAlignment=VAlign_Center}}
  }
  Text_NoLeaderboardEntries {
   widgetClassPath=&CommonTextBlock
   parent=Border_1
   widget=Text_NoLeaderboardEntries
   properties {
    autoWrapText=true
    justification=Center
    text="No Leaderboard Available \r\nTry Again Later"
    font {
     bForceMonospaced=false
     bMaterialIsStencil=false
     fontMaterial=null
     fontObject=&Burbank
     letterSpacing=0
     monospacedWidth=1
     size=24
     skewAmount=0
     typefaceFontName=Burbank-PlayerPower
     outlineSettings {
      bApplyOutlineToDropShadows=false
      bMiteredCorners=false
      bSeparateFillAlpha=false
      outlineMaterial=null
      outlineSize=0
      outlineColor=<rgba 0 0 0 1>
     }
    }
    shadowOffset=<0 0>
   }
   slot {class=BorderSlot, properties {verticalAlignment=VAlign_Center, padding=<ltrb 64 32 64 32>}}
  }
  Overlay_NoLeaderboardEntries_MissingParts {
   widgetClassPath=&Overlay
   bIsVariable=true
   parent=ContentSwitcher
   widget=Overlay_NoLeaderboardEntries_MissingParts
   properties {visibility=HitTestInvisible}
   slot {class=CommonVisibilitySwitcherSlot}
  }
  VerticalBox_1 {
   widgetClassPath=&VerticalBox
   parent=Overlay_NoLeaderboardEntries_MissingParts
   widget=VerticalBox_1
   properties {renderOpacity=0.25}
   slot {class=OverlaySlot, properties {horizontalAlignment=HAlign_Fill, verticalAlignment=VAlign_Fill}}
  }
  Image_5 {
   widgetClassPath=&Image
   parent=VerticalBox_1
   widget=Image_5
   properties {brush {
     bIsDynamicallyLoaded=false
     drawAs=Image
     imageType=NoImage
     mirroring=NoMirror
     resourceName=None
     resourceObject=&M_UI_Pilgrim_LeaderBoard_Backing
     tiling=NoTile
     imageSize=<32 64>
     margin=<ltrb 0 0 0 0>
     outlineSettings {
      bUseBrushTransparency=false
      roundingType=HalfHeightRadius
      width=0
      color {colorUseRule=UseColor_Specified, specifiedColor=<rgba 0 0 0 0>}
      cornerRadii=<0 0 0 0>
     }
     tintColor {colorUseRule=UseColor_Specified, specifiedColor=<rgba 1 1 1 1>}
     uVRegion {bIsValid=false, max=<0 0>, min=<0 0>}
    }}
   slot {class=VerticalBoxSlot}
  }
  Image_6 {
   widgetClassPath=&Image
   parent=VerticalBox_1
   widget=Image_6
   properties {brush {
     bIsDynamicallyLoaded=false
     drawAs=Image
     imageType=NoImage
     mirroring=NoMirror
     resourceName=None
     resourceObject=&M_UI_Pilgrim_LeaderBoard_Backing
     tiling=NoTile
     imageSize=<32 64>
     margin=<ltrb 0 0 0 0>
     outlineSettings {
      bUseBrushTransparency=false
      roundingType=HalfHeightRadius
      width=0
      color {colorUseRule=UseColor_Specified, specifiedColor=<rgba 0 0 0 0>}
      cornerRadii=<0 0 0 0>
     }
     tintColor {colorUseRule=UseColor_Specified, specifiedColor=<rgba 1 1 1 1>}
     uVRegion {bIsValid=false, max=<0 0>, min=<0 0>}
    }}
   slot {class=VerticalBoxSlot, properties {padding=<ltrb 0 4 0 0>}}
  }
  Image_7 {
   widgetClassPath=&Image
   parent=VerticalBox_1
   widget=Image_7
   properties {brush {
     bIsDynamicallyLoaded=false
     drawAs=Image
     imageType=NoImage
     mirroring=NoMirror
     resourceName=None
     resourceObject=&M_UI_Pilgrim_LeaderBoard_Backing
     tiling=NoTile
     imageSize=<32 64>
     margin=<ltrb 0 0 0 0>
     outlineSettings {
      bUseBrushTransparency=false
      roundingType=HalfHeightRadius
      width=0
      color {colorUseRule=UseColor_Specified, specifiedColor=<rgba 0 0 0 0>}
      cornerRadii=<0 0 0 0>
     }
     tintColor {colorUseRule=UseColor_Specified, specifiedColor=<rgba 1 1 1 1>}
     uVRegion {bIsValid=false, max=<0 0>, min=<0 0>}
    }}
   slot {class=VerticalBoxSlot, properties {padding=<ltrb 0 4 0 0>}}
  }
  Border {
   widgetClassPath=&Border
   parent=Overlay_NoLeaderboardEntries_MissingParts
   widget=Border
   properties {
    verticalAlignment=VAlign_Center
    background {
     bIsDynamicallyLoaded=false
     drawAs=Image
     imageType=NoImage
     mirroring=NoMirror
     resourceName=None
     resourceObject=&M_UI_Pilgrim_Difficulty_Backing
     tiling=NoTile
     imageSize=<32 32>
     margin=<ltrb 0.5 0.5 0.5 0.5>
     outlineSettings {
      bUseBrushTransparency=false
      roundingType=HalfHeightRadius
      width=0
      color {colorUseRule=UseColor_Specified, specifiedColor=<rgba 0 0 0 0>}
      cornerRadii=<0 0 0 0>
     }
     tintColor {colorUseRule=UseColor_Specified, specifiedColor=<rgba 0.0039 0.0118 0.098 0>}
     uVRegion {bIsValid=false, max=<0 0>, min=<0 0>}
    }
    padding=<ltrb 64 32 64 32>
   }
   slot {class=OverlaySlot, properties {horizontalAlignment=HAlign_Center, verticalAlignment=VAlign_Center}}
  }
  Text_NoLeaderboardEntries_MissingParts {
   widgetClassPath=&CommonTextBlock
   parent=Border
   widget=Text_NoLeaderboardEntries_MissingParts
   properties {
    autoWrapText=true
    justification=Center
    text=Part has no Leaderboard
    font {
     bForceMonospaced=false
     bMaterialIsStencil=false
     fontMaterial=null
     fontObject=&Burbank
     letterSpacing=0
     monospacedWidth=1
     size=24
     skewAmount=0
     typefaceFontName=Burbank-PlayerPower
     outlineSettings {
      bApplyOutlineToDropShadows=false
      bMiteredCorners=false
      bSeparateFillAlpha=false
      outlineMaterial=null
      outlineSize=0
      outlineColor=<rgba 0 0 0 1>
     }
    }
    shadowOffset=<0 0>
   }
   slot {class=BorderSlot, properties {verticalAlignment=VAlign_Center, padding=<ltrb 64 32 64 32>}}
  }
 ]
}
