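# Pilgrim Quickplay results widget showing one player's loadout: avatar, instrument icon, difficulty and name.
# PreConstruct applies a borderless local-player avatar material and sets each element's visibility from Hide flags, with HideAvatarIcon also collapsing the spacer.
# Setter functions take a player state, difficulty, instrument and name to fill the child widgets.
# The instrument icon is chosen by writing a looked-up index to the dynamic material's Icon parameter.
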WidgetBlueprint /PilgrimQuickplay/Widgets/Results/WBP_Pilgrim_Quickplay_PlayerLoadout
graphs [
 EventGraph {nodes [
   Event PreConstruct()
   > SetBrushFromMaterial(
    self=Get PlayerImage!(self:WBP Pilgrim Avatar Icon Object Reference=Get AvatarIcon() -> (AvatarIcon:WBP Pilgrim Avatar Icon Object Reference)) -> (PlayerImage:Image Object Reference)
    Material=/SparksUICommon/Materials/MI_UI_BandMember_Avatar_Master_LocalPlayerWithouBorder
   )
   > SetVisibility(
    self=[Get AvatarIcon(), Get Spacer() -> (Spacer:Spacer Object Reference)]
    InVisibility=Select(Option 0=Visible, Option 1=Collapsed, Index=Get HideAvatarIcon() -> (HideAvatarIcon:bool))
   )
   > SetVisibility(
    self=Get Image_InstrumentIcon() -> (Image_InstrumentIcon:Image Object Reference)
    InVisibility=Select(Option 0=Visible, Option 1=Hidden, Index=Get HideInstrument() -> (HideInstrument:bool))
   )
   > SetVisibility(
    self=Get RichText_PlayerName() -> (RichText_PlayerName:Common Rich Text Block Object Reference)
    InVisibility=Select(Option 0=Visible, Option 1=Collapsed, Index=Get HideName() -> (HideName:bool))
   )
   > SetVisibility(
    self=Get WBP_Pilgrim_Results_Difficulty() -> (WBP_Pilgrim_Results_Difficulty:WBP Pilgrim Results Difficulty Object Reference)
    InVisibility=Select(Option 0=Visible, Option 1=Hidden, Index=Get HideDifficulty() -> (HideDifficulty:bool))
   )
  ]}
 SetAvatarIconImg {nodes [
   n1 = SetAvatarIconImg!() -> (then, PlayerState:Fort Player State Object Reference) {is=title K2Node_FunctionEntry}
   > n2 = Cast To FortPlayerStateAthena!(execute, Object=@n1) -> (then, CastFailed, AsFort Player State Athena:Fort Player State Athena Object Reference) {is=title K2Node_DynamicCast}
   > SetPlayerImage!(execute, self:WBP Pilgrim Avatar Icon Object Reference=Get AvatarIcon(), FortPlayerStateAthena:Fort Player State Athena Object Reference=@n2) -> (then)
  ]}
 SetDifficultyText {nodes [
   n1 = SetDifficultyText!() -> (then, Difficulty:EPilgrimSongDifficulty Enum=DifficultyEasy) {is=title K2Node_FunctionEntry}
   > SetDifficulty!(
    execute
    self:WBP Pilgrim Results Difficulty Object Reference=Get WBP_Pilgrim_Results_Difficulty()
    Difficulty:EPilgrimSongDifficulty Enum=@n1
   ) -> (then)
  ]}
 SetInstrumentIconImg {nodes [
   n1 = SetInstrumentIconImg!() -> (then, Instrument:EPilgrimTrackType Enum=TrackGuitar) {is=title K2Node_FunctionEntry}
   > n2 = GetDynamicMaterial(self=Get Image_InstrumentIcon())
   n3 = PilgrimTrackTypeToSparksInstrument!(PilgrimType:EPilgrimTrackType Enum=@n1) -> (ReturnValue:ESparksInstrumentType Enum=Guitar)
   > SetScalarParameterValue(
    self=@n2
    ParameterName=Icon
    Value="To Float (Integer)"(InInt=GetInstrumentIconIndex!(
      InstrumentType:EPilgrimTrackType Enum=@n1
      SongMetaData:Pilgrim Song Metadata Object Reference
      StartingInstrumentIndex:ESparksInstrumentType Enum=@n3
      bUseKeyTarIcon:bool=false
     ) -> (InstrumentIdexIcon:int=0))
   )
  ]}
 SetPlayerName {nodes [
   n1 = SetPlayerName!() -> (then, PlayerName:Text) {is=title K2Node_FunctionEntry}
   > SetText(self=Get RichText_PlayerName(), InText=@n1)
  ]}
]
variables [
 HideAvatarIcon {type=boolean}
 HideDifficulty {type=boolean}
 HideInstrument {type=boolean}
 HideName {type=boolean}
]
widgets {
 info {inheritedWidgetCount=0, parentClass=&UserWidget, rootWidgetClass=&HorizontalBox, widgetCount=9}
 widgets [
  HorizontalBox_3 {widgetClassPath=&HorizontalBox, widget=HorizontalBox_3}
  Overlay_0 {
   widgetClassPath=&Overlay
   parent=HorizontalBox_3
   widget=Overlay_0
   slot {class=HorizontalBoxSlot, properties {horizontalAlignment=HAlign_Left}}
  }
  AvatarIcon {
   widgetClassPath=/PilgrimCore/Widgets/Player/WBP_Pilgrim_AvatarIcon
   bIsVariable=true
   parent=Overlay_0
   widget=AvatarIcon
   properties {playerIconSizeOverride=<171 171>, renderTransform {angle=0, scale=<1 1>, shear=<0 0>, translation=<6 0>}}
   slot {class=OverlaySlot}
  }
  VerticalBox_1 {
   widgetClassPath=&VerticalBox
   parent=Overlay_0
   widget=VerticalBox_1
   slot {class=OverlaySlot, properties {verticalAlignment=VAlign_Center, padding=<ltrb 5 0 0 0>}}
  }
  Image_InstrumentIcon {
   widgetClassPath=&Image
   bIsVariable=true
   parent=VerticalBox_1
   widget=Image_InstrumentIcon
   properties {
    brush {
     bIsDynamicallyLoaded=false
     drawAs=Image
     imageType=FullColor
     mirroring=NoMirror
     resourceName=None
     resourceObject=/PilgrimQuickplay/Widgets/Materials/M_UI_Icon_Instrument_Type_SDF_Inst_BG
     tiling=NoTile
     imageSize=<100 100>
     margin=<ltrb 0 0 0 0>
     outlineSettings {
      bUseBrushTransparency=false
      roundingType=HalfHeightRadius
      width=0
      color {colorUseRule=UseColor_Specified, specifiedColor=<rgba 0 0 0 0>}
      cornerRadii=<0 0 0 0>
     }
     tintColor {colorUseRule=UseColor_Specified, specifiedColor=<rgba 1 1 1 1>}
     uVRegion {bIsValid=false, max=<0 0>, min=<0 0>}
    }
    renderTransform {angle=0, scale=<1 1>, shear=<0 0>, translation=<-18 -14>}
   }
   slot {class=VerticalBoxSlot, properties {horizontalAlignment=HAlign_Left, verticalAlignment=VAlign_Top}}
  }
  Spacer {
   widgetClassPath=&Spacer
   bIsVariable=true
   parent=VerticalBox_1
   widget=Spacer
   properties {size=<1 46>}
   slot {class=VerticalBoxSlot}
  }
  WBP_Pilgrim_Results_Difficulty {
   widgetClassPath=&/WBP_Pilgrim_Results_Difficulty
   bIsVariable=true
   parent=VerticalBox_1
   widget=WBP_Pilgrim_Results_Difficulty
   slot {class=VerticalBoxSlot, properties {horizontalAlignment=HAlign_Left, verticalAlignment=VAlign_Bottom}}
  }
  SizeBox_71 {
   widgetClassPath=&SizeBox
   parent=VerticalBox_1
   widget=SizeBox_71
   properties {bOverride_MaxDesiredWidth=true, maxDesiredWidth=171}
   slot {class=VerticalBoxSlot}
  }
  RichText_PlayerName {
   widgetClassPath=&CommonRichTextBlock
   bIsVariable=true
   parent=SizeBox_71
   widget=RichText_PlayerName
   properties {
    bOverrideDefaultStyle=true
    defaultTextStyleOverrideClass=/FortUILibrary/Styles/Typography/_TextStyle_Body1._TextStyle_Body1_C
    scrollStyle=/PilgrimQuickplay/Widgets/ScrollStyles/TextScrollStyle_PilgrimQuickplay_Setlist.TextScrollStyle_PilgrimQuickplay_Setlist_C
    text=Player Name
    defaultTextStyleOverride {
     overflowPolicy=Clip
     transformPolicy=None
     colorAndOpacity {colorUseRule=UseColor_Specified, specifiedColor=<rgba 1 1 1 1>}
     font {
      bForceMonospaced=false
      bMaterialIsStencil=false
      fontMaterial=null
      fontObject=&BurbankSmall
      letterSpacing=0
      monospacedWidth=1
      size=24
      skewAmount=0
      typefaceFontName=Black
      outlineSettings {
       bApplyOutlineToDropShadows=false
       bMiteredCorners=false
       bSeparateFillAlpha=false
       outlineMaterial=null
       outlineSize=0
       outlineColor=<rgba 0 0 0 1>
      }
     }
     highlightColor {colorUseRule=UseColor_Specified, specifiedColor=<rgba 0 0 0 1>}
     highlightShape {
      bIsDynamicallyLoaded=false
      drawAs=Image
      imageType=NoImage
      mirroring=NoMirror
      resourceName=None
      resourceObject=null
      tiling=NoTile
      imageSize=<32 32>
      margin=<ltrb 0 0 0 0>
      outlineSettings {
       bUseBrushTransparency=false
       roundingType=HalfHeightRadius
       width=0
       color {colorUseRule=UseColor_Specified, specifiedColor=<rgba 0 0 0 0>}
       cornerRadii=<0 0 0 0>
      }
      tintColor {colorUseRule=UseColor_Specified, specifiedColor=<rgba 1 1 1 1>}
      uVRegion {bIsValid=false, max=<0 0>, min=<0 0>}
     }
     selectedBackgroundColor {colorUseRule=UseColor_Foreground, specifiedColor=<rgba 1 0 1 1>}
     shadowColorAndOpacity=<rgba 0 0 0 1>
     shadowOffset=<0 0>
     strikeBrush {
      bIsDynamicallyLoaded=false
      drawAs=Image
      imageType=NoImage
      mirroring=NoMirror
      resourceName=None
      resourceObject=null
      tiling=NoTile
      imageSize=<32 32>
      margin=<ltrb 0 0 0 0>
      outlineSettings {
       bUseBrushTransparency=false
       roundingType=HalfHeightRadius
       width=0
       color {colorUseRule=UseColor_Specified, specifiedColor=<rgba 0 0 0 0>}
       cornerRadii=<0 0 0 0>
      }
      tintColor {colorUseRule=UseColor_Specified, specifiedColor=<rgba 1 1 1 1>}
      uVRegion {bIsValid=false, max=<0 0>, min=<0 0>}
     }
     underlineBrush {
      bIsDynamicallyLoaded=false
      drawAs=Image
      imageType=NoImage
      mirroring=NoMirror
      resourceName=None
      resourceObject=null
      tiling=NoTile
      imageSize=<32 32>
      margin=<ltrb 0 0 0 0>
      outlineSettings {
       bUseBrushTransparency=false
       roundingType=HalfHeightRadius
       width=0
       color {colorUseRule=UseColor_Specified, specifiedColor=<rgba 0 0 0 0>}
       cornerRadii=<0 0 0 0>
      }
      tintColor {colorUseRule=UseColor_Specified, specifiedColor=<rgba 1 1 1 1>}
      uVRegion {bIsValid=false, max=<0 0>, min=<0 0>}
     }
    }
   }
   slot {class=SizeBoxSlot, properties {horizontalAlignment=HAlign_Left, padding=<ltrb 0 8 0 0>}}
  }
 ]
}
