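# Setlist preview row button for the Pilgrim Quickplay song menu, showing a picked song or a waiting placeholder.
# Its list item object supplies and updates song and player data, driving the state switcher, text, lazy-loaded album art, avatar and missing-parts view model.
# Hover or selection plays a glow and starts the playspace's song preview; leaving fades it out.
# Pressing only plays a sound; a separate function removes songs through the setlist editor component.
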WidgetBlueprint /PilgrimQuickplay/Widgets/SetlistAndSongDisplay/WBP_PilgrimQuickplay_SongMenu_SetlistPreviewEntry
animations [
 OnHovered {
  binding_count=2
  length_seconds=0.15
  object=&self:OnHovered
  sequence {bindings [
    Image_SelectionGlow {
     id=#1
     possessed_class=Image
     tracks [{
       type=MovieSceneFloatTrack
       display_name=Render Opacity
       sections [{type=MovieSceneFloatSection, channels [None {type=MovieSceneScriptingFloatChannel, keys [[0,0],[3,1]]}]}]
      }]
    }
    AddedSong {
     id=#2
     possessed_class=Border
     tracks [{
       type=MovieSceneWidgetMaterialTrack
       sections [{type=MovieSceneParameterSection, channels [Anim_GradientRamp_FadeColor {type=MovieSceneScriptingFloatChannel, keys [[0,0],[3,1]]}]}]
      }]
    }
   ]}
 }
 OnSongRemoved {
  binding_count=1
  length_seconds=0.05
  object=&self:OnSongRemoved
  sequence {bindings [AddedSong {
     id=#3
     possessed_class=Border
     tracks [{
       type=MovieSceneWidgetMaterialTrack
       sections [{
         type=MovieSceneParameterSection
         end_frame=1
         start_frame=0
         channels [Anim_SwitchAddedSong {type=MovieSceneScriptingFloatChannel, const [1,2]}]
        }]
      }]
    }]}
 }
 OnSongAdded {
  binding_count=1
  length_seconds=0.05
  object=&self:OnSongAdded
  sequence {bindings [AddedSong {
     id=#3
     possessed_class=Border
     tracks [{
       type=MovieSceneWidgetMaterialTrack
       sections [{
         type=MovieSceneParameterSection
         end_frame=1
         start_frame=0
         channels [Anim_SwitchAddedSong {type=MovieSceneScriptingFloatChannel, const [1,2]}]
        }]
      }]
    }]}
 }
]
graphs [
 EventGraph {nodes [
   n1 = OnPlayerAndSongUpdate!() -> (OutputDelegate, then, SongMetaData:Pilgrim Song Metadata Object Reference, SongAddedByPlayer:Player State Object Reference) {is=title K2Node_CustomEvent}
   > UpdatePlayerAndSongData!(
    execute
    self
    SongAddedByPlayer:Player State Object Reference=@n1.SongAddedByPlayer
    SongMetadata:Pilgrim Song Metadata Object Reference=@n1.SongMetaData
   ) -> (then)
   > ReceiveSongMetadata!(
    execute
    self:BP Pilgrim Quickplay Missing Parts VM Object Reference=Get WBP_PilgrimQuickplay_MissingPartsVM() -> (WBP_PilgrimQuickplay_MissingPartsVM:BP Pilgrim Quickplay Missing Parts VM Object Reference)
    Song Metadata:Pilgrim Song Metadata Object Reference=@n1.SongMetaData
   ) -> (then)
   n2 = UpdatePlayerAndSongData!() -> (OutputDelegate, then, SongAddedByPlayer:Player State Object Reference, SongMetadata:Pilgrim Song Metadata Object Reference) {is=title K2Node_CustomEvent}
   > n3 = Sequence()
   > n4 = Is Valid(InputObject=@n2.SongAddedByPlayer)
   SetVisibility(execute=@n4.Is Not Valid, self=Get AvatarIcon() -> (AvatarIcon:WBP Pilgrim Avatar Icon Object Reference), InVisibility=Hidden)
   SetVisibility(execute=@n4.Is Valid, self=Get AvatarIcon(), InVisibility=SelfHitTestInvisible)
   > n5 = Cast To FortPlayerStateAthena!(execute, Object=@n2.SongAddedByPlayer) -> (then, CastFailed, AsFort Player State Athena:Fort Player State Athena Object Reference) {is=title K2Node_DynamicCast}
   > SetPlayerImage!(execute, self:WBP Pilgrim Avatar Icon Object Reference=Get AvatarIcon(), FortPlayerStateAthena:Fort Player State Athena Object Reference=@n5) -> (then)
   n6 = Is Valid(exec=@n3.then_1, InputObject=@n2.SongMetadata)
   OnSongRemovedAnimation(execute=@n6.Is Not Valid)
   > Set Song Metadata(Song Metadata:Pilgrim Song Metadata Object Reference) -> (Output_Get:Pilgrim Song Metadata Object Reference)
   OnSongAddedAnimation(execute=@n6.Is Valid)
   > "SetText (Text)"(
    self=Get SongTitle() -> (SongTitle:Common Text Object Reference)
    InText=GetTitleText!(self:Sparks Song Data Object Reference=@n2.SongMetadata) -> (ReturnValue:Text)
   )
   > "SetText (Text)"(
    self=Get ArtistName() -> (ArtistName:Common Text Object Reference)
    InText=GetArtistText!(self:Sparks Song Data Object Reference=@n2.SongMetadata) -> (ReturnValue:Text)
   )
   > "SetText (Text)"(
    self=Get SongRelease() -> (SongRelease:Common Text Object Reference)
    InText="To Text (Integer)"(Value=Get Year!(self:Pilgrim Song Metadata Object Reference=@n2.SongMetadata) -> (Year:int=0), bUseGrouping=false)
   )
   > "SetText (Text)"(
    self=Get SongDuration() -> (SongDuration:Common Text Object Reference)
    InText=As Timespan!(InTimespan:"Timespan Structure (by ref)"=Get Duration!(self:Pilgrim Song Metadata Object Reference=@n2.SongMetadata) -> (Duration:Timespan Structure)) -> (ReturnValue:Text)
   )
   > SetVisibility(self=Get ReleaseDurationDivider() -> (ReleaseDurationDivider:Common Text Object Reference), InVisibility=SelfHitTestInvisible)
   > LazyLoadAlbumArt!(
    execute
    Song Metadata:Pilgrim Song Metadata Object Reference=@n2.SongMetadata
    Lazy Image:Fort Lazy Image Object Reference=Get ImageAlbumArt() -> (ImageAlbumArt:Fort Lazy Image Object Reference)
   ) -> (then)
   > Set Song Metadata(Song Metadata=@n2.SongMetadata)
   OnSongRemovedAnimation() {is=title K2Node_CustomEvent}
   > "Play Animation with Finished event (legacy)"!(
    execute:Exec
    Widget:User Widget Object Reference=Self-Reference()
    InAnimation:Widget Animation Object Reference=Get OnSongRemoved() -> (OnSongRemoved:Widget Animation Object Reference)
    StartAtTime:float=0
    NumLoopsToPlay:int=1
    PlayMode:EUMGSequencePlayMode Enum
    PlaybackSpeed:float=1
   ) -> (then:Exec, Result:UMGSequence Player Object Reference, Finished:Exec) {is=title K2Node_PlayAnimation}
   > SetActiveWidgetIndex(
    self:Common Visibility Switcher Object Reference=Get SongEntryStatusSwitcher() -> (SongEntryStatusSwitcher:Common Visibility Switcher Object Reference)
    Index=2
   )
   > "SetText (Text)"(self=Get SongTitle())
   > "SetText (Text)"(self=Get ArtistName())
   > "SetText (Text)"(self=Get SongRelease())
   > "SetText (Text)"(self=Get SongDuration())
   > SetVisibility(self=Get ReleaseDurationDivider(), InVisibility=Collapsed)
   OnSongAddedAnimation() {is=title K2Node_CustomEvent}
   > SetActiveWidgetIndex(self=Get SongEntryStatusSwitcher(), Index=1)
   > PlayAnimationForward(InAnimation=Get OnSongAdded() -> (OnSongAdded:Widget Animation Object Reference))
   Event PreConstruct()
   > SetBrushFromMaterial(
    self=Get PlayerImage!(self:WBP Pilgrim Avatar Icon Object Reference=Get AvatarIcon()) -> (PlayerImage:Image Object Reference)
    Material=/SparksUICommon/Materials/MI_UI_BandMember_Avatar_Master_LocalPlayerWithouBorder
   )
   Event Construct()
   > SetHoverStateDefaultValues()
   n7 = Event On Unhovered()
   PlayAnimationReverse(execute=[@n7.then, @n9.then], InAnimation=Get OnHovered() -> (OnHovered:Widget Animation Object Reference))
   n8 = GetOwningPlayer()
   > StopCurrentPreview!(
    execute
    self:Sparks Song Previewer Component Object Reference=Get SongPreviewer!(self:BP Pilgrim Quickplay Playspace Object Reference=GetSparksMusicPlayspaceOfClass!(
      PlayspaceType:Sparks Music Playspace Class Reference=/PilgrimQuickplay/Playspaces/BP_PilgrimQuickplayPlayspace.BP_PilgrimQuickplayPlayspace_C
      Actor:Actor Object Reference=@n8
      bMatchChildTypes:bool=false
     ) -> (ReturnValue:BP Pilgrim Quickplay Playspace Object Reference)) -> (SongPreviewer:BPC Sparks Song Previewer Object Reference)
    bWithFadeout:bool=true
   ) -> (then)
   n9 = Event On Deselected()
   n10 = Event OnListItemObjectSet()
   > n11 = Cast To BP_SongEntryHorizontal_ItemObj!(execute, Object=@n10.ListItemObject) -> (then, CastFailed, AsBP Song Entry Horizontal Item Obj:BP Song Entry Horizontal Item Obj Object Reference) {is=title K2Node_DynamicCast}
   > n12 = Set Song Metadata(Song Metadata=Get SongMetadata!(self:BP Song Entry Horizontal Item Obj Object Reference=@n11) -> (SongMetadata:Pilgrim Song Metadata Object Reference))
   > Bind Event to On Update Song and Player(self:BP Song Entry Horizontal Item Obj Object Reference=@n11, Delegate=@n1) {is=title K2Node_AddDelegate}
   > n13 = Sequence()
   n14 = Get ListNumber!(self:BP Song Entry Horizontal Item Obj Object Reference=@n11) -> (ListNumber:int=0)
   > "SetText (Text)"(
    self=Get SongNumber() -> (SongNumber:Common Text Object Reference)
    InText=Format Text!(Format="LOCTEXT(\"{number}.\")", number:Text="To Text (Integer)"(Value=@n14)) -> (Result)
   )
   UpdatePlayerAndSongData(
    execute=@n13.then_1
    SongAddedByPlayer=Get SongAddedByPlayerState!(self:BP Song Entry Horizontal Item Obj Object Reference=@n11) -> (SongAddedByPlayerState:Player State Object Reference)
    SongMetadata=@n12
   ) {is=title K2Node_CallFunction}
   > ReceiveSongMetadata(self=Get WBP_PilgrimQuickplay_MissingPartsVM(), Song Metadata=@n12)
   Event On Pressed() {is=title K2Node_Event}
   > Is Valid(InputObject=Get Song Metadata() -> (Song Metadata:Pilgrim Song Metadata Object Reference))
   > PlaySound2D(Sound=Get Pressed Remove Song Sound() -> (Pressed Remove Song Sound:Sound Base Object Reference))
   n15 = Event On Selected()
   PlayAnimationForward(execute=[@n15.then, @n19.then], InAnimation=Get OnHovered())
   n16 = Get Song Metadata()
   > Is Valid(InputObject=@n16)
   n17 = Get SongPreviewer(self=GetSparksMusicPlayspaceOfClass(
     PlayspaceType=&BP_PilgrimQuickplayPlayspace_C
     Actor=GetOwningPlayer()
     bMatchChildTypes=false
    ))
   n18 = GetShortName!(self:Sparks Song Data Object Reference=@n16) -> (ReturnValue:Name=None)
   > Branch(Condition=NOT Boolean(A=IsPlayingPreviewForSong!(self:Sparks Song Previewer Component Object Reference=@n17, Song:Name=@n18) -> (ReturnValue:bool)))
   > StartPreviewByShortName!(execute, self:Sparks Song Previewer Component Object Reference=@n17, ShortName:Name=@n18) -> (then)
   Event On Focused() {is=title K2Node_Event}
   > Call On Setlist Entry Focus Received!(execute:Exec, self:WBP Pilgrim Quickplay Song Menu Setlist Preview Entry Object Reference) -> (then:Exec) {is=title K2Node_CallDelegate}
   n19 = Event On Hovered()
  ]}
 OnSetlistEntryFocusReceived {nodes [OnSetlistEntryFocusReceived() {is=title K2Node_FunctionEntry}]}
 RemoveSongFromSetlist {nodes [
   n1 = RemoveSongFromSetlist!() -> (then, Song Metadata:Pilgrim Song Metadata Object Reference) {is=title K2Node_FunctionEntry}
   > n2 = Set OwningPlayerState(OwningPlayerState:Player State Object Reference=Get PlayerState!(self:Controller Object Reference=GetOwningPlayer()) -> (PlayerState:Player State Object Reference)) -> (Output_Get:Player State Object Reference)
   > Remove Song!(
    execute
    self:BP Pilgrim Custom Setlist Editor Component Object Reference=GetComponentByClass!(
     self:Actor Object Reference=GetPlayerController!(self:Player State Object Reference=@n2) -> (ReturnValue:Player Controller Object Reference)
     ComponentClass:Actor Component Class Reference=/PilgrimCore/Player/Components/BP_PilgrimCustomSetlistEditorComponent.BP_PilgrimCustomSetlistEditorComponent_C
    ) -> (ReturnValue:BP Pilgrim Custom Setlist Editor Component Object Reference)
    SongToRemove:Name=GetShortName(self=@n1) -> (ReturnValue=None)
   ) -> (then)
  ]}
 SetHoverStateDefaultValues {nodes [
   SetHoverStateDefaultValues() {is=title K2Node_FunctionEntry}
   > Set RenderOpacity!(execute, RenderOpacity:float=0, self:Widget Object Reference=Get Image_SelectionGlow() -> (Image_SelectionGlow:Image Object Reference)) -> (then, Output_Get:float)
  ]}
]
variables [
 OnSetlistEntryFocusReceived {}
 OwningPlayerState {type=PlayerState, default=null}
 PlayerInstrumentChoice {type=EPilgrimTrackType, default=TrackGuitar}
 Pressed Remove Song Sound {type=SoundBase, default=/Game/Sounds/Fort_UI_Sounds/FrontEnd/DefaultSounds/PressedButtons/MSS_UI_Button_Pressed_Default}
 Song Metadata {type=PilgrimSongMetadata, default=null}
 SongSelectMenu {type=WBP_PilgrimQuickplay_SongSelectMenu2_C, default=null}
]
widgets {
 info {inheritedWidgetCount=1, parentClass=&CommonButtonBase, rootWidgetClass=&Overlay, widgetCount=26}
 widgets [
  Overlay_30 {widgetClassPath=&Overlay, widget=Overlay_30}
  SongEntryStatusSwitcher {
   widgetClassPath=&CommonVisibilitySwitcher
   bIsVariable=true
   parent=Overlay_30
   widget=SongEntryStatusSwitcher
   properties {activeWidgetIndex=2}
   slot {class=OverlaySlot, properties {horizontalAlignment=HAlign_Fill, verticalAlignment=VAlign_Fill}}
  }
  HorizontalBoxAddaSong {
   widgetClassPath=&HorizontalBox
   parent=SongEntryStatusSwitcher
   widget=HorizontalBoxAddaSong
   slot {class=CommonVisibilitySwitcherSlot, properties {padding=<ltrb -2 -2 -2 -2>}}
  }
  ImageAddIcon {
   widgetClassPath=&Image
   bIsVariable=true
   parent=HorizontalBoxAddaSong
   widget=ImageAddIcon
   properties {brush {
     bIsDynamicallyLoaded=false
     drawAs=Image
     imageType=FullColor
     mirroring=NoMirror
     resourceName=None
     resourceObject=/SparksUICommon/Materials/M_UI_Add_Icon
     tiling=NoTile
     imageSize=<96 96>
     margin=<ltrb 0 0 0 0>
     outlineSettings {
      bUseBrushTransparency=false
      roundingType=HalfHeightRadius
      width=0
      color {colorUseRule=UseColor_Specified, specifiedColor=<rgba 0 0 0 0>}
      cornerRadii=<0 0 0 0>
     }
     tintColor {colorUseRule=UseColor_Specified, specifiedColor=<rgba 1 1 1 1>}
     uVRegion {bIsValid=false, max=<0 0>, min=<0 0>}
    }}
   slot {class=HorizontalBoxSlot, properties {horizontalAlignment=HAlign_Left, verticalAlignment=VAlign_Center}}
  }
  Addasong {
   widgetClassPath=&CommonTextBlock
   bIsVariable=true
   parent=HorizontalBoxAddaSong
   widget=Addasong
   properties {
    text=Add a song
    font {
     bForceMonospaced=false
     bMaterialIsStencil=false
     fontMaterial=null
     fontObject=&BurbankSmall
     letterSpacing=0
     monospacedWidth=1
     size=24
     skewAmount=0
     typefaceFontName=Black
     outlineSettings {
      bApplyOutlineToDropShadows=false
      bMiteredCorners=false
      bSeparateFillAlpha=false
      outlineMaterial=null
      outlineSize=0
      outlineColor=<rgba 0 0 0 1>
     }
    }
    shadowOffset=<0 0>
   }
   slot {class=HorizontalBoxSlot, properties {verticalAlignment=VAlign_Center, padding=<ltrb 24 0 0 0>, size {sizeRule=Fill, value=1}}}
  }
  SongNumber {
   widgetClassPath=&CommonTextBlock
   bIsVariable=true
   parent=HorizontalBoxAddaSong
   widget=SongNumber
   properties {
    text=1
    visibility=Collapsed
    colorAndOpacity {colorUseRule=UseColor_Specified, specifiedColor=<rgba 1 1 1 0.6>}
    font {
     bForceMonospaced=false
     bMaterialIsStencil=false
     fontMaterial=null
     fontObject=&BurbankSmall
     letterSpacing=0
     monospacedWidth=1
     size=36
     skewAmount=0
     typefaceFontName=Black
     outlineSettings {
      bApplyOutlineToDropShadows=false
      bMiteredCorners=false
      bSeparateFillAlpha=false
      outlineMaterial=null
      outlineSize=0
      outlineColor=<rgba 0 0 0 1>
     }
    }
    shadowOffset=<0 0>
   }
   slot {
    class=HorizontalBoxSlot
    properties {horizontalAlignment=HAlign_Right, verticalAlignment=VAlign_Center, padding=<ltrb 0 0 18 0>, size {sizeRule=Fill, value=1}}
   }
  }
  AddedSong {
   widgetClassPath=&Border
   parent=SongEntryStatusSwitcher
   widget=AddedSong
   properties {
    background {
     bIsDynamicallyLoaded=false
     drawAs=Image
     imageType=FullColor
     mirroring=NoMirror
     resourceName=None
     resourceObject=/SparksUICommon/Materials/MI_UI_Sparks_SetList_BackingAdded
     tiling=NoTile
     imageSize=<32 32>
     margin=<ltrb 0 0 0 0>
     outlineSettings {
      bUseBrushTransparency=false
      roundingType=HalfHeightRadius
      width=0
      color {colorUseRule=UseColor_Specified, specifiedColor=<rgba 0 0 0 0>}
      cornerRadii=<0 0 0 0>
     }
     tintColor {colorUseRule=UseColor_Specified, specifiedColor=<rgba 1 1 1 1>}
     uVRegion {bIsValid=false, max=<0 0>, min=<0 0>}
    }
    padding=<ltrb 0 0 22 0>
   }
   slot {class=CommonVisibilitySwitcherSlot}
  }
  HorizontalBoxTrack {
   widgetClassPath=&HorizontalBox
   parent=AddedSong
   widget=HorizontalBoxTrack
   slot {class=BorderSlot, properties {padding=<ltrb 0 0 22 0>}}
  }
  ImageAlbumArt {
   widgetClassPath=&FortLazyImage
   bIsVariable=true
   parent=HorizontalBoxTrack
   widget=ImageAlbumArt
   properties {
    materialTextureParamName=T_AlbumArt
    brush {
     bIsDynamicallyLoaded=false
     drawAs=Image
     imageType=FullColor
     mirroring=NoMirror
     resourceName=None
     resourceObject=/SparksUICommon/Materials/MI_UI_Sparks_SetList_AlbumArt_Rounded_Inst
     tiling=NoTile
     imageSize=<94 94>
     margin=<ltrb 0 0 0 0>
     outlineSettings {
      bUseBrushTransparency=false
      roundingType=HalfHeightRadius
      width=0
      color {colorUseRule=UseColor_Specified, specifiedColor=<rgba 0 0 0 0>}
      cornerRadii=<0 0 0 0>
     }
     tintColor {colorUseRule=UseColor_Specified, specifiedColor=<rgba 1 1 1 1>}
     uVRegion {bIsValid=false, max=<0 0>, min=<0 0>}
    }
    loadingBackgroundBrush {
     bIsDynamicallyLoaded=false
     drawAs=Image
     imageType=NoImage
     mirroring=NoMirror
     resourceName=None
     resourceObject=null
     tiling=NoTile
     imageSize=<92 92>
     margin=<ltrb 0 0 0 0>
     outlineSettings {
      bUseBrushTransparency=false
      roundingType=HalfHeightRadius
      width=0
      color {colorUseRule=UseColor_Specified, specifiedColor=<rgba 0 0 0 0>}
      cornerRadii=<0 0 0 0>
     }
     tintColor {colorUseRule=UseColor_Specified, specifiedColor=<rgba 1 1 1 1>}
     uVRegion {bIsValid=false, max=<0 0>, min=<0 0>}
    }
   }
   slot {class=HorizontalBoxSlot, properties {horizontalAlignment=HAlign_Left}}
  }
  VerticalBox_0 {
   widgetClassPath=&VerticalBox
   parent=HorizontalBoxTrack
   widget=VerticalBox_0
   properties {clipping=ClipToBounds}
   slot {class=HorizontalBoxSlot, properties {verticalAlignment=VAlign_Center, padding=<ltrb 24 0 0 0>, size {sizeRule=Fill, value=1}}}
  }
  SongTitle {
   widgetClassPath=&CommonTextBlock
   bIsVariable=true
   parent=VerticalBox_0
   widget=SongTitle
   properties {
    scrollStyle=/PilgrimQuickplay/Widgets/ScrollStyles/TextScrollStyle_PilgrimQuickplay_Setlist.TextScrollStyle_PilgrimQuickplay_Setlist_C
    text="[Song Title]"
    font {
     bForceMonospaced=false
     bMaterialIsStencil=false
     fontMaterial=null
     fontObject=&BurbankSmall
     letterSpacing=0
     monospacedWidth=1
     size=24
     skewAmount=0
     typefaceFontName=Black
     outlineSettings {
      bApplyOutlineToDropShadows=false
      bMiteredCorners=false
      bSeparateFillAlpha=false
      outlineMaterial=null
      outlineSize=0
      outlineColor=<rgba 0 0 0 1>
     }
    }
    shadowOffset=<0 0>
   }
   slot {class=VerticalBoxSlot, properties {padding=<ltrb 0 0 5 0>}}
  }
  ArtistName {
   widgetClassPath=&CommonTextBlock
   bIsVariable=true
   parent=VerticalBox_0
   widget=ArtistName
   properties {
    scrollStyle=&TextScrollStyle_PilgrimQuickplay_Setlist_C
    text="[Artist Name]"
    colorAndOpacity {colorUseRule=UseColor_Specified, specifiedColor=<rgba 1 1 1 0.8>}
    font {
     bForceMonospaced=false
     bMaterialIsStencil=false
     fontMaterial=null
     fontObject=&BurbankSmall
     letterSpacing=0
     monospacedWidth=1
     size=24
     skewAmount=0
     typefaceFontName=Black
     outlineSettings {
      bApplyOutlineToDropShadows=false
      bMiteredCorners=false
      bSeparateFillAlpha=false
      outlineMaterial=null
      outlineSize=0
      outlineColor=<rgba 0 0 0 1>
     }
    }
    shadowOffset=<0 0>
   }
   slot {class=VerticalBoxSlot}
  }
  ReleaseDurationHorizontalBox {
   widgetClassPath=&HorizontalBox
   parent=VerticalBox_0
   widget=ReleaseDurationHorizontalBox
   slot {class=VerticalBoxSlot, properties {verticalAlignment=VAlign_Top, size {sizeRule=Fill, value=1}}}
  }
  SongRelease {
   widgetClassPath=&CommonTextBlock
   bIsVariable=true
   parent=ReleaseDurationHorizontalBox
   widget=SongRelease
   properties {
    clipping=OnDemand
    text=2019
    textTransformPolicy=ToUpper
    visibility=HitTestInvisible
    colorAndOpacity {colorUseRule=UseColor_Specified, specifiedColor=<rgba 1 1 1 0.5>}
    font {
     bForceMonospaced=false
     bMaterialIsStencil=false
     fontMaterial=null
     fontObject=&HeadingNow
     letterSpacing=0
     monospacedWidth=1
     size=13
     skewAmount=0
     typefaceFontName=86 Bold
     outlineSettings {
      bApplyOutlineToDropShadows=false
      bMiteredCorners=false
      bSeparateFillAlpha=false
      outlineMaterial=null
      outlineSize=0
      outlineColor=<rgba 0 0 0 1>
     }
    }
    shadowOffset=<0 0>
   }
   slot {class=HorizontalBoxSlot}
  }
  ReleaseDurationDivider {
   widgetClassPath=&CommonTextBlock
   bIsVariable=true
   parent=ReleaseDurationHorizontalBox
   widget=ReleaseDurationDivider
   properties {
    clipping=OnDemand
    text=|
    textTransformPolicy=ToUpper
    visibility=HitTestInvisible
    colorAndOpacity {colorUseRule=UseColor_Specified, specifiedColor=<rgba 1 1 1 0.5>}
    font {
     bForceMonospaced=false
     bMaterialIsStencil=false
     fontMaterial=null
     fontObject=&Roboto
     letterSpacing=0
     monospacedWidth=1
     size=13
     skewAmount=0
     typefaceFontName=Black
     outlineSettings {
      bApplyOutlineToDropShadows=false
      bMiteredCorners=false
      bSeparateFillAlpha=false
      outlineMaterial=null
      outlineSize=0
      outlineColor=<rgba 0 0 0 1>
     }
    }
    shadowOffset=<0 0>
   }
   slot {class=HorizontalBoxSlot, properties {padding=<ltrb 5 0 5 0>}}
  }
  SongDuration {
   widgetClassPath=&CommonTextBlock
   bIsVariable=true
   parent=ReleaseDurationHorizontalBox
   widget=SongDuration
   properties {
    clipping=OnDemand
    text=3:09
    textTransformPolicy=ToUpper
    visibility=HitTestInvisible
    colorAndOpacity {colorUseRule=UseColor_Specified, specifiedColor=<rgba 1 1 1 0.5>}
    font {
     bForceMonospaced=false
     bMaterialIsStencil=false
     fontMaterial=null
     fontObject=&Roboto
     letterSpacing=0
     monospacedWidth=1
     size=13
     skewAmount=0
     typefaceFontName=Black
     outlineSettings {
      bApplyOutlineToDropShadows=false
      bMiteredCorners=false
      bSeparateFillAlpha=false
      outlineMaterial=null
      outlineSize=0
      outlineColor=<rgba 0 0 0 1>
     }
    }
    shadowOffset=<0 0>
   }
   slot {class=HorizontalBoxSlot}
  }
  EmoteOnly {
   widgetClassPath=&/WBP_PilgrimQuickplay_EmoteOnly
   bIsVariable=true
   parent=ReleaseDurationHorizontalBox
   widget=EmoteOnly
   properties {bOverride_Cursor=true}
   slot {class=HorizontalBoxSlot, properties {horizontalAlignment=HAlign_Right, verticalAlignment=VAlign_Center, size {sizeRule=Fill, value=1}}}
  }
  Spacer_73 {widgetClassPath=&Spacer, parent=HorizontalBoxTrack, widget=Spacer_73, properties {size=<34 1>}, slot {class=HorizontalBoxSlot}}
  AvatarIcon {
   widgetClassPath=/PilgrimCore/Widgets/Player/WBP_Pilgrim_AvatarIcon
   bIsVariable=true
   parent=HorizontalBoxTrack
   widget=AvatarIcon
   properties {playerIconSizeOverride=<75 75>}
   slot {class=HorizontalBoxSlot, properties {horizontalAlignment=HAlign_Center, verticalAlignment=VAlign_Center}}
  }
  Waiting {
   widgetClassPath=&Border
   parent=SongEntryStatusSwitcher
   widget=Waiting
   properties {
    background {
     bIsDynamicallyLoaded=false
     drawAs=Image
     imageType=FullColor
     mirroring=NoMirror
     resourceName=None
     resourceObject=/SparksUICommon/Materials/MI_UI_Sparks_SetList_BackingEmpty
     tiling=NoTile
     imageSize=<32 32>
     margin=<ltrb 0 0 0 0>
     outlineSettings {
      bUseBrushTransparency=false
      roundingType=HalfHeightRadius
      width=0
      color {colorUseRule=UseColor_Specified, specifiedColor=<rgba 0 0 0 0>}
      cornerRadii=<0 0 0 0>
     }
     tintColor {colorUseRule=UseColor_Specified, specifiedColor=<rgba 1 1 1 1>}
     uVRegion {bIsValid=false, max=<0 0>, min=<0 0>}
    }
    padding=<ltrb 0 0 22 0>
   }
   slot {class=CommonVisibilitySwitcherSlot}
  }
  HorizontalBoxWaiting {
   widgetClassPath=&HorizontalBox
   parent=Waiting
   widget=HorizontalBoxWaiting
   slot {class=BorderSlot, properties {padding=<ltrb 0 0 22 0>}}
  }
  Border_0 {
   widgetClassPath=&Border
   parent=HorizontalBoxWaiting
   widget=Border_0
   properties {
    horizontalAlignment=HAlign_Center
    verticalAlignment=VAlign_Center
    background {
     bIsDynamicallyLoaded=false
     drawAs=Image
     imageType=FullColor
     mirroring=NoMirror
     resourceName=None
     resourceObject=/SparksUICommon/Materials/MI_UI_Setlist_CD_Underlayer
     tiling=NoTile
     imageSize=<86 86>
     margin=<ltrb 0 0 0 0>
     outlineSettings {
      bUseBrushTransparency=false
      roundingType=HalfHeightRadius
      width=0
      color {colorUseRule=UseColor_Specified, specifiedColor=<rgba 0 0 0 0>}
      cornerRadii=<0 0 0 0>
     }
     tintColor {colorUseRule=UseColor_Specified, specifiedColor=<rgba 1 1 1 1>}
     uVRegion {bIsValid=false, max=<0 0>, min=<0 0>}
    }
    padding=<ltrb 20 17 20 17>
   }
   slot {class=HorizontalBoxSlot, properties {padding=<ltrb 4 4 4 4>}}
  }
  ImageAddIcon_1 {
   widgetClassPath=&Image
   bIsVariable=true
   parent=Border_0
   widget=ImageAddIcon_1
   properties {brush {
     bIsDynamicallyLoaded=false
     drawAs=Image
     imageType=FullColor
     mirroring=NoMirror
     resourceName=None
     resourceObject=/SparksUICommon/Materials/MI_UI_Setlist_CD_Icon
     tiling=NoTile
     imageSize=<52 52>
     margin=<ltrb 0 0 0 0>
     outlineSettings {
      bUseBrushTransparency=false
      roundingType=HalfHeightRadius
      width=0
      color {colorUseRule=UseColor_Specified, specifiedColor=<rgba 0 0 0 0>}
      cornerRadii=<0 0 0 0>
     }
     tintColor {colorUseRule=UseColor_Specified, specifiedColor=<rgba 1 1 1 1>}
     uVRegion {bIsValid=false, max=<0 0>, min=<0 0>}
    }}
   slot {class=BorderSlot, properties {horizontalAlignment=HAlign_Center, verticalAlignment=VAlign_Center, padding=<ltrb 20 17 20 17>}}
  }
  Waitingonsetlistpick {
   widgetClassPath=&CommonTextBlock
   bIsVariable=true
   parent=HorizontalBoxWaiting
   widget=Waitingonsetlistpick
   properties {
    text=Waiting on setlist pick
    colorAndOpacity {colorUseRule=UseColor_Specified, specifiedColor=<rgba 0.5796 0.6636 1 0.6>}
    font {
     bForceMonospaced=false
     bMaterialIsStencil=false
     fontMaterial=null
     fontObject=&BurbankSmall
     letterSpacing=0
     monospacedWidth=1
     size=24
     skewAmount=0
     typefaceFontName=Black
     outlineSettings {
      bApplyOutlineToDropShadows=false
      bMiteredCorners=false
      bSeparateFillAlpha=false
      outlineMaterial=null
      outlineSize=0
      outlineColor=<rgba 0 0 0 1>
     }
    }
    shadowOffset=<0 0>
   }
   slot {class=HorizontalBoxSlot, properties {verticalAlignment=VAlign_Center, padding=<ltrb 24 0 0 0>, size {sizeRule=Fill, value=1}}}
  }
  Image_SelectionGlow {
   widgetClassPath=&Image
   bIsVariable=true
   parent=Overlay_30
   widget=Image_SelectionGlow
   properties {
    visibility=HitTestInvisible
    brush {
     bIsDynamicallyLoaded=false
     drawAs=Image
     imageType=FullColor
     mirroring=NoMirror
     resourceName=None
     resourceObject=/PilgrimCore/Widgets/Materials/MI_UI_Pilgrim_Rounded_Outline_SelectedMainSong_Selected
     tiling=NoTile
     imageSize=<32 32>
     margin=<ltrb 0.5 0.5 0.5 0.5>
     outlineSettings {
      bUseBrushTransparency=false
      roundingType=HalfHeightRadius
      width=0
      color {colorUseRule=UseColor_Specified, specifiedColor=<rgba 0 0 0 0>}
      cornerRadii=<0 0 0 0>
     }
     tintColor {colorUseRule=UseColor_Specified, specifiedColor=<rgba 1 1 1 1>}
     uVRegion {bIsValid=false, max=<0 0>, min=<0 0>}
    }
    renderTransform {angle=0, scale=<1.01 1.03>, shear=<0 0>, translation=<0 0>}
   }
   slot {class=OverlaySlot, properties {horizontalAlignment=HAlign_Fill, verticalAlignment=VAlign_Fill, padding=<ltrb 2 0 2 0>}}
  }
  InputActionWidget {widgetClassPath=&CommonActionWidget, bInherited=true, widget=null}
 ]
}
