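# Quickplay activatable widget that previews setlist songs in a CommonListView, which is also its focus target, and lets the player remove them.
# Hovering broadcasts OnEntryHovered and caches the entry.
# Clicking or RemoveSongFromSetlist removes that cached song, not the clicked item, through the player controller's setlist editor component, then broadcasts OnRemoveSongFromSetlist.
# Set Song Status for Player checks CustomSetlist ownership, removes or rejects by display status, and reports the result.
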WidgetBlueprint /PilgrimQuickplay/Widgets/SetlistAndSongDisplay/Duplicated/WBP_PilgrimQuickplay_SetlistPreviewList
graphs [
 BP_GetDesiredFocusTarget {nodes [
   Get Desired Focus Target()
   > Return Node!(execute, ReturnValue:Widget Object Reference=Get SetlistPreviewList() -> (SetlistPreviewList:Common List View Object Reference))
   GetDesiredFocusTarget!(self:Common Activatable Widget Object Reference) -> (ReturnValue:Widget Object Reference)
  ]}
 EventGraph {nodes [
   n1 = "On Item Clicked (SetlistPreviewList)"!() -> (OutputDelegate, then, Item:Object Reference) {is=title K2Node_ComponentBoundEvent}
   n2 = Set OwningPlayerState(
    execute=[@n1.then, @n6.then]
    OwningPlayerState:Player State Object Reference=Get PlayerState!(self:Controller Object Reference=GetOwningPlayer()) -> (PlayerState:Player State Object Reference)
   ) -> (Output_Get:Player State Object Reference)
   n3 = Get SongMetadata!(self:BP Song Entry Horizontal Item Obj Object Reference=Get Song Entry Reference() -> (Song Entry Reference:BP Song Entry Horizontal Item Obj Object Reference)) -> (SongMetadata:Pilgrim Song Metadata Object Reference)
   > Remove Song!(
    execute
    self:BP Pilgrim Custom Setlist Editor Component Object Reference=GetComponentByClass!(
     self:Actor Object Reference=GetPlayerController!(self:Player State Object Reference=@n2) -> (ReturnValue:Player Controller Object Reference)
     ComponentClass:Actor Component Class Reference=/PilgrimCore/Player/Components/BP_PilgrimCustomSetlistEditorComponent.BP_PilgrimCustomSetlistEditorComponent_C
    ) -> (ReturnValue:BP Pilgrim Custom Setlist Editor Component Object Reference)
    SongToRemove:Name=GetShortName!(self:Sparks Song Data Object Reference=@n3) -> (ReturnValue:Name=None)
   ) -> (then)
   > Call On Remove Song from Setlist!(execute:Exec, self:WBP Pilgrim Quickplay Setlist Preview List Object Reference, Song Meta Data:Pilgrim Song Metadata Object Reference=@n3) -> (then:Exec) {is=title K2Node_CallDelegate}
   n4 = "On Item Is Hovered Changed (SetlistPreviewList)"!() -> (OutputDelegate, then, Item:Object Reference, bIsHovered:bool) {is=title K2Node_ComponentBoundEvent}
   > Call On Entry Hovered!(execute:Exec, self:WBP Pilgrim Quickplay Setlist Preview List Object Reference, IsItemHovered:bool=@n4.bIsHovered) -> (then:Exec) {is=title K2Node_CallDelegate}
   > Branch(Condition=@n4.bIsHovered)
   > n5 = Cast To BP_SongEntryHorizontal_ItemObj!(execute, Object=@n4.Item) -> (then, CastFailed, AsBP Song Entry Horizontal Item Obj:BP Song Entry Horizontal Item Obj Object Reference) {is=title K2Node_DynamicCast}
   > Set Song Entry Reference(Song Entry Reference:BP Song Entry Horizontal Item Obj Object Reference=@n5) -> (Output_Get:BP Song Entry Horizontal Item Obj Object Reference)
   n6 = RemoveSongFromSetlist() {is=title K2Node_CustomEvent}
   Event PreConstruct()
   Event Construct()
   Event Tick()
  ]}
 OnEntryHovered {nodes [OnEntryHovered!() -> (then, IsItemHovered:bool=~) {is=title K2Node_FunctionEntry}]}
 OnRemoveSongFromSetlist {nodes [OnRemoveSongFromSetlist!() -> (then, Song Meta Data:Pilgrim Song Metadata Object Reference) {is=title K2Node_FunctionEntry}]}
 Set Song Status for Player {nodes [
   n1 = Set Song Status for Player!() -> (then, DisplayStatus:Song Display Status Struct Structure) {is=title K2Node_FunctionEntry}
   n2 = Break Song Display Status Struct!(SongDisplayStatusStruct:"Song Display Status Struct Structure (by ref)"=@n1) -> (
    SongShortName_19_#1:Name
    DisplayStatus_14_#2:ESongDisplayStatus Enum
    SongChooserPlayerState_17_#3:Player State Object Reference
   ) {is=title K2Node_BreakStruct}
   > Is Valid(InputObject=@n2.SongChooserPlayerState_17_#3)
   n3 = Get CustomSetlist() -> (CustomSetlist:Pilgrim Custom Setlist Object Reference)
   > Branch(Condition=HasAnyoneAddedSong!(self:Pilgrim Custom Setlist Object Reference=@n3, SongShortName:Name=@n2) -> (OutAddedByPlayer:Player State Object Reference, ReturnValue:bool).ReturnValue)
   > Branch(Condition=HasPlayerAddedSong!(
     self:Pilgrim Custom Setlist Object Reference=@n3
     SongShortName:Name=@n2
     Player:Player State Object Reference=@n2.SongChooserPlayerState_17_#3
    ) -> (ReturnValue:bool))
   > n4 = Switch on ESongDisplayStatus!(execute, Selection:ESongDisplayStatus Enum=@n2.DisplayStatus_14_#2) -> (NewEnumerator2:Exec, NewEnumerator0:Exec, NewEnumerator1:Exec) {is=title K2Node_SwitchEnum}
   Set StatusWasChanged(execute=@n4.NewEnumerator0, StatusWasChanged:bool=false) -> (Output_Get:bool)
   > Set ReasonForNoChange(ReasonForNoChange:EPilgrimCantAddSongReason Enum=SongAlreadyAdded) -> (Output_Get:EPilgrimCantAddSongReason Enum=None)
   PrintString!(
    execute=@n4.NewEnumerator1
    InString="Player trying to set a song in custom setlist with 'Disabled' display status, should not be possible"
    bPrintToScreen
    bPrintToLog
    TextColor
    Duration
    Key
   ) -> (then)
   Remove Song(
    execute=@n4.NewEnumerator2
    self=GetComponentByClass(
     self=GetPlayerController(self=@n2.SongChooserPlayerState_17_#3)
     ComponentClass=&BP_PilgrimCustomSetlistEditorComponent_C
    )
    SongToRemove=@n2
   )
   > Set StatusWasChanged(StatusWasChanged=true)
   > Return Node!(
    execute
    SongStatusChanged:bool=Get StatusWasChanged() -> (StatusWasChanged:bool)
    CantChangeReason:EPilgrimCantAddSongReason Enum=Get ReasonForNoChange() -> (ReasonForNoChange:EPilgrimCantAddSongReason Enum=None)
   )
  ]}
]
variables [
 CustomSetlist {type=PilgrimCustomSetlist, default=null}
 NewDisplayStatus {type=ESongDisplayStatus, default=None}
 OnEntryHovered {}
 OnRemoveSongFromSetlist {}
 OwningPlayerState {type=PlayerState, default=null}
 ReasonForNoChange {type=EPilgrimCantAddSongReason, default=None}
 Song Entry Reference {type=BP_SongEntryHorizontal_ItemObj_C, default=null}
 StatusWasChanged {type=boolean}
]
widgets {
 info {inheritedWidgetCount=0, parentClass=&CommonActivatableWidget, rootWidgetClass=&CommonListView, widgetCount=1}
 widgets [SetlistPreviewList {
   widgetClassPath=&CommonListView
   bIsVariable=true
   widget=SetlistPreviewList
   properties {
    bReturnFocusToSelection=true
    clipping=Inherit
    entryWidgetClass=/PilgrimQuickplay/Widgets/SetlistAndSongDisplay/WBP_PilgrimQuickplay_SongMenu_SetlistPreviewEntry.WBP_PilgrimQuickplay_SongMenu_SetlistPreviewEntry_C
    horizontalEntrySpacing=6
    navigation=&self:WidgetTree.SetlistPreviewList.WidgetNavigation_1
    numDesignerPreviewEntries=4
    verticalEntrySpacing=6
   }
  }]
}
