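# Activatable Quickplay song leaderboard screen showing a song's title, artist, album art and ranked entries.
# Enum-built tab bars (solo/band, instrument/band size, competitor group, season/all-time) determine the leaderboard context sent to the viewmodel, with analytics on changes.
# Scrolling requests further pages and shows a separate local-player row whenever theirs is off-screen.
# The band tab respects a friends-leaderboard option, and the screen closes upon a quickplay-state-began event.
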WidgetBlueprint /PilgrimQuickplay/Widgets/Leaderboards/WBP_PilgrimQuickplay_Leaderboard_Song
graphs [
 BP_GetDesiredFocusTarget {nodes [
   Get Desired Focus Target()
   > Return Node!(execute, ReturnValue:Widget Object Reference=Get LeaderboardEntryList() -> (LeaderboardEntryList:Common List View Object Reference))
  ]}
 BP_GetDesiredInputConfig {nodes [
   Get Desired Input Config() {is=title K2Node_FunctionEntry}
   > Return Node(ReturnValue:UIInput Config Structure=Make UIInput Config!(
     bIgnoreMoveInput:bool=true
     bIgnoreLookInput:bool=False
     InputMode:ECommonInputMode Enum=Menu
     MouseCaptureMode:EMouseCaptureMode Enum=NoCapture
     MouseLockMode:EMouseLockMode Enum=DoNotLock
     bHideCursorDuringViewportCapture:bool=True
    ) -> (UIInputConfig:UIInput Config Structure) {is=title K2Node_MakeStruct})
  ]}
 ChangeSelectedItemInList {nodes [
   n1 = ChangeSelectedItemInList!() -> (then, ListView:Common List View Object Reference, DoesLoop:bool=~, DirectionOfChange:int) {is=title K2Node_FunctionEntry}
   > Set MaxListIndex(MaxListIndex:int=int - int(A=GetNumItems!(self:List View Object Reference=@n1) -> (ReturnValue:int=0))) -> (Output_Get:int=0)
   n2 = GetIndexForItem!(
    self:List View Object Reference=@n1
    Item:Object Reference=Get Selected Item!(self:List View Object Reference=@n1) -> (ReturnValue:Object Reference) {is=title K2Node_CallFunction}
   ) -> (ReturnValue:int=0)
   > Set CurrentIndex(CurrentIndex:int=@n2) -> (Output_Get:int=0)
   > Set NewIndex(NewIndex:int=int + int(A=@n2, B=@n1.DirectionOfChange)) -> (Output_Get:int=0)
   > n3 = Sequence()
   > n4 = Branch(Condition=@n1.DoesLoop)
   > n5 = Branch(Condition="integer < integer"(A=Get NewIndex() -> (NewIndex:int=0)))
   > Set NewIndex(NewIndex=Get MaxListIndex() -> (MaxListIndex:int=0)) -> (Output_Get=0)
   Branch(execute=@n5.else, Condition="integer > integer"(A=Get NewIndex() -> (NewIndex=0), B=Get MaxListIndex() -> (MaxListIndex=0)))
   > Set NewIndex(NewIndex=0) -> (Output_Get=0)
   Set NewIndex(
    execute=@n4.else
    NewIndex="Clamp (Integer)"(Value=Get NewIndex() -> (NewIndex=0), Min=0, Max=Get MaxListIndex() -> (MaxListIndex=0)) -> (ReturnValue=0)
   ) -> (Output_Get=0)
   Branch(execute=@n3.then_1, Condition="Not Equal (Integer)"(A=Get CurrentIndex() -> (CurrentIndex:int=0), B=Get NewIndex() -> (NewIndex=0)))
   > SetSelectedIndex!(execute, self:List View Object Reference=@n1, Index:int=Get NewIndex() -> (NewIndex=0)) -> (then)
  ]}
 CreateCategoriesFromArray {nodes [
   n1 = CreateCategoriesFromArray!() -> (then, Categories:Array of EPilgrimQuickplay_LeaderboardFilters Enums=NewEnumerator0, TopNav Widget:WBP UIKit Tabs Base Object Reference) {is=title K2Node_FunctionEntry}
   > Clear(TargetArray:Array of UI Kit Tab Group Button View Model Object References=Get Local Tabs View Models() -> (Local Tabs View Models:Array of UI Kit Tab Group Button View Model Object References))
   > n2 = For Each Loop(Array:Array of EPilgrimQuickplay_LeaderboardFilters Enums=@n1) -> (Array Element:EPilgrimQuickplay_LeaderboardFilters Enum)
   > n3 = Construct Object from Class(Class=/Script/FortniteUI.UIKitTabGroupButtonViewModel) -> (ReturnValue:UI Kit Tab Group Button View Model Object Reference)
   n4 = Enum to String!(Enumerator:EPilgrimQuickplay_LeaderboardFilters Enum=@n2) -> (ReturnValue:String)
   > Set Label!(execute, Label:Text="To Text (String)"(InString=@n4), self:UI Kit Selectable Button View Model Object Reference=@n3) -> (then, Output_Get:Text)
   > Set TabID!(execute, TabID:Name=String To Name(InString=@n4), self:UI Kit Tab Group Button View Model Object Reference=@n3) -> (then, Output_Get:Name=None)
   > Add(
    TargetArray:Array of UI Kit Tab Group Button View Model Object References=Get Local Tabs View Models()
    NewItem:"UI Kit Tab Group Button View Model Object Reference (by ref)"=@n3
   ) {is=title K2Node_CallArrayFunction}
   Set Tabs Data!(
    execute=@n2.Completed
    self:WBP UIKit Tabs Base Object Reference=@n1.TopNav Widget
    InTabs Data:Array of UI Kit Tab Group Button View Model Object References=Get Local Tabs View Models()
   ) -> (then) {is=title K2Node_CallFunction}
  ]}
 CreateEmptyCategoriesFromArray {nodes [
   n1 = CreateEmptyCategoriesFromArray!() -> (then, Categories:Array of EPilgrimQuickplay_LeaderboardFilters Enums=NewEnumerator0, TopNav Widget:WBP UIKit Tabs Base Object Reference) {is=title K2Node_FunctionEntry}
   > Clear(TargetArray=Get Local Tabs View Models())
   > n2 = For Each Loop(Array=@n1)
   > n3 = Construct Object from Class(Class=&UIKitTabGroupButtonViewModel)
   > Set TabID(TabID=String To Name(InString=Enum to String(Enumerator=@n2)), self=@n3) -> (Output_Get=None)
   > Add(TargetArray=Get Local Tabs View Models(), NewItem=@n3)
   Set Tabs Data(execute=@n2.Completed, self=@n1.TopNav Widget, InTabs Data=Get Local Tabs View Models())
  ]}
 EventGraph {nodes [
   "On Clicked (BackButton)"!() -> (OutputDelegate, then, Button:Common Button Base Object Reference) {is=title K2Node_ComponentBoundEvent}
   > HandleBackButton()
   n1 = "On Tab Selected (TopNavBar_EntryBox_ActivityType)"!() -> (OutputDelegate, then, TabId:Name=None) {is=title K2Node_ComponentBoundEvent}
   > FireFilterChangedAnalytics!(execute, self, FilterName:String=SoloFilter, FilterValue:Name=@n1.TabId) -> (then)
   > n2 = Update All Tabs!(execute, self, Update Activity:bool=false, Update Competitor:bool=false, UpdateSeasonDisplay:bool=false) -> (then, SubFilterUpdated:bool, CompetitorUpdated:bool, SeasonDisplayUpdated:bool, ActivityUpdated:bool)
   > UpdateTabType!(execute, self, Activity Type Press:bool=@n2.ActivityUpdated) -> (then)
   > n3 = Branch(Condition=@n2.ActivityUpdated)
   n4 = Get TopNav_EntryBox_SubFilters() -> (TopNav_EntryBox_SubFilters:WBP UIKit Tabs Top Nav Bar Object Reference)
   > n5 = SelectTabByIndex!(
    execute
    self:UIKit Tab Group Object Reference=@n4
    Index:int=Get LastInstrumentSelection() -> (LastInstrumentSelection:int=0)
    bSuppressClickFeedback:bool=true
   ) -> (then, ReturnValue:bool)
   Do Once(execute=@n12.then, Reset=[@n16.then, @n19.then, @n10.then, @n11.then, @n5.then, @n6.then])
   > RefreshFocusedPlayerEntry()
   n6 = SelectTabByIndex(execute=@n3.else, self=@n4, Index=Get LastActivitySelection() -> (LastActivitySelection:int=0), bSuppressClickFeedback=true)
   n7 = "On Tab Selected (TopNav_EntryBox_SubFilters)"!() -> (OutputDelegate, then, TabId:Name=None) {is=title K2Node_ComponentBoundEvent}
   > RequestLeaderboardData()
   > "Set  Instrument Text Tab"!(execute, self, InstrumentName:Name=@n7.TabId, IsActivityUpdated:bool=@n2.ActivityUpdated) -> (then) {is=title K2Node_CallFunction}
   > FirePartSelectedAnalytics!(execute, self, InName:Name=@n7.TabId) -> (then)
   > n8 = Branch(Condition=@n2.ActivityUpdated)
   n9 = GetSelectedTabIndex!(self:UIKit Tab Group Object Reference=Get TopNav_EntryBox_SubFilters()) -> (ReturnValue:int=0)
   > n10 = Set LastInstrumentSelection(LastInstrumentSelection:int=@n9) -> (Output_Get:int=0)
   n11 = Set LastActivitySelection(execute=@n8.else, LastActivitySelection:int=@n9) -> (Output_Get:int=0)
   n12 = "On Entries Generated (LeaderboardEntryList)"!() -> (OutputDelegate, then, NumEntries:int=0) {is=title K2Node_ComponentBoundEvent}
   n13 = "On List View Scrolled (LeaderboardEntryList)"!() -> (OutputDelegate, then, ItemOffset:float, DistanceRemaining:float) {is=title K2Node_ComponentBoundEvent}
   > RefreshFocusedPlayerEntry()
   > Branch(Condition="float < float"(A=@n13.DistanceRemaining, B=0.5))
   > RequestNextLeaderboardPage(self:Pilgrim Quickplay Leaderboard VM Object Reference=Get PilgrimQuickplayLeaderboardVM() -> (PilgrimQuickplayLeaderboardVM:Pilgrim Quickplay Leaderboard VM Object Reference))
   n14 = "On UIKit Tab Selected (TopNavBar_EntryBox_CompetitorGroup)"!() -> (OutputDelegate, then, TabID:Name=None, TabVM:MVVM Base Viewmodel Object Reference) {is=title K2Node_ComponentBoundEvent}
   > FireFilterChangedAnalytics(FilterName=EveryoneFilter, FilterValue=@n14.TabID)
   > n15 = Update All Tabs(Update Activity=Should Force Activity Update!() -> (Should Update:bool), Update Competitor=false, UpdateSeasonDisplay=false)
   > Branch(Condition=@n15)
   > n16 = RequestLeaderboardData()
   n17 = "On UIKit Tab Selected (TopNavBar_EntryBox_SeasonDisplay)"!() -> (OutputDelegate, then, TabID:Name=None, TabVM:MVVM Base Viewmodel Object Reference) {is=title K2Node_ComponentBoundEvent}
   > FireFilterChangedAnalytics(FilterName=SeasonFilter, FilterValue=@n17.TabID)
   > n18 = Update All Tabs(Update Activity=false, Update Competitor=false, UpdateSeasonDisplay=true)
   > Branch(Condition=@n18.SeasonDisplayUpdated)
   > RequestLeaderboardData()
   > n19 = UpdateComponentAllTimeSelection()
   RequestLeaderboardData() {is=title K2Node_CustomEvent}
   > n20 = GetCurrentStateDataForLeaderboardRequest!(execute, self) -> (
    then
    Params Success:bool
    Song ShortName:Name=None
    SpotlightGroup:String
    LeaderboardType:EPilgrimLeaderboardType Enum=SoloGuitar
    FortPlayerController:Fort Player Controller Object Reference
    IsFriendsOnly:bool
    CenteredOnPlayer:bool
    Page:int=0
    IsSpotlightCompetition:bool
    IsAllTimeScore:bool
   )
   > SetLeaderboardContext!(
    execute
    self:Pilgrim Quickplay Leaderboard VM Object Reference=Get PilgrimQuickplayLeaderboardVM()
    InSongShortName:Name=@n20.Song ShortName
    InLeaderboardType:EPilgrimLeaderboardType Enum=@n20.LeaderboardType
    bInFriendsOnly:bool=@n20.IsFriendsOnly
    bInAllTime:bool=@n20.IsAllTimeScore
   ) -> (then)
   n21 = On Band Friend Leaderboard Option Changed!() -> (OutputDelegate, then, bOptionEnabled:bool=~) {is=title K2Node_CustomEvent}
   > n22 = Set Enable Band Friend Leaderboard Option(Enable Band Friend Leaderboard Option:bool=@n21.bOptionEnabled) -> (Output_Get:bool)
   n23 = NOT Boolean(A=@n22)
   > Update All Tabs(Update Activity=@n23, Update Competitor=@n23, UpdateSeasonDisplay=false)
   n24 = OnTabButtonCreation!() -> (OutputDelegate, then, TabId:Name, TabButton:Common Button Base Object Reference) {is=title K2Node_CustomEvent}
   > n25 = Cast To WBP_UIKit_ButtonGeneric_Base!(execute, Object=@n24.TabButton) -> (then, CastFailed, AsWBP UIKit Button Generic Base:WBP UIKit Button Generic Base Object Reference) {is=title K2Node_DynamicCast}
   n26 = Get TopNav_EntryBox_SubFilters()
   > n27 = GetIconMaterialByTabIndex!(execute, self, TabIndex:int=GetTabIndexByID!(self:UIKit Tab Group Object Reference=@n26, TabNameID:Name=@n24.TabId) -> (ReturnValue:int=0)) -> (then, OutputPin:Material Instance Dynamic Object Reference)
   > n28 = SetIconMaterial!(execute, self:WBP UIKit Button Generic Base Object Reference=@n25, Material Instance:Material Instance Object Reference=@n27) -> (then)
   n29 = UpdateTabType!() -> (OutputDelegate, then, Activity Type Press:bool=~) {is=title K2Node_CustomEvent}
   n30 = Branch(execute=@n28.then, Condition=@n29.Activity Type Press)
   > SetText!(execute, self:WBP UIKit Button Generic Base Object Reference=@n25, Text:Text) -> (then)
   SetIconSize!(execute=@n30.else, self:WBP UIKit Button Generic Base Object Reference=@n25, Icon:Vector 2D Structure=(X=0,Y=0)) -> (then)
   Event OnInitialized()
   > Fire Event Main View Opened!(execute, View:Widget Object Reference=Self-Reference(), AdditionalInfo:Flow Analytics Additional Info Structure, bIsOverlay:bool=false) -> (then)
   > InitializeFilters()
   > Update Tab Actions()
   Event Construct()
   > n31 = Sequence!(execute) -> (then_0, then_1, then_2:Exec, then_3:Exec)
   Bind Event to On Tab Button Creation(execute=[@n29.then, @n31.then_0], self:Common Tab List Widget Base Object Reference=@n26, Delegate=@n24) {is=title K2Node_AddDelegate}
   > On Activity Type Selected Refresh Sub Type Filters()
   > Update All-Time Selection from Component()
   n32 = PilgrimWaitForEventRouter!(
    execute:Exec=@n31.then_1
    ContextActor:Actor Object Reference=GetSparksMusicPlayspaceOfClass!(
     PlayspaceType:Sparks Music Playspace Class Reference=/Script/PilgrimQuickplayRuntime.PilgrimQuickplayPlayspace
     Actor:Actor Object Reference=GetOwningPlayer()
     bMatchChildTypes:bool=false
    ) -> (ReturnValue:Pilgrim Quickplay Playspace Object Reference)
   ) -> (then:Exec, OnEventRouterAvailable:Exec, EventRouter:Gameplay Event Router Component Object Reference) {is=title K2Node_AsyncAction}
   n33 = StartListeningToEvent!(
    execute:Exec=@n32.OnEventRouterAvailable
    Target:Gameplay Event Router Component Object Reference=@n32
    EventType:Script Struct Object Reference=/Script/PilgrimQPSharedRuntime.PilgrimEvent_QuickplayState_Began
    Context:Object Reference
    EventBubblingRule:EEventBubblingRule Enum
   ) -> (
    then:Exec
    AsyncTaskProxy:Async Action Start Listening To Event Object Reference
    OnEventReceived:Exec
    Payload_StateId:EPilgrimQuickplayState Enum
   )
   DeactivateWidget(execute=@n33.OnEventReceived)
   n34 = GetPlayerLeaderboardComponent!() -> (LeaderboardComponent:Pilgrim Leaderboard Player Component Object Reference)
   Is Valid(exec=@n31.then_2, InputObject=@n34)
   > Set Enable Band Friend Leaderboard Option(Enable Band Friend Leaderboard Option=IsBandFriendLeaderboardEnabled!(self:Pilgrim Leaderboard Player Component Object Reference=@n34) -> (ReturnValue:bool))
   > Bind Event to On Band Friend Leaderboard Option Changed(self:Pilgrim Leaderboard Player Component Object Reference=@n34, Delegate=@n21) {is=title K2Node_AddDelegate}
   Update All Tabs(execute=@n31.then_3, Update Activity=true, Update Competitor=true, UpdateSeasonDisplay=true)
   > Update Season Text()
   Event On Activated()
   > RequestLeaderboardData() {is=title K2Node_CallFunction}
  ]}
 FireFilterChangedAnalytics {nodes [
   n1 = FireFilterChangedAnalytics!() -> (then, FilterName:String, FilterValue:Name) {is=title K2Node_FunctionEntry}
   > n2 = Cast To FortPlayerController(Object=GetOwningPlayer())
   > "Fire Event Element Interacted (Non-Localized)"!(
    execute
    FortPC:Fort Player Controller Object Reference=@n2
    ElementName:String=@n1
    AdditionalInfo:Flow Analytics Additional Info Structure="Add String (Non-Localized)"!(
     Info:"Flow Analytics Additional Info Structure (by ref)"=Make Flow Analytics Additional Info!() -> (FlowAnalyticsAdditionalInfo:Flow Analytics Additional Info Structure) {is=title K2Node_MakeStruct}
     Key:String=FilterValue
     Value:String="To String (Name)"(InName=@n1.FilterValue)
    ) -> (ReturnValue:Flow Analytics Additional Info Structure)
   ) -> (then)
  ]}
 FirePartSelectedAnalytics {nodes [
   n1 = FirePartSelectedAnalytics!() -> (then, InName:Name) {is=title K2Node_FunctionEntry}
   > n2 = Cast To FortPlayerController(Object=GetOwningPlayer())
   > "Fire Event Sub View Opened (Non-Localized)"!(
    execute
    FortPC:Fort Player Controller Object Reference=@n2
    ViewName:String=Append(A=FestivalLeaderboards_, B="To String (Name)"(InName=@n1))
    AdditionalInfo:Flow Analytics Additional Info Structure
    bIsOverlay:bool=false
   ) -> (then)
  ]}
 Get Categories Array {nodes [
   Get Categories Array() {is=title K2Node_FunctionEntry}
   n1 = IsGuitarPeripheralAllowed!() -> (ReturnValue:bool)
   n2 = Get Categories_Instruments_WithPeripherals() -> (Categories_Instruments_WithPeripherals:Array of EPilgrimQuickplay_LeaderboardFilters Enums=NewEnumerator0)
   n3 = Get Categories_Instruments() -> (Categories_Instruments:Array of EPilgrimQuickplay_LeaderboardFilters Enums=NewEnumerator0)
   > Return Node(ReturnValue:Array of EPilgrimQuickplay_LeaderboardFilters Enums=Select(
     Option 0:Array of EPilgrimQuickplay_LeaderboardFilters Enums=Get Categories_BandSize() -> (Categories_BandSize:Array of EPilgrimQuickplay_LeaderboardFilters Enums=NewEnumerator0)
     Option 1:Array of EPilgrimQuickplay_LeaderboardFilters Enums=Select(Option 0=@n3, Option 1=@n2, Index=@n1)
     Index=Is Showing Solo Leaderboard!() -> (ReturnValue:bool)
    ) -> (ReturnValue:Array of EPilgrimQuickplay_LeaderboardFilters Enums))
  ]}
 Get Filter Subtype {nodes [
   Get Filter Subtype() {is=title K2Node_FunctionEntry}
   n1 = Get TopNav_EntryBox_SubFilters()
   > Return Node!(
    execute
    Subtype:EPilgrimQuickplay_LeaderboardFilters Enum="Get (a copy)"(
     Array:Array of EPilgrimQuickplay_LeaderboardFilters Enums=Get Categories Array!() -> (ReturnValue:Array of EPilgrimQuickplay_LeaderboardFilters Enums=NewEnumerator0) {is=title K2Node_CallFunction}
     Dimension 1=GetSelectedTabIndex(self=@n1) -> (ReturnValue=0)
    ) -> (Output:EPilgrimQuickplay_LeaderboardFilters Enum)
   )
  ]}
 Get Leaderboard Type {nodes [
   n1 = Get Leaderboard Type!() -> (then, Selection:EPilgrimQuickplay_LeaderboardFilters Enum=NewEnumerator0) {is=title K2Node_FunctionEntry}
   > n2 = Switch on EPilgrimQuickplay_LeaderboardFilters!(execute, Selection:EPilgrimQuickplay_LeaderboardFilters Enum=@n1) -> (
    NewEnumerator0:Exec
    NewEnumerator1:Exec
    NewEnumerator2:Exec
    NewEnumerator3:Exec
    NewEnumerator4:Exec
    NewEnumerator5:Exec
    NewEnumerator6:Exec
    NewEnumerator7:Exec
    NewEnumerator8:Exec
    NewEnumerator9:Exec
    NewEnumerator10:Exec
    NewEnumerator12:Exec
    NewEnumerator11:Exec
    NewEnumerator13:Exec
    NewEnumerator14:Exec
    NewEnumerator15:Exec
    NewEnumerator16:Exec
    NewEnumerator17:Exec
    NewEnumerator18:Exec
    NewEnumerator19:Exec
   ) {is=title K2Node_SwitchEnum}
   Return Node!(
    execute=[@n2.NewEnumerator0, @n2.NewEnumerator1, @n2.NewEnumerator2, @n2.NewEnumerator3]
    LeaderboardType:EPilgrimLeaderboardType Enum=Invalid
   )
   Return Node(execute=@n2.NewEnumerator10, LeaderboardType=BandQuad)
   Return Node(execute=[@n2.NewEnumerator11, @n2.NewEnumerator12], LeaderboardType=Invalid)
   Return Node(execute=@n2.NewEnumerator13, LeaderboardType=SoloPeripheralGuitar)
   Return Node(execute=@n2.NewEnumerator14, LeaderboardType=SoloPeripheralBass)
   Return Node(execute=@n2.NewEnumerator15, LeaderboardType=SoloPeripheralDrum)
   Return Node(execute=[@n2.NewEnumerator16, @n2.NewEnumerator17], LeaderboardType=Invalid)
   Return Node(execute=@n2.NewEnumerator18, LeaderboardType=SoloPeripheralCymbal)
   Return Node(execute=@n2.NewEnumerator19, LeaderboardType=SoloPeripheralVocals)
   Return Node(execute=@n2.NewEnumerator4, LeaderboardType=SoloGuitar)
   Return Node(execute=@n2.NewEnumerator5, LeaderboardType=SoloDrum)
   Return Node(execute=@n2.NewEnumerator6, LeaderboardType=SoloVocals)
   Return Node(execute=@n2.NewEnumerator7, LeaderboardType=SoloBass)
   Return Node(execute=@n2.NewEnumerator8, LeaderboardType=BandDuo)
   Return Node(execute=@n2.NewEnumerator9, LeaderboardType=BandTrio)
  ]}
 GetCurrentLeaderboardType {nodes [
   GetCurrentLeaderboardType() {is=title K2Node_FunctionEntry}
   > n1 = Get Leaderboard Type!(
    execute
    self
    Selection:EPilgrimQuickplay_LeaderboardFilters Enum=Get Filter Subtype!() -> (Subtype:EPilgrimQuickplay_LeaderboardFilters Enum=NewEnumerator0) {is=title K2Node_CallFunction}
   ) -> (then, LeaderboardType:EPilgrimLeaderboardType Enum=SoloGuitar) {is=title K2Node_CallFunction}
   > Return Node(LeaderboardType=@n1)
  ]}
 GetCurrentStateDataForLeaderboardRequest {nodes [
   GetCurrentStateDataForLeaderboardRequest() {is=title K2Node_FunctionEntry}
   > n1 = GetCurrentLeaderboardType!(execute, self) -> (then, LeaderboardType:EPilgrimLeaderboardType Enum=SoloGuitar) {is=title K2Node_CallFunction}
   > n2 = Branch(Condition="Not Equal (Enum)"!(A:EPilgrimLeaderboardType Enum=@n1, B:EPilgrimLeaderboardType Enum=Invalid) -> (ReturnValue:bool))
   n3 = Get CurrentSongShortName() -> (CurrentSongShortName:Name=None)
   > n4 = Branch(Condition="Not Equal (Name)"(A=@n3) -> (ReturnValue=~))
   n5 = Is Friends Only!() -> (IsFriendsOnly:bool)
   n6 = GetFortPlayerControllerFromActor!(Actor:Actor Object Reference=GetOwningPlayer()) -> (ReturnValue:Fort Player Controller Object Reference)
   > Return Node!(
    execute
    Params Success:bool=true
    Song ShortName:Name=@n3
    SpotlightGroup:String
    LeaderboardType=@n1
    FortPlayerController:Fort Player Controller Object Reference=@n6
    IsFriendsOnly:bool=@n5
    CenteredOnPlayer:bool=false
    Page:int=0
    IsSpotlightCompetition:bool=false
    IsAllTimeScore:bool=IsAllTimeScoring!() -> (ShouldShowAllTime:bool)
   )
   Reroute Node(InputPin=@n4.else)
   Reroute Node(InputPin=@n2.else)
   PrintString!(
    execute
    InString=Append!(
     A="Request data> Song Short Name: "
     B="To String (Name)"(InName=@n3)
     C:String=", LeaderboardType: "
     D:String=Enum to String(Enumerator:EPilgrimLeaderboardType Enum=@n1)
     E:String=", RequestingPlayer: "
     F:String=GetPlayerNameSafe!(Actor:Actor Object Reference=@n6) -> (ReturnValue:String)
     G:String=", Friends Only: "
     H:String="To String (Boolean)"(InBool=@n5)
    ) -> (ReturnValue)
    bPrintToScreen
    bPrintToLog
    TextColor=(R=0.1812,G=1,B=0.273,A=1)
    Duration=30
    Key
   ) -> (then)
   PrintString(InString="No valid leaderboard type found, no request will be made")
   PrintString(InString="No song tag set, no leaderboard will be requested")
  ]}
 GetIconMaterialByTabIndex {nodes [
   n1 = GetIconMaterialByTabIndex!() -> (then, TabIndex:int) {is=title K2Node_FunctionEntry}
   > n2 = CreateDynamicMaterialInstance!(execute, Parent=/PilgrimQuickplay/Widgets/Materials/M_UI_Icon_Instrument_Type_MSDF, OptionalName=None, CreationFlags=None) -> (then, ReturnValue)
   n3 = "Get (a copy)"(
    Array:Array of EPilgrimTrackType Enums=Get InstrumentsTabTypes() -> (InstrumentsTabTypes:Array of EPilgrimTrackType Enums=TrackGuitar)
    Dimension 1=@n1
   ) -> (Output:EPilgrimTrackType Enum)
   n4 = PilgrimTrackTypeToSparksInstrument!(PilgrimType:EPilgrimTrackType Enum=@n3) -> (ReturnValue:ESparksInstrumentType Enum=Guitar)
   > SetScalarParameterValue(
    self=@n2
    ParameterName=Icon
    Value="To Float (Integer)"(InInt=GetInstrumentIconIndex!(
      InstrumentType:EPilgrimTrackType Enum=@n3
      SongMetaData:Pilgrim Song Metadata Object Reference
      StartingInstrumentIndex:ESparksInstrumentType Enum=@n4
      bUseKeyTarIcon:bool=false
     ) -> (InstrumentIdexIcon:int=0))
   )
   > SetScalarParameterValue(self=@n2, ParameterName=Size, Value=0.7)
   > Return Node!(execute, OutputPin:Material Instance Dynamic Object Reference=@n2)
  ]}
 GetLeaderboardFilterTypeFromTrackType {nodes [
   n1 = GetLeaderboardFilterTypeFromTrackType!() -> (then, TrackType:EPilgrimTrackType Enum=TrackGuitar) {is=title K2Node_FunctionEntry}
   > Return Node!(
    execute
    LeaderboardFilter:EPilgrimQuickplay_LeaderboardFilters Enum=Select!(
     TrackGuitar:EPilgrimQuickplay_LeaderboardFilters Enum=NewEnumerator4
     TrackBass:EPilgrimQuickplay_LeaderboardFilters Enum=NewEnumerator7
     TrackVocals:EPilgrimQuickplay_LeaderboardFilters Enum=NewEnumerator6
     TrackDrum:EPilgrimQuickplay_LeaderboardFilters Enum=NewEnumerator5
     TrackPlasticGuitar:EPilgrimQuickplay_LeaderboardFilters Enum=NewEnumerator13
     TrackPlasticBass:EPilgrimQuickplay_LeaderboardFilters Enum=NewEnumerator14
     TrackPlasticDrum:EPilgrimQuickplay_LeaderboardFilters Enum=NewEnumerator15
     TrackProVocals:EPilgrimQuickplay_LeaderboardFilters Enum=NewEnumerator19
     TrackProCymbals:EPilgrimQuickplay_LeaderboardFilters Enum=NewEnumerator18
     TrackEvents:EPilgrimQuickplay_LeaderboardFilters Enum=NewEnumerator11
     TrackNone:EPilgrimQuickplay_LeaderboardFilters Enum=NewEnumerator11
     Index:EPilgrimTrackType Enum=@n1
    ) -> (ReturnValue:EPilgrimQuickplay_LeaderboardFilters Enum)
   )
  ]}
 GetPlayerLeaderboardComponent {nodes [
   GetPlayerLeaderboardComponent() {is=title K2Node_FunctionEntry}
   > Return Node!(
    execute
    LeaderboardComponent:Pilgrim Leaderboard Player Component Object Reference=GetComponentByClass!(
     self:Actor Object Reference=GetOwningPlayer()
     ComponentClass:Actor Component Class Reference=/Script/PilgrimCoreRuntime.PilgrimLeaderboardPlayerComponent
    ) -> (ReturnValue:Pilgrim Leaderboard Player Component Object Reference)
   )
  ]}
 HandleBackButton {nodes [
   HandleBackButton() {is=title K2Node_FunctionEntry}
   > ClearQueuedLeaderboardRequests!(execute, ContextObject:Object Reference=Self-Reference()) -> (then)
   > DeactivateWidget()
  ]}
 InitializeFilters {nodes [
   InitializeFilters() {is=title K2Node_FunctionEntry}
   > n1 = Sequence!(execute) -> (then_0, then_1, then_2)
   > CreateCategoriesFromArray!(
    execute
    self
    Categories:Array of EPilgrimQuickplay_LeaderboardFilters Enums=Get Categories_CompetitorGroups() -> (Categories_CompetitorGroups:Array of EPilgrimQuickplay_LeaderboardFilters Enums=NewEnumerator0)
    TopNav Widget:WBP UIKit Tabs Base Object Reference=Get TopNavBar_EntryBox_CompetitorGroup() -> (TopNavBar_EntryBox_CompetitorGroup:WBP Pilgrim Quickplay Leaderboard Nav Bar Object Reference)
   ) -> (then) {is=title K2Node_CallFunction}
   CreateCategoriesFromArray(
    execute=@n1.then_2
    Categories=Get Categories_SeasonDisplay() -> (Categories_SeasonDisplay:Array of EPilgrimQuickplay_LeaderboardFilters Enums=NewEnumerator0)
    TopNav Widget=Get TopNavBar_EntryBox_SeasonDisplay() -> (TopNavBar_EntryBox_SeasonDisplay:WBP Pilgrim Quickplay Leaderboard Nav Bar Object Reference)
   )
   CreateCategoriesFromArray(
    execute=@n1.then_1
    Categories=Get Categories_ActivityType() -> (Categories_ActivityType:Array of EPilgrimQuickplay_LeaderboardFilters Enums=NewEnumerator0)
    TopNav Widget=Get TopNavBar_EntryBox_ActivityType() -> (TopNavBar_EntryBox_ActivityType:WBP Pilgrim Quickplay Leaderboard Nav Bar Object Reference)
   )
  ]}
 Is Band Tab Enabled {nodes [
   Is Band Tab Enabled() {is=title K2Node_FunctionEntry}
   n1 = Is Friends Only()
   > Return Node!(
    execute
    bEnabled:bool=OR Boolean(A=Get Enable Band Friend Leaderboard Option() -> (Enable Band Friend Leaderboard Option:bool), B=NOT Boolean(A=@n1))
   )
  ]}
 Is Friends Only {nodes [
   Is Friends Only() {is=title K2Node_FunctionEntry}
   n1 = GetSelectedTabIndex(self=Get TopNavBar_EntryBox_CompetitorGroup()) -> (ReturnValue=0)
   n2 = Get Categories_CompetitorGroups() -> (Categories_CompetitorGroups=NewEnumerator0)
   > n3 = Branch(Condition=Is Valid Index!(TargetArray:Array of EPilgrimQuickplay_LeaderboardFilters Enums=@n2, IndexToTest:int=@n1) -> (ReturnValue:bool))
   > Return Node!(
    execute
    IsFriendsOnly:bool="Equal (Enum)"(
     A:EPilgrimQuickplay_LeaderboardFilters Enum="Get (a ref)"(Array:Array of EPilgrimQuickplay_LeaderboardFilters Enums=@n2, Dimension 1=@n1) -> (Output:"EPilgrimQuickplay_LeaderboardFilters Enum (by ref)")
     B:EPilgrimQuickplay_LeaderboardFilters Enum=Literal enum EPilgrimQuickplay_LeaderboardFilters!(Enum:EPilgrimQuickplay_LeaderboardFilters Enum=NewEnumerator0) -> (ReturnValue:EPilgrimQuickplay_LeaderboardFilters Enum) {is=title K2Node_EnumLiteral}
    )
   )
   Return Node(execute=@n3.else, IsFriendsOnly=false)
  ]}
 Is Showing Solo Leaderboard {nodes [
   Is Showing Solo Leaderboard() {is=title K2Node_FunctionEntry}
   n1 = GetSelectedTabIndex(self=Get TopNavBar_EntryBox_ActivityType()) -> (ReturnValue=0)
   n2 = Get Categories_ActivityType() -> (Categories_ActivityType=NewEnumerator0)
   > n3 = Branch(Condition=Is Valid Index(TargetArray=@n2, IndexToTest=@n1))
   > Return Node!(
    execute
    ReturnValue:bool="Equal (Enum)"(A="Get (a ref)"(Array=@n2, Dimension 1=@n1), B=Literal enum EPilgrimQuickplay_LeaderboardFilters(Enum=NewEnumerator2))
   )
   Return Node(execute=@n3.else, ReturnValue=false)
  ]}
 IsAllTimeScoring {nodes [
   IsAllTimeScoring() {is=title K2Node_FunctionEntry}
   n1 = Get TopNavBar_EntryBox_SeasonDisplay()
   > n2 = Branch(Condition="Equal (Integer)"(A=GetSelectedTabIndex(self=@n1) -> (ReturnValue=0)))
   > Return Node!(execute, ShouldShowAllTime:bool=false)
   Return Node(execute=@n2.else, ShouldShowAllTime=true)
  ]}
 On Activity Type Selected Refresh Sub Type Filters {nodes [
   On Activity Type Selected Refresh Sub Type Filters() {is=title K2Node_FunctionEntry}
   n1 = Get Categories Array() -> (ReturnValue=NewEnumerator0)
   n2 = Get TopNav_EntryBox_SubFilters()
   > CreateCategoriesFromArray(Categories=@n1, TopNav Widget=@n2)
   CreateEmptyCategoriesFromArray!(execute, self, Categories:Array of EPilgrimQuickplay_LeaderboardFilters Enums=@n1, TopNav Widget:WBP UIKit Tabs Base Object Reference=@n2) -> (then) {is=title K2Node_CallFunction}
  ]}
 RefreshCurrentSongInfo {nodes [
   RefreshCurrentSongInfo() {is=title K2Node_FunctionEntry}
   n1 = Get CurrentSongShortName() -> (CurrentSongShortName=None)
   > Branch(Condition="Not Equal (Name)"(A=@n1) -> (ReturnValue=~))
   > n2 = Sequence()
   n3 = GetSongByShortName!(
    self:Pilgrim Song Catalog Object Reference=GetPilgrimSongCatalog!() -> (ReturnValue:Pilgrim Song Catalog Object Reference)
    ShortName:Name=@n1
   ) -> (ReturnValue:Pilgrim Song Metadata Object Reference)
   > "SetText (Text)"(
    self=Get SongTitle() -> (SongTitle:Common Text Object Reference)
    InText=GetTitleText!(self:Sparks Song Data Object Reference=@n3) -> (ReturnValue:Text)
   )
   "SetText (Text)"(
    execute=@n2.then_1
    self=Get SongArtist() -> (SongArtist:Common Text Object Reference)
    InText=GetArtistText!(self:Sparks Song Data Object Reference=@n3) -> (ReturnValue:Text)
   )
   LazyLoadAlbumArt!(
    execute=@n2.then_2
    Song Metadata:Pilgrim Song Metadata Object Reference=@n3
    Lazy Image:Fort Lazy Image Object Reference=Get AlbumArt() -> (AlbumArt:Fort Lazy Image Object Reference)
   ) -> (then)
  ]}
 RefreshFocusedPlayerEntry {nodes [
   RefreshFocusedPlayerEntry() {is=title K2Node_FunctionEntry}
   > n1 = Get!(execute:Exec) -> (then:Exec, else:Exec, PilgrimQuickplayLeaderboardVM:Pilgrim Quickplay Leaderboard VM Object Reference)
   > n2 = Get!(execute, self:Pilgrim Quickplay Leaderboard VM Object Reference=@n1) -> (then, else, LocalPlayerItem:Pilgrim Quickplay Leaderboard Item VM Object Reference)
   > SetPilgrimQuickplayLeaderboardItemVM!(
    execute
    self:WBP Pilgrim Quickplay Leaderboard Entry Object Reference=Get FocusedPlayerEntry() -> (FocusedPlayerEntry:WBP Pilgrim Quickplay Leaderboard Entry Object Reference)
    Viewmodel:Pilgrim Quickplay Leaderboard Item VM Object Reference=@n2
   ) -> (then)
   > n3 = Branch(Condition=Is Item Visible!(self:List View Object Reference=Get LeaderboardEntryList(), Item:Object Reference=@n2) -> (ReturnValue:bool))
   n4 = Get FocusedPlayerEntry()
   > SetVisibility(self=@n4, InVisibility=Collapsed)
   SetVisibility(execute=@n3.else, self=@n4, InVisibility=SelfHitTestInvisible)
   SetVisibility(execute=[@n2.else, @n1.else], self=Get FocusedPlayerEntry(), InVisibility=Collapsed)
  ]}
 SelectNextItemInList {nodes [
   n1 = SelectNextItemInList!() -> (then, ListView:Common List View Object Reference, DoesLoop:bool=~) {is=title K2Node_FunctionEntry}
   > ChangeSelectedItemInList!(execute, self, ListView:Common List View Object Reference=@n1, DoesLoop:bool=@n1.DoesLoop, DirectionOfChange:int=1) -> (then) {is=title K2Node_CallFunction}
  ]}
 SelectPrevItemInList {nodes [
   n1 = SelectPrevItemInList!() -> (then, ListView:Common List View Object Reference, DoesLoop:bool=~) {is=title K2Node_FunctionEntry}
   > ChangeSelectedItemInList(ListView=@n1, DoesLoop=@n1.DoesLoop, DirectionOfChange=-1)
  ]}
 "Set  Instrument Text Tab" {nodes [
   n1 = "Set  Instrument Text Tab"!() -> (then, InstrumentName:Name, IsActivityUpdated:bool=~) {is=title K2Node_FunctionEntry}
   n2 = Get Text_Instrument() -> (Text_Instrument:Common Text Object Reference)
   > SetVisibility(
    self=@n2
    InVisibility=Select!(Option 0:ESlateVisibility Enum=Collapsed, Option 1:ESlateVisibility Enum=SelfHitTestInvisible, Index:bool=@n1.IsActivityUpdated) -> (ReturnValue:ESlateVisibility Enum)
   )
   n3 = "To String (Name)"(InName=@n1)
   > "SetText (Text)"(self=@n2, InText="To Text (String)"(InString=Replace(SourceString=@n3, From=Peripheral, To=Pro)))
  ]}
 SetCategoryInstruments {nodes [
   n1 = SetCategoryInstruments!() -> (then, CategoryInstrumentsList:Array of EPilgrimTrackType Enums=TrackGuitar) {is=title K2Node_FunctionEntry}
   n2 = Get Categories_Instruments() -> (Categories_Instruments=NewEnumerator0)
   > Clear(TargetArray:Array of EPilgrimQuickplay_LeaderboardFilters Enums=@n2)
   > n3 = For Each Loop(Array:Array of EPilgrimTrackType Enums=@n1) -> (Array Element:EPilgrimTrackType Enum)
   > n4 = GetLeaderboardFilterTypeFromTrackType!(execute, self, TrackType:EPilgrimTrackType Enum=@n3) -> (then, LeaderboardFilter:EPilgrimQuickplay_LeaderboardFilters Enum=NewEnumerator0) {is=title K2Node_CallFunction}
   > Add(TargetArray:Array of EPilgrimQuickplay_LeaderboardFilters Enums=@n2, NewItem:"EPilgrimQuickplay_LeaderboardFilters Enum (by ref)"=@n4)
  ]}
 SetCategoryInstrumentsWithPeripherals {nodes [
   n1 = SetCategoryInstrumentsWithPeripherals!() -> (then, CategoryInstrumentsPeripheralList:Array of EPilgrimTrackType Enums=TrackGuitar) {is=title K2Node_FunctionEntry}
   n2 = Get Categories_Instruments_WithPeripherals() -> (Categories_Instruments_WithPeripherals=NewEnumerator0)
   > Clear(TargetArray=@n2)
   > n3 = For Each Loop(Array=@n1)
   > n4 = GetLeaderboardFilterTypeFromTrackType(TrackType=@n3) -> (LeaderboardFilter=NewEnumerator0)
   > Add(TargetArray=@n2, NewItem=@n4)
  ]}
 SetContentState {nodes [
   n1 = SetContentState!() -> (then, State:EPilgrimQuickplayLeaderboardContentState Enum=Loading) {is=title K2Node_FunctionEntry}
   > SetVisibility(
    self=[
     Get Content_NoLeaderboardEntries_MissingPart() -> (Content_NoLeaderboardEntries_MissingPart:Common Border Object Reference)
     Get Content_NoLeaderboardEntries() -> (Content_NoLeaderboardEntries:Common Border Object Reference)
     Get Content_LoadingEntries() -> (Content_LoadingEntries:Horizontal Box Object Reference)
     Get Content_LeaderboardRequestFailed() -> (Content_LeaderboardRequestFailed:Common Border Object Reference)
     Get LeaderboardEntryList()
    ]
    InVisibility=Collapsed
   )
   > n2 = Switch on EPilgrimQuickplayLeaderboardContentState!(execute, Selection:EPilgrimQuickplayLeaderboardContentState Enum=@n1) -> (Loading:Exec, Error:Exec, MissingPart:Exec, NoEntries:Exec, HasItems:Exec) {is=title K2Node_SwitchEnum}
   SetVisibility(execute=@n2.Error, self=Get Content_LeaderboardRequestFailed(), InVisibility=SelfHitTestInvisible)
   n3 = Get LeaderboardEntryList()
   SetVisibility(execute=@n2.HasItems, self=@n3, InVisibility=SelfHitTestInvisible)
   > SetSelectedIndex(self=@n3, Index=0)
   > ScrollIndexIntoView!(execute, self:List View Object Reference=@n3, Index:int=0) -> (then)
   > SetFocus(self=@n3)
   SetVisibility(execute=@n2.Loading, self=Get Content_LoadingEntries(), InVisibility=SelfHitTestInvisible)
   SetVisibility(execute=@n2.MissingPart, self=Get Content_NoLeaderboardEntries_MissingPart(), InVisibility=SelfHitTestInvisible)
   SetVisibility(execute=@n2.NoEntries, self=Get Content_NoLeaderboardEntries(), InVisibility=SelfHitTestInvisible)
  ]}
 SetCurrentSongTag {nodes [
   n1 = SetCurrentSongTag!() -> (then, Song ShortName:Name) {is=title K2Node_FunctionEntry}
   > Set CurrentSongShortName(CurrentSongShortName:Name=@n1) -> (Output_Get:Name=None)
   > RefreshCurrentSongInfo() {is=title K2Node_CallFunction}
   > RequestLeaderboardData()
  ]}
 SetFocusedPlayerEntry {nodes [
   n1 = SetFocusedPlayerEntry!() -> (then, VM:Pilgrim Quickplay Leaderboard Item VM Object Reference) {is=title K2Node_FunctionEntry}
   > n2 = Is Valid(InputObject=@n1)
   n3 = Get FocusedPlayerEntry()
   SetVisibility(execute=@n2.Is Not Valid, self=@n3, InVisibility=Collapsed)
   SetVisibility(execute=@n2.Is Valid, self=@n3, InVisibility=SelfHitTestInvisible)
   > SetPilgrimQuickplayLeaderboardItemVM(self=@n3, Viewmodel=@n1)
   > RefreshFocusedPlayerEntry() {is=title K2Node_CallFunction}
  ]}
 SetLoadingMore {nodes [
   n1 = SetLoadingMore!() -> (then, IsLoadingMore:bool=~) {is=title K2Node_FunctionEntry}
   > SetVisibility(
    self=Get Content_LoadingEntries_BelowList() -> (Content_LoadingEntries_BelowList:Horizontal Box Object Reference)
    InVisibility=Select(Option 0=Collapsed, Option 1=SelfHitTestInvisible, Index=@n1)
   )
  ]}
 SetStartingActiveTab {nodes [
   n1 = SetStartingActiveTab!() -> (then, SelectedPart:EPilgrimTrackType Enum=TrackGuitar) {is=title K2Node_FunctionEntry}
   > n2 = GetLeaderboardFilterTypeFromTrackType(TrackType=@n1) -> (LeaderboardFilter=NewEnumerator0)
   n3 = Get TopNav_EntryBox_SubFilters()
   > n4 = GetTabIndexByID!(execute:Exec, self=@n3, TabNameID=String To Name(InString=Enum to String(Enumerator:EPilgrimQuickplay_LeaderboardFilters Enum=@n2))) -> (then:Exec, ReturnValue=0)
   > SelectTabByIndex(self=@n3, Index=@n4, bSuppressClickFeedback=true)
  ]}
 Should Force Activity Update {nodes [
   Should Force Activity Update() {is=title K2Node_FunctionEntry}
   n1 = NOT Boolean(A=Is Friends Only() {is=title K2Node_CallFunction})
   n2 = Is Showing Solo Leaderboard() {is=title K2Node_CallFunction}
   n3 = Get Enable Band Friend Leaderboard Option()
   > Return Node!(execute, Should Update:bool=NOT Boolean(A=OR Boolean!(A=@n2, B=@n3, C:bool=@n1) -> (ReturnValue)))
  ]}
 Update All Tabs {nodes [
   n1 = Update All Tabs!() -> (then, Update Activity:bool=true, Update Competitor:bool=true, UpdateSeasonDisplay:bool=~) {is=title K2Node_FunctionEntry}
   > n2 = Sequence!(execute) -> (then_0, then_1, then_2, then_3:Exec, then_4:Exec)
   > Branch(Condition=@n1.UpdateSeasonDisplay)
   n3 = Get TopNavBar_EntryBox_SeasonDisplay()
   > n4 = Set SeasonDisplayPreviouslySelected(SeasonDisplayPreviouslySelected:int=GetSelectedTabIndex(self=@n3) -> (ReturnValue=0)) -> (Output_Get:int=0)
   > SelectTabByIndex(self=@n3, Index=@n4, bSuppressClickFeedback=true)
   Branch(execute=@n2.then_1, Condition=@n1.Update Competitor)
   n5 = Get TopNavBar_EntryBox_CompetitorGroup()
   > n6 = Set CompetitorGroupPreviouslySelected(CompetitorGroupPreviouslySelected:int=GetSelectedTabIndex(self=@n5) -> (ReturnValue=0)) -> (Output_Get:int=0)
   > SelectTabByIndex(self=@n5, Index=@n6, bSuppressClickFeedback=true)
   Branch(execute=@n2.then_2, Condition=@n1)
   n7 = Get CompetitorGroupPreviouslySelected() -> (CompetitorGroupPreviouslySelected:int=0)
   n8 = Get TopNavBar_EntryBox_ActivityType()
   n9 = GetSelectedTabIndex(self=@n8) -> (ReturnValue=0)
   > n10 = Branch(Condition=AND Boolean(A="Equal (Integer)"(A=@n9, B=1), B="Equal (Integer)"(A=@n7)))
   > n11 = Set ActivityTypePreviouslySelected(ActivityTypePreviouslySelected:int=0) -> (Output_Get:int=0)
   > SelectTabByIndex(self=@n8, Index=@n11, bSuppressClickFeedback=true)
   n12 = Set ActivityTypePreviouslySelected(execute=@n10.else, ActivityTypePreviouslySelected=@n9) -> (Output_Get=0)
   > SelectTabByIndex(self=@n8, Index=@n12, bSuppressClickFeedback=true)
   n13 = Get TopNav_EntryBox_SubFilters()
   n14 = Set SubFilterPreviouslySelected(execute=@n2.then_3, SubFilterPreviouslySelected:int=GetSelectedTabIndex(self=@n13) -> (ReturnValue=0)) -> (Output_Get:int=0)
   > SelectTabByIndex(self=@n13, Index=@n14, bSuppressClickFeedback=true)
   n15 = Literal enum EPilgrimQuickplay_LeaderboardFilters(Enum=NewEnumerator3)
   SetTabEnabled!(
    execute=@n2.then_4
    self:Common Tab List Widget Base Object Reference=Get TopNavBar_EntryBox_ActivityType()
    TabNameID:Name=String To Name(InString=Enum to String(Enumerator=@n15))
    bEnable:bool=Is Band Tab Enabled!() -> (bEnabled:bool) {is=title K2Node_CallFunction}
   ) -> (then)
   n16 = Get TopNav_EntryBox_SubFilters()
   n17 = Get TopNavBar_EntryBox_CompetitorGroup()
   n18 = Get TopNavBar_EntryBox_ActivityType()
   n19 = Get TopNavBar_EntryBox_SeasonDisplay()
   > Return Node!(
    execute
    SubFilterUpdated:bool="Equal (Integer)"(A=GetSelectedTabIndex(self=@n16) -> (ReturnValue=0), B=Get SubFilterPreviouslySelected() -> (SubFilterPreviouslySelected:int=0))
    CompetitorUpdated:bool="Equal (Integer)"(A=GetSelectedTabIndex(self=@n17) -> (ReturnValue=0), B=Get CompetitorGroupPreviouslySelected() -> (CompetitorGroupPreviouslySelected=0))
    SeasonDisplayUpdated:bool="Equal (Integer)"(A=GetSelectedTabIndex(self=@n19) -> (ReturnValue=0), B=Get SeasonDisplayPreviouslySelected() -> (SeasonDisplayPreviouslySelected:int=0))
    ActivityUpdated:bool="Equal (Integer)"(A=GetSelectedTabIndex(self=@n18) -> (ReturnValue=0), B=Get ActivityTypePreviouslySelected() -> (ActivityTypePreviouslySelected:int=0))
   )
  ]}
 Update All-Time Selection from Component {nodes [
   Update All-Time Selection from Component() {is=title K2Node_FunctionEntry}
   > n1 = For Each Loop with Break!(
    Exec:Exec
    Array:Array of EPilgrimQuickplay_LeaderboardFilters Enums=Get Categories_SeasonDisplay() -> (Categories_SeasonDisplay=NewEnumerator0)
    Break:Exec=@n3.then
   ) -> (LoopBody:Exec, Array Element:EPilgrimQuickplay_LeaderboardFilters Enum, Array Index:int, Completed:Exec)
   n2 = GetClientSelectedAllTimeOption!(self:Pilgrim Leaderboard Player Component Object Reference=GetPlayerLeaderboardComponent() {is=title K2Node_CallFunction}) -> (ReturnValue:bool)
   > Branch(Condition="Equal (Enum)"(
     A=@n1
     B=Select(
      Option 0:EPilgrimQuickplay_LeaderboardFilters Enum=NewEnumerator16
      Option 1:EPilgrimQuickplay_LeaderboardFilters Enum=NewEnumerator17
      Index=@n2
     ) -> (ReturnValue:EPilgrimQuickplay_LeaderboardFilters Enum)
    ))
   > n3 = SelectTabByIndex(self=Get TopNavBar_EntryBox_SeasonDisplay(), Index=@n1.Array Index, bSuppressClickFeedback=true)
  ]}
 Update Season Text {nodes [
   Update Season Text() {is=title K2Node_FunctionEntry}
   > n1 = EvaluateCurveTableRow!(
    execute
    CurveTableRowHandle:Curve Table Row Handle Structure=Make Curve Table Row Handle!(CurveTable:Curve Table Object Reference=/PilgrimQuickplay/DataTables/PilgrimQuickplayGameData, RowName:Name=SeasonData.SeasonNumber) -> (CurveTableRowHandle:Curve Table Row Handle Structure) {is=title K2Node_MakeStruct}
    InXY:float=0
    ContextString:String
   ) -> (then, OutXY:float, ReturnValue:bool)
   > n2 = Branch(Condition=@n1.ReturnValue)
   > "SetText (Text)"(
    self=Get SeasonText() -> (SeasonText:Common Text Object Reference)
    InText=Format Text!(Format="LOCTEXT(\"SEASON {SeasonNumber}\")", SeasonNumber:float=@n1) -> (Result)
   )
   "SetText (Text)"(execute=@n2.else, self=Get SeasonText())
  ]}
 Update Tab Actions {nodes [
   n1 = Get TopNavBar_EntryBox_CompetitorGroup()
   SetInputAction!(
    execute
    self:Common Action Widget Object Reference=Get PrevTabAction!(self:WBP UIKit Tabs Base Object Reference=@n1) -> (PrevTabAction:Common Action Widget Object Reference)
    InputActionRow:Data Table Row Handle Structure=Get Input Action Row Competitor Group() -> (Input Action Row Competitor Group:Data Table Row Handle Structure)
   ) -> (then)
   > SetInputActions!(
    execute
    self:UIKit Tab Group Object Reference=@n1
    NextTab:Data Table Row Handle Structure=Get Input Action Row Competitor Group()
    PreviousTab:Data Table Row Handle Structure
   ) -> (then)
   n2 = Get TopNavBar_EntryBox_ActivityType()
   > SetInputAction(
    self=Get PrevTabAction(self=@n2)
    InputActionRow=Get Input Action Row Activity Type() -> (Input Action Row Activity Type:Data Table Row Handle Structure)
   )
   > SetInputActions(self=@n2, NextTab=Get Input Action Row Activity Type())
   Update Tab Actions() {is=title K2Node_FunctionEntry}
  ]}
 UpdateComponentAllTimeSelection {nodes [
   UpdateComponentAllTimeSelection() {is=title K2Node_FunctionEntry}
   n1 = GetSelectedTabIndex(self=Get TopNavBar_EntryBox_SeasonDisplay()) -> (ReturnValue=0)
   n2 = Get Categories_SeasonDisplay() -> (Categories_SeasonDisplay=NewEnumerator0)
   > SetClientSelectedAllTimeOption!(
    execute
    self:Pilgrim Leaderboard Player Component Object Reference=GetPlayerLeaderboardComponent()
    SelectedAllTimeOption:bool="Equal (Enum)"(
     A="Get (a copy)"(Array:Array of EPilgrimQuickplay_LeaderboardFilters Enums=@n2, Dimension 1=@n1) -> (Output:EPilgrimQuickplay_LeaderboardFilters Enum)
     B=NewEnumerator17
    )
   ) -> (then)
  ]}
]
variables [
 Categories_ActivityType {type="array<EPilgrimQuickplay_LeaderboardFilters>", default [Solo,Band]}
 Categories_BandSize {type="array<EPilgrimQuickplay_LeaderboardFilters>", default [Duo,Trio,Squad]}
 Categories_CompetitorGroups {type="array<EPilgrimQuickplay_LeaderboardFilters>", default [Everyone,Friends]}
 Categories_Instruments {type="array<EPilgrimQuickplay_LeaderboardFilters>", default [Tap Vocals, Bass, Lead, Drums]}
 Categories_Instruments_WithPeripherals {
  type="array<EPilgrimQuickplay_LeaderboardFilters>"
  default [Tap Vocals, Bass, Lead, Drums, Karaoke, Pro Bass, Pro Lead, Pro Drums, Pro Drums + Cymbals]
 }
 Categories_SeasonDisplay {type="array<EPilgrimQuickplay_LeaderboardFilters>", default [Season,All-Time]}
 CurrentPage {type=integer}
 CurrentSongShortName {type=string, default=None}
 Enable Band Friend Leaderboard Option {type=boolean}
 Input Action Row Activity Type {type=DataTableRowHandle, default {dataTable=&InputActionDataTable, rowName=Leaderboard_Activity_Tab}}
 Input Action Row Competitor Group {type=DataTableRowHandle, default {dataTable=&InputActionDataTable, rowName=Leaderboard_CompetitorGroup_Tab}}
 InstrumentsTabTypes {
  type="array<EPilgrimTrackType>"
  default [TrackVocals, TrackBass, TrackGuitar, TrackDrum, TrackProVocals, TrackPlasticBass, TrackPlasticGuitar, TrackPlasticDrum, TrackProCymbals]
 }
 LastActivitySelection {type=integer}
 LastInstrumentSelection {type=integer}
 Local Tabs View Models {type="array<UIKitTabGroupButtonViewModel>", default []}
 MaxRetries {type=integer}
 RemainingRetries {type=integer}
]
widgets {
 info {
  inheritedWidgetCount=0
  parentClass=/Script/PilgrimQuickplayUI.PilgrimCommonActivatableWidget
  rootWidgetClass=&Overlay
  widgetCount=60
 }
 widgets [
  Overlay_31 {widgetClassPath=&Overlay, widget=Overlay_31}
  BG {
   widgetClassPath=&Border
   parent=Overlay_31
   widget=BG
   properties {
    visibility=SelfHitTestInvisible
    background {
     bIsDynamicallyLoaded=false
     drawAs=Image
     imageType=FullColor
     mirroring=NoMirror
     resourceName=None
     resourceObject=/SparksUICommon/Materials/MI_UI_Sparks_Fullscreen_Background_Triangles_LessPurple
     tiling=NoTile
     imageSize=<32 32>
     margin=<ltrb 0 0 0 0>
     outlineSettings {
      bUseBrushTransparency=false
      roundingType=HalfHeightRadius
      width=0
      color {colorUseRule=UseColor_Specified, specifiedColor=<rgba 0 0 0 0>}
      cornerRadii=<0 0 0 0>
     }
     tintColor {colorUseRule=UseColor_Specified, specifiedColor=<rgba 1 1 1 1>}
     uVRegion {bIsValid=false, max=<0 0>, min=<0 0>}
    }
    padding=<ltrb 0 0 0 0>
   }
   slot {class=OverlaySlot, properties {horizontalAlignment=HAlign_Fill, verticalAlignment=VAlign_Fill}}
  }
  HorizontalBox-Backgrounds {
   widgetClassPath=&HorizontalBox
   parent=BG
   widget=HorizontalBox-Backgrounds
   slot {class=BorderSlot, properties {padding=<ltrb 0 0 0 0>}}
  }
  BG_left {
   widgetClassPath=&Border
   parent=HorizontalBox-Backgrounds
   widget=BG_left
   properties {
    renderOpacity=0.7
    background {
     bIsDynamicallyLoaded=false
     drawAs=Image
     imageType=FullColor
     mirroring=NoMirror
     resourceName=None
     resourceObject=/PilgrimQuickplay/Widgets/SetlistAndSongDisplay/Duplicated/Materials/MI_Leaderboards_BGwithFade_Inst_Inst_Inst
     tiling=NoTile
     imageSize=<0 0>
     margin=<ltrb 0 0 0 0>
     outlineSettings {
      bUseBrushTransparency=false
      roundingType=HalfHeightRadius
      width=0
      color {colorUseRule=UseColor_Specified, specifiedColor=<rgba 0 0 0 0>}
      cornerRadii=<0 0 0 0>
     }
     tintColor {colorUseRule=UseColor_Specified, specifiedColor=<rgba 1 1 1 1>}
     uVRegion {bIsValid=false, max=<0 0>, min=<0 0>}
    }
    padding=<ltrb 0 0 0 0>
   }
   slot {class=HorizontalBoxSlot, properties {size {sizeRule=Fill, value=0.4}}}
  }
  Spacer_92 {
   widgetClassPath=&Spacer
   parent=HorizontalBox-Backgrounds
   widget=Spacer_92
   slot {class=HorizontalBoxSlot, properties {size {sizeRule=Fill, value=0.5}}}
  }
  HorizontalBox_43 {
   widgetClassPath=&HorizontalBox
   parent=Overlay_31
   widget=HorizontalBox_43
   slot {class=OverlaySlot, properties {horizontalAlignment=HAlign_Fill, verticalAlignment=VAlign_Fill}}
  }
  Border_60 {
   widgetClassPath=&Border
   parent=HorizontalBox_43
   widget=Border_60
   properties {brushColor=<rgba 1 1 1 0>, padding=<ltrb 48 48 48 30>}
   slot {class=HorizontalBoxSlot}
  }
  SZ_LeftPanel {
   widgetClassPath=&SafeZone
   bIsVariable=true
   parent=Border_60
   widget=SZ_LeftPanel
   properties {padRight=false}
   slot {class=BorderSlot, properties {padding=<ltrb 48 48 48 30>}}
  }
  SizeBox_LeaderboardsContainerSize {
   widgetClassPath=&SizeBox
   parent=SZ_LeftPanel
   widget=SizeBox_LeaderboardsContainerSize
   properties {bOverride_WidthOverride=true, widthOverride=500}
   slot {class=SafeZoneSlot}
  }
  DetailsVerticalBox {widgetClassPath=&VerticalBox, parent=SizeBox_LeaderboardsContainerSize, widget=DetailsVerticalBox, slot {class=SizeBoxSlot}}
  PageTitle {
   widgetClassPath=&CommonTextBlock
   parent=DetailsVerticalBox
   widget=PageTitle
   properties {
    scrollStyle=/PilgrimQuickplay/Widgets/ScrollStyles/TextScrollStyle_PilgrimQuickplay_Setlist.TextScrollStyle_PilgrimQuickplay_Setlist_C
    style=/FortUILibrary/Styles/Typography/Foundation/TextStyle_BurbankSmall_L_White.TextStyle_BurbankSmall_L_White_C
    text=LEADERBOARDs
    visibility=HitTestInvisible
    font {
     bForceMonospaced=false
     bMaterialIsStencil=false
     fontMaterial=null
     fontObject=&BurbankSmall
     letterSpacing=0
     monospacedWidth=1
     size=36
     skewAmount=0
     typefaceFontName=Black
     outlineSettings {
      bApplyOutlineToDropShadows=false
      bMiteredCorners=false
      bSeparateFillAlpha=false
      outlineMaterial=null
      outlineSize=0
      outlineColor=<rgba 0 0 0 1>
     }
    }
    shadowOffset=<0 0>
   }
   slot {class=VerticalBoxSlot, properties {verticalAlignment=VAlign_Center}}
  }
  SeasonText {
   widgetClassPath=&CommonTextBlock
   bIsVariable=true
   parent=DetailsVerticalBox
   widget=SeasonText
   properties {
    text=SEASON 2
    font {
     bForceMonospaced=false
     bMaterialIsStencil=false
     fontMaterial=null
     fontObject=&HeadingNow
     letterSpacing=0
     monospacedWidth=1
     size=15.75
     skewAmount=0
     typefaceFontName=76 Bold
     outlineSettings {
      bApplyOutlineToDropShadows=false
      bMiteredCorners=false
      bSeparateFillAlpha=false
      outlineMaterial=null
      outlineSize=0
      outlineColor=<rgba 0 0 0 1>
     }
    }
    shadowOffset=<0 0>
   }
   slot {class=VerticalBoxSlot, properties {verticalAlignment=VAlign_Bottom}}
  }
  TopHeaderInfo {
   widgetClassPath=&HorizontalBox
   parent=DetailsVerticalBox
   widget=TopHeaderInfo
   slot {class=VerticalBoxSlot, properties {padding=<ltrb 0 -18 0 -20>}}
  }
  Logo {
   widgetClassPath=&Image
   bIsVariable=true
   parent=TopHeaderInfo
   widget=Logo
   properties {
    visibility=Collapsed
    brush {
     bIsDynamicallyLoaded=false
     drawAs=Image
     imageType=NoImage
     mirroring=NoMirror
     resourceName=None
     resourceObject=/PilgrimQuickplay/Widgets/Results/SongResults/T_UI_ProgressFestivalLogo
     tiling=NoTile
     imageSize=<160 80>
     margin=<ltrb 0 0 0 0>
     outlineSettings {
      bUseBrushTransparency=false
      roundingType=HalfHeightRadius
      width=0
      color {colorUseRule=UseColor_Specified, specifiedColor=<rgba 0 0 0 0>}
      cornerRadii=<0 0 0 0>
     }
     tintColor {colorUseRule=UseColor_Specified, specifiedColor=<rgba 1 1 1 1>}
     uVRegion {bIsValid=false, max=<0 0>, min=<0 0>}
    }
   }
   slot {class=HorizontalBoxSlot, properties {horizontalAlignment=HAlign_Left, verticalAlignment=VAlign_Top}}
  }
  ScaleBox_2 {
   widgetClassPath=&ScaleBox
   parent=DetailsVerticalBox
   widget=ScaleBox_2
   properties {stretch=ScaleBySafeZone}
   slot {class=VerticalBoxSlot, properties {padding=<ltrb 0 45 0 0>}}
  }
  SizeBox_0 {
   widgetClassPath=&SizeBox
   parent=ScaleBox_2
   widget=SizeBox_0
   properties {
    bOverride_HeightOverride=true
    bOverride_MaxAspectRatio=true
    bOverride_MinAspectRatio=true
    bOverride_WidthOverride=true
    heightOverride=500
    widthOverride=500
   }
   slot {class=ScaleBoxSlot, properties {horizontalAlignment=HAlign_Left}}
  }
  AlbumArt {
   widgetClassPath=&FortLazyImage
   bIsVariable=true
   parent=SizeBox_0
   widget=AlbumArt
   properties {
    materialTextureParamName=T_AlbumArt
    brush {
     bIsDynamicallyLoaded=false
     drawAs=Image
     imageType=FullColor
     mirroring=NoMirror
     resourceName=None
     resourceObject=/SparksUICommon/Materials/MI_UI_Sparks_SetList_AlbumArt_Rounded
     tiling=NoTile
     imageSize=<320 320>
     margin=<ltrb 0 0 0 0>
     outlineSettings {
      bUseBrushTransparency=false
      roundingType=HalfHeightRadius
      width=0
      color {colorUseRule=UseColor_Specified, specifiedColor=<rgba 0 0 0 0>}
      cornerRadii=<0 0 0 0>
     }
     tintColor {colorUseRule=UseColor_Specified, specifiedColor=<rgba 1 1 1 1>}
     uVRegion {bIsValid=false, max=<0 0>, min=<0 0>}
    }
   }
   slot {class=SizeBoxSlot}
  }
  Spacer_3 {widgetClassPath=&Spacer, parent=DetailsVerticalBox, widget=Spacer_3, properties {size=<1 20>}, slot {class=VerticalBoxSlot}}
  SongTitle {
   widgetClassPath=&CommonTextBlock
   bIsVariable=true
   parent=DetailsVerticalBox
   widget=SongTitle
   properties {
    scrollStyle=&TextScrollStyle_PilgrimQuickplay_Setlist_C
    style=/FortUILibrary/Styles/Typography/Foundation/TextStyle_BurbankSmall_M_White.TextStyle_BurbankSmall_M_White_C
    text=SONG TITLE
    visibility=HitTestInvisible
    font {
     bForceMonospaced=false
     bMaterialIsStencil=false
     fontMaterial=null
     fontObject=&BurbankSmall
     letterSpacing=0
     monospacedWidth=1
     size=24
     skewAmount=0
     typefaceFontName=Black
     outlineSettings {
      bApplyOutlineToDropShadows=false
      bMiteredCorners=false
      bSeparateFillAlpha=false
      outlineMaterial=null
      outlineSize=0
      outlineColor=<rgba 0 0 0 1>
     }
    }
    shadowOffset=<0 0>
   }
   slot {class=VerticalBoxSlot, properties {verticalAlignment=VAlign_Center}}
  }
  SongArtist {
   widgetClassPath=&CommonTextBlock
   bIsVariable=true
   parent=DetailsVerticalBox
   widget=SongArtist
   properties {
    clipping=OnDemand
    scrollStyle=&TextScrollStyle_PilgrimQuickplay_Setlist_C
    style=/FortUILibrary/Styles/Typography/Foundation/TextStyle_BurbankSmall_S_White.TextStyle_BurbankSmall_S_White_C
    text=ARTIST
    visibility=HitTestInvisible
    font {
     bForceMonospaced=false
     bMaterialIsStencil=false
     fontMaterial=null
     fontObject=&BurbankSmall
     letterSpacing=0
     monospacedWidth=1
     size=18
     skewAmount=0
     typefaceFontName=Black
     outlineSettings {
      bApplyOutlineToDropShadows=false
      bMiteredCorners=false
      bSeparateFillAlpha=false
      outlineMaterial=null
      outlineSize=0
      outlineColor=<rgba 0 0 0 1>
     }
    }
    shadowOffset=<0 0>
   }
   slot {class=VerticalBoxSlot, properties {verticalAlignment=VAlign_Center}}
  }
  Spacer_1 {widgetClassPath=&Spacer, parent=DetailsVerticalBox, widget=Spacer_1, properties {size=<1 30>}, slot {class=VerticalBoxSlot}}
  LeaderboardFilterLabel {
   widgetClassPath=&CommonTextBlock
   parent=DetailsVerticalBox
   widget=LeaderboardFilterLabel
   properties {
    scrollStyle=&TextScrollStyle_PilgrimQuickplay_Setlist_C
    style=&TextStyle_BurbankSmall_S_White_C
    text=Leaderboard Type
    visibility=HitTestInvisible
    font {
     bForceMonospaced=false
     bMaterialIsStencil=false
     fontMaterial=null
     fontObject=&BurbankSmall
     letterSpacing=0
     monospacedWidth=1
     size=18
     skewAmount=0
     typefaceFontName=Black
     outlineSettings {
      bApplyOutlineToDropShadows=false
      bMiteredCorners=false
      bSeparateFillAlpha=false
      outlineMaterial=null
      outlineSize=0
      outlineColor=<rgba 0 0 0 1>
     }
    }
    shadowOffset=<0 0>
   }
   slot {class=VerticalBoxSlot, properties {verticalAlignment=VAlign_Center}}
  }
  Spacer_4 {widgetClassPath=&Spacer, parent=DetailsVerticalBox, widget=Spacer_4, properties {size=<1 10>}, slot {class=VerticalBoxSlot}}
  Tab_SeasonDisplayType {
   widgetClassPath=&Overlay
   parent=DetailsVerticalBox
   widget=Tab_SeasonDisplayType
   slot {class=VerticalBoxSlot, properties {horizontalAlignment=HAlign_Left, verticalAlignment=VAlign_Center}}
  }
  TabSeasonDisplayBacking {
   widgetClassPath=&Image
   parent=Tab_SeasonDisplayType
   widget=TabSeasonDisplayBacking
   properties {
    visibility=SelfHitTestInvisible
    brush {
     bIsDynamicallyLoaded=false
     drawAs=Image
     imageType=NoImage
     mirroring=NoMirror
     resourceName=None
     resourceObject=/PilgrimCore/Widgets/Materials/MI_Pilgrim_TabBacking_Panel
     tiling=NoTile
     imageSize=<32 32>
     margin=<ltrb 0 0 0 0>
     outlineSettings {
      bUseBrushTransparency=false
      roundingType=HalfHeightRadius
      width=0
      color {colorUseRule=UseColor_Specified, specifiedColor=<rgba 0 0 0 0>}
      cornerRadii=<0 0 0 0>
     }
     tintColor {colorUseRule=UseColor_Specified, specifiedColor=<rgba 1 1 1 1>}
     uVRegion {bIsValid=false, max=<0 0>, min=<0 0>}
    }
   }
   slot {class=OverlaySlot, properties {horizontalAlignment=HAlign_Fill, verticalAlignment=VAlign_Fill, padding=<ltrb -8 0 -8 0>}}
  }
  TopNavBar_EntryBox_SeasonDisplay {
   widgetClassPath=&/WBP_PilgrimQuickplay_Leaderboard_NavBar
   bIsVariable=true
   parent=Tab_SeasonDisplayType
   widget=TopNavBar_EntryBox_SeasonDisplay
   properties {
    bDeferRebuildingTabList=true
    buttonClass=/PilgrimCore/Widgets/SparksUILibrary/WBP_Sparks_TextTabs_Button_Selectable.WBP_Sparks_TextTabs_Button_Selectable_C
    initialSelectedTabIndex=0
    minButtonWidth=188
    tabSpacing=8
    iconSize=<0 0>
    localTabsViewModels [null,null]
    nextTabInputActionData {dataTable=&InputActionDataTable, rowName=GenericForward}
    previousTabInputActionData {dataTable=&InputActionDataTable, rowName=GenericForward}
    tabGroupPadding=<ltrb 0 8 0 8>
   }
   slot {class=OverlaySlot, properties {horizontalAlignment=HAlign_Fill, verticalAlignment=VAlign_Fill}}
  }
  Spacer {widgetClassPath=&Spacer, parent=DetailsVerticalBox, widget=Spacer, properties {size=<1 10>}, slot {class=VerticalBoxSlot}}
  Tab_CompetitorGroup {
   widgetClassPath=&Overlay
   parent=DetailsVerticalBox
   widget=Tab_CompetitorGroup
   slot {class=VerticalBoxSlot, properties {horizontalAlignment=HAlign_Left, verticalAlignment=VAlign_Center}}
  }
  TabCompetitorBacking {
   widgetClassPath=&Image
   parent=Tab_CompetitorGroup
   widget=TabCompetitorBacking
   properties {
    visibility=SelfHitTestInvisible
    brush {
     bIsDynamicallyLoaded=false
     drawAs=Image
     imageType=NoImage
     mirroring=NoMirror
     resourceName=None
     resourceObject=&MI_Pilgrim_TabBacking_Panel
     tiling=NoTile
     imageSize=<32 32>
     margin=<ltrb 0 0 0 0>
     outlineSettings {
      bUseBrushTransparency=false
      roundingType=HalfHeightRadius
      width=0
      color {colorUseRule=UseColor_Specified, specifiedColor=<rgba 0 0 0 0>}
      cornerRadii=<0 0 0 0>
     }
     tintColor {colorUseRule=UseColor_Specified, specifiedColor=<rgba 1 1 1 1>}
     uVRegion {bIsValid=false, max=<0 0>, min=<0 0>}
    }
   }
   slot {class=OverlaySlot, properties {horizontalAlignment=HAlign_Fill, verticalAlignment=VAlign_Fill, padding=<ltrb -8 0 -8 0>}}
  }
  TopNavBar_EntryBox_CompetitorGroup {
   widgetClassPath=&WBP_PilgrimQuickplay_Leaderboard_NavBar
   bIsVariable=true
   parent=Tab_CompetitorGroup
   widget=TopNavBar_EntryBox_CompetitorGroup
   properties {
    bDeferRebuildingTabList=true
    buttonClass=&WBP_Sparks_TextTabs_Button_Selectable_C
    initialSelectedTabIndex=0
    minButtonWidth=188
    tabSpacing=8
    iconSize=<0 0>
    localTabsViewModels [null,null]
    nextTabInputActionData {dataTable=/PilgrimQuickplay/Widgets/InputActionDataTables/FM_PilgrimQuickplay_InputActionDataTable, rowName=LeaderboardTertiaryPrev}
    previousTabInputActionData {dataTable=&FM_PilgrimQuickplay_InputActionDataTable, rowName=LeaderboardTertiaryPrev}
    tabGroupPadding=<ltrb 0 8 0 8>
   }
   slot {class=OverlaySlot, properties {horizontalAlignment=HAlign_Fill, verticalAlignment=VAlign_Fill}}
  }
  Spacer_5 {widgetClassPath=&Spacer, parent=DetailsVerticalBox, widget=Spacer_5, properties {size=<1 10>}, slot {class=VerticalBoxSlot}}
  Tab_ActivityType {
   widgetClassPath=&Overlay
   parent=DetailsVerticalBox
   widget=Tab_ActivityType
   slot {class=VerticalBoxSlot, properties {horizontalAlignment=HAlign_Left, verticalAlignment=VAlign_Center}}
  }
  TabActivityBacking {
   widgetClassPath=&Image
   parent=Tab_ActivityType
   widget=TabActivityBacking
   properties {
    visibility=SelfHitTestInvisible
    brush {
     bIsDynamicallyLoaded=false
     drawAs=Image
     imageType=NoImage
     mirroring=NoMirror
     resourceName=None
     resourceObject=&MI_Pilgrim_TabBacking_Panel
     tiling=NoTile
     imageSize=<32 32>
     margin=<ltrb 0 0 0 0>
     outlineSettings {
      bUseBrushTransparency=false
      roundingType=HalfHeightRadius
      width=0
      color {colorUseRule=UseColor_Specified, specifiedColor=<rgba 0 0 0 0>}
      cornerRadii=<0 0 0 0>
     }
     tintColor {colorUseRule=UseColor_Specified, specifiedColor=<rgba 1 1 1 1>}
     uVRegion {bIsValid=false, max=<0 0>, min=<0 0>}
    }
   }
   slot {class=OverlaySlot, properties {horizontalAlignment=HAlign_Fill, verticalAlignment=VAlign_Fill, padding=<ltrb -8 0 -8 0>}}
  }
  TopNavBar_EntryBox_ActivityType {
   widgetClassPath=&WBP_PilgrimQuickplay_Leaderboard_NavBar
   bIsVariable=true
   parent=Tab_ActivityType
   widget=TopNavBar_EntryBox_ActivityType
   properties {
    bDeferRebuildingTabList=true
    buttonClass=&WBP_Sparks_TextTabs_Button_Selectable_C
    initialSelectedTabIndex=0
    minButtonWidth=188
    tabSpacing=8
    iconSize=<0 0>
    localTabsViewModels [null,null]
    nextTabInputActionData {dataTable=&FM_PilgrimQuickplay_InputActionDataTable, rowName=LeaderboardTertiaryNext}
    previousTabInputActionData {dataTable=&FM_PilgrimQuickplay_InputActionDataTable, rowName=LeaderboardTertiaryNext}
    tabGroupPadding=<ltrb 0 8 0 8>
   }
   slot {class=OverlaySlot, properties {horizontalAlignment=HAlign_Fill, verticalAlignment=VAlign_Fill}}
  }
  SZ_LeaderBoard {
   widgetClassPath=&SafeZone
   parent=HorizontalBox_43
   widget=SZ_LeaderBoard
   properties {padLeft=false}
   slot {class=HorizontalBoxSlot, properties {padding=<ltrb 50 50 0 0>, size {sizeRule=Fill, value=1}}}
  }
  EntriesVerticalBox {widgetClassPath=&VerticalBox, parent=SZ_LeaderBoard, widget=EntriesVerticalBox, slot {class=SafeZoneSlot}}
  Text_Instrument {
   widgetClassPath=&CommonTextBlock
   bIsVariable=true
   parent=EntriesVerticalBox
   widget=Text_Instrument
   properties {
    justification=Center
    text=INSTRUMENT
    textTransformPolicy=ToUpper
    visibility=Collapsed
    colorAndOpacity {colorUseRule=UseColor_Specified, specifiedColor=<rgba 0.5 0.8333 1 1>}
    font {
     bForceMonospaced=false
     bMaterialIsStencil=false
     fontMaterial=null
     fontObject=&HeadingNow
     letterSpacing=0
     monospacedWidth=1
     size=13.5
     skewAmount=0
     typefaceFontName=76 Bold Italic
     outlineSettings {
      bApplyOutlineToDropShadows=false
      bMiteredCorners=false
      bSeparateFillAlpha=false
      outlineMaterial=null
      outlineSize=0
      outlineColor=<rgba 0 0 0 1>
     }
    }
    shadowOffset=<0 0>
   }
   slot {class=VerticalBoxSlot, properties {horizontalAlignment=HAlign_Center, verticalAlignment=VAlign_Center, padding=<ltrb 0 0 0 4>}}
  }
  Tab_Subfilters {
   widgetClassPath=&Overlay
   parent=EntriesVerticalBox
   widget=Tab_Subfilters
   slot {class=VerticalBoxSlot, properties {horizontalAlignment=HAlign_Center, verticalAlignment=VAlign_Center}}
  }
  TabSubFiltersBacking {
   widgetClassPath=&Image
   parent=Tab_Subfilters
   widget=TabSubFiltersBacking
   properties {
    visibility=SelfHitTestInvisible
    brush {
     bIsDynamicallyLoaded=false
     drawAs=Image
     imageType=NoImage
     mirroring=NoMirror
     resourceName=None
     resourceObject=&MI_Pilgrim_TabBacking_Panel
     tiling=NoTile
     imageSize=<32 32>
     margin=<ltrb 0 0 0 0>
     outlineSettings {
      bUseBrushTransparency=false
      roundingType=HalfHeightRadius
      width=0
      color {colorUseRule=UseColor_Specified, specifiedColor=<rgba 0 0 0 0>}
      cornerRadii=<0 0 0 0>
     }
     tintColor {colorUseRule=UseColor_Specified, specifiedColor=<rgba 1 1 1 1>}
     uVRegion {bIsValid=false, max=<0 0>, min=<0 0>}
    }
   }
   slot {class=OverlaySlot, properties {horizontalAlignment=HAlign_Fill, verticalAlignment=VAlign_Fill, padding=<ltrb -8 0 -8 0>}}
  }
  TopNav_EntryBox_SubFilters {
   widgetClassPath=/Game/UI/Frontend/TopBar/WBP_UIKit_Tabs_TopNavBar
   bIsVariable=true
   parent=Tab_Subfilters
   widget=TopNav_EntryBox_SubFilters
   properties {
    bDeferRebuildingTabList=true
    buttonClass=&WBP_Sparks_TextTabs_Button_Selectable_C
    initialSelectedTabIndex=0
    inputActions=Bumpers
    minButtonHeight=64
    minButtonWidth=64
    tabSpacing=8
    iconSize=<64 64>
    tabGroupPadding=<ltrb 0 8 0 8>
   }
   slot {class=OverlaySlot, properties {horizontalAlignment=HAlign_Center, verticalAlignment=VAlign_Fill}}
  }
  ContentOverlay {
   widgetClassPath=&Overlay
   parent=EntriesVerticalBox
   widget=ContentOverlay
   slot {class=VerticalBoxSlot, properties {padding=<ltrb 0 25 50 0>, size {sizeRule=Fill, value=1}}}
  }
  Content_LoadingEntries_BelowList {
   widgetClassPath=&HorizontalBox
   bIsVariable=true
   parent=ContentOverlay
   widget=Content_LoadingEntries_BelowList
   properties {visibility=Hidden}
   slot {class=OverlaySlot, properties {horizontalAlignment=HAlign_Fill, verticalAlignment=VAlign_Bottom, padding=<ltrb 0 0 0 -54>}}
  }
  ScaleBox {
   widgetClassPath=&ScaleBox
   parent=Content_LoadingEntries_BelowList
   widget=ScaleBox
   properties {stretchDirection=DownOnly}
   slot {class=HorizontalBoxSlot, properties {verticalAlignment=VAlign_Center, padding=<ltrb 25 0 25 0>}}
  }
  CommonTextBlock {
   widgetClassPath=&CommonTextBlock
   parent=ScaleBox
   widget=CommonTextBlock
   properties {
    text=Loading Leaderboard Entries
    textTransformPolicy=ToUpper
    font {
     bForceMonospaced=false
     bMaterialIsStencil=false
     fontMaterial=null
     fontObject=&BurbankSmall
     letterSpacing=0
     monospacedWidth=1
     size=18
     skewAmount=0
     typefaceFontName=Bold
     outlineSettings {
      bApplyOutlineToDropShadows=false
      bMiteredCorners=false
      bSeparateFillAlpha=false
      outlineMaterial=null
      outlineSize=0
      outlineColor=<rgba 0 0 0 1>
     }
    }
    shadowOffset=<0 0>
   }
   slot {class=ScaleBoxSlot}
  }
  Image {
   widgetClassPath=&Image
   bIsVariable=true
   parent=Content_LoadingEntries_BelowList
   widget=Image
   properties {
    visibility=SelfHitTestInvisible
    brush {
     bIsDynamicallyLoaded=false
     drawAs=Image
     imageType=NoImage
     mirroring=NoMirror
     resourceName=None
     resourceObject=/Game/UI/Foundation/Materials/Throbber_Material-Monochrome
     tiling=NoTile
     imageSize=<32 32>
     margin=<ltrb 0 0 0 0>
     outlineSettings {
      bUseBrushTransparency=false
      roundingType=HalfHeightRadius
      width=0
      color {colorUseRule=UseColor_Specified, specifiedColor=<rgba 0 0 0 0>}
      cornerRadii=<0 0 0 0>
     }
     tintColor {colorUseRule=UseColor_Specified, specifiedColor=<rgba 1 1 1 1>}
     uVRegion {bIsValid=false, max=<0 0>, min=<0 0>}
    }
   }
   slot {class=HorizontalBoxSlot, properties {horizontalAlignment=HAlign_Center, verticalAlignment=VAlign_Center}}
  }
  Content_LeaderboardRequestFailed {
   widgetClassPath=&CommonBorder
   bIsVariable=true
   parent=ContentOverlay
   widget=Content_LeaderboardRequestFailed
   properties {visibility=Hidden, padding=<ltrb 10 60 10 25>}
   slot {class=OverlaySlot, properties {horizontalAlignment=HAlign_Center, verticalAlignment=VAlign_Center}}
  }
  Text_LeaderboardRequestFailed {
   widgetClassPath=&CommonTextBlock
   bIsVariable=true
   parent=Content_LeaderboardRequestFailed
   widget=Text_LeaderboardRequestFailed
   properties {
    autoWrapText=true
    justification=Center
    text="No Leaderboard Available \r\nCheck Back Later"
    font {
     bForceMonospaced=false
     bMaterialIsStencil=false
     fontMaterial=null
     fontObject=&Burbank
     letterSpacing=0
     monospacedWidth=1
     size=24
     skewAmount=0
     typefaceFontName=Burbank-PlayerPower
     outlineSettings {
      bApplyOutlineToDropShadows=false
      bMiteredCorners=false
      bSeparateFillAlpha=false
      outlineMaterial=null
      outlineSize=0
      outlineColor=<rgba 0 0 0 1>
     }
    }
    shadowOffset=<0 0>
   }
   slot {class=BorderSlot, properties {padding=<ltrb 10 60 10 25>}}
  }
  Content_LoadingEntries {
   widgetClassPath=&HorizontalBox
   bIsVariable=true
   parent=ContentOverlay
   widget=Content_LoadingEntries
   properties {visibility=Hidden}
   slot {class=OverlaySlot, properties {horizontalAlignment=HAlign_Center, verticalAlignment=VAlign_Center}}
  }
  ScaleBox_0 {
   widgetClassPath=&ScaleBox
   parent=Content_LoadingEntries
   widget=ScaleBox_0
   properties {stretchDirection=DownOnly}
   slot {class=HorizontalBoxSlot, properties {verticalAlignment=VAlign_Center, padding=<ltrb 25 0 25 0>}}
  }
  CommonTextBlock_0 {
   widgetClassPath=&CommonTextBlock
   parent=ScaleBox_0
   widget=CommonTextBlock_0
   properties {
    text=Loading Leaderboard Entries
    textTransformPolicy=ToUpper
    font {
     bForceMonospaced=false
     bMaterialIsStencil=false
     fontMaterial=null
     fontObject=&BurbankSmall
     letterSpacing=0
     monospacedWidth=1
     size=18
     skewAmount=0
     typefaceFontName=Bold
     outlineSettings {
      bApplyOutlineToDropShadows=false
      bMiteredCorners=false
      bSeparateFillAlpha=false
      outlineMaterial=null
      outlineSize=0
      outlineColor=<rgba 0 0 0 1>
     }
    }
    shadowOffset=<0 0>
   }
   slot {class=ScaleBoxSlot}
  }
  Image_0 {
   widgetClassPath=&Image
   bIsVariable=true
   parent=Content_LoadingEntries
   widget=Image_0
   properties {
    visibility=SelfHitTestInvisible
    brush {
     bIsDynamicallyLoaded=false
     drawAs=Image
     imageType=NoImage
     mirroring=NoMirror
     resourceName=None
     resourceObject=&Throbber_Material-Monochrome
     tiling=NoTile
     imageSize=<32 32>
     margin=<ltrb 0 0 0 0>
     outlineSettings {
      bUseBrushTransparency=false
      roundingType=HalfHeightRadius
      width=0
      color {colorUseRule=UseColor_Specified, specifiedColor=<rgba 0 0 0 0>}
      cornerRadii=<0 0 0 0>
     }
     tintColor {colorUseRule=UseColor_Specified, specifiedColor=<rgba 1 1 1 1>}
     uVRegion {bIsValid=false, max=<0 0>, min=<0 0>}
    }
   }
   slot {class=HorizontalBoxSlot, properties {horizontalAlignment=HAlign_Center, verticalAlignment=VAlign_Center}}
  }
  Content_NoLeaderboardEntries {
   widgetClassPath=&CommonBorder
   bIsVariable=true
   parent=ContentOverlay
   widget=Content_NoLeaderboardEntries
   properties {visibility=Hidden, padding=<ltrb 10 60 10 25>}
   slot {class=OverlaySlot, properties {horizontalAlignment=HAlign_Fill, verticalAlignment=VAlign_Center}}
  }
  Text_NoLeaderboardEntries {
   widgetClassPath=&CommonTextBlock
   bIsVariable=true
   parent=Content_NoLeaderboardEntries
   widget=Text_NoLeaderboardEntries
   properties {
    autoWrapText=true
    justification=Center
    text=No Scores Submitted Yet
    font {
     bForceMonospaced=false
     bMaterialIsStencil=false
     fontMaterial=null
     fontObject=&Burbank
     letterSpacing=0
     monospacedWidth=1
     size=24
     skewAmount=0
     typefaceFontName=Burbank-PlayerPower
     outlineSettings {
      bApplyOutlineToDropShadows=false
      bMiteredCorners=false
      bSeparateFillAlpha=false
      outlineMaterial=null
      outlineSize=0
      outlineColor=<rgba 0 0 0 1>
     }
    }
    shadowOffset=<0 0>
   }
   slot {class=BorderSlot, properties {padding=<ltrb 10 60 10 25>}}
  }
  Content_NoLeaderboardEntries_MissingPart {
   widgetClassPath=&CommonBorder
   bIsVariable=true
   parent=ContentOverlay
   widget=Content_NoLeaderboardEntries_MissingPart
   properties {visibility=Hidden, padding=<ltrb 10 60 10 25>}
   slot {class=OverlaySlot, properties {horizontalAlignment=HAlign_Fill, verticalAlignment=VAlign_Center}}
  }
  Text_NoLeaderboardEntries_MissingPart {
   widgetClassPath=&CommonTextBlock
   bIsVariable=true
   parent=Content_NoLeaderboardEntries_MissingPart
   widget=Text_NoLeaderboardEntries_MissingPart
   properties {
    autoWrapText=true
    justification=Center
    text=Song has no Part for this Instrument
    font {
     bForceMonospaced=false
     bMaterialIsStencil=false
     fontMaterial=null
     fontObject=&Burbank
     letterSpacing=0
     monospacedWidth=1
     size=24
     skewAmount=0
     typefaceFontName=Burbank-PlayerPower
     outlineSettings {
      bApplyOutlineToDropShadows=false
      bMiteredCorners=false
      bSeparateFillAlpha=false
      outlineMaterial=null
      outlineSize=0
      outlineColor=<rgba 0 0 0 1>
     }
    }
    shadowOffset=<0 0>
   }
   slot {class=BorderSlot, properties {padding=<ltrb 10 60 10 25>}}
  }
  LeaderboardEntryList {
   widgetClassPath=&CommonListView
   bIsVariable=true
   parent=ContentOverlay
   widget=LeaderboardEntryList
   properties {
    bAllowDragging=false
    bReturnFocusToSelection=true
    entryWidgetClass=&/WBP_PilgrimQuickplay_LeaderboardEntry.WBP_PilgrimQuickplay_LeaderboardEntry_C
    visibility=SelfHitTestInvisible
    scrollBarStyle {
     thickness=12
     draggedThumbImage {
      bIsDynamicallyLoaded=false
      drawAs=RoundedBox
      imageType=FullColor
      mirroring=NoMirror
      resourceName=../../../Engine/Content/Slate/Common/Scrollbar_Thumb.png
      resourceObject=null
      tiling=NoTile
      imageSize=<12 12>
      margin=<ltrb 0.5 0.5 0.5 0.5>
      outlineSettings {
       bUseBrushTransparency=false
       roundingType=FixedRadius
       width=2
       color {colorUseRule=UseColor_Specified, specifiedColor=<rgba 0 0 0 0>}
       cornerRadii=<6 6 6 6>
      }
      tintColor {colorUseRule=UseColor_Specified, specifiedColor=<rgba 1 1 1 1>}
      uVRegion {bIsValid=false, max=<0 0>, min=<0 0>}
     }
     horizontalBackgroundImage {
      bIsDynamicallyLoaded=false
      drawAs=NoDrawType
      imageType=NoImage
      mirroring=NoMirror
      resourceName=None
      resourceObject=null
      tiling=NoTile
      imageSize=<0 0>
      margin=<ltrb 0 0 0 0>
      outlineSettings {
       bUseBrushTransparency=false
       roundingType=HalfHeightRadius
       width=0
       color {colorUseRule=UseColor_Specified, specifiedColor=<rgba 0 0 0 0>}
       cornerRadii=<0 0 0 0>
      }
      tintColor {colorUseRule=UseColor_Specified, specifiedColor=<rgba 1 1 1 1>}
      uVRegion {bIsValid=false, max=<0 0>, min=<0 0>}
     }
     horizontalBottomSlotImage {
      bIsDynamicallyLoaded=false
      drawAs=NoDrawType
      imageType=FullColor
      mirroring=NoMirror
      resourceName=../../../Engine/Content/Slate/Common/Scrollbar_Background_Vertical.png
      resourceObject=null
      tiling=NoTile
      imageSize=<8 8>
      margin=<ltrb 0 0 0 0>
      outlineSettings {
       bUseBrushTransparency=false
       roundingType=HalfHeightRadius
       width=0
       color {colorUseRule=UseColor_Specified, specifiedColor=<rgba 0 0 0 0>}
       cornerRadii=<0 0 0 0>
      }
      tintColor {colorUseRule=UseColor_Specified, specifiedColor=<rgba 1 1 1 1>}
      uVRegion {bIsValid=false, max=<0 0>, min=<0 0>}
     }
     horizontalTopSlotImage {
      bIsDynamicallyLoaded=false
      drawAs=NoDrawType
      imageType=FullColor
      mirroring=NoMirror
      resourceName=../../../Engine/Content/Slate/Common/Scrollbar_Background_Vertical.png
      resourceObject=null
      tiling=NoTile
      imageSize=<8 8>
      margin=<ltrb 0 0 0 0>
      outlineSettings {
       bUseBrushTransparency=false
       roundingType=HalfHeightRadius
       width=0
       color {colorUseRule=UseColor_Specified, specifiedColor=<rgba 0 0 0 0>}
       cornerRadii=<0 0 0 0>
      }
      tintColor {colorUseRule=UseColor_Specified, specifiedColor=<rgba 1 1 1 1>}
      uVRegion {bIsValid=false, max=<0 0>, min=<0 0>}
     }
     hoveredThumbImage {
      bIsDynamicallyLoaded=false
      drawAs=RoundedBox
      imageType=FullColor
      mirroring=NoMirror
      resourceName=../../../Engine/Content/Slate/Common/Scrollbar_Thumb.png
      resourceObject=null
      tiling=NoTile
      imageSize=<12 12>
      margin=<ltrb 0.5 0.5 0.5 0.5>
      outlineSettings {
       bUseBrushTransparency=false
       roundingType=FixedRadius
       width=2
       color {colorUseRule=UseColor_Specified, specifiedColor=<rgba 0 0 0 0>}
       cornerRadii=<6 6 6 6>
      }
      tintColor {colorUseRule=UseColor_Specified, specifiedColor=<rgba 1 1 1 1>}
      uVRegion {bIsValid=false, max=<0 0>, min=<0 0>}
     }
     normalThumbImage {
      bIsDynamicallyLoaded=false
      drawAs=RoundedBox
      imageType=FullColor
      mirroring=NoMirror
      resourceName=../../../Engine/Content/Slate/Common/Scrollbar_Thumb.png
      resourceObject=null
      tiling=NoTile
      imageSize=<12 12>
      margin=<ltrb 0.5 0.5 0.5 0.5>
      outlineSettings {
       bUseBrushTransparency=false
       roundingType=FixedRadius
       width=2
       color {colorUseRule=UseColor_Specified, specifiedColor=<rgba 0 0 0 0>}
       cornerRadii=<6 6 6 6>
      }
      tintColor {colorUseRule=UseColor_Specified, specifiedColor=<rgba 1 1 1 0.75>}
      uVRegion {bIsValid=false, max=<0 0>, min=<0 0>}
     }
     verticalBackgroundImage {
      bIsDynamicallyLoaded=false
      drawAs=Box
      imageType=NoImage
      mirroring=NoMirror
      resourceName=None
      resourceObject=/FortUILibrary/UIKit/ScrollBar/MaterialInstances/MI_UIKit_ScrollbarBackground
      tiling=NoTile
      imageSize=<12 12>
      margin=<ltrb 0.5 0.5 0.5 0.5>
      outlineSettings {
       bUseBrushTransparency=false
       roundingType=HalfHeightRadius
       width=0
       color {colorUseRule=UseColor_Specified, specifiedColor=<rgba 0 0 0 0>}
       cornerRadii=<0 0 0 0>
      }
      tintColor {colorUseRule=UseColor_Specified, specifiedColor=<rgba 1 1 1 1>}
      uVRegion {bIsValid=false, max=<0 0>, min=<0 0>}
     }
     verticalBottomSlotImage {
      bIsDynamicallyLoaded=false
      drawAs=NoDrawType
      imageType=FullColor
      mirroring=NoMirror
      resourceName=../../../Engine/Content/Slate/Common/Scrollbar_Background_Vertical.png
      resourceObject=null
      tiling=NoTile
      imageSize=<8 8>
      margin=<ltrb 0 0 0 0>
      outlineSettings {
       bUseBrushTransparency=false
       roundingType=HalfHeightRadius
       width=0
       color {colorUseRule=UseColor_Specified, specifiedColor=<rgba 0 0 0 0>}
       cornerRadii=<0 0 0 0>
      }
      tintColor {colorUseRule=UseColor_Specified, specifiedColor=<rgba 1 1 1 1>}
      uVRegion {bIsValid=false, max=<0 0>, min=<0 0>}
     }
     verticalTopSlotImage {
      bIsDynamicallyLoaded=false
      drawAs=NoDrawType
      imageType=FullColor
      mirroring=NoMirror
      resourceName=../../../Engine/Content/Slate/Common/Scrollbar_Background_Vertical.png
      resourceObject=null
      tiling=NoTile
      imageSize=<8 8>
      margin=<ltrb 0 0 0 0>
      outlineSettings {
       bUseBrushTransparency=false
       roundingType=HalfHeightRadius
       width=0
       color {colorUseRule=UseColor_Specified, specifiedColor=<rgba 0 0 0 0>}
       cornerRadii=<0 0 0 0>
      }
      tintColor {colorUseRule=UseColor_Specified, specifiedColor=<rgba 1 1 1 1>}
      uVRegion {bIsValid=false, max=<0 0>, min=<0 0>}
     }
    }
   }
   slot {class=OverlaySlot, properties {horizontalAlignment=HAlign_Fill, verticalAlignment=VAlign_Fill, padding=<ltrb 0 0 0 75>}}
  }
  FocusedPlayerEntry {
   widgetClassPath=&/WBP_PilgrimQuickplay_LeaderboardEntry
   bIsVariable=true
   parent=ContentOverlay
   widget=FocusedPlayerEntry
   slot {class=OverlaySlot, properties {horizontalAlignment=HAlign_Fill, verticalAlignment=VAlign_Bottom}}
  }
  Overlay_0 {
   widgetClassPath=&Overlay
   parent=EntriesVerticalBox
   widget=Overlay_0
   slot {class=VerticalBoxSlot, properties {horizontalAlignment=HAlign_Right, verticalAlignment=VAlign_Top, padding=<ltrb 0 20 20 20>}}
  }
  WBP_CaptureForPostBufferUpdate {
   widgetClassPath=&WBP_CaptureForPostBufferUpdate
   bIsVariable=true
   parent=Overlay_0
   slot {class=OverlaySlot, properties {horizontalAlignment=HAlign_Fill, verticalAlignment=VAlign_Fill}}
  }
  BackButton {
   widgetClassPath=/FortUILibrary/UIKit/Buttons/WBP_UIKit_Button_Quiet
   bIsVariable=true
   parent=Overlay_0
   widget=BackButton
   properties {bEnableOnClickedAnalytics=true, text=Exit, triggeringInputAction {dataTable=&InputActionDataTable, rowName=GenericBack}}
   slot {class=OverlaySlot}
  }
 ]
}
