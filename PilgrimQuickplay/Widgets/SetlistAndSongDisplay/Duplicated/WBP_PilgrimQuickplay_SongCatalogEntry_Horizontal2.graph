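# Horizontal song catalog list entry for Pilgrim Quickplay, showing album art, title, artist and the player's personal-best stats.
# Binding a PilgrimSongCatalogItemVM triggers an async personal-best request, delayed by a timer unless leaderboard data is cached and cancelled on destruct.
# Selection expands the entry, starts the song preview and builds a favorite, archive and leaderboard context menu.
# Mobile toggles collapse on reselect; animations drive size, material and visibility changes.
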WidgetBlueprint /PilgrimQuickplay/Widgets/SetlistAndSongDisplay/Duplicated/WBP_PilgrimQuickplay_SongCatalogEntry_Horizontal2
animations [
 Selected_Animated {
  binding_count=17
  length_seconds=0.45
  object=&self:Selected_Animated
  sequence {bindings [
    SongArtist {
     id=#1
     possessed_class=CommonTextBlock
     tracks [{
       type=MovieSceneColorTrack
       display_name=Color and Opacity
       sections [{
         type=MovieSceneColorSection
         end_frame=6
         start_frame=0
         channels [
          Color.R {type=MovieSceneScriptingFloatChannel, const [1,2]}
          Color.G {type=MovieSceneScriptingFloatChannel, const [1,2]}
          Color.B {type=MovieSceneScriptingFloatChannel, const [1,2]}
          Color.A {type=MovieSceneScriptingFloatChannel, const [0.6,2]}
         ]
        }]
      }]
    }
    IntensitiesScaleBox {
     id=#2
     possessed_class=ScaleBox
     tracks [{
       type=MovieSceneFloatTrack
       display_name=User Specified Scale
       sections [{type=MovieSceneFloatSection, end_frame=6, start_frame=0, channels [None {type=MovieSceneScriptingFloatChannel, keys [[0,0],[6,1.5]]}]}]
      }]
    }
    ContextMenu {
     id=#3
     possessed_class=WBP_UIKit_ContextMenuAnchorButton_C
     tracks [
      {
       type=MovieSceneByteTrack
       display_name=Visibility
       sections [{type=MovieSceneByteSection, end_frame=6, start_frame=0, channels [None {type=MovieSceneScriptingByteChannel, keys [[0,1],[1,0]]}]}]
      }
      {
       type=MovieSceneColorTrack
       display_name=Color and Opacity
       sections [{
         type=MovieSceneColorSection
         end_frame=6
         start_frame=0
         channels [
          Color.R {type=MovieSceneScriptingFloatChannel, const [1,1]}
          Color.G {type=MovieSceneScriptingFloatChannel, const [1,1]}
          Color.B {type=MovieSceneScriptingFloatChannel, const [1,1]}
          Color.A {type=MovieSceneScriptingFloatChannel, keys [[0,0],[6,1]]}
         ]
        }]
      }
     ]
    }
    StatusIconBox {
     id=#4
     possessed_class=HorizontalBox
     tracks [
      {
       type=MovieSceneByteTrack
       display_name=Visibility
       sections [{
         type=MovieSceneByteSection
         end_frame=6
         start_frame=0
         channels [None {type=MovieSceneScriptingByteChannel, thinned_from=5, keys [[0,1],[1,1],[3,4],[6,4]]}]
        }]
      }
      {
       type=MovieSceneFloatTrack
       display_name=Render Opacity
       sections [{type=MovieSceneFloatSection, end_frame=6, start_frame=0, channels [None {type=MovieSceneScriptingFloatChannel, keys [[0,0],[3,0],[6,1]]}]}]
      }
     ]
    }
    GridSlot_0 {
     id=#5
     parent_id=#6
     possessed_class=GridSlot
     tracks [{
       type=MovieSceneMarginTrack
       display_name=Padding
       sections [{
         type=MovieSceneMarginSection
         end_frame=6
         start_frame=0
         channels [
          Left {type=MovieSceneScriptingFloatChannel, keys [[0,3.5],[6,0.75]]}
          Top {type=MovieSceneScriptingFloatChannel}
          Right {type=MovieSceneScriptingFloatChannel}
          Bottom {type=MovieSceneScriptingFloatChannel}
         ]
        }]
      }]
    }
    Image_SelectionGlow {
     id=#6
     possessed_class=Image
     tracks [{
       type=MovieSceneWidgetMaterialTrack
       sections [{
         type=MovieSceneParameterSection
         end_frame=6
         start_frame=0
         channels [
          Corner Radius {type=MovieSceneScriptingFloatChannel, keys [[0,0.117],[6,0.05]]}
          CornerRadius {type=MovieSceneScriptingFloatChannel, keys [[0,0.2],[6,0.075]]}
          OutlineThicknes {type=MovieSceneScriptingFloatChannel, keys [[0,0.015],[6,0.004]]}
         ]
        }]
      }]
    }
    OverlaySlot_2 {
     id=#7
     parent_id=#8
     possessed_class=OverlaySlot
     tracks [{
       type=MovieSceneMarginTrack
       display_name=Padding
       sections [{
         type=MovieSceneMarginSection
         end_frame=6
         start_frame=0
         channels [
          Left {type=MovieSceneScriptingFloatChannel}
          Top {type=MovieSceneScriptingFloatChannel, keys [[0,0],[6,23]]}
          Right {type=MovieSceneScriptingFloatChannel}
          Bottom {type=MovieSceneScriptingFloatChannel, keys [[0,-15],[6,0]]}
         ]
        }]
      }]
    }
    SongTitle {
     id=#9
     possessed_class=CommonTextBlock
     tracks [{
       type=MovieSceneColorTrack
       display_name=Color and Opacity
       sections [{
         type=MovieSceneColorSection
         end_frame=6
         start_frame=0
         channels [
          Color.R {type=MovieSceneScriptingFloatChannel, const [1,2]}
          Color.G {type=MovieSceneScriptingFloatChannel, const [1,2]}
          Color.B {type=MovieSceneScriptingFloatChannel, const [1,2]}
          Color.A {type=MovieSceneScriptingFloatChannel}
         ]
        }]
      }]
    }
    CommonBorder_Backplate {
     id=#10
     possessed_class=CommonBorder
     tracks [{
       type=MovieSceneWidgetMaterialTrack
       sections [{
         type=MovieSceneParameterSection
         channels [
          CornerRadius {type=MovieSceneScriptingFloatChannel, keys [[0,0.175],[6,0.05]]}
          EdgeXOffset {type=MovieSceneScriptingFloatChannel, keys [[3,0],[9,-0.055]]}
          Opacity {type=MovieSceneScriptingFloatChannel, const [0.8,2]}
          OffsetGradient_X {type=MovieSceneScriptingFloatChannel, keys [[0,-0.2],[6,-0.4]]}
          EaseCurve_Amount {type=MovieSceneScriptingFloatChannel, keys [[0,1],[6,0]]}
          RampGradientTrianglesMixer {type=MovieSceneScriptingFloatChannel, keys [[0,0],[6,0.4]]}
          RGB1R {type=MovieSceneScriptingFloatChannel, keys [[0,0.4263],[6,0.0382]]}
          RGB1G {type=MovieSceneScriptingFloatChannel, keys [[0,0.0352],[6,0.0331]]}
          RGB1B {type=MovieSceneScriptingFloatChannel, keys [[0,0.7396],[6,0.6939]]}
          RGB1A {type=MovieSceneScriptingFloatChannel, const [1,1]}
          RGB2R {type=MovieSceneScriptingFloatChannel, keys [[0,0.042],[6,0]]}
          RGB2G {type=MovieSceneScriptingFloatChannel, keys [[0,0.0352],[6,0.4]]}
          RGB2B {type=MovieSceneScriptingFloatChannel, keys [[0,0.3073],[6,1]]}
          RGB2A {type=MovieSceneScriptingFloatChannel, const [1,2]}
         ]
        }]
      }]
    }
    FortLazyImage_Icon_Side {
     id=#11
     possessed_class=FortLazyImage
     tracks [{
       type=MovieSceneFloatTrack
       display_name=Render Opacity
       sections [{type=MovieSceneFloatSection, end_frame=6, start_frame=0, channels [None {type=MovieSceneScriptingFloatChannel, keys [[0,1],[6,0]]}]}]
      }]
    }
    BottomSpacer {
     id=#12
     possessed_class=Spacer
     tracks [{
       type=MovieSceneByteTrack
       display_name=Visibility
       sections [{
         type=MovieSceneByteSection
         end_frame=6
         start_frame=0
         channels [None {type=MovieSceneScriptingByteChannel, thinned_from=4, keys [[0,0],[1,1],[6,1]]}]
        }]
      }]
    }
    IntensitiesBorder {
     id=#13
     possessed_class=CommonBorder
     tracks [
      {
       type=MovieSceneColorTrack
       display_name=Content Color and Opacity
       sections [{
         type=MovieSceneColorSection
         end_frame=6
         start_frame=0
         channels [
          Color.R {type=MovieSceneScriptingFloatChannel, const [1,1]}
          Color.G {type=MovieSceneScriptingFloatChannel, const [1,1]}
          Color.B {type=MovieSceneScriptingFloatChannel, const [1,1]}
          Color.A {type=MovieSceneScriptingFloatChannel, keys [[0,0],[6,1]]}
         ]
        }]
      }
      {
       type=MovieSceneByteTrack
       display_name=Visibility
       sections [{type=MovieSceneByteSection, end_frame=6, start_frame=0, channels [None {type=MovieSceneScriptingByteChannel, keys [[0,1],[1,4],[6,4]]}]}]
      }
     ]
    }
    albumCover {
     id=#14
     possessed_class=FortLazyImage
     tracks [{
       type=MovieSceneWidgetMaterialTrack
       sections [{
         type=MovieSceneParameterSection
         end_frame=6
         start_frame=0
         channels [
          Corner_BottomLeft {type=MovieSceneScriptingFloatChannel, keys [[0,0.2],[6,0.1]]}
          Corner_TopLeft {type=MovieSceneScriptingFloatChannel, keys [[0,0.2],[6,0.1]]}
          Corner_BottomRight {type=MovieSceneScriptingFloatChannel, keys [[0,0.2],[6,0.1]]}
          Corner_TopRight {type=MovieSceneScriptingFloatChannel, keys [[0,0.2],[6,0.1]]}
         ]
        }]
      }]
    }
    AlbumArtSizeBox {
     id=#15
     possessed_class=SizeBox
     tracks [
      {
       type=MovieSceneFloatTrack
       display_name=Width Override
       sections [{type=MovieSceneFloatSection, end_frame=6, start_frame=0, channels [None {type=MovieSceneScriptingFloatChannel, keys [[0,138],[6,250]]}]}]
      }
      {
       type=MovieSceneFloatTrack
       display_name=Height Override
       sections [{type=MovieSceneFloatSection, end_frame=6, start_frame=0, channels [None {type=MovieSceneScriptingFloatChannel, keys [[0,138],[6,250]]}]}]
      }
     ]
    }
    HorizontalBoxSlot_0 {
     id=#16
     parent_id=#17
     possessed_class=HorizontalBoxSlot
     tracks [{
       type=MovieSceneMarginTrack
       display_name=Padding
       sections [{
         type=MovieSceneMarginSection
         end_frame=6
         start_frame=0
         channels [
          Left {type=MovieSceneScriptingFloatChannel}
          Top {type=MovieSceneScriptingFloatChannel, const [0,2]}
          Right {type=MovieSceneScriptingFloatChannel}
          Bottom {type=MovieSceneScriptingFloatChannel}
         ]
        }]
      }]
    }
    HighScoreInfo {id=#8, possessed_class=WrapBox, tracks []}
    TitleArtistBox {id=#17, possessed_class=VerticalBox, tracks []}
   ]}
 }
 Hovered_Animation {
  binding_count=1
  length_seconds=0.3
  object=&self:Hovered_Animation
  sequence {bindings [Image_SelectionGlow {
     id=#18
     possessed_class=Image
     tracks [{
       type=MovieSceneColorTrack
       display_name=Color and Opacity
       sections [{
         type=MovieSceneColorSection
         end_frame=6
         start_frame=0
         channels [
          Color.R {type=MovieSceneScriptingFloatChannel, keys [[3,0],[6,1]]}
          Color.G {type=MovieSceneScriptingFloatChannel, keys [[3,0],[6,1]]}
          Color.B {type=MovieSceneScriptingFloatChannel, keys [[3,0],[6,1]]}
          Color.A {type=MovieSceneScriptingFloatChannel, keys [[3,0],[6,1]]}
         ]
        }]
      }]
    }]}
 }
]
graphs [
 ClearEntry {nodes [
   ClearEntry() {is=title K2Node_FunctionEntry}
   > SetVisibility(self=Get FullComboIndicator() -> (FullComboIndicator:WBP Pilgrim Quickplay Full Combo Indicator Object Reference), InVisibility=Collapsed)
   > SetVisibility(self=Get PercentageText() -> (PercentageText:Common Text Object Reference), InVisibility=Collapsed)
   > UpdateDifficultyState!(
    execute
    self:WBP Pilgrim Results Difficulty Object Reference=Get DifficultyIcon() -> (DifficultyIcon:WBP Pilgrim Results Difficulty Object Reference)
    SongIsUnplayed:bool=true
   ) -> (then)
   > ResetCompletedVisuals(self:WBP Pilgrim Quickplay Local Star Bar Object Reference=Get WBP_PilgrimQuickplay_LocalStarBar() -> (WBP_PilgrimQuickplay_LocalStarBar:WBP Pilgrim Quickplay Local Star Bar Object Reference))
   > SetVisibility(self=Get PercentageText(), InVisibility=Collapsed)
   > SetVisibility(
    self=Get WBP_PilgrimQuickplay_SeasonDisplay() -> (WBP_PilgrimQuickplay_SeasonDisplay:WBP Pilgrim Quickplay Season Display Object Reference)
    InVisibility=Collapsed
   )
  ]}
 DebugPrintMessage {nodes [
   n1 = DebugPrintMessage!() -> (then, Message:String, TextColor:Linear Color Structure=(R=1,G=1,B=1,A=1)) {is=title K2Node_FunctionEntry}
   n2 = Get Shortname() -> (Shortname:Name=None)
   n3 = Get DebugButtonInput() -> (DebugButtonInput:bool)
   > PrintString!(
    execute
    InString=Append!(A=@n1, B=" ", C:String="To String (Name)"(InName=@n2)) -> (ReturnValue)
    bPrintToScreen=@n3
    bPrintToLog=@n3
    TextColor=@n1.TextColor
    Duration=Get DebugDuration() -> (DebugDuration:float)
    Key
   ) -> (then)
  ]}
 Does Player Have Leaderboard Cache Data {nodes [
   Does Player Have Leaderboard Cache Data() {is=title K2Node_FunctionEntry}
   > n1 = GetPilgrimQuickplayPlayerStateComponent!(
    execute
    PlayerState:Player State Object Reference=Get PlayerState!(self:Controller Object Reference=GetOwningPlayer()) -> (PlayerState:Player State Object Reference)
   ) -> (then, ReturnValue:Pilgrim Quickplay Player State Component Object Reference)
   > n2 = PlayerHasSongPartLeaderboardCacheData!(
    execute
    self:Pilgrim Leaderboard Player Component Object Reference=GetComponentByClass!(
     self:Actor Object Reference=GetFirstLocalFortPlayerController!(bIncludeNonSpectators:bool=true, bIncludeSpectators:bool=false) -> (ReturnValue:Fort Player Controller Object Reference)
     ComponentClass:Actor Component Class Reference=/Script/PilgrimCoreRuntime.PilgrimLeaderboardPlayerComponent
    ) -> (ReturnValue:Pilgrim Leaderboard Player Component Object Reference)
    SongShortName:Name=Get Shortname() -> (Shortname=None)
    Instrument:EPilgrimTrackType Enum=GetInstrumentChoice!(self:Pilgrim Core Player State Component Object Reference=@n1) -> (ReturnValue:EPilgrimTrackType Enum=TrackGuitar)
   ) -> (then, ReturnValue:bool)
   > Return Node!(execute, Has Cache:bool=@n2)
  ]}
 EventGraph {nodes [
   n1 = SetBrushFromLazyDisplayAsset_DontMatchSize!(
    execute
    self:Fort Lazy Image Object Reference=Get albumCover() -> (albumCover:Fort Lazy Image Object Reference)
    LazyObject:"Soft Object Reference (by ref)"=Get EmptyAlbumCover() -> (EmptyAlbumCover:Texture 2D Soft Object Reference)
   ) -> (then)
   Set Shortname(execute=[@n1.then, @n12.then], Shortname:Name=None) -> (Output_Get:Name=None)
   n2 = RequestPersonalBest() {is=title K2Node_CustomEvent}
   Clear and Invalidate Timer by Handle(execute=[@n7.then, @n2.then], Handle=Get StartTimerHandle() -> (StartTimerHandle:Timer Handle Structure))
   > Set StartTimerHandle(StartTimerHandle:Timer Handle Structure) -> (Output_Get:Timer Handle Structure)
   > n3 = Is Valid(InputObject=Get PilgrimRequestPersonalBestHandle() -> (PilgrimRequestPersonalBestHandle:Pilgrim Async Action Request Personal Best Object Reference))
   n4 = GetPlayerInfoForPersonalBest!(execute=@n3.Is Not Valid, self) -> (
    then
    RequestingPlayer:Fort Player Controller Object Reference
    PlayerAccountID:Unique Net Id Structure
    InstrumentType:EPilgrimTrackType Enum=TrackGuitar
    SongShortName:Name=None
   )
   > n5 = PilgrimRequestPersonalBest!(
    execute:Exec
    RequestingPlayer:Fort Player Controller Object Reference=@n4
    PlayerAccountId:"Unique Net Id Structure (by ref)"=@n4.PlayerAccountID
    SongShortName:Name=@n4.SongShortName
    BestScoreFlags:Bitmask Integer=16
    InstrumentType:EPilgrimTrackType Enum=@n4.InstrumentType
    BandSize:int=1
    SetlistIndex:int=-1
    bUseCache:bool=true
    FreshnessTime:int=GetLeaderboardsSongStarsMaxCacheSeconds!() -> (ReturnValue:int=0)
    RequestPriority:EPilgrimLeaderboardRequestPriority Enum=Low
   ) -> (
    then:Exec
    AsyncTaskProxy:Pilgrim Async Action Request Personal Best Object Reference
    OnSuccess:Exec
    OnFailure:Exec
    PersonalBestResult:Pilgrim Personal Best Result Structure
    RequestParams:Pilgrim Personal Best Params Structure
   ) {is=title K2Node_AsyncAction}
   > Set PilgrimRequestPersonalBestHandle(PilgrimRequestPersonalBestHandle:Pilgrim Async Action Request Personal Best Object Reference=@n5) -> (Output_Get:Pilgrim Async Action Request Personal Best Object Reference)
   ClearEntry(execute=@n5.OnFailure) {is=title K2Node_CallFunction}
   > Set PilgrimRequestPersonalBestHandle()
   > DebugPrintMessage!(
    execute
    self
    Message:String="Error!! Failed to get leaderboard request! Clearing out the entry"
    TextColor:Linear Color Structure=(R=0.2195,G=0,B=0.0024,A=1)
   ) -> (then) {is=title K2Node_CallFunction}
   Set Horizontal Entry Highscore Data!(
    execute=@n5.OnSuccess
    self
    PersonalBestResult:Pilgrim Personal Best Result Structure=@n5.PersonalBestResult
    PilgrimPersonalBestParams:"Pilgrim Personal Best Params Structure (by ref)"=@n5.RequestParams
   ) -> (then) {is=title K2Node_CallFunction}
   > ResetSpacerForSongScrolling!(
    execute
    self
    PersonalBestResults:Pilgrim Personal Best Result Structure=@n5.PersonalBestResult
    PersonalBestParams:Pilgrim Personal Best Params Structure=@n5.RequestParams
   ) -> (then)
   > Set PilgrimRequestPersonalBestHandle()
   Load Personal Best Leaderboard() {is=title K2Node_CustomEvent}
   > n6 = Does Player Have Leaderboard Cache Data!(execute, self) -> (then, Has Cache:bool) {is=title K2Node_CallFunction}
   > n7 = Branch(Condition=@n6)
   Clear and Invalidate Timer by Handle(execute=@n7.else, Handle=Get StartTimerHandle())
   > n8 = Set Timer by Event(Delegate=@n2, Time=GetRequestPersonalBestTimerDelaySeconds!() -> (ReturnValue:float), bLooping=false)
   > Set StartTimerHandle(StartTimerHandle=@n8)
   Cleanup Handles() {is=title K2Node_CustomEvent}
   > Clear and Invalidate Timer by Handle(Handle=Get StartTimerHandle())
   > Is Valid(InputObject=Get PilgrimRequestPersonalBestHandle())
   > Cancel(self:Cancellable Async Action Object Reference=Get PilgrimRequestPersonalBestHandle())
   > Set PilgrimRequestPersonalBestHandle()
   n9 = SetAlbumCoverFromRef!() -> (OutputDelegate, then, ImageReference:"Fort Image Reference Structure (by ref)") {is=title K2Node_CustomEvent}
   > SetBrushFromImageReference!(
    execute
    self:Fort Lazy Image Object Reference=Get albumCover()
    ImageReference:"Fort Image Reference Structure (by ref)"=@n9.ImageReference
    bMatchSize:bool=false
    DownloadPriority:ECMSDownloadPriority Enum=VeryHigh
   ) -> (then)
   n10 = MobileDeselect() {is=title K2Node_CustomEvent}
   TryCollapse(execute=[@n10.then, @n26.then])
   > SetVisibility(self=[Get WBP_PilgrimQuickplay_SeasonDisplay(), Get PercentageText()], InVisibility=Collapsed)
   > StartStopSongPreview!(execute, self, Start:bool=false) -> (then)
   > Close Context Menu()
   Event On Hovered()
   > DebugPrintMessage(Message=Hovered, TextColor=(R=1,G=0.9216,B=0.491,A=1))
   > n11 = Branch(Condition=IsUsingTouch(PlayerController=GetOwningPlayer()))
   PlayAnimationForward(
    execute=@n11.else
    InAnimation=Get Hovered_Animation() -> (Hovered_Animation:Widget Animation Object Reference)
    PlaybackSpeed=Get SelectAnimPlaybackSpeed() -> (SelectAnimPlaybackSpeed:double)
   )
   > Load Personal Best Leaderboard() {is=title K2Node_CallFunction}
   Event Construct()
   > Set IsExpanded(IsExpanded:bool=false) -> (Output_Get:bool)
   Event On Unfocused() {is=title K2Node_Event}
   > DebugPrintMessage(Message=UnFocused, TextColor=(R=0,G=0.9823,B=1,A=1))
   > Load Personal Best Leaderboard()
   Event On Focused() {is=title K2Node_Event}
   > DebugPrintMessage(Message=Focused, TextColor=(R=0,G=0.9823,B=1,A=1))
   n12 = Event On Entry Released() {is=title K2Node_Event}
   n13 = Event On Clicked()
   n14 = Get Selected_Animated() -> (Selected_Animated:Widget Animation Object Reference)
   n15 = IsMobileGame()
   n16 = Get IsExpanded() -> (IsExpanded:bool)
   n17 = Branch(
    execute=[@n13.then, @n18.then]
    Condition=OR Boolean!(A=NOT Boolean(A=@n15), B=NOT Boolean(A=@n16), C:bool=IsAnimationPlaying(InAnimation=@n14)) -> (ReturnValue)
   )
   > TryExpand()
   > StartStopSongPreview(Start=true)
   > Load Personal Best Leaderboard()
   > Create Context Menu()
   MobileDeselect(execute=@n17.else) {is=title K2Node_CallFunction}
   n18 = Event On Selected()
   Event On Unhovered()
   > DebugPrintMessage(Message=Unhovered, TextColor=(R=1,G=1,B=1,A=1))
   > n19 = Branch(Condition=IsMobileGame())
   PlayAnimationReverse(execute=@n19.else, InAnimation=Get Hovered_Animation(), PlaybackSpeed=Get SelectAnimPlaybackSpeed())
   n20 = Event OnListItemObjectSet()
   > ClearEntry()
   > n21 = Sequence()
   > n22 = Cast To PilgrimSongCatalogItemVM!(execute, Object=@n20.ListItemObject) -> (then, CastFailed, AsPilgrim Song Catalog Item VM:Pilgrim Song Catalog Item VM Object Reference) {is=title K2Node_DynamicCast}
   > SetItemVMs!(execute, self, Viewmodel:Pilgrim Song Catalog Item VM Object Reference=@n22) -> (then)
   > n23 = Get!(execute:Exec) -> (then:Exec, else:Exec, PilgrimSongCatalogItemVM:Pilgrim Song Catalog Item VM Object Reference)
   > n24 = GetSongMetadata!(execute, self:Pilgrim Song Catalog Item VM Object Reference=@n23) -> (then, ReturnValue:Pilgrim Song Metadata Object Reference)
   > n25 = Is Valid(InputObject=@n24)
   ClearEntry(execute=@n25.Is Not Valid)
   Set SongMetadata(execute=@n25.Is Valid, SongMetadata:Pilgrim Song Metadata Object Reference=@n24) -> (Output_Get:Pilgrim Song Metadata Object Reference)
   > Set Shortname(Shortname=GetShortName!(self:Sparks Song Data Object Reference=@n24) -> (ReturnValue:Name=None)) -> (Output_Get=None)
   > SetInstrumentIcon!(
    execute
    self
    InstrumentType:EPilgrimTrackType Enum=Get InstrumentForNextSong!(self:Pilgrim Instrument VM Object Reference=Get PilgrimInstrumentVM() -> (PilgrimInstrumentVM:Pilgrim Instrument VM Object Reference)) -> (InstrumentForNextSong:EPilgrimTrackType Enum=TrackGuitar)
   ) -> (then)
   > Load Personal Best Leaderboard()
   ClearEntry(execute=@n22.CastFailed)
   SetupStarCount!(execute=@n21.then_1, self:WBP Pilgrim Quickplay Local Star Bar Object Reference=Get WBP_PilgrimQuickplay_LocalStarBar(), StarCount:int=0) -> (then)
   > Finish OnListItemObjectSet()
   Event Destruct()
   > Cleanup Handles() {is=title K2Node_CallFunction}
   Event OnInitialized()
   > Set IsExpanded(IsExpanded=false)
   > Update Context Content()
   n26 = Event On Deselected()
   Event PreConstruct()
   > Set Personal Best Stats!(
    execute
    self
    Score:int=100
    Full Combo:bool=false
    Accuracy:float=79
    Instrument:EPilgrimTrackType Enum=TrackGuitar
    Stars Earned:int=3
    Difficulty:EPilgrimSongDifficulty Enum=DifficultyEasy
    SeasonNumber:int=0
   ) -> (then) {is=title K2Node_CallFunction}
  ]}
 EventGraph_ContextMenu {nodes [
   n1 = On Favorite Item!() -> (OutputDelegate, then, InteractionNameID:Name) {is=title K2Node_CustomEvent}
   > n2 = On Handle Favorite Item Iteraction()
   n3 = Close Context Menu(execute=[@n7.then, @n2.then, @n5.then])
   RemoveAllInteractions(
    execute=[@n8.then, @n3.then]
    self:UIKit Context Menu Action Context Object Reference=Get Context Menu Action Context() -> (Context Menu Action Context:UIKit Context Menu Action Context Object Reference)
   )
   > n4 = Make Favorite Context Menu Iteraction!(execute, self, InInteractionDelegate:"Delegate (by ref)"=@n1) -> (then)
   n5 = On Archive Item!() -> (OutputDelegate, then, InteractionNameID:Name) {is=title K2Node_CustomEvent}
   n6 = Make Archive Context Menu Iteraction!(execute=@n4.then, self, InInteractionDelegate:"Delegate (by ref)"=@n5) -> (then)
   n7 = On Leaderboard Iteraction!() -> (OutputDelegate, then, InteractionNameID:Name) {is=title K2Node_CustomEvent}
   Make Leaderboard Context Menu Iteraction!(execute=@n6.then, self, InInteractionDelegate:"Delegate (by ref)"=@n7) -> (then)
   > SetContext!(
    execute
    self:WBP UIKit Context Menu Anchor Button Base Object Reference=Get ContextMenu() -> (ContextMenu:WBP UIKit Context Menu Anchor Button Object Reference)
    InteractionContext_Data:UIKit Context Menu Action Context Object Reference=Get Context Menu Action Context()
   ) -> (then)
   Create Context Menu() {is=title K2Node_CustomEvent}
   > n8 = Branch(Condition=IsContextMenuReady!() -> (ReturnValue:bool))
   Open Context Menu() {is=title K2Node_CustomEvent}
   n9 = Get ContextMenu()
   > SetVisibility(
    self=[
     Get ContextMenuButton!(self:WBP UIKit Context Menu Anchor Button Base Object Reference=@n9) -> (ContextMenuButton:WBP UIKit Button Generic Base Object Reference)
     @n9
    ]
    InVisibility=Visible
   )
   > Toggle Quick Actions!(execute, self:WBP UIKit Context Menu Anchor Button Base Object Reference=Get ContextMenu(), Enabled:bool=true) -> (then)
   n10 = Get ContextMenu()
   > Branch(Condition=IsOpen!(self:Menu Anchor Object Reference=Get Anchor!(self:UIKit Context Menu Anchor Object Reference=@n10) -> (Anchor:Menu Anchor Object Reference)) -> (ReturnValue:bool))
   > Create Context Menu() {is=title K2Node_CallFunction}
   Finish OnListItemObjectSet() {is=title K2Node_CustomEvent}
   > Update Context Content()
   > n11 = Is Valid(InputObject=Get Context Menu Action Context())
   n12 = Construct Object from Class(execute=@n11.Is Not Valid, Class=/Script/FortniteUI.UIKitContextMenuActionContext) -> (ReturnValue:UIKit Context Menu Action Context Object Reference)
   > Set Context Menu Action Context(Context Menu Action Context:UIKit Context Menu Action Context Object Reference=@n12) -> (Output_Get:UIKit Context Menu Action Context Object Reference)
   > Branch(Condition=IsUsingTouch(PlayerController=GetOwningPlayer()))
   > Create Context Menu()
   Close Context Menu() {is=title K2Node_CustomEvent}
   > Close(self:Menu Anchor Object Reference=Get Anchor(self=Get ContextMenu()))
   Update Context Content() {is=title K2Node_CustomEvent}
   > n13 = Branch(Condition=IsContextMenuReady())
   > SetVisibility(self=Get ContextMenu_Favorite_Overlay() -> (ContextMenu_Favorite_Overlay:Overlay Object Reference), InVisibility=Visible)
   SetVisibility(execute=@n13.else, self=Get ContextMenu_Favorite_Overlay(), InVisibility=Collapsed)
  ]}
 GetPlayerInfoForPersonalBest {nodes [
   GetPlayerInfoForPersonalBest() {is=title K2Node_FunctionEntry}
   n1 = GetOwningPlayer()
   > n2 = GetPilgrimQuickplayPlayerStateComponent(PlayerState=Get PlayerState(self=@n1))
   > n3 = Cast To FortPlayerController(Object=@n1)
   > Return Node!(
    execute
    RequestingPlayer:Fort Player Controller Object Reference=@n3
    PlayerAccountID:Unique Net Id Structure=GetGameAccountId!(self:Fort Player Controller Object Reference=@n3) -> (ReturnValue:Unique Net Id Structure)
    InstrumentType:EPilgrimTrackType Enum=GetInstrumentChoice(self=@n2) -> (ReturnValue=TrackGuitar)
    SongShortName:Name=Get Shortname() -> (Shortname=None)
   )
  ]}
 IsContextMenuEnabled {nodes [
   IsContextMenuEnabled() {is=title K2Node_FunctionEntry}
   > n1 = GetConsoleVariableBoolValue!(execute, VariableName:String=Pilgrim.SongCatalogItemVM.bEnable) -> (then, ReturnValue:bool)
   > Return Node!(execute, ReturnValue:bool=@n1)
  ]}
 IsContextMenuReady {nodes [
   IsContextMenuReady() {is=title K2Node_FunctionEntry}
   > n1 = Get()
   > n2 = GetConsoleVariableBoolValue(VariableName=Pilgrim.SongCatalogItemVM.bEnable)
   > Return Node(ReturnValue=@n2)
   Return Node(execute=@n1.else, ReturnValue=false)
  ]}
 Make Archive Context Menu Iteraction {nodes [
   n1 = Make Archive Context Menu Iteraction!() -> (then, InInteractionDelegate:"Delegate (by ref)") {is=title K2Node_FunctionEntry}
   n2 = Get PilgrimSongCatalogItemVM() -> (PilgrimSongCatalogItemVM:Pilgrim Song Catalog Item VM Object Reference)
   n3 = Select(
    Option 0:Text="LOCTEXT(\"Archive Item\")"
    Option 1:Text="LOCTEXT(\"Archive Item\")"
    Index=Get bIsArchived!(self:Fort Item VM Object Reference=@n2) -> (bIsArchived:bool)
   ) -> (ReturnValue:Text)
   n4 = Get bCanArchive!(self:Fort Item VM Object Reference=@n2) -> (bCanArchive:bool)
   > AddInteraction!(
    execute
    self:UIKit Context Menu Action Context Object Reference=Get Context Menu Action Context()
    InActionInfo:"UIKit Context Menu Action Info Structure (by ref)"=Make UIKit Context Menu Action Info!(
     DisplayName:Text=Select(Option 0="LOCTEXT(\"Archiving Disabled for Unowned\")", Option 1=@n3, Index=@n4)
     IconTexture:Texture 2D Soft Object Reference
     IconMaterial:Material Instance Soft Object Reference
     SortPriority:int=Get ArchiveContextMenuSortPriority() -> (ArchiveContextMenuSortPriority:int=0)
     bIsAvailable:bool=True
     bIsEnabled:bool=True
     bIsSelected:bool=False
     bAddDivider:bool=False
     bPositionDividerAfter:bool=False
     bCloseMenuOnSelect:bool=False
     InteractionNameID:Name=None
     Interaction:Data Table Row Handle Structure=Get FavoriteInput() -> (FavoriteInput:Data Table Row Handle Structure)
     PressedDelegate:Delegate="(null).None"
     ReleasedDelegate:Delegate="(null).None"
    ) -> (UIKitContextMenuActionInfo:UIKit Context Menu Action Info Structure) {is=title K2Node_MakeStruct}
    InInteractionDelegate:"Delegate (by ref)"=@n1
   ) -> (then)
  ]}
 Make Favorite Context Menu Iteraction {nodes [
   n1 = Make Favorite Context Menu Iteraction!() -> (then, InInteractionDelegate:"Delegate (by ref)") {is=title K2Node_FunctionEntry}
   n2 = Get PilgrimSongCatalogItemVM()
   n3 = Select(
    Option 0="LOCTEXT(\"Favorite Item\")"
    Option 1="LOCTEXT(\"Unfavorite Item\")"
    Index=Get bIsFavorite!(self:Fort Item VM Object Reference=@n2) -> (bIsFavorite:bool)
   )
   n4 = Get bCanFavorite!(self:Fort Item VM Object Reference=@n2) -> (bCanFavorite:bool)
   > AddInteraction(
    self=Get Context Menu Action Context()
    InActionInfo=Make UIKit Context Menu Action Info(
     DisplayName=Select(Option 0="LOCTEXT(\"Favoriting Disabled for Unowned\")", Option 1=@n3, Index=@n4)
     SortPriority=Get FavoriteContextMenuSortPriority() -> (FavoriteContextMenuSortPriority:int=0)
     bIsAvailable=True
     bIsEnabled=True
     bIsSelected=False
     bAddDivider=False
     bPositionDividerAfter=False
     bCloseMenuOnSelect=False
     InteractionNameID=None
     Interaction=Get FavoriteInput()
     PressedDelegate="(null).None"
     ReleasedDelegate="(null).None"
    )
    InInteractionDelegate=@n1
   )
  ]}
 Make Leaderboard Context Menu Iteraction {nodes [
   n1 = Make Leaderboard Context Menu Iteraction!() -> (then, InInteractionDelegate:"Delegate (by ref)") {is=title K2Node_FunctionEntry}
   > AddInteraction(
    self=Get Context Menu Action Context()
    InActionInfo=Make UIKit Context Menu Action Info(
     DisplayName="LOCTEXT(\"Leaderboards\")"
     SortPriority=Get LeaderboardContextMenuSortPriority_0() -> (LeaderboardContextMenuSortPriority_0:int=0)
     bIsAvailable=True
     bIsEnabled=True
     bIsSelected=False
     bAddDivider=False
     bPositionDividerAfter=False
     bCloseMenuOnSelect=False
     InteractionNameID=None
     Interaction=Get FavoriteInput()
     PressedDelegate="(null).None"
     ReleasedDelegate="(null).None"
    )
    InInteractionDelegate=@n1
   )
  ]}
 On Handle Favorite Item Iteraction {nodes [
   On Handle Favorite Item Iteraction() {is=title K2Node_FunctionEntry}
   n1 = Get PilgrimSongCatalogItemVM()
   > Branch(Condition=Get bCanFavorite(self=@n1))
   > SetFavoriteStatus!(
    execute
    self:Pilgrim Song Catalog Item VM Object Reference=@n1
    bFavorite:bool=NOT Boolean(A=Get bIsFavorite(self=@n1))
    ShouldDeferMarking:bool=false
   ) -> (then, ReturnValue:bool)
  ]}
 ResetSpacerForSongScrolling {nodes [
   n1 = ResetSpacerForSongScrolling!() -> (then, PersonalBestResults:Pilgrim Personal Best Result Structure, PersonalBestParams:Pilgrim Personal Best Params Structure) {is=title K2Node_FunctionEntry}
   > n2 = ForEach EPilgrimPersonalBestType!(execute:Exec, SkipHidden:bool) -> (LoopBody:Exec, EnumValue:EPilgrimPersonalBestType Enum, then:Exec) {is=title K2Node_ForEachElementInEnum}
   Return Node!(execute=@n2.then)
   n3 = "To Integer (Byte)"!(InByte:Byte=Break Pilgrim Personal Best Params!(PilgrimPersonalBestParams:"Pilgrim Personal Best Params Structure (by ref)"=@n1.PersonalBestParams) -> (
     PlayerAccountId:Unique Net Id Structure
     SongShortName:Name
     InstrumentType:EPilgrimTrackType Enum
     ControllerId:int
     BandSize:int
     SetlistIndex:int
     bUseCache:bool=~
     FreshnessTime:int
     Priority:EPilgrimLeaderboardRequestPriority Enum
     SpotlightGroup:String
     BestScoreFlags:EPilgrimPersonalBestType Enum
    ) {is=title K2Node_BreakStruct}.BestScoreFlags) -> (ReturnValue:int=0)
   n4 = "To Integer (Byte)"(InByte=@n2) -> (ReturnValue=0)
   > Branch(Condition="To Boolean (Integer)"!(InInt:int=Bitwise AND(A:int=@n4, B:int=@n3) -> (ReturnValue:int=0) {is=title K2Node_CommutativeAssociativeBinaryOperator}) -> (ReturnValue:bool))
   > n5 = Switch on EPilgrimPersonalBestType!(execute, Selection:EPilgrimPersonalBestType Enum=@n2) -> (SoloScore:Exec, BandScore:Exec, SpotlightScore:Exec, AlltimeScore:Exec) {is=title K2Node_SwitchEnum}
   > n6 = Branch(Condition=Get IsExpanded())
   n7 = Break Pilgrim Personal Best Result!(PilgrimPersonalBestResult:"Pilgrim Personal Best Result Structure (by ref)"=Get PersonalBestResults() -> (PersonalBestResults:Pilgrim Personal Best Result Structure)) -> (
    BestSolo:Pilgrim Leaderboard Entry Structure
    BestBand:Pilgrim Leaderboard Entry Structure
    BestSpotlight:Pilgrim Leaderboard Entry Structure
    BestAlltime:Pilgrim Leaderboard Entry Structure
   ) {is=title K2Node_BreakStruct}
   n8 = Branch(
    execute=@n6.else
    Condition=Break Pilgrim Leaderboard Entry!(PilgrimLeaderboardEntry:"Pilgrim Leaderboard Entry Structure (by ref)"=@n7) -> (
     bHasLocalPlayer:bool=~
     PlayerAccountIdToDisplayNameMap:Map of Strings to Strings
     Rank:int
     Percentile:float=~
     ScoreData_Score:int
     ScoreData_StarsEarned:int
     ScoreData_Accuracy:float=~
     ScoreData_bFullCombo:bool=~
     ScoreData_Difficulty:EPilgrimSongDifficulty Enum
     ScoreData_InstrumentsPlayed:Array of EPilgrimTrackType Enums
     ScoreData_SeasonNumber:int
     DateTime:Date Time Structure
     FriendRank:int
    ) {is=title K2Node_BreakStruct}.ScoreData_bFullCombo
   )
   SetVisibility(execute=[@n8.else, @n14.else, @n10.else, @n12.else], self=Get WBP_PilgrimQuickplay_SeasonDisplay(), InVisibility=Collapsed)
   n9 = Branch(execute=@n5.SpotlightScore, Condition=Get IsExpanded())
   n10 = Branch(
    execute=@n9.else
    Condition=Break Pilgrim Leaderboard Entry(PilgrimLeaderboardEntry=@n7.BestSpotlight) -> (bHasLocalPlayer=~, Percentile=~, ScoreData_Accuracy=~, ScoreData_bFullCombo=~).ScoreData_bFullCombo
   )
   n11 = Branch(execute=@n5.AlltimeScore, Condition=Get IsExpanded())
   n12 = Branch(
    execute=@n11.else
    Condition=Break Pilgrim Leaderboard Entry(PilgrimLeaderboardEntry=@n7.BestAlltime) -> (bHasLocalPlayer=~, Percentile=~, ScoreData_Accuracy=~, ScoreData_bFullCombo=~).ScoreData_bFullCombo
   )
   n13 = Branch(execute=@n5.BandScore, Condition=Get IsExpanded())
   n14 = Branch(
    execute=@n13.else
    Condition=Break Pilgrim Leaderboard Entry(PilgrimLeaderboardEntry=@n7.BestBand) -> (bHasLocalPlayer=~, Percentile=~, ScoreData_Accuracy=~, ScoreData_bFullCombo=~).ScoreData_bFullCombo
   )
  ]}
 Set Horizontal Entry Highscore Data {nodes [
   n1 = Set Horizontal Entry Highscore Data!() -> (then, PersonalBestResult:Pilgrim Personal Best Result Structure, PilgrimPersonalBestParams:"Pilgrim Personal Best Params Structure (by ref)") {is=title K2Node_FunctionEntry}
   n2 = Break Pilgrim Personal Best Params(PilgrimPersonalBestParams=@n1.PilgrimPersonalBestParams) -> (bUseCache=~)
   n3 = "To Integer (Byte)"(InByte=@n2.BestScoreFlags) -> (ReturnValue=0)
   n4 = ForEach EPilgrimPersonalBestType(execute=@n15.then)
   n5 = "To Integer (Byte)"(InByte=@n4) -> (ReturnValue=0)
   n6 = Bitwise AND(A=@n5, B=@n3) -> (ReturnValue=0)
   n7 = "To Boolean (Integer)"(InInt=@n6)
   n8 = PrintText!(
    execute
    InText=Format Text!(
     Format="LOCTEXT(\"Checking Personal Best Params {EnumValue} && {BestScoreFlags} = {BitwiseAnd} == {OutCheck}\")"
     EnumValue:int=@n5
     BestScoreFlags:int=@n3
     BitwiseAnd:int=@n6
     OutCheck:bool=@n7
    ) -> (Result)
    bPrintToScreen
    bPrintToLog
    TextColor=(R=0.7483,G=0,B=0,A=1)
    Duration=8
    Key
   ) -> (then)
   Branch(execute=[@n8.then, @n4.LoopBody], Condition=@n7)
   > n9 = Switch on EPilgrimPersonalBestType(Selection=@n4)
   n10 = Get PersonalBestResult() -> (PersonalBestResult:Pilgrim Personal Best Result Structure)
   > Set Personal Best From Score Data!(
    execute
    self
    PilgrimLeaderboardScoreData:"Pilgrim Leaderboard Score Data Structure (by ref)"=Break Pilgrim Leaderboard Entry!(PilgrimLeaderboardEntry=Break Pilgrim Personal Best Result(PilgrimPersonalBestResult=@n10)) -> (
     bHasLocalPlayer=~
     PlayerAccountIdToDisplayNameMap
     Rank
     Percentile=~
     ScoreData:Pilgrim Leaderboard Score Data Structure
     DateTime
     FriendRank
    ).ScoreData
   ) -> (then) {is=title K2Node_CallFunction}
   n11 = Get PersonalBestResult()
   Set Personal Best From Score Data(
    execute=@n9.SpotlightScore
    PilgrimLeaderboardScoreData=Break Pilgrim Leaderboard Entry(PilgrimLeaderboardEntry=Break Pilgrim Personal Best Result(PilgrimPersonalBestResult=@n11).BestSpotlight) -> (bHasLocalPlayer=~, Percentile=~).ScoreData
   )
   n12 = Get PersonalBestResult()
   Set Personal Best From Score Data(
    execute=@n9.AlltimeScore
    PilgrimLeaderboardScoreData=Break Pilgrim Leaderboard Entry(PilgrimLeaderboardEntry=Break Pilgrim Personal Best Result(PilgrimPersonalBestResult=@n12).BestAlltime) -> (bHasLocalPlayer=~, Percentile=~).ScoreData
   )
   n13 = Get PersonalBestResult()
   Set Personal Best From Score Data(
    execute=@n9.BandScore
    PilgrimLeaderboardScoreData=Break Pilgrim Leaderboard Entry(PilgrimLeaderboardEntry=Break Pilgrim Personal Best Result(PilgrimPersonalBestResult=@n13).BestBand) -> (bHasLocalPlayer=~, Percentile=~).ScoreData
   )
   n14 = "To Integer (Byte)"(InByte=0) -> (ReturnValue=0)
   Floor(A=Power!(Base:double=2, Exp:double="To Float (Integer)"(InInt=@n14)) -> (ReturnValue:double) {is=title K2Node_CallFunction}) {is=title K2Node_CallFunction}
   Return Node(execute=@n4.then)
   n15 = Branch(execute=@n1.then, Condition="Equal (Name)"(A=Get Shortname() -> (Shortname=None), B=@n2.SongShortName))
  ]}
 Set Icon and Status Text {nodes [
   n1 = Set Icon and Status Text!() -> (then, Status Icon:"Texture 2D Soft Object Reference (by ref)", Status Text:Text) {is=title K2Node_FunctionEntry}
   > n2 = Branch(Condition=IsValidSoftObjectReference(SoftObjectReference=@n1))
   > SetBrushFromLazyTexture!(
    execute
    self:Common Lazy Image Object Reference=[
     Get FortLazyImage_Icon_Internal() -> (FortLazyImage_Icon_Internal:Fort Lazy Image Object Reference)
     Get FortLazyImage_Icon_Side() -> (FortLazyImage_Icon_Side:Fort Lazy Image Object Reference)
    ]
    LazyTexture:"Texture 2D Soft Object Reference (by ref)"=@n1
    bMatchSize:bool=false
   ) -> (then)
   > SetVisibility(
    self=[
     Get FortLazyImage_Icon_Side()
     Get CommonTextBlock_States() -> (CommonTextBlock_States:Common Text Object Reference)
     Get FortLazyImage_Icon_Internal()
    ]
    InVisibility=SelfHitTestInvisible
   )
   > "SetText (Text)"(self=Get CommonTextBlock_States(), InText=@n1.Status Text)
   SetVisibility(
    execute=@n2.else
    self=[Get FortLazyImage_Icon_Internal(), Get CommonTextBlock_States(), Get FortLazyImage_Icon_Side()]
    InVisibility=Collapsed
   )
  ]}
 Set Personal Best From Score Data {nodes [
   n1 = Set Personal Best From Score Data!() -> (then, PilgrimLeaderboardScoreData:"Pilgrim Leaderboard Score Data Structure (by ref)") {is=title K2Node_FunctionEntry}
   n2 = Break Pilgrim Leaderboard Score Data!(PilgrimLeaderboardScoreData:"Pilgrim Leaderboard Score Data Structure (by ref)"=@n1) -> (
    Score:int
    StarsEarned:int
    Accuracy:float=~
    bFullCombo:bool=~
    Difficulty:EPilgrimSongDifficulty Enum
    InstrumentsPlayed:Array of EPilgrimTrackType Enums
    SeasonNumber:int
   ) {is=title K2Node_BreakStruct}
   > Branch(Condition="integer > integer"(A=Length(TargetArray:Array of EPilgrimTrackType Enums=@n2.InstrumentsPlayed)))
   > Set Personal Best Stats(
    Score=@n2
    Full Combo=@n2.bFullCombo
    Accuracy=@n2.Accuracy
    Instrument="Get (a copy)"(Array:Array of EPilgrimTrackType Enums=@n2.InstrumentsPlayed) -> (Output:EPilgrimTrackType Enum)
    Stars Earned=@n2.StarsEarned
    Difficulty=@n2.Difficulty
    SeasonNumber=@n2.SeasonNumber
   )
   > n3 = Branch(Condition=Get IsExpanded())
   SetVisibility(execute=@n3.else, self=Get PercentageText(), InVisibility=Collapsed)
  ]}
 Set Personal Best Stats {nodes [
   n1 = Set Personal Best Stats!() -> (
    then
    Score:int
    Full Combo:bool=~
    Accuracy:float=~
    Instrument:EPilgrimTrackType Enum=TrackGuitar
    Stars Earned:int
    Difficulty:EPilgrimSongDifficulty Enum=DifficultyEasy
    SeasonNumber:int
   ) {is=title K2Node_FunctionEntry}
   > n2 = Sequence!(execute) -> (then_0, then_1, then_2:Exec, then_3:Exec)
   > SetPercentageText!(execute, self, PlayerPercentage:double=@n1.Accuracy, Full Combo:bool=@n1.Full Combo) -> (then)
   n3 = Get WBP_PilgrimQuickplay_LocalStarBar()
   Set Visibility(execute=@n2.then_1, Visibility=SelfHitTestInvisible, self=@n3)
   > SetupStarCount(self=@n3, StarCount=@n1.Stars Earned)
   n4 = Get DifficultyIcon()
   UpdateDifficultyState(execute=@n2.then_2, self=@n4, SongIsUnplayed=false)
   > SetDifficulty!(execute, self:WBP Pilgrim Results Difficulty Object Reference=@n4, Difficulty:EPilgrimSongDifficulty Enum=@n1.Difficulty) -> (then)
   UpdateSeasonDisplay!(execute=@n2.then_3, self, SeasonNumber:int=@n1.SeasonNumber) -> (then)
  ]}
 SetInstrumentIcon {nodes [
   n1 = SetInstrumentIcon!() -> (then, InstrumentType:EPilgrimTrackType Enum=TrackGuitar) {is=title K2Node_FunctionEntry}
   > n2 = Get!(execute) -> (then, else, SongMetadata:Pilgrim Song Metadata Object Reference)
   > SetInstrumentIcon!(
    execute
    self:WBP Pilgrim Quickplay Song Intensity Entry Object Reference=Get InstrumentIntensity() -> (InstrumentIntensity:WBP Pilgrim Quickplay Song Intensity Entry Object Reference)
    InstrumentType=@n1
    StartingInstrument:ESparksInstrumentType Enum=GetStartingInstrumentType!(self:Pilgrim Song Metadata Object Reference=@n2, TrackType:EPilgrimTrackType Enum=@n1) -> (ReturnValue:ESparksInstrumentType Enum=Guitar)
   ) -> (then) {is=title K2Node_CallFunction}
  ]}
 SetItemVMs {nodes [
   n1 = SetItemVMs!() -> (then, Viewmodel:Pilgrim Song Catalog Item VM Object Reference) {is=title K2Node_FunctionEntry}
   > SetPilgrimSongCatalogItemVM!(execute, self, Viewmodel:Pilgrim Song Catalog Item VM Object Reference=@n1) -> (then)
   > n2 = Cast To PilgrimSongCatalogItemDetailsVM!(execute, Object=Get ItemDetails!(self:Fort Item VM Object Reference=@n1) -> (ItemDetails:Fort Item Details VM Object Reference)) -> (then, CastFailed, AsPilgrim Song Catalog Item Details VM:Pilgrim Song Catalog Item Details VM Object Reference) {is=title K2Node_DynamicCast}
   > SetPilgrimSongCatalogItemDetailsVM!(execute, self, Viewmodel:Pilgrim Song Catalog Item Details VM Object Reference=@n2) -> (then)
   SetPilgrimSongCatalogItemDetailsVM(execute=@n2.CastFailed)
  ]}
 SetPercentageText {nodes [
   n1 = SetPercentageText!() -> (then, PlayerPercentage:double=~, Full Combo:bool=~) {is=title K2Node_FunctionEntry}
   > n2 = Branch(Condition=@n1.Full Combo)
   > SetVisibility(self=Get FullComboIndicator(), InVisibility=SelfHitTestInvisible)
   > SetVisibility(self=Get PercentageText(), InVisibility=Collapsed)
   > SetActiveWidgetIndex(
    self:Common Visibility Switcher Object Reference=Get FullCombo_Switcher() -> (FullCombo_Switcher:Common Visibility Switcher Object Reference)
    Index=1
   )
   > Branch(Condition=Get IsExpanded())
   > SetVisibility(self=Get WBP_PilgrimQuickplay_SeasonDisplay(), InVisibility=SelfHitTestInvisible)
   n3 = Branch(execute=@n2.else, Condition=Get IsExpanded())
   > "SetText (Text)"(
    self=Get PercentageText()
    InText=Format Text!(Format="LOCTEXT(\"{Accuracy}%\")", Accuracy:int=Round(A=@n1) {is=title K2Node_CallFunction}) -> (Result)
   )
   > SetVisibility(self=Get FullComboIndicator(), InVisibility=Collapsed)
   > SetVisibility(self=Get PercentageText(), InVisibility=SelfHitTestInvisible)
   > SetActiveWidgetIndex(self=Get FullCombo_Switcher(), Index=0)
   > SetVisibility(self=Get WBP_PilgrimQuickplay_SeasonDisplay(), InVisibility=SelfHitTestInvisible)
   SetVisibility(execute=@n3.else, self=Get FullComboIndicator(), InVisibility=Collapsed)
   > SetVisibility(self=Get PercentageText(), InVisibility=Collapsed)
   > SetVisibility(self=Get WBP_PilgrimQuickplay_SeasonDisplay(), InVisibility=Collapsed)
  ]}
 StartStopSongPreview {nodes [
   n1 = StartStopSongPreview!() -> (then, Start:bool=~) {is=title K2Node_FunctionEntry}
   > n2 = Branch(Condition=@n1)
   n3 = Get SongPreviewer!(self:BP Pilgrim Quickplay Playspace Object Reference=GetSparksMusicPlayspaceOfClass!(
     PlayspaceType:Sparks Music Playspace Class Reference=/PilgrimQuickplay/Playspaces/BP_PilgrimQuickplayPlayspace.BP_PilgrimQuickplayPlayspace_C
     Actor:Actor Object Reference=GetOwningPlayer()
     bMatchChildTypes:bool=false
    ) -> (ReturnValue:BP Pilgrim Quickplay Playspace Object Reference)) -> (SongPreviewer:BPC Sparks Song Previewer Object Reference)
   n4 = IsPlayingPreviewForSong!(self:Sparks Song Previewer Component Object Reference=@n3, Song:Name=Get Shortname() -> (Shortname=None)) -> (ReturnValue:bool)
   > Branch(Condition=NOT Boolean(A=@n4))
   > StartPreviewByShortName!(execute, self:Sparks Song Previewer Component Object Reference=@n3, ShortName:Name=Get Shortname() -> (Shortname=None)) -> (then)
   Branch(execute=@n2.else, Condition=@n4)
   > StopCurrentPreview!(execute, self:Sparks Song Previewer Component Object Reference=@n3, bWithFadeout:bool=true) -> (then)
  ]}
 TryCollapse {nodes [
   TryCollapse() {is=title K2Node_FunctionEntry}
   n1 = Get Selected_Animated()
   > Branch(Condition=OR Boolean!(A=IsAnimationPlaying(InAnimation=@n1), B=Get IsExpanded()) -> (ReturnValue))
   > PlayAnimationReverse(InAnimation=Get Selected_Animated(), PlaybackSpeed=Get SelectAnimPlaybackSpeed())
   > Set IsExpanded(IsExpanded=false)
  ]}
 TryExpand {nodes [
   TryExpand() {is=title K2Node_FunctionEntry}
   n1 = Get IsExpanded()
   n2 = Get Selected_Animated()
   > Branch(Condition=OR Boolean(A=IsAnimationPlaying(InAnimation=@n2), B=NOT Boolean(A=@n1)))
   > PlayAnimationForward(InAnimation=Get Selected_Animated(), PlaybackSpeed=Get SelectAnimPlaybackSpeed())
   > Set IsExpanded(IsExpanded=true)
  ]}
 Update Status Icon {nodes [
   n1 = Update Status Icon!() -> (then, Status:ESparksSongStatusFlags Enum=None) {is=title K2Node_FunctionEntry}
   > Set Icon and Status Text!(
    execute
    self
    Status Icon:"Texture 2D Soft Object Reference (by ref)"=Select!(
     None:"Texture 2D Soft Object Reference (by ref)"
     Disabled:"Texture 2D Soft Object Reference (by ref)"
     Owned:"Texture 2D Soft Object Reference (by ref)"
     Shared:"Texture 2D Soft Object Reference (by ref)"=Get SideIconShared() -> (SideIconShared:Texture 2D Soft Object Reference)
     New:"Texture 2D Soft Object Reference (by ref)"=Get SideIconNew() -> (SideIconNew:Texture 2D Soft Object Reference)
     Featured:"Texture 2D Soft Object Reference (by ref)"=Get SideIconFeatured() -> (SideIconFeatured:Texture 2D Soft Object Reference)
     Spotlighted:"Texture 2D Soft Object Reference (by ref)"=Get SideIconSpotlight() -> (SideIconSpotlight:Texture 2D Soft Object Reference)
     Index:ESparksSongStatusFlags Enum=@n1
    ) -> (ReturnValue:"Texture 2D Soft Object Reference (by ref)")
    Status Text:Text=Select(
     None:Text
     Disabled:Text
     Owned:Text
     Shared:Text=Get StatusTextShared() -> (StatusTextShared:Text)
     New:Text=Get StatusTextNew() -> (StatusTextNew:Text)
     Featured:Text=Get StatusTextFeatured() -> (StatusTextFeatured:Text)
     Spotlighted:Text=Get StatusTextSpotlight() -> (StatusTextSpotlight:Text)
     Index=@n1
    ) -> (ReturnValue:Text)
   ) -> (then) {is=title K2Node_CallFunction}
  ]}
 UpdateForFlawless {nodes [
   n1 = UpdateForFlawless!() -> (then, FullComboIndex:int) {is=title K2Node_FunctionEntry}
   > n2 = Switch on Int!(execute:Exec, Selection:int=@n1) -> (Default:Exec, 0:Exec, 1:Exec, 2:Exec, 3:Exec)
   SetActiveWidgetIndex(
    execute=@n2.0
    self=Get CommonVisibilitySwitcher!(self:WBP Pilgrim Quickplay Full Combo Indicator Object Reference=Get FullComboIndicator()) -> (CommonVisibilitySwitcher:Common Visibility Switcher Object Reference)
    Index=0
   )
   SetActiveWidgetIndex(execute=@n2.1, self=Get CommonVisibilitySwitcher(self=Get FullComboIndicator()), Index=1)
   SetActiveWidgetIndex(execute=@n2.2, self=Get CommonVisibilitySwitcher(self=Get FullComboIndicator()), Index=2)
   SetActiveWidgetIndex(execute=@n2.3, self=Get CommonVisibilitySwitcher(self=Get FullComboIndicator()), Index=3)
  ]}
 UpdateSeasonDisplay {nodes [
   n1 = UpdateSeasonDisplay!() -> (then, SeasonNumber:int) {is=title K2Node_FunctionEntry}
   > n2 = EvaluateCurveTableRow!(
    execute
    CurveTableRowHandle:Curve Table Row Handle Structure=Make Curve Table Row Handle!(CurveTable:Curve Table Object Reference=/PilgrimQuickplay/DataTables/PilgrimQuickplayGameData, RowName:Name=SeasonData.SeasonNumber) -> (CurveTableRowHandle:Curve Table Row Handle Structure) {is=title K2Node_MakeStruct}
    InXY:float=0
    ContextString:String
   ) -> (then, OutXY:float, ReturnValue:bool)
   > n3 = Branch(Condition="Equal (Float)"(A=@n2, B:int=@n1))
   > n4 = SetCurrentSeasonBorder(self:WBP Pilgrim Quickplay Season Display Object Reference=Get WBP_PilgrimQuickplay_SeasonDisplay())
   "SetText (Text)"(
    execute=[@n4.then, @n5.then]
    self=Get SeasonText!(self:WBP Pilgrim Quickplay Season Display Object Reference=Get WBP_PilgrimQuickplay_SeasonDisplay()) -> (SeasonText:Common Text Object Reference)
    InText=Format Text!(Format="LOCTEXT(\"S{SeasonNumber}\")", SeasonNumber:int=@n1) -> (Result)
   )
   n5 = RevertSeasonBorder(execute=@n3.else, self:WBP Pilgrim Quickplay Season Display Object Reference=Get WBP_PilgrimQuickplay_SeasonDisplay())
  ]}
]
variables [
 ArchiveContextMenuSortPriority {type=integer, default=3000}
 Context Menu Action Context {type=UIKitContextMenuActionContext, default=null}
 DebugButtonInput {type=boolean}
 DebugDuration {type=number, default=10}
 EmptyAlbumCover {type=Texture2D, default=/FortniteCoreUI/Customization/Art/T_UI_EmptySlot}
 FavoriteContextMenuSortPriority {type=integer, default=2000}
 FavoriteInput {type=DataTableRowHandle, default {dataTable=/FortUILibrary/Input/LockerInputActionsDataTable, rowName=ContextFavorite}}
 IsExpanded {type=boolean}
 LeaderboardContextMenuSortPriority_0 {type=integer, default=1000}
 PilgrimRequestPersonalBestHandle {type=PilgrimAsyncAction_RequestPersonalBest, default=null}
 SelectAnimPlaybackSpeed {type=number, default=5}
 Shortname {type=string, default=None}
 SideIconFeatured {type=Texture2D, default=/FMCore/Textures/T_Icon_Featured}
 SideIconNew {type=Texture2D, default=/FMCore/Textures/T_Icon_New}
 SideIconShared {type=Texture2D, default=/FMCore/Textures/T_Icon_Shared}
 SideIconSpotlight {type=Texture2D, default=/FMCore/Textures/T_Icon_Spotlight}
 SongMetadata {type=PilgrimSongMetadata, default=null}
 StartTimerHandle {type=TimerHandle}
 StatusTextFeatured {type=string, default=Featured}
 StatusTextNew {type=string, default=New}
 StatusTextShared {type=string, default=Shared}
 StatusTextSpotlight {type=string, default=Spotlight}
]
widgets {
 info {inheritedWidgetCount=1, parentClass=&CommonButtonBase, rootWidgetClass=&GridPanel, widgetCount=45}
 widgets [
  MainContainer {widgetClassPath=&GridPanel, widget=MainContainer, properties {columnFill [1], rowFill [1]}}
  CommonBorder_Backplate {
   widgetClassPath=&CommonBorder
   bIsVariable=true
   parent=MainContainer
   widget=CommonBorder_Backplate
   properties {
    background {
     bIsDynamicallyLoaded=false
     drawAs=Image
     imageType=NoImage
     mirroring=NoMirror
     resourceName=None
     resourceObject=/SparksUICommon/Materials/MI_UI_Sparks_CatalogueEntry
     tiling=NoTile
     imageSize=<32 32>
     margin=<ltrb 0.5 0.5 0.5 0.5>
     outlineSettings {
      bUseBrushTransparency=false
      roundingType=FixedRadius
      width=0
      color {colorUseRule=UseColor_Specified, specifiedColor=<rgba 0 0 0 0>}
      cornerRadii=<10 10 10 10>
     }
     tintColor {colorUseRule=UseColor_Specified, specifiedColor=<rgba 1 1 1 1>}
     uVRegion {bIsValid=false, max=<0 0>, min=<0 0>}
    }
    padding=<ltrb 0 0 0 0>
   }
   slot {class=GridSlot, properties {padding=<ltrb 2 0 32 0>}}
  }
  VerticalBox_21 {
   widgetClassPath=&VerticalBox
   parent=MainContainer
   widget=VerticalBox_21
   slot {class=GridSlot, properties {layer=1, padding=<ltrb 0 0 32 0>}}
  }
  HorizontalBox_34 {widgetClassPath=&HorizontalBox, parent=VerticalBox_21, widget=HorizontalBox_34, slot {class=VerticalBoxSlot}}
  AlbumArtSizeBox {
   widgetClassPath=&SizeBox
   parent=HorizontalBox_34
   widget=AlbumArtSizeBox
   properties {
    bOverride_MaxAspectRatio=true
    bOverride_MinAspectRatio=true
    bOverride_MinDesiredHeight=true
    bOverride_MinDesiredWidth=true
    heightOverride=115
    minDesiredHeight=138
    minDesiredWidth=138
    widthOverride=118
   }
   slot {class=HorizontalBoxSlot}
  }
  albumCover {
   widgetClassPath=&FortLazyImage
   bIsVariable=true
   parent=AlbumArtSizeBox
   widget=albumCover
   properties {
    materialTextureParamName=T_AlbumArt
    brush {
     bIsDynamicallyLoaded=false
     drawAs=Image
     imageType=FullColor
     mirroring=NoMirror
     resourceName=None
     resourceObject=/SparksUICommon/Materials/MI_UI_Sparks_SetList_AlbumArt_Rounded_Inst
     tiling=NoTile
     imageSize=<138 138>
     margin=<ltrb 0 0 0 0>
     outlineSettings {
      bUseBrushTransparency=false
      roundingType=HalfHeightRadius
      width=0
      color {colorUseRule=UseColor_Specified, specifiedColor=<rgba 0 0 0 0>}
      cornerRadii=<0 0 0 0>
     }
     tintColor {colorUseRule=UseColor_Specified, specifiedColor=<rgba 1 1 1 1>}
     uVRegion {bIsValid=false, max=<0 0>, min=<0 0>}
    }
    loadingThrobberBrush {
     bIsDynamicallyLoaded=false
     drawAs=Image
     imageType=FullColor
     mirroring=NoMirror
     resourceName=None
     resourceObject=/FortUILibrary/Systems/Items/MaterialInstances/MI_UIKit_Item_Loading
     tiling=NoTile
     imageSize=<138 138>
     margin=<ltrb 0 0 0 0>
     outlineSettings {
      bUseBrushTransparency=false
      roundingType=HalfHeightRadius
      width=0
      color {colorUseRule=UseColor_Specified, specifiedColor=<rgba 0 0 0 0>}
      cornerRadii=<0 0 0 0>
     }
     tintColor {colorUseRule=UseColor_Specified, specifiedColor=<rgba 1 1 1 1>}
     uVRegion {bIsValid=false, max=<0 0>, min=<0 0>}
    }
   }
   slot {class=SizeBoxSlot}
  }
  SongInfoContent {
   widgetClassPath=&Overlay
   parent=HorizontalBox_34
   widget=SongInfoContent
   slot {class=HorizontalBoxSlot, properties {padding=<ltrb 24 24 36 24>, size {sizeRule=Fill, value=1}}}
  }
  SongInfoBox {
   widgetClassPath=&HorizontalBox
   parent=SongInfoContent
   widget=SongInfoBox
   slot {class=OverlaySlot, properties {horizontalAlignment=HAlign_Fill}}
  }
  TitleArtistBox {
   widgetClassPath=&VerticalBox
   parent=SongInfoBox
   widget=TitleArtistBox
   slot {class=HorizontalBoxSlot, properties {size {sizeRule=Fill, value=0.9}}}
  }
  SongTitle {
   widgetClassPath=&CommonTextBlock
   bIsVariable=true
   parent=TitleArtistBox
   widget=SongTitle
   properties {
    scrollStyle=/PilgrimQuickplay/Widgets/ScrollStyles/TextScrollStyle_PilgrimQuickplay_Setlist.TextScrollStyle_PilgrimQuickplay_Setlist_C
    text="[Song Title]"
    textTransformPolicy=ToUpper
    visibility=HitTestInvisible
    font {
     bForceMonospaced=false
     bMaterialIsStencil=false
     fontMaterial=null
     fontObject=&BurbankSmall
     letterSpacing=0
     monospacedWidth=1
     size=24
     skewAmount=0
     typefaceFontName=Black
     outlineSettings {
      bApplyOutlineToDropShadows=false
      bMiteredCorners=false
      bSeparateFillAlpha=false
      outlineMaterial=null
      outlineSize=0
      outlineColor=<rgba 0 0 0 1>
     }
    }
    shadowOffset=<0 0>
   }
   slot {class=VerticalBoxSlot, properties {verticalAlignment=VAlign_Top, size {sizeRule=Fill, value=1}}}
  }
  SongArtist {
   widgetClassPath=&CommonTextBlock
   bIsVariable=true
   parent=TitleArtistBox
   widget=SongArtist
   properties {
    clipping=OnDemand
    scrollStyle=&TextScrollStyle_PilgrimQuickplay_Setlist_C
    text="[Artist]"
    textTransformPolicy=ToUpper
    visibility=HitTestInvisible
    colorAndOpacity {colorUseRule=UseColor_Specified, specifiedColor=<rgba 1 1 1 0.6>}
    font {
     bForceMonospaced=false
     bMaterialIsStencil=false
     fontMaterial=null
     fontObject=&BurbankSmall
     letterSpacing=0
     monospacedWidth=1
     size=13
     skewAmount=0
     typefaceFontName=Black
     outlineSettings {
      bApplyOutlineToDropShadows=false
      bMiteredCorners=false
      bSeparateFillAlpha=false
      outlineMaterial=null
      outlineSize=0
      outlineColor=<rgba 0 0 0 1>
     }
    }
    shadowOffset=<0 0>
   }
   slot {class=VerticalBoxSlot, properties {verticalAlignment=VAlign_Top, size {sizeRule=Fill, value=1}}}
  }
  ReleaseDurationHorizontalBox {
   widgetClassPath=&HorizontalBox
   parent=TitleArtistBox
   widget=ReleaseDurationHorizontalBox
   slot {class=VerticalBoxSlot, properties {verticalAlignment=VAlign_Top, size {sizeRule=Fill, value=1}}}
  }
  SongReleaseDuration {
   widgetClassPath=&CommonTextBlock
   parent=ReleaseDurationHorizontalBox
   widget=SongReleaseDuration
   properties {
    clipping=OnDemand
    text=2019 | 3:09
    textTransformPolicy=ToUpper
    visibility=HitTestInvisible
    colorAndOpacity {colorUseRule=UseColor_Specified, specifiedColor=<rgba 1 1 1 0.5>}
    font {
     bForceMonospaced=false
     bMaterialIsStencil=false
     fontMaterial=null
     fontObject=&HeadingNow
     letterSpacing=0
     monospacedWidth=1
     size=13
     skewAmount=0
     typefaceFontName=86 Bold
     outlineSettings {
      bApplyOutlineToDropShadows=false
      bMiteredCorners=false
      bSeparateFillAlpha=false
      outlineMaterial=null
      outlineSize=0
      outlineColor=<rgba 0 0 0 1>
     }
    }
    shadowOffset=<0 0>
   }
   slot {class=HorizontalBoxSlot}
  }
  FortLazyImage_Icon_Side {
   widgetClassPath=&FortLazyImage
   bIsVariable=true
   parent=ReleaseDurationHorizontalBox
   widget=FortLazyImage_Icon_Side
   properties {
    brush {
     bIsDynamicallyLoaded=false
     drawAs=Image
     imageType=NoImage
     mirroring=NoMirror
     resourceName=None
     resourceObject=&T_Icon_Featured
     tiling=NoTile
     imageSize=<21 21>
     margin=<ltrb 0 0 0 0>
     outlineSettings {
      bUseBrushTransparency=false
      roundingType=HalfHeightRadius
      width=0
      color {colorUseRule=UseColor_Specified, specifiedColor=<rgba 0 0 0 0>}
      cornerRadii=<0 0 0 0>
     }
     tintColor {colorUseRule=UseColor_Specified, specifiedColor=<rgba 1 1 1 1>}
     uVRegion {bIsValid=false, max=<0 0>, min=<0 0>}
    }
    colorAndOpacity=<rgba 0.4125 0.5089 0.9647 1>
   }
   slot {class=HorizontalBoxSlot, properties {horizontalAlignment=HAlign_Right, verticalAlignment=VAlign_Center, padding=<ltrb 10 0 0 0>}}
  }
  StatusIconBox {
   widgetClassPath=&HorizontalBox
   parent=TitleArtistBox
   widget=StatusIconBox
   properties {visibility=Collapsed}
   slot {class=VerticalBoxSlot, properties {horizontalAlignment=HAlign_Left, padding=<ltrb 0 8 0 0>}}
  }
  FortLazyImage_Icon_Internal {
   widgetClassPath=&FortLazyImage
   bIsVariable=true
   parent=StatusIconBox
   widget=FortLazyImage_Icon_Internal
   properties {
    brush {
     bIsDynamicallyLoaded=false
     drawAs=Image
     imageType=NoImage
     mirroring=NoMirror
     resourceName=None
     resourceObject=&T_Icon_Featured
     tiling=NoTile
     imageSize=<32 32>
     margin=<ltrb 0 0 0 0>
     outlineSettings {
      bUseBrushTransparency=false
      roundingType=HalfHeightRadius
      width=0
      color {colorUseRule=UseColor_Specified, specifiedColor=<rgba 0 0 0 0>}
      cornerRadii=<0 0 0 0>
     }
     tintColor {colorUseRule=UseColor_Specified, specifiedColor=<rgba 1 1 1 1>}
     uVRegion {bIsValid=false, max=<0 0>, min=<0 0>}
    }
    colorAndOpacity=<rgba 0.4125 0.5089 0.9647 1>
   }
   slot {class=HorizontalBoxSlot, properties {horizontalAlignment=HAlign_Right, verticalAlignment=VAlign_Center, padding=<ltrb 0 0 10 0>}}
  }
  CommonTextBlock_States {
   widgetClassPath=&CommonTextBlock
   bIsVariable=true
   parent=StatusIconBox
   widget=CommonTextBlock_States
   properties {
    text="[Status]"
    visibility=Collapsed
    colorAndOpacity {colorUseRule=UseColor_Specified, specifiedColor=<rgba 0.4125 0.5089 0.9647 1>}
    font {
     bForceMonospaced=false
     bMaterialIsStencil=false
     fontMaterial=null
     fontObject=&NotoSans
     letterSpacing=0
     monospacedWidth=1
     size=24
     skewAmount=0
     typefaceFontName=NotoSans-Regular
     outlineSettings {
      bApplyOutlineToDropShadows=false
      bMiteredCorners=false
      bSeparateFillAlpha=false
      outlineMaterial=null
      outlineSize=0
      outlineColor=<rgba 0 0 0 1>
     }
    }
    shadowOffset=<0 0>
   }
   slot {class=HorizontalBoxSlot, properties {horizontalAlignment=HAlign_Right, verticalAlignment=VAlign_Center}}
  }
  HighScoreInfo {
   widgetClassPath=&VerticalBox
   parent=SongInfoContent
   widget=HighScoreInfo
   slot {class=OverlaySlot, properties {horizontalAlignment=HAlign_Right, verticalAlignment=VAlign_Bottom, padding=<ltrb 0 0 0 -15>}}
  }
  TopSpacer {
   widgetClassPath=&Spacer
   parent=HighScoreInfo
   widget=TopSpacer
   slot {class=VerticalBoxSlot, properties {horizontalAlignment=HAlign_Left, size {sizeRule=Fill, value=1}}}
  }
  IntensitiesBorder {
   widgetClassPath=&CommonBorder
   parent=HighScoreInfo
   widget=IntensitiesBorder
   properties {visibility=Collapsed, contentColorAndOpacity=<rgba 0 0 0 0>, padding=<ltrb 0 0 0 0>}
   slot {class=VerticalBoxSlot, properties {horizontalAlignment=HAlign_Right}}
  }
  IntensitiesScaleBox {
   widgetClassPath=&ScaleBox
   parent=IntensitiesBorder
   widget=IntensitiesScaleBox
   properties {stretch=UserSpecified, userSpecifiedScale=0}
   slot {class=BorderSlot, properties {padding=<ltrb 0 0 0 0>}}
  }
  InstrumentIntensity {
   widgetClassPath=/PilgrimQuickplay/Widgets/SetlistAndSongDisplay/WBP_PilgrimQuickplay_SongIntensityEntry
   bIsVariable=true
   parent=IntensitiesScaleBox
   widget=InstrumentIntensity
   properties {visibility=HitTestInvisible}
   slot {class=ScaleBoxSlot, properties {horizontalAlignment=HAlign_Fill, verticalAlignment=VAlign_Fill}}
  }
  MissingParts_Switcher {
   widgetClassPath=&CommonVisibilitySwitcher
   bIsVariable=true
   parent=HighScoreInfo
   widget=MissingParts_Switcher
   slot {class=VerticalBoxSlot}
  }
  HB_Leaderboard {widgetClassPath=&HorizontalBox, parent=MissingParts_Switcher, widget=HB_Leaderboard, slot {class=CommonVisibilitySwitcherSlot}}
  WBP_PilgrimQuickplay_SeasonDisplay {
   widgetClassPath=/PilgrimQuickplay/Widgets/WBP_PilgrimQuickplay_SeasonDisplay
   bIsVariable=true
   parent=HB_Leaderboard
   widget=WBP_PilgrimQuickplay_SeasonDisplay
   slot {class=HorizontalBoxSlot, properties {verticalAlignment=VAlign_Center}}
  }
  FullCombo_Switcher {
   widgetClassPath=&CommonVisibilitySwitcher
   bIsVariable=true
   parent=HB_Leaderboard
   widget=FullCombo_Switcher
   slot {class=HorizontalBoxSlot, properties {horizontalAlignment=HAlign_Center, verticalAlignment=VAlign_Center, padding=<ltrb 5 0 0 0>}}
  }
  PercentageText {
   widgetClassPath=&CommonTextBlock
   bIsVariable=true
   parent=FullCombo_Switcher
   widget=PercentageText
   properties {
    text=99%
    visibility=Collapsed
    colorAndOpacity {colorUseRule=UseColor_Specified, specifiedColor=<rgba 0.4125 0.5089 0.9647 1>}
    font {
     bForceMonospaced=false
     bMaterialIsStencil=false
     fontMaterial=null
     fontObject=&HeadingNow
     letterSpacing=0
     monospacedWidth=1
     size=18
     skewAmount=0
     typefaceFontName=74 Regular
     outlineSettings {
      bApplyOutlineToDropShadows=false
      bMiteredCorners=false
      bSeparateFillAlpha=false
      outlineMaterial=null
      outlineSize=0
      outlineColor=<rgba 1 1 1 1>
     }
    }
    shadowOffset=<0 0>
   }
   slot {
    class=CommonVisibilitySwitcherSlot
    properties {horizontalAlignment=HAlign_Center, verticalAlignment=VAlign_Center, padding=<ltrb 10 0 10 0>}
   }
  }
  FullComboIndicator {
   widgetClassPath=/PilgrimQuickplay/Widgets/WBP_PilgrimQuickplay_FullComboIndicator
   bIsVariable=true
   parent=FullCombo_Switcher
   widget=FullComboIndicator
   properties {showIconOnly=true, visibility=Collapsed}
   slot {class=CommonVisibilitySwitcherSlot, properties {horizontalAlignment=HAlign_Center, verticalAlignment=VAlign_Center}}
  }
  SizeBox_0 {
   widgetClassPath=&SizeBox
   parent=HB_Leaderboard
   widget=SizeBox_0
   properties {bOverride_MaxDesiredHeight=true, bOverride_MaxDesiredWidth=true, maxDesiredHeight=48, maxDesiredWidth=216}
   slot {class=HorizontalBoxSlot, properties {horizontalAlignment=HAlign_Right, verticalAlignment=VAlign_Center}}
  }
  WBP_PilgrimQuickplay_LocalStarBar {
   widgetClassPath=/PilgrimQuickplay/Widgets/WBP_PilgrimQuickplay_LocalStarBar
   bIsVariable=true
   parent=SizeBox_0
   widget=WBP_PilgrimQuickplay_LocalStarBar
   properties {visibility=Collapsed}
   slot {class=SizeBoxSlot, properties {horizontalAlignment=HAlign_Center, verticalAlignment=VAlign_Center, padding=<ltrb 7 0 0 0>}}
  }
  DifficultyIcon {
   widgetClassPath=/PilgrimQuickplay/Widgets/Results/WBP_Pilgrim_Results_Difficulty
   bIsVariable=true
   parent=HB_Leaderboard
   widget=DifficultyIcon
   slot {class=HorizontalBoxSlot, properties {horizontalAlignment=HAlign_Center, verticalAlignment=VAlign_Center, padding=<ltrb 10 0 0 0>}}
  }
  HB_MissingPart {
   widgetClassPath=&HorizontalBox
   parent=MissingParts_Switcher
   widget=HB_MissingPart
   slot {class=CommonVisibilitySwitcherSlot, properties {horizontalAlignment=HAlign_Center, verticalAlignment=VAlign_Center, padding=<ltrb 10 0 0 0>}}
  }
  MissingPartIcon {
   widgetClassPath=&Image
   parent=HB_MissingPart
   widget=MissingPartIcon
   properties {
    visibility=SelfHitTestInvisible
    brush {
     bIsDynamicallyLoaded=false
     drawAs=Image
     imageType=FullColor
     mirroring=NoMirror
     resourceName=None
     resourceObject=/PilgrimQuickplay/UI/MI_UI_MissingPartsAlertIcon
     tiling=NoTile
     imageSize=<23 23>
     margin=<ltrb 0 0 0 0>
     outlineSettings {
      bUseBrushTransparency=false
      roundingType=HalfHeightRadius
      width=0
      color {colorUseRule=UseColor_Specified, specifiedColor=<rgba 0 0 0 0>}
      cornerRadii=<0 0 0 0>
     }
     tintColor {colorUseRule=UseColor_Specified, specifiedColor=<rgba 0.9301 1 0.0103 1>}
     uVRegion {bIsValid=false, max=<0 0>, min=<0 0>}
    }
   }
   slot {class=HorizontalBoxSlot, properties {verticalAlignment=VAlign_Center}}
  }
  Spacer_65 {widgetClassPath=&Spacer, parent=HB_MissingPart, widget=Spacer_65, properties {size=<4 1>}, slot {class=HorizontalBoxSlot}}
  NoPartsAvailableText {
   widgetClassPath=&CommonTextBlock
   parent=HB_MissingPart
   widget=NoPartsAvailableText
   properties {
    text=No Part Available
    colorAndOpacity {colorUseRule=UseColor_Specified, specifiedColor=<rgba 0.4125 0.5089 0.9647 1>}
    font {
     bForceMonospaced=false
     bMaterialIsStencil=false
     fontMaterial=null
     fontObject=&HeadingNow
     letterSpacing=0
     monospacedWidth=1
     size=18
     skewAmount=0
     typefaceFontName=74 Regular
     outlineSettings {
      bApplyOutlineToDropShadows=false
      bMiteredCorners=false
      bSeparateFillAlpha=false
      outlineMaterial=null
      outlineSize=0
      outlineColor=<rgba 1 1 1 1>
     }
    }
    shadowOffset=<0 0>
   }
   slot {class=HorizontalBoxSlot}
  }
  BottomSpacer {
   widgetClassPath=&Spacer
   parent=HighScoreInfo
   widget=BottomSpacer
   slot {class=VerticalBoxSlot, properties {horizontalAlignment=HAlign_Left, size {sizeRule=Fill, value=1}}}
  }
  ContextMenu_Favorite_Overlay {
   widgetClassPath=&Overlay
   bIsVariable=true
   parent=SongInfoContent
   widget=ContextMenu_Favorite_Overlay
   slot {class=OverlaySlot, properties {horizontalAlignment=HAlign_Right, verticalAlignment=VAlign_Fill, padding=<ltrb 0 12 12 0>}}
  }
  CommonVisibilitySwitcher_Favorite_Icons {
   widgetClassPath=&CommonVisibilitySwitcher
   bIsVariable=true
   parent=ContextMenu_Favorite_Overlay
   widget=CommonVisibilitySwitcher_Favorite_Icons
   slot {class=OverlaySlot, properties {horizontalAlignment=HAlign_Right, padding=<ltrb 0 -35 -30 0>}}
  }
  FortLazyImage_Icon_Favorite_Base {
   widgetClassPath=&FortLazyImage
   parent=CommonVisibilitySwitcher_Favorite_Icons
   widget=FortLazyImage_Icon_Favorite_Base
   properties {
    brush {
     bIsDynamicallyLoaded=false
     drawAs=Image
     imageType=FullColor
     mirroring=NoMirror
     resourceName=None
     resourceObject=/FortUILibrary/Foundation/Buttons/Materials/Favorite/MI_UI_Favorite_Base_White
     tiling=NoTile
     imageSize=<32 32>
     margin=<ltrb 0 0 0 0>
     outlineSettings {
      bUseBrushTransparency=false
      roundingType=HalfHeightRadius
      width=0
      color {colorUseRule=UseColor_Specified, specifiedColor=<rgba 0 0 0 0>}
      cornerRadii=<0 0 0 0>
     }
     tintColor {colorUseRule=UseColor_Specified, specifiedColor=<rgba 1 1 1 1>}
     uVRegion {bIsValid=false, max=<0 0>, min=<0 0>}
    }
    colorAndOpacity=<rgba 0.4125 0.5089 0.9647 1>
   }
   slot {class=CommonVisibilitySwitcherSlot}
  }
  FortLazyImage_Icon_Favorited {
   widgetClassPath=&FortLazyImage
   parent=CommonVisibilitySwitcher_Favorite_Icons
   widget=FortLazyImage_Icon_Favorited
   properties {
    brush {
     bIsDynamicallyLoaded=false
     drawAs=Image
     imageType=FullColor
     mirroring=NoMirror
     resourceName=None
     resourceObject=/FortUILibrary/Foundation/Buttons/Materials/Favorite/MI_UI_Favorite_Base_Modal
     tiling=NoTile
     imageSize=<32 32>
     margin=<ltrb 0 0 0 0>
     outlineSettings {
      bUseBrushTransparency=false
      roundingType=HalfHeightRadius
      width=0
      color {colorUseRule=UseColor_Specified, specifiedColor=<rgba 0 0 0 0>}
      cornerRadii=<0 0 0 0>
     }
     tintColor {colorUseRule=UseColor_Specified, specifiedColor=<rgba 1 1 1 1>}
     uVRegion {bIsValid=false, max=<0 0>, min=<0 0>}
    }
    colorAndOpacity=<rgba 0.4125 0.5089 0.9647 1>
   }
   slot {class=CommonVisibilitySwitcherSlot}
  }
  ContextMenu {
   widgetClassPath=/FortUILibrary/UIKit/ContextMenu/WBP_UIKit_ContextMenuAnchorButton
   bIsVariable=true
   parent=ContextMenu_Favorite_Overlay
   widget=ContextMenu
   properties {
    button Background Material=/FortUILibrary/UIKit/Buttons/MaterialInstances/CTAButtons/MI_UIKit_BgQuiet
    buttonClass=/FortUILibrary/UIKit/Buttons/WBP_UIKit_Button_Contextual.WBP_UIKit_Button_Contextual_C
    buttonIcon=/FortUILibrary/Textures/Icons/T_UI_Icon_UIKit_ContextMenu
    buttonText=""
    visibility=Collapsed
    button Padding=<ltrb 8 0 8 0>
    button Text Font {
     bForceMonospaced=false
     bMaterialIsStencil=false
     fontMaterial=/FortUILibrary/UIKit/Buttons/MaterialInstances/GenericButtons/MI_UIKit_GenericText_Quiet
     fontObject=&Burbank
     letterSpacing=180
     monospacedWidth=1
     size=26.25
     skewAmount=0
     typefaceFontName=BurbankRegular-Black
     outlineSettings {
      bApplyOutlineToDropShadows=false
      bMiteredCorners=false
      bSeparateFillAlpha=false
      outlineMaterial=null
      outlineSize=0
      outlineColor=<rgba 0 0 0 1>
     }
    }
   }
   slot {class=OverlaySlot, properties {verticalAlignment=VAlign_Bottom}}
  }
  AddedToSetlistIndicator {
   widgetClassPath=&Overlay
   parent=HorizontalBox_34
   widget=AddedToSetlistIndicator
   properties {visibility=Hidden}
   slot {class=HorizontalBoxSlot, properties {horizontalAlignment=HAlign_Right}}
  }
  AddedTag {
   widgetClassPath=&Image
   parent=AddedToSetlistIndicator
   widget=AddedTag
   properties {
    brush {
     bIsDynamicallyLoaded=false
     drawAs=Image
     imageType=NoImage
     mirroring=NoMirror
     resourceName=None
     resourceObject=/Game/UI/GreyBox/white
     tiling=NoTile
     imageSize=<16 16>
     margin=<ltrb 0 0 0 0>
     outlineSettings {
      bUseBrushTransparency=false
      roundingType=HalfHeightRadius
      width=0
      color {colorUseRule=UseColor_Specified, specifiedColor=<rgba 0 0 0 0>}
      cornerRadii=<0 0 0 0>
     }
     tintColor {colorUseRule=UseColor_Specified, specifiedColor=<rgba 1 1 1 1>}
     uVRegion {bIsValid=false, max=<0 0>, min=<0 0>}
    }
    colorAndOpacity=<rgba 0.6795 0.6514 0.0052 1>
   }
   slot {class=OverlaySlot, properties {horizontalAlignment=HAlign_Fill, verticalAlignment=VAlign_Fill}}
  }
  Image_SelectionGlow {
   widgetClassPath=&Image
   bIsVariable=true
   parent=MainContainer
   widget=Image_SelectionGlow
   properties {
    visibility=HitTestInvisible
    brush {
     bIsDynamicallyLoaded=false
     drawAs=Image
     imageType=FullColor
     mirroring=NoMirror
     resourceName=None
     resourceObject=/PilgrimCore/Widgets/Materials/MI_UI_Pilgrim_Rounded_Outline_SelectedMainSong
     tiling=NoTile
     imageSize=<32 32>
     margin=<ltrb 0.5 0.5 0.5 0.5>
     outlineSettings {
      bUseBrushTransparency=false
      roundingType=HalfHeightRadius
      width=0
      color {colorUseRule=UseColor_Specified, specifiedColor=<rgba 0 0 0 0>}
      cornerRadii=<0 0 0 0>
     }
     tintColor {colorUseRule=UseColor_Specified, specifiedColor=<rgba 1 1 1 1>}
     uVRegion {bIsValid=false, max=<0 0>, min=<0 0>}
    }
    colorAndOpacity=<rgba 0 0 0 0>
    renderTransform {angle=0, scale=<1.01 1.03>, shear=<0 0>, translation=<0 0>}
   }
   slot {class=GridSlot, properties {layer=2, padding=<ltrb 3.5 0 32 0>}}
  }
  InputActionWidget {widgetClassPath=&CommonActionWidget, bInherited=true, widget=null}
 ]
}
