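# Quickplay song catalog widget that lists filtered songs, or a "No Songs Found" message when the list is empty.
# Clicks, double-clicks and selection changes drive SongSelected, which adds or removes songs via the player controller's custom setlist editor component and broadcasts selection changes.
# Touch input gets special toggle-selection handling that caches the selected entry.
# It also opens an async-created song leaderboard screen, rate-limited by a CVar-timed cooldown.
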WidgetBlueprint /PilgrimQuickplay/Widgets/SetlistAndSongDisplay/Duplicated/WBP_PilgrimQuickplay_SongCatalog2
graphs [
 AddRemoveSongForPlayer {nodes [
   n1 = AddRemoveSongForPlayer!() -> (then, SongShortname:Name, Add:bool=~, PlayerState:Player State Object Reference) {is=title K2Node_FunctionEntry}
   > n2 = Sequence()
   > n3 = Is Valid(InputObject=@n1.PlayerState)
   PrintString!(
    execute=@n3.Is Not Valid
    InString=BP_SongSelectionStatusDisplay cannot change song status for invalid player state
    bPrintToScreen
    bPrintToLog
    TextColor
    Duration
    Key
   ) -> (then)
   n4 = Branch(execute=@n3.Is Valid, Condition=@n1.Add)
   n5 = GetCustomSetlist!() -> (ReturnValue:Pilgrim Custom Setlist Object Reference)
   n6 = Get PlayerState() -> (PlayerState:Player State Object Reference)
   > n7 = CanAddSongForPlayer!(execute, self:Pilgrim Custom Setlist Object Reference=@n5, SongShortName:Name=@n1, Player:Player State Object Reference=@n6) -> (then, OutReason:EPilgrimCantAddSongReason Enum=None, OutAddedByPlayer:Player State Object Reference, ReturnValue:bool)
   > Branch(Condition=@n7.ReturnValue)
   n8 = GetComponentByClass!(
    self:Actor Object Reference=GetPlayerController!(self:Player State Object Reference=@n6) -> (ReturnValue:Player Controller Object Reference)
    ComponentClass:Actor Component Class Reference=/PilgrimCore/Player/Components/BP_PilgrimCustomSetlistEditorComponent.BP_PilgrimCustomSetlistEditorComponent_C
   ) -> (ReturnValue:BP Pilgrim Custom Setlist Editor Component Object Reference)
   n9 = Get SongShortname() -> (SongShortname:Name=None)
   > Add Song!(execute, self:BP Pilgrim Custom Setlist Editor Component Object Reference=@n8, SongToAdd:Name=@n9) -> (then)
   > Set WasChanged(WasChanged:bool=true) -> (Output_Get:bool)
   > FortShippingLog!(
    execute
    WorldContextObject:Object Reference=Self-Reference()
    InString:String=Append(A="Client attempting to add song to setlist: ")
    bLogAsWarning:bool=false
   ) -> (then)
   Branch(
    execute=@n4.else
    Condition=HasPlayerAddedSong!(self:Pilgrim Custom Setlist Object Reference=@n5, SongShortName:Name=@n1, Player:Player State Object Reference=@n6) -> (ReturnValue:bool)
   )
   > Remove Song!(execute, self:BP Pilgrim Custom Setlist Editor Component Object Reference=@n8, SongToRemove:Name=@n9) -> (then)
   > Set WasChanged(WasChanged=true)
   Return Node!(execute=@n2.then_1, SongStatusChanged:bool=Get WasChanged() -> (WasChanged:bool))
  ]}
 AdjustOffset {nodes [
   n1 = "On List View Scrolled (SongEntryList)"!() -> (OutputDelegate, then, ItemOffset:float, DistanceRemaining:float) {is=title K2Node_ComponentBoundEvent}
   > AdjustListOffset!(execute, self, Offset:double=@n1.ItemOffset) -> (then)
   "On Item Scrolled Into View (SongEntryList)"!() -> (OutputDelegate, then, Item:Object Reference, Widget:User Widget Object Reference) {is=title K2Node_ComponentBoundEvent}
   > AdjustListOffset(Offset=GetScrollOffset!(self:List View Base Object Reference=Get SongEntryList() -> (SongEntryList:Common List View Object Reference)) -> (ReturnValue:float))
   n2 = AdjustListOffset!() -> (OutputDelegate, then, Offset:double=~) {is=title K2Node_CustomEvent}
   > Branch(Condition="float <= float"(A=@n2.Offset) -> (ReturnValue=~))
   > ScrollToTop(self:List View Base Object Reference=Get SongEntryList())
  ]}
 BP_GetDesiredFocusTarget {nodes [
   Get Desired Focus Target()
   > Return Node!(execute, ReturnValue:Widget Object Reference=Get SongEntryList())
  ]}
 EventGraph {nodes [
   n1 = "On Item Clicked (SongEntryList)"!() -> (OutputDelegate, then, Item:Object Reference) {is=title K2Node_ComponentBoundEvent}
   n2 = GetOwningPlayer()
   > Branch(Condition=AND Boolean(
     A=IsUsingInputType!(InputType:ECommonInputType Enum=Touch, PlayerController:Player Controller Object Reference=@n2) -> (ReturnValue:bool)
     B=Get IsUsingPilgrimInverseForTouch() -> (IsUsingPilgrimInverseForTouch:bool)
    ))
   > n3 = Cast To PilgrimSongCatalogItemVM!(execute, Object=@n1.Item) -> (then, CastFailed, AsPilgrim Song Catalog Item VM:Pilgrim Song Catalog Item VM Object Reference) {is=title K2Node_DynamicCast}
   n4 = Get SongEntryList()
   > Branch(Condition="Equal (Object)"(A=@n3, B=Get Selected Item!(self:List View Object Reference=@n4) -> (ReturnValue:Object Reference) {is=title K2Node_CallFunction}))
   > n5 = Set CachedSelectedSongEntry(CachedSelectedSongEntry:Pilgrim Song Catalog Item VM Object Reference=@n3) -> (Output_Get:Pilgrim Song Catalog Item VM Object Reference)
   > Set Selected Item!(execute, self:List View Object Reference=Get SongEntryList(), Item:Object Reference=@n5) -> (then) {is=title K2Node_CallFunction}
   > ToggleItemSelection()
   > n6 = Cast To PilgrimSongCatalogItemDetailsVM!(
    execute
    Object=Get ItemDetails!(self:Fort Item VM Object Reference=Get CachedSelectedSongEntry() -> (CachedSelectedSongEntry:Pilgrim Song Catalog Item VM Object Reference)) -> (ItemDetails:Fort Item Details VM Object Reference)
   ) -> (then, CastFailed, AsPilgrim Song Catalog Item Details VM:Pilgrim Song Catalog Item Details VM Object Reference) {is=title K2Node_DynamicCast}
   > Branch(Condition=Get bIsAddedToSetlist!(self:Pilgrim Song Catalog Item Details VM Object Reference=@n6) -> (bIsAddedToSetlist:bool))
   > Set CachedSelectedSongEntry()
   > Set Selected Item(self=Get SongEntryList())
   n7 = "On Item Double Clicked (SongEntryList)"!() -> (OutputDelegate, then, Item:Object Reference) {is=title K2Node_ComponentBoundEvent}
   n8 = Cast To PilgrimSongCatalogItemVM(execute=[@n7.then, @n24.then], Object=Get Selected Item(self=Get SongEntryList()))
   > n9 = Cast To PilgrimSongCatalogItemDetailsVM(Object=Get ItemDetails(self=@n8))
   > SongSelected!(execute, self, SongListItem:Pilgrim Song Catalog Item VM Object Reference=@n8, Add:bool=NOT Boolean(A=Get bIsAddedToSetlist(self=@n9))) -> (then)
   n10 = "On Item Selection Changed (SongEntryList)"!() -> (OutputDelegate, then, Item:Object Reference, bIsSelected:bool) {is=title K2Node_ComponentBoundEvent}
   > n11 = Branch(Condition=@n10.bIsSelected)
   > SetScrollIntoViewAlignment!(execute, self:List View Object Reference=Get SongEntryList(), NewScrollIntoViewAlignment:EScrollIntoViewAlignment Enum=IntoView) -> (then)
   > Scroll Item Into View!(execute, self:List View Object Reference=Get SongEntryList(), Item:Object Reference=@n10.Item) -> (then)
   > n12 = Cast To PilgrimSongCatalogItemVM(Object=@n10.Item)
   > Call On Item Selection Changed!(execute:Exec, self:WBP Pilgrim Quickplay Song Catalog 2 Object Reference, Item:Pilgrim Song Catalog Item VM Object Reference=@n12) -> (then:Exec) {is=title K2Node_CallDelegate}
   Branch(execute=@n11.else, Condition=IsUsingInputType(InputType=Touch, PlayerController=GetOwningPlayer()))
   > n13 = Branch(Condition=Get IsUsingPilgrimInverseForTouch())
   > n14 = Branch(Condition=IsValid(Object=Get CachedSelectedSongEntry()))
   > Set Selected Item(self=Get SongEntryList(), Item=Get CachedSelectedSongEntry())
   Call On Item Deselected!(execute:Exec=[@n14.else, @n13.else], self:WBP Pilgrim Quickplay Song Catalog 2 Object Reference) -> (then:Exec) {is=title K2Node_CallDelegate}
   Leaderboard Selected() {is=title K2Node_CustomEvent}
   n15 = Get SongEntryList()
   > Branch(Condition="integer > integer"(A=Get Num Items Selected!(self:List View Object Reference=@n15) -> (ReturnValue:int=0) {is=title K2Node_CallFunction}))
   n16 = GetNumericCVar!(CVar:String=Sparks.LeaderboardOpenDelayMS) -> (OutValue:float, ReturnValue:bool)
   n17 = Does Timer Exist by Handle!(Handle:Timer Handle Structure=Get Timer() -> (Timer:Timer Handle Structure)) -> (ReturnValue:bool)
   > n18 = Branch(Condition=AND Boolean(A=NOT Boolean(A=@n17), B=@n16.ReturnValue))
   > n19 = CreateWidgetAsync!(
    execute:Exec
    UserWidgetSoftClass:User Widget Soft Class Reference=/PilgrimQuickplay/Widgets/Leaderboards/WBP_PilgrimQuickplay_Leaderboard_Song.WBP_PilgrimQuickplay_Leaderboard_Song_C
    OwningPlayer:Player Controller Object Reference=GetOwningPlayer()
    bSuspendInputUntilComplete:bool=true
   ) -> (then:Exec, OnComplete:Exec, OnFailure:Exec, UserWidget:User Widget Object Reference) {is=title K2Node_AsyncAction}
   PushContentWidget!(execute=@n19.OnComplete, PushWidget:Widget Object Reference=@n19) -> (then)
   > n20 = Cast To WBP_PilgrimQuickplay_Leaderboard_Song!(execute, Object=@n19) -> (then, CastFailed, AsWBP Pilgrim Quickplay Leaderboard Song:WBP Pilgrim Quickplay Leaderboard Song Object Reference) {is=title K2Node_DynamicCast}
   > n21 = SetCurrentSongTag!(
    execute
    self:WBP Pilgrim Quickplay Leaderboard Song Object Reference=@n20
    Song ShortName:Name=GetSelectedSongShortName!() -> (ShortName:Name=None)
   ) -> (then)
   n22 = CreateLeaderboardCooldown() {is=title K2Node_CustomEvent}
   n23 = Set Timer by Event(execute=@n21.then, Delegate=@n22, Time=float / float(A=@n16, B=1000), bLooping=false)
   > Set Timer(Timer:Timer Handle Structure=@n23) -> (Output_Get:Timer Handle Structure)
   PrintString(execute=@n18.else, InString="Timer not finish, cant open leaderboard screen")
   n24 = ToggleItemSelection() {is=title K2Node_CustomEvent}
   n25 = SongSelected!() -> (OutputDelegate, then, SongListItem:Pilgrim Song Catalog Item VM Object Reference, Add:bool=~) {is=title K2Node_CustomEvent}
   > n26 = AddRemoveSongForPlayer!(
    execute
    self
    SongShortname:Name=GetShortName!(self:Sparks Song Data Object Reference=GetSongMetadata!(self:Pilgrim Song Catalog Item VM Object Reference=@n25.SongListItem) -> (ReturnValue:Pilgrim Song Metadata Object Reference)) -> (ReturnValue:Name=None)
    Add:bool=@n25.Add
    PlayerState:Player State Object Reference=Get PlayerState!(self:Controller Object Reference=GetOwningPlayer()) -> (PlayerState)
   ) -> (then, SongStatusChanged:bool) {is=title K2Node_CallFunction}
   > Branch(Condition=@n26)
   > Call On Item Selection Changed(Item=@n25.SongListItem)
   Event Construct()
   n27 = GetOwningPlayer()
   > SetSelectionMode!(
    execute
    self:List View Object Reference=Get SongEntryList()
    SelectionMode:ESelectionMode Enum=Select(
     Option 0:ESelectionMode Enum=Single
     Option 1:ESelectionMode Enum=SingleToggle
     Index=IsUsingInputType(InputType=Touch, PlayerController=@n27)
    ) -> (ReturnValue:ESelectionMode Enum)
   ) -> (then)
   > n28 = IsPlatformTraitSupported!(
    execute
    TraitToCheck:"Gameplay Tag Structure (by ref)"=MakeLiteralGameplayTag!(Value:Gameplay Tag Structure=(TagName="Platform.Trait.Input.SupportsTouch.UI")) -> (ReturnValue:Gameplay Tag Structure)
   ) -> (then, ReturnValue:bool)
   > Set IsUsingPilgrimInverseForTouch(IsUsingPilgrimInverseForTouch:bool=@n28) -> (Output_Get:bool)
   Clear and Invalidate Timer by Handle(execute=@n22.then, Handle=Get Timer())
   Event Destruct()
  ]}
 GetSelectedSongShortName {nodes [
   GetSelectedSongShortName() {is=title K2Node_FunctionEntry}
   > n1 = Cast To PilgrimSongCatalogItemVM(Object=Get Selected Item(self=Get SongEntryList()))
   > Return Node!(execute, ShortName:Name=GetShortName(self=GetSongMetadata(self=@n1)) -> (ReturnValue=None))
   Return Node(execute=@n1.CastFailed, ShortName=None)
  ]}
 HideMusicEntries {nodes [
   HideMusicEntries() {is=title K2Node_FunctionEntry}
   > SetActiveWidget(
    self:Widget Switcher Object Reference=Get LibraryVisibilitySwitcher() -> (LibraryVisibilitySwitcher:Common Activatable Widget Switcher Object Reference)
    Widget=Get NoSongsFoundText() -> (NoSongsFoundText:Common Text Object Reference)
   )
  ]}
 NavigateToSongItem {nodes [
   n1 = NavigateToSongItem!() -> (then, Item:Object Reference) {is=title K2Node_FunctionEntry}
   > n2 = Cast To PilgrimSongCatalogItemVM(Object=@n1)
   n3 = Get SongEntryList()
   > Set Selected Item(self=@n3, Item=@n2)
   > SetScrollIntoViewAlignment(self=@n3, NewScrollIntoViewAlignment=TopOrLeft)
   > Scroll Item Into View(self=@n3, Item=@n2)
  ]}
 OnItemDeselected {nodes [OnItemDeselected() {is=title K2Node_FunctionEntry}]}
 OnItemSelectionChanged {nodes [OnItemSelectionChanged!() -> (then, Item:Pilgrim Song Catalog Item VM Object Reference) {is=title K2Node_FunctionEntry}]}
 SetFilteredItemsSongCatalogVM {nodes [
   n1 = SetFilteredItemsSongCatalogVM!() -> (then, FilteredItems:Array of Pilgrim Song Catalog Item VM Object References) {is=title K2Node_FunctionEntry}
   n2 = Get SongEntryList()
   > Set List Items!(execute, self:List View Object Reference=@n2, InListItems:Array of Object References=@n1) -> (then) {is=title K2Node_CallFunction}
   > n3 = Branch(Condition="integer <= integer"(A=GetNumItems!(self:List View Object Reference=@n2) -> (ReturnValue:int=0)))
   > HideMusicEntries() {is=title K2Node_CallFunction}
   ShowMusicEntries(execute=@n3.else)
  ]}
 ShowMusicEntries {nodes [
   ShowMusicEntries() {is=title K2Node_FunctionEntry}
   > SetActiveWidget(self=Get LibraryVisibilitySwitcher(), Widget=Get HorizontalBox_SongEntries() -> (HorizontalBox_SongEntries:Horizontal Box Object Reference))
  ]}
]
variables [
 CachedSelectedSongEntry {type=PilgrimSongCatalogItemVM, default=null}
 IsUsingPilgrimInverseForTouch {type=boolean}
 OnItemDeselected {}
 OnItemSelectionChanged {}
 Timer {type=TimerHandle}
]
widgets {
 info {inheritedWidgetCount=0, parentClass=&CommonActivatableWidget, rootWidgetClass=&Overlay, widgetCount=6}
 widgets [
  SongCatalog {widgetClassPath=&Overlay, widget=SongCatalog}
  LibraryVisibilitySwitcher {
   widgetClassPath=&CommonActivatableWidgetSwitcher
   bIsVariable=true
   parent=SongCatalog
   widget=LibraryVisibilitySwitcher
   slot {class=OverlaySlot, properties {horizontalAlignment=HAlign_Fill}}
  }
  HorizontalBox_SongEntries {
   widgetClassPath=&HorizontalBox
   bIsVariable=true
   parent=LibraryVisibilitySwitcher
   widget=HorizontalBox_SongEntries
   slot {class=WidgetSwitcherSlot}
  }
  SongEntryList {
   widgetClassPath=&CommonListView
   bIsVariable=true
   parent=HorizontalBox_SongEntries
   widget=SongEntryList
   properties {
    entryWidgetClass=&/WBP_PilgrimQuickplay_SongCatalogEntry_Horizontal2.WBP_PilgrimQuickplay_SongCatalogEntry_Horizontal2_C
    navigation=&self:WidgetTree.SongEntryList.WidgetNavigation_0
    numDesignerPreviewEntries=8
    verticalEntrySpacing=6
    visibility=SelfHitTestInvisible
    scrollBarStyle {
     thickness=12
     draggedThumbImage {
      bIsDynamicallyLoaded=false
      drawAs=RoundedBox
      imageType=FullColor
      mirroring=NoMirror
      resourceName=../../../Engine/Content/Slate/Common/Scrollbar_Thumb.png
      resourceObject=null
      tiling=NoTile
      imageSize=<12 12>
      margin=<ltrb 0.5 0.5 0.5 0.5>
      outlineSettings {
       bUseBrushTransparency=false
       roundingType=FixedRadius
       width=2
       color {colorUseRule=UseColor_Specified, specifiedColor=<rgba 0 0 0 0>}
       cornerRadii=<6 6 6 6>
      }
      tintColor {colorUseRule=UseColor_Specified, specifiedColor=<rgba 1 1 1 1>}
      uVRegion {bIsValid=false, max=<0 0>, min=<0 0>}
     }
     horizontalBackgroundImage {
      bIsDynamicallyLoaded=false
      drawAs=NoDrawType
      imageType=NoImage
      mirroring=NoMirror
      resourceName=None
      resourceObject=null
      tiling=NoTile
      imageSize=<0 0>
      margin=<ltrb 0 0 0 0>
      outlineSettings {
       bUseBrushTransparency=false
       roundingType=HalfHeightRadius
       width=0
       color {colorUseRule=UseColor_Specified, specifiedColor=<rgba 0 0 0 0>}
       cornerRadii=<0 0 0 0>
      }
      tintColor {colorUseRule=UseColor_Specified, specifiedColor=<rgba 1 1 1 1>}
      uVRegion {bIsValid=false, max=<0 0>, min=<0 0>}
     }
     horizontalBottomSlotImage {
      bIsDynamicallyLoaded=false
      drawAs=NoDrawType
      imageType=FullColor
      mirroring=NoMirror
      resourceName=../../../Engine/Content/Slate/Common/Scrollbar_Background_Vertical.png
      resourceObject=null
      tiling=NoTile
      imageSize=<8 8>
      margin=<ltrb 0 0 0 0>
      outlineSettings {
       bUseBrushTransparency=false
       roundingType=HalfHeightRadius
       width=0
       color {colorUseRule=UseColor_Specified, specifiedColor=<rgba 0 0 0 0>}
       cornerRadii=<0 0 0 0>
      }
      tintColor {colorUseRule=UseColor_Specified, specifiedColor=<rgba 1 1 1 1>}
      uVRegion {bIsValid=false, max=<0 0>, min=<0 0>}
     }
     horizontalTopSlotImage {
      bIsDynamicallyLoaded=false
      drawAs=NoDrawType
      imageType=FullColor
      mirroring=NoMirror
      resourceName=../../../Engine/Content/Slate/Common/Scrollbar_Background_Vertical.png
      resourceObject=null
      tiling=NoTile
      imageSize=<8 8>
      margin=<ltrb 0 0 0 0>
      outlineSettings {
       bUseBrushTransparency=false
       roundingType=HalfHeightRadius
       width=0
       color {colorUseRule=UseColor_Specified, specifiedColor=<rgba 0 0 0 0>}
       cornerRadii=<0 0 0 0>
      }
      tintColor {colorUseRule=UseColor_Specified, specifiedColor=<rgba 1 1 1 1>}
      uVRegion {bIsValid=false, max=<0 0>, min=<0 0>}
     }
     hoveredThumbImage {
      bIsDynamicallyLoaded=false
      drawAs=RoundedBox
      imageType=FullColor
      mirroring=NoMirror
      resourceName=../../../Engine/Content/Slate/Common/Scrollbar_Thumb.png
      resourceObject=null
      tiling=NoTile
      imageSize=<12 12>
      margin=<ltrb 0.5 0.5 0.5 0.5>
      outlineSettings {
       bUseBrushTransparency=false
       roundingType=FixedRadius
       width=2
       color {colorUseRule=UseColor_Specified, specifiedColor=<rgba 0 0 0 0>}
       cornerRadii=<6 6 6 6>
      }
      tintColor {colorUseRule=UseColor_Specified, specifiedColor=<rgba 1 1 1 1>}
      uVRegion {bIsValid=false, max=<0 0>, min=<0 0>}
     }
     normalThumbImage {
      bIsDynamicallyLoaded=false
      drawAs=RoundedBox
      imageType=FullColor
      mirroring=NoMirror
      resourceName=../../../Engine/Content/Slate/Common/Scrollbar_Thumb.png
      resourceObject=null
      tiling=NoTile
      imageSize=<12 12>
      margin=<ltrb 0.5 0.5 0.5 0.5>
      outlineSettings {
       bUseBrushTransparency=false
       roundingType=FixedRadius
       width=2
       color {colorUseRule=UseColor_Specified, specifiedColor=<rgba 0 0 0 0>}
       cornerRadii=<6 6 6 6>
      }
      tintColor {colorUseRule=UseColor_Specified, specifiedColor=<rgba 1 1 1 0.75>}
      uVRegion {bIsValid=false, max=<0 0>, min=<0 0>}
     }
     verticalBackgroundImage {
      bIsDynamicallyLoaded=false
      drawAs=Box
      imageType=NoImage
      mirroring=NoMirror
      resourceName=None
      resourceObject=/FortUILibrary/UIKit/ScrollBar/MaterialInstances/MI_UIKit_ScrollbarBackground
      tiling=NoTile
      imageSize=<12 12>
      margin=<ltrb 0.5 0.5 0.5 0.5>
      outlineSettings {
       bUseBrushTransparency=false
       roundingType=HalfHeightRadius
       width=0
       color {colorUseRule=UseColor_Specified, specifiedColor=<rgba 0 0 0 0>}
       cornerRadii=<0 0 0 0>
      }
      tintColor {colorUseRule=UseColor_Specified, specifiedColor=<rgba 1 1 1 1>}
      uVRegion {bIsValid=false, max=<0 0>, min=<0 0>}
     }
     verticalBottomSlotImage {
      bIsDynamicallyLoaded=false
      drawAs=NoDrawType
      imageType=FullColor
      mirroring=NoMirror
      resourceName=../../../Engine/Content/Slate/Common/Scrollbar_Background_Vertical.png
      resourceObject=null
      tiling=NoTile
      imageSize=<8 8>
      margin=<ltrb 0 0 0 0>
      outlineSettings {
       bUseBrushTransparency=false
       roundingType=HalfHeightRadius
       width=0
       color {colorUseRule=UseColor_Specified, specifiedColor=<rgba 0 0 0 0>}
       cornerRadii=<0 0 0 0>
      }
      tintColor {colorUseRule=UseColor_Specified, specifiedColor=<rgba 1 1 1 1>}
      uVRegion {bIsValid=false, max=<0 0>, min=<0 0>}
     }
     verticalTopSlotImage {
      bIsDynamicallyLoaded=false
      drawAs=NoDrawType
      imageType=FullColor
      mirroring=NoMirror
      resourceName=../../../Engine/Content/Slate/Common/Scrollbar_Background_Vertical.png
      resourceObject=null
      tiling=NoTile
      imageSize=<8 8>
      margin=<ltrb 0 0 0 0>
      outlineSettings {
       bUseBrushTransparency=false
       roundingType=HalfHeightRadius
       width=0
       color {colorUseRule=UseColor_Specified, specifiedColor=<rgba 0 0 0 0>}
       cornerRadii=<0 0 0 0>
      }
      tintColor {colorUseRule=UseColor_Specified, specifiedColor=<rgba 1 1 1 1>}
      uVRegion {bIsValid=false, max=<0 0>, min=<0 0>}
     }
    }
   }
   slot {class=HorizontalBoxSlot, properties {size {sizeRule=Fill, value=1}}}
  }
  NoSongsFoundText {
   widgetClassPath=&CommonTextBlock
   bIsVariable=true
   parent=LibraryVisibilitySwitcher
   widget=NoSongsFoundText
   properties {
    justification=Center
    style=/FortUILibrary/Styles/Typography/Foundation/TextStyle_BurbankSmall_L_White.TextStyle_BurbankSmall_L_White_C
    text=No Songs Found
    font {
     bForceMonospaced=false
     bMaterialIsStencil=false
     fontMaterial=null
     fontObject=&BurbankSmall
     letterSpacing=0
     monospacedWidth=1
     size=36
     skewAmount=0
     typefaceFontName=Black
     outlineSettings {
      bApplyOutlineToDropShadows=false
      bMiteredCorners=false
      bSeparateFillAlpha=false
      outlineMaterial=null
      outlineSize=0
      outlineColor=<rgba 0 0 0 1>
     }
    }
    shadowOffset=<0 0>
   }
   slot {class=WidgetSwitcherSlot, properties {verticalAlignment=VAlign_Center}}
  }
  HorizontalBox_429 {
   widgetClassPath=&HorizontalBox
   parent=SongCatalog
   widget=HorizontalBox_429
   properties {visibility=Visible}
   slot {class=OverlaySlot, properties {horizontalAlignment=HAlign_Center, verticalAlignment=VAlign_Bottom}}
  }
 ]
}
