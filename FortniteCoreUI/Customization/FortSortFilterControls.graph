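# FortSortFilterControls is a reusable search and sort/filter header widget with sort and filter tags, an extra-info named slot, a search field and a Sort + Filter button.
# Trigger-bound buttons focus the search field or broadcast SortFilterClicked, while entering the focus path broadcasts OnSearchOrFilterFocused.
# If search already has focus, gamepad users get focus bounced through the widget and back.
# Custom navigation functions return configurable widgets below the text box and button.
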WidgetBlueprint /FortniteCoreUI/Customization/FortSortFilterControls
graphs [
 EventGraph {nodes [
   "On Button Clicked (Button_TextSearch)"!() -> (OutputDelegate, then, Button:Common Button Legacy Object Reference) {is=title K2Node_ComponentBoundEvent}
   n1 = GetOwningPlayer()
   n2 = Get EditableText_Search() -> (EditableText_Search:Editable Text Object Reference)
   > n3 = Branch(Condition=NOT Boolean(A=HasUserFocus!(self:Widget Object Reference=@n2, PlayerController:Player Controller Object Reference=@n1) -> (ReturnValue:bool)))
   > SetFocus(self=@n2)
   RefocusEditableTextForGamepad(execute=@n3.else)
   "On Clicked (Button_SortAndFilter)"!() -> (OutputDelegate, then, Button:Common Button Base Object Reference) {is=title K2Node_ComponentBoundEvent}
   > Call Sort Filter Clicked!(execute:Exec, self:Fort Sort Filter Controls Object Reference) -> (then:Exec) {is=title K2Node_CallDelegate}
   Event OnAddedToFocusPath!() -> (OutputDelegate, then, InFocusEvent:Focus Event Structure) {is=title K2Node_Event}
   > Call On Search or Filter Focused!(execute:Exec, self:Fort Sort Filter Controls Object Reference) -> (then:Exec) {is=title K2Node_CallDelegate}
  ]}
 FormatSortFilterLabel {nodes [
   n1 = FormatSortFilterLabel!() -> (then, StringFormat:Text, SortFilterName:Text) {is=title K2Node_FunctionEntry}
   > Return Node!(execute, NewParam2:Text=Format Text!(Format=@n1, 0:Text=@n1.SortFilterName) -> (Result))
  ]}
 NavigateDownFromButtom {nodes [
   NavigateDownFromButtom!() -> (then, Navigation:EUINavigation Enum) {is=title K2Node_FunctionEntry}
   > Return Node!(execute, ReturnValue:Widget Object Reference=Get WidgetBelowButton() -> (WidgetBelowButton:Widget Object Reference))
  ]}
 NavigateDownFromTextBox {nodes [
   NavigateDownFromTextBox!() -> (then, Navigation:EUINavigation Enum) {is=title K2Node_FunctionEntry}
   > Return Node(ReturnValue=Get WidgetBelowTextBox() -> (WidgetBelowTextBox:Widget Object Reference))
  ]}
 OnSearchOrFilterFocused {nodes [OnSearchOrFilterFocused() {is=title K2Node_FunctionEntry}]}
 RefocusEditableTextForGamepad {nodes [
   RefocusEditableTextForGamepad() {is=title K2Node_FunctionEntry}
   > Branch(Condition=IsUsingInputType!(InputType:ECommonInputType Enum=Gamepad, PlayerController:Player Controller Object Reference=GetOwningPlayer()) -> (ReturnValue:bool))
   > SetFocus()
   > SetFocus(self=Get EditableText_Search())
  ]}
 SortFilterClicked {nodes [SortFilterClicked() {is=title K2Node_FunctionEntry}]}
]
variables [
 OnSearchOrFilterFocused {}
 SortFilterClicked {}
 WidgetBelowButton {type=Widget, default=null}
 WidgetBelowTextBox {type=Widget, default=null}
]
widgets {
 info {inheritedWidgetCount=0, namedSlotCount=1, parentClass=&UserWidget, rootWidgetClass=&VerticalBox, widgetCount=23}
 widgets [
  VerticalBox_1 {widgetClassPath=&VerticalBox, widget=VerticalBox_1}
  SizeBox_0 {
   widgetClassPath=&SizeBox
   parent=VerticalBox_1
   widget=SizeBox_0
   properties {bOverride_HeightOverride=true, heightOverride=32}
   slot {class=VerticalBoxSlot, properties {padding=<ltrb 0 0 0 18>}}
  }
  HorizontalBox_0 {
   widgetClassPath=&HorizontalBox
   parent=SizeBox_0
   widget=HorizontalBox_0
   properties {flowDirectionPreference=Culture}
   slot {class=SizeBoxSlot}
  }
  Overlay_3 {widgetClassPath=&Overlay, parent=HorizontalBox_0, widget=Overlay_3, slot {class=HorizontalBoxSlot}}
  Border_0 {
   widgetClassPath=&Border
   parent=Overlay_3
   widget=Border_0
   properties {
    verticalAlignment=VAlign_Center
    background {
     bIsDynamicallyLoaded=false
     drawAs=Box
     imageType=FullColor
     mirroring=NoMirror
     resourceName=None
     resourceObject=/FortUILibrary/Systems/Tag/MaterialInstances/MI_UIKit_TagBackground
     tiling=NoTile
     imageSize=<32 32>
     margin=<ltrb 0.5 0 0.5 0>
     outlineSettings {
      bUseBrushTransparency=false
      roundingType=HalfHeightRadius
      width=0
      color {colorUseRule=UseColor_Specified, specifiedColor=<rgba 0 0 0 0>}
      cornerRadii=<0 0 0 0>
     }
     tintColor {colorUseRule=UseColor_Specified, specifiedColor=<rgba 0.0044 0.011 0.0704 1>}
     uVRegion {bIsValid=false, max=<0 0>, min=<0 0>}
    }
    contentColorAndOpacity=<rgba 0.0044 0.011 0.0704 1>
    padding=<ltrb 20 0 20 0>
    renderTransform {angle=0, scale=<1 1>, shear=<-10 0>, translation=<0 0>}
   }
   slot {class=OverlaySlot, properties {horizontalAlignment=HAlign_Fill, verticalAlignment=VAlign_Fill}}
  }
  Text_SortTag {
   widgetClassPath=&CommonTextBlock
   parent=Overlay_3
   widget=Text_SortTag
   properties {
    justification=Center
    mobileFontSizeMultiplier=1.3
    scrollStyle=&TextScrollStyle-Base_C
    text="Sort: [PH]"
    textTransformPolicy=ToUpper
    colorAndOpacity {colorUseRule=UseColor_Specified, specifiedColor=<rgba 0.4852 0.4342 0.7231 1>}
    font {
     bForceMonospaced=false
     bMaterialIsStencil=false
     fontMaterial=null
     fontObject=&HeadingNow
     letterSpacing=0
     monospacedWidth=1
     size=12
     skewAmount=0
     typefaceFontName=75 Medium Italic
     outlineSettings {
      bApplyOutlineToDropShadows=false
      bMiteredCorners=false
      bSeparateFillAlpha=false
      outlineMaterial=null
      outlineSize=0
      outlineColor=<rgba 0 0 0 1>
     }
    }
    renderTransform {angle=0, scale=<1 1>, shear=<-8 0>, translation=<0 0>}
    shadowOffset=<0 0>
   }
   slot {class=OverlaySlot, properties {horizontalAlignment=HAlign_Fill, verticalAlignment=VAlign_Center, padding=<ltrb 40 0 40 0>}}
  }
  Overlay_2 {
   widgetClassPath=&Overlay
   parent=HorizontalBox_0
   widget=Overlay_2
   slot {class=HorizontalBoxSlot, properties {padding=<ltrb 8 0 0 0>}}
  }
  Border {
   widgetClassPath=&Border
   parent=Overlay_2
   widget=Border
   properties {
    background {
     bIsDynamicallyLoaded=false
     drawAs=Box
     imageType=FullColor
     mirroring=NoMirror
     resourceName=None
     resourceObject=&MI_UIKit_TagBackground
     tiling=NoTile
     imageSize=<32 32>
     margin=<ltrb 0.5 0 0.5 0>
     outlineSettings {
      bUseBrushTransparency=false
      roundingType=HalfHeightRadius
      width=0
      color {colorUseRule=UseColor_Specified, specifiedColor=<rgba 0 0 0 0>}
      cornerRadii=<0 0 0 0>
     }
     tintColor {colorUseRule=UseColor_Specified, specifiedColor=<rgba 0.0044 0.011 0.0704 1>}
     uVRegion {bIsValid=false, max=<0 0>, min=<0 0>}
    }
    contentColorAndOpacity=<rgba 0.0044 0.0273 0.2502 1>
    padding=<ltrb 4 0 0 0>
    renderTransform {angle=0, scale=<1 1>, shear=<-10 0>, translation=<0 0>}
   }
   slot {class=OverlaySlot, properties {horizontalAlignment=HAlign_Fill, verticalAlignment=VAlign_Fill}}
  }
  Text_FilterTag {
   widgetClassPath=&CommonTextBlock
   parent=Overlay_2
   widget=Text_FilterTag
   properties {
    justification=Center
    mobileFontSizeMultiplier=1.3
    scrollStyle=&TextScrollStyle-Base_C
    text="Filter: [PH]"
    textTransformPolicy=ToUpper
    colorAndOpacity {colorUseRule=UseColor_Specified, specifiedColor=<rgba 0.4852 0.4342 0.7231 1>}
    font {
     bForceMonospaced=false
     bMaterialIsStencil=false
     fontMaterial=null
     fontObject=&HeadingNow
     letterSpacing=0
     monospacedWidth=1
     size=12
     skewAmount=0
     typefaceFontName=75 Medium Italic
     outlineSettings {
      bApplyOutlineToDropShadows=false
      bMiteredCorners=false
      bSeparateFillAlpha=false
      outlineMaterial=null
      outlineSize=0
      outlineColor=<rgba 0 0 0 1>
     }
    }
    renderTransform {angle=0, scale=<1 1>, shear=<-8 0>, translation=<0 0>}
    shadowOffset=<0 0>
   }
   slot {class=OverlaySlot, properties {horizontalAlignment=HAlign_Fill, verticalAlignment=VAlign_Center, padding=<ltrb 40 0 40 0>}}
  }
  AdditionalHeaderInfo {
   widgetClassPath=&NamedSlot
   parent=HorizontalBox_0
   widget=AdditionalHeaderInfo
   slot {class=HorizontalBoxSlot, properties {padding=<ltrb 8 0 0 0>}}
  }
  SearchBar_HB {widgetClassPath=&HorizontalBox, parent=VerticalBox_1, widget=SearchBar_HB, slot {class=VerticalBoxSlot}}
  SizeBox_1 {
   widgetClassPath=&SizeBox
   parent=SearchBar_HB
   widget=SizeBox_1
   properties {bOverride_HeightOverride=true, bOverride_WidthOverride=true, heightOverride=40, widthOverride=800}
   slot {class=HorizontalBoxSlot, properties {padding=<ltrb 0 0 12 0>, size {sizeRule=Fill, value=1}}}
  }
  Overlay_0 {widgetClassPath=&Overlay, parent=SizeBox_1, widget=Overlay_0, slot {class=SizeBoxSlot}}
  Border_2 {
   widgetClassPath=&Border
   parent=Overlay_0
   widget=Border_2
   properties {
    background {
     bIsDynamicallyLoaded=false
     drawAs=Image
     imageType=FullColor
     mirroring=NoMirror
     resourceName=None
     resourceObject=&/MI_UI_SparksSeachBarBacking
     tiling=NoTile
     imageSize=<32 32>
     margin=<ltrb 0.5 0.5 0.5 0.5>
     outlineSettings {
      bUseBrushTransparency=false
      roundingType=HalfHeightRadius
      width=0
      color {colorUseRule=UseColor_Specified, specifiedColor=<rgba 0 0 0 0>}
      cornerRadii=<0 0 0 0>
     }
     tintColor {colorUseRule=UseColor_Specified, specifiedColor=<rgba 1 1 1 0.2>}
     uVRegion {bIsValid=false, max=<0 0>, min=<0 0>}
    }
    padding=<ltrb 0 0 0 0>
   }
   slot {class=OverlaySlot, properties {horizontalAlignment=HAlign_Fill, verticalAlignment=VAlign_Fill}}
  }
  HorizontalBox_2 {
   widgetClassPath=&HorizontalBox
   parent=Overlay_0
   widget=HorizontalBox_2
   slot {class=OverlaySlot, properties {horizontalAlignment=HAlign_Fill, verticalAlignment=VAlign_Fill}}
  }
  FortMobileResize {
   widgetClassPath=&ScaleBox
   parent=HorizontalBox_2
   widget=FortMobileResize
   properties {stretch=None, stretchDirection=UpOnly}
   slot {class=HorizontalBoxSlot, properties {size {sizeRule=Fill, value=1}}}
  }
  Border_1 {
   widgetClassPath=&Border
   parent=FortMobileResize
   widget=Border_1
   properties {
    verticalAlignment=VAlign_Center
    background {
     bIsDynamicallyLoaded=false
     drawAs=Image
     imageType=FullColor
     mirroring=NoMirror
     resourceName=None
     resourceObject=null
     tiling=NoTile
     imageSize=<32 32>
     margin=<ltrb 0 0 0 0>
     outlineSettings {
      bUseBrushTransparency=false
      roundingType=HalfHeightRadius
      width=0
      color {colorUseRule=UseColor_Specified, specifiedColor=<rgba 0 0 0 0>}
      cornerRadii=<0 0 0 0>
     }
     tintColor {colorUseRule=UseColor_Specified, specifiedColor=<rgba 1 1 1 0>}
     uVRegion {bIsValid=false, max=<0 0>, min=<0 0>}
    }
    padding=<ltrb 24 0 0 0>
   }
   slot {class=ScaleBoxSlot, properties {horizontalAlignment=HAlign_Fill, verticalAlignment=VAlign_Fill}}
  }
  EditableText_Search {
   widgetClassPath=/Script/UMG.EditableText
   bIsVariable=true
   parent=Border_1
   widget=EditableText_Search
   properties {
    hintText=Search
    navigation=&self:WidgetTree.EditableText_Search.WidgetNavigation_0
    virtualKeyboardTrigger=OnAllFocusEvents
    widgetStyle {
     backgroundImageComposing {
      bIsDynamicallyLoaded=false
      drawAs=Border
      imageType=FullColor
      mirroring=NoMirror
      resourceName=../../../Engine/Content/Slate/Old/HyperlinkDotted.png
      resourceObject=null
      tiling=Both
      imageSize=<0 0>
      margin=<ltrb 0 0 0 0.1875>
      outlineSettings {
       bUseBrushTransparency=false
       roundingType=HalfHeightRadius
       width=0
       color {colorUseRule=UseColor_Specified, specifiedColor=<rgba 0 0 0 0>}
       cornerRadii=<0 0 0 0>
      }
      tintColor {colorUseRule=UseColor_Specified, specifiedColor=<rgba 1 1 1 1>}
      uVRegion {bIsValid=false, max=<0 0>, min=<0 0>}
     }
     backgroundImageSelected {
      bIsDynamicallyLoaded=false
      drawAs=Box
      imageType=FullColor
      mirroring=NoMirror
      resourceName=../../../Engine/Content/Slate/Common/EditableTextSelectionBackground.png
      resourceObject=null
      tiling=NoTile
      imageSize=<0 0>
      margin=<ltrb 0.25 0.25 0.25 0.25>
      outlineSettings {
       bUseBrushTransparency=false
       roundingType=HalfHeightRadius
       width=0
       color {colorUseRule=UseColor_Specified, specifiedColor=<rgba 0 0 0 0>}
       cornerRadii=<0 0 0 0>
      }
      tintColor {colorUseRule=UseColor_Specified, specifiedColor=<rgba 1 1 1 1>}
      uVRegion {bIsValid=false, max=<0 0>, min=<0 0>}
     }
     caretImage {
      bIsDynamicallyLoaded=false
      drawAs=Image
      imageType=FullColor
      mirroring=NoMirror
      resourceName=../../../Engine/Content/Slate/Old/White.png
      resourceObject=null
      tiling=NoTile
      imageSize=<16 16>
      margin=<ltrb 0 0 0 0>
      outlineSettings {
       bUseBrushTransparency=false
       roundingType=HalfHeightRadius
       width=0
       color {colorUseRule=UseColor_Specified, specifiedColor=<rgba 0 0 0 0>}
       cornerRadii=<0 0 0 0>
      }
      tintColor {colorUseRule=UseColor_Specified, specifiedColor=<rgba 1 1 1 1>}
      uVRegion {bIsValid=false, max=<0 0>, min=<0 0>}
     }
     colorAndOpacity {colorUseRule=UseColor_Specified, specifiedColor=<rgba 1 1 1 1>}
     font {
      bForceMonospaced=false
      bMaterialIsStencil=false
      fontMaterial=null
      fontObject=&HeadingNow
      letterSpacing=10
      monospacedWidth=1
      size=18
      skewAmount=0
      typefaceFontName=75 Medium
      outlineSettings {
       bApplyOutlineToDropShadows=false
       bMiteredCorners=false
       bSeparateFillAlpha=false
       outlineMaterial=null
       outlineSize=0
       outlineColor=<rgba 0 0 0 1>
      }
     }
     hintColorAndOpacity {colorUseRule=UseColor_Foreground, specifiedColor=<rgba 1 0 1 1>}
     selectedBackgroundColor {colorUseRule=UseColor_Foreground, specifiedColor=<rgba 1 0 1 1>}
    }
   }
   slot {class=BorderSlot, properties {verticalAlignment=VAlign_Center, padding=<ltrb 24 0 0 0>}}
  }
  SizeBox_TextSearch {
   widgetClassPath=&SizeBox
   parent=HorizontalBox_2
   widget=SizeBox_TextSearch
   properties {
    bOverride_MinDesiredWidth=true
    heightOverride=48
    minDesiredWidth=48
    navigation=&self:WidgetTree.SizeBox_TextSearch.WidgetNavigation_0
    widthOverride=48
   }
   slot {class=HorizontalBoxSlot, properties {padding=<ltrb 0 0 24 0>}}
  }
  Button_TextSearch {
   widgetClassPath=/FortUILibrary/Foundation/Buttons/PanelButton
   bIsVariable=true
   parent=SizeBox_TextSearch
   widget=Button_TextSearch
   properties {
    bDisplayInputActionWhenNotInteractable=false
    bInteractableWhenSelected=true
    bIsFocusable=false
    inputActionUseRimBrush=false
    inputActionVerticalAlignment=VAlign_Center
    navigation=&self:WidgetTree.Button_TextSearch.WidgetNavigation_0
    onButtonBaseClicked=(&self:WidgetTree.Button_TextSearch.HandleOnButtonClicked)
    onButtonBaseDoubleClicked=(&self:WidgetTree.Button_TextSearch.HandleOnButtonDoubleClicked)
    onButtonBaseHovered=(&self:WidgetTree.Button_TextSearch.HandleOnButtonHovered)
    onButtonBaseUnhovered=(&self:WidgetTree.Button_TextSearch.HandleOnButtonUnhovered)
    onSelectedChangedBase=(&self:WidgetTree.Button_TextSearch.HandleOnSelectedChanged)
    style=/FortUILibrary/Foundation/Buttons/ButtonStyles/Athena/ButtonStyle_TransparentList_Nothing.ButtonStyle_TransparentList_Nothing_C
    inputActionRenderTranslation=<-50 0>
    triggeringInputAction {dataTable=&InputActionDataTable, rowName=GenericLeftTrigger}
   }
   slot {class=SizeBoxSlot, properties {horizontalAlignment=HAlign_Right, verticalAlignment=VAlign_Center}}
  }
  SearchIcon {
   widgetClassPath=&Image
   namedSlotHost=Button_TextSearch
   widget=SearchIcon
   properties {brush {
     bIsDynamicallyLoaded=false
     drawAs=Image
     imageType=FullColor
     mirroring=NoMirror
     resourceName=None
     resourceObject=/FortUILibrary/Textures/Icons/T_UI_Icon_UIKit_Search
     tiling=NoTile
     imageSize=<32 32>
     margin=<ltrb 0 0 0 0>
     outlineSettings {
      bUseBrushTransparency=false
      roundingType=HalfHeightRadius
      width=0
      color {colorUseRule=UseColor_Specified, specifiedColor=<rgba 0 0 0 0>}
      cornerRadii=<0 0 0 0>
     }
     tintColor {colorUseRule=UseColor_Specified, specifiedColor=<rgba 0.5776 0.5776 0.5776 1>}
     uVRegion {bIsValid=false, max=<0 0>, min=<0 0>}
    }}
  }
  MaxWidth_SB {
   widgetClassPath=&SizeBox
   parent=SearchBar_HB
   widget=MaxWidth_SB
   properties {bOverride_MaxDesiredWidth=true, maxDesiredWidth=550}
   slot {class=HorizontalBoxSlot}
  }
  Button_SortAndFilter {
   widgetClassPath=&WBP_UIKit_Button_Regular
   bIsVariable=true
   parent=MaxWidth_SB
   widget=Button_SortAndFilter
   properties {
    minHeight=90
    mobileMinHeight=70
    navigation=&self:WidgetTree.Button_SortAndFilter.WidgetNavigation_0
    overrideDefaultHeight=true
    text=Sort + Filter
    triggeringInputAction {dataTable=&InputActionDataTable, rowName=GenericRightTrigger}
   }
   slot {class=SizeBoxSlot, properties {verticalAlignment=VAlign_Top}}
  }
 ]
}
