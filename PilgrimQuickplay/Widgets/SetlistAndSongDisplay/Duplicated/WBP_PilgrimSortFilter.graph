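# Pilgrim Quickplay sort and filter controls widget, a child of FortSortFilterControls that adds no widgets of its own.
# On Construct it reads the guitar-mode client setting through the owning player's Pilgrim controller component.
# That flag selects the standard or guitar SortFilterRequest input row as the sort button's triggering action.
# Two down-navigation functions return the widget below the button or text box, substituting its desired focus target when activatable.
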WidgetBlueprint /PilgrimQuickplay/Widgets/SetlistAndSongDisplay/Duplicated/WBP_PilgrimSortFilter
graphs [
 EventGraph {nodes [
   Event Construct()
   > Parent: Construct()
   > n1 = Sequence()
   n2 = GetOwningPlayer()
   > Is Valid(InputObject=@n2)
   n3 = GetComponentByClass!(
    self:Actor Object Reference=@n2
    ComponentClass:Actor Component Class Reference=/PilgrimCore/Player/Components/BP_PilgrimPlayerControllerComponent.BP_PilgrimPlayerControllerComponent_C
   ) -> (ReturnValue:BP Pilgrim Player Controller Component Object Reference)
   > Is Valid(InputObject=@n3)
   n4 = GetPilgrimClientSettingRecordPartition!(self:Pilgrim Player Controller Component Object Reference=@n3) -> (ReturnValue:Pilgrim Client Setting Record Partition Object Reference)
   > Is Valid(InputObject=@n4)
   > Set IsSwitchGuitarModeEnabled(IsSwitchGuitarModeEnabled:bool=GetEnableGuitarMode!(self:Pilgrim Client Setting Record Partition Object Reference=@n4) -> (ReturnValue:bool)) -> (Output_Get:bool)
   SetTriggeringInputAction!(
    execute=@n1.then_1
    self:Common Button Base Object Reference=Get Button_SortAndFilter() -> (Button_SortAndFilter:WBP UIKit Button Regular Object Reference)
    InputActionRow:"Data Table Row Handle Structure (by ref)"=Select(
     Option 0:"Data Table Row Handle Structure (by ref)"=Make Data Table Row Handle!(
      DataTable:Data Table Object Reference=/PilgrimQuickplay/Widgets/InputActionDataTables/FM_PilgrimQuickplay_InputActionDataTable
      RowName:Name=SortFilterRequest
     ) -> (DataTableRowHandle:Data Table Row Handle Structure) {is=title K2Node_MakeStruct}
     Option 1:"Data Table Row Handle Structure (by ref)"=Make Data Table Row Handle(DataTable=&FM_PilgrimQuickplay_InputActionDataTable, RowName=SortFilterRequest_SwitchGuitar)
     Index=Get IsSwitchGuitarModeEnabled() -> (IsSwitchGuitarModeEnabled:bool)
    ) -> (ReturnValue:"Data Table Row Handle Structure (by ref)")
   ) -> (then)
   Event PreConstruct()
   Event Tick()
  ]}
 NavigateDownFromButtom {nodes [
   NavigateDownFromButtom!() -> (then, Navigation:EUINavigation Enum) {is=title K2Node_FunctionEntry}
   n1 = Get WidgetBelowButton() -> (WidgetBelowButton:Widget Object Reference)
   n2 = Cast To CommonActivatableWidget!(Object=@n1) -> (AsCommon Activatable Widget:Common Activatable Widget Object Reference, bSuccess:bool=~) {is=title K2Node_DynamicCast}
   > Return Node!(
    execute
    ReturnValue:Widget Object Reference=Select(
     Option 0:Widget Object Reference=@n1
     Option 1:Widget Object Reference=GetDesiredFocusTarget!(self:Common Activatable Widget Object Reference=@n2) -> (ReturnValue:Widget Object Reference)
     Index=@n2.bSuccess
    ) -> (ReturnValue:Widget Object Reference)
   )
  ]}
 NavigateDownFromTextBox {nodes [
   NavigateDownFromTextBox!() -> (then, Navigation:EUINavigation Enum) {is=title K2Node_FunctionEntry}
   n1 = Get WidgetBelowTextBox() -> (WidgetBelowTextBox:Widget Object Reference)
   n2 = Cast To CommonActivatableWidget(Object=@n1) -> (bSuccess=~)
   > Return Node(ReturnValue=Select(Option 0=@n1, Option 1=GetDesiredFocusTarget(self=@n2), Index=@n2.bSuccess))
  ]}
]
variables [IsSwitchGuitarModeEnabled {type=boolean}]
widgets {
 info {
  inheritedWidgetCount=0
  parentClass=/FortniteCoreUI/Customization/FortSortFilterControls.FortSortFilterControls_C
  rootWidgetClass=null
  widgetCount=0
 }
 widgets []
}
