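# Pilgrim quickplay modal dialog for sorting, ordering and filtering the song catalog, built on the UIKit dialog base.
# On activation it snapshots active items from the sort/filter and song catalog viewmodels, then tracks section clicks as pending selections.
# Button 0 commits selections and broadcasts SortFilterAccepted, button 1 restores the snapshot, and both deactivate the widget.
# Focus targets the first Sort entry, and Pilgrim Inverse mode overrides dialog width.
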WidgetBlueprint /PilgrimQuickplay/Widgets/SetlistAndSongDisplay/SortFilterModal/WBP_Pilgrim_SortFilter_Modal
animations [Intro {
  binding_count=0
  length_seconds=0.2986
  object=&self:Intro
  sequence {bindings []}
 }]
graphs [
 BP_GetDesiredFocusTarget {nodes [
   Get Desired Focus Target()
   n1 = Get Sort_Section() -> (Sort_Section:WBP Pilgrim Sort Filter Filters Section Object Reference)
   > n2 = Set ListOfAllEntries(ListOfAllEntries:Array of User Widget Object References=GetAllEntries!(self:Dynamic Entry Box Base Object Reference=Get DynBox_Entries_Col1!(self:WBP Pilgrim Sort Filter Filters Section Object Reference=@n1) -> (DynBox_Entries_Col1:Dynamic Entry Box Object Reference)) -> (ReturnValue:Array of User Widget Object References)) -> (Output_Get:Array of User Widget Object References)
   > Append Array!(
    execute
    TargetArray:Array of User Widget Object References=@n2
    SourceArray:Array of User Widget Object References=GetAllEntries(self=Get DynBox_Entries_Col2!(self:WBP Pilgrim Sort Filter Filters Section Object Reference=@n1) -> (DynBox_Entries_Col2:Dynamic Entry Box Object Reference))
   ) -> (then)
   > Return Node!(
    execute
    ReturnValue:Widget Object Reference="Get (a copy)"(Array:Array of User Widget Object References=Get ListOfAllEntries() -> (ListOfAllEntries:Array of User Widget Object References)) -> (Output:User Widget Object Reference)
   )
  ]}
 EventGraph {nodes [
   n1 = "On Item Clicked (Filter_Section)"!() -> (OutputDelegate, then, Item Filter VM:Fort Sort Filter Item VM Object Reference) {is=title K2Node_ComponentBoundEvent}
   > n2 = Cast To FortFilterItemVM!(execute, Object=@n1.Item Filter VM) -> (then, CastFailed, AsFort Filter Item VM:Fort Filter Item VM Object Reference) {is=title K2Node_DynamicCast}
   > Set Saved Filter Selection(Saved Filter Selection:Fort Filter Item VM Object Reference=@n2) -> (Output_Get:Fort Filter Item VM Object Reference)
   n3 = "On Item Clicked (Sort_Section)"!() -> (OutputDelegate, then, Item Filter VM:Fort Sort Filter Item VM Object Reference) {is=title K2Node_ComponentBoundEvent}
   > n4 = Cast To FortSortItemVM!(execute, Object=@n3.Item Filter VM) -> (then, CastFailed, AsFort Sort Item VM:Fort Sort Item VM Object Reference) {is=title K2Node_DynamicCast}
   > Set Saved Sort Selection(Saved Sort Selection:Fort Sort Item VM Object Reference=@n4) -> (Output_Get:Fort Sort Item VM Object Reference)
   n5 = "On Item Clicked (OrderBy_Section)"!() -> (OutputDelegate, then, Item Filter VM:Fort Sort Filter Item VM Object Reference) {is=title K2Node_ComponentBoundEvent}
   > n6 = Cast To PilgrimSongCatalogItemSecondaryFilterSortOrderItemVM!(execute, Object=@n5.Item Filter VM) -> (
    then
    CastFailed
    AsPilgrim Song Catalog Item Secondary Filter Sort Order Item VM:Pilgrim Song Catalog Item Secondary Filter Sort Order Item VM Object Reference
   ) {is=title K2Node_DynamicCast}
   > Set Saved Order By Selection(Saved Order By Selection:Pilgrim Song Catalog Item Secondary Filter Sort Order Item VM Object Reference=@n6) -> (Output_Get:Pilgrim Song Catalog Item Secondary Filter Sort Order Item VM Object Reference)
   n7 = ButtonClicked!() -> (OutputDelegate, then, ButtonIndex:int) {is=title K2Node_CustomEvent}
   > Branch(Condition=Get bIsActive() -> (bIsActive:bool))
   > n8 = Sequence()
   > n9 = Switch on Int!(execute:Exec, Selection:int=@n7.ButtonIndex) -> (0:Exec, 1:Exec) {bHasDefaultPin=false}
   > Set ActiveSortItem!(
    execute
    ActiveSortItem:Fort Sort Item VM Object Reference=Get Saved Sort Selection() -> (Saved Sort Selection:Fort Sort Item VM Object Reference)
    self:Fort Sort Filter VM Object Reference=Get FortSortFilterVM() -> (FortSortFilterVM:Fort Sort Filter VM Object Reference)
   ) -> (then, Output_Get:Fort Sort Item VM Object Reference)
   > Set ActiveFilterItem!(
    execute
    ActiveFilterItem:Fort Filter Item VM Object Reference=Get Saved Filter Selection() -> (Saved Filter Selection:Fort Filter Item VM Object Reference)
    self:Fort Sort Filter VM Object Reference=Get FortSortFilterVM()
   ) -> (then, Output_Get:Fort Filter Item VM Object Reference)
   > Set ActiveSortOrderItem!(
    execute
    ActiveSortOrderItem:Pilgrim Song Catalog Item Secondary Filter Sort Order Item VM Object Reference=Get Saved Order By Selection() -> (Saved Order By Selection:Pilgrim Song Catalog Item Secondary Filter Sort Order Item VM Object Reference)
    self:Pilgrim Song Catalog VM Object Reference=Get PilgrimSongCatalogVM() -> (PilgrimSongCatalogVM:Pilgrim Song Catalog VM Object Reference)
   ) -> (then, Output_Get:Pilgrim Song Catalog Item Secondary Filter Sort Order Item VM Object Reference)
   > Call Sort Filter Accepted!(execute:Exec, self:WBP Pilgrim Sort Filter Modal Object Reference) -> (then:Exec) {is=title K2Node_CallDelegate}
   RestoreFiltersToSaved(execute=@n9.1)
   DeactivateWidget(execute=@n8.then_1)
   n10 = Event PreConstruct()
   > Branch(Condition=@n10.IsDesignTime)
   > n11 = Construct Object from Class(Class=&/DialogVM_Pilgrim_SortFilter.DialogVM_Pilgrim_SortFilter_C) -> (ReturnValue:Dialog VM Pilgrim Sort Filter Object Reference)
   > SetDialogData!(
    execute
    self:WBP UIKit Dialog Internal Object Reference=Get WBP_UIKit_Dialog_Base() -> (WBP_UIKit_Dialog_Base:WBP UIKit Dialog Internal Object Reference)
    Data:UI Kit Dialog Viewmodel Object Reference=@n11
   ) -> (then)
   > SetupForPilgrimInverse!(execute, self, Index:bool=Get PreviewPilgrimInverse() -> (PreviewPilgrimInverse:bool)) -> (then)
   Event On Deactivated()
   > Parent: On Deactivated() {is=title K2Node_CallParentFunction}
   Event Construct()
   > SetupForPilgrimInverse(Index=IsPilgrimInverse!(OwningPlayerController:Player Controller Object Reference=GetOwningPlayer()) -> (IsPilgrimInverse:bool))
   Event On Activated()
   > Parent: On Activated() {is=title K2Node_CallParentFunction}
   > PlayAnimation(InAnimation=Get Intro())
   > n12 = Construct Object from Class(Class=&DialogVM_Pilgrim_SortFilter_C)
   > SetDialogData(self=Get WBP_UIKit_Dialog_Base(), Data=@n12)
   n13 = Get WBP_UIKit_Dialog_Base()
   > Unbind all Events from Dialog Button Clicked Dispatcher!(execute:Exec, self:WBP UIKit Dialog Internal Object Reference=@n13) -> (then:Exec) {is=title K2Node_ClearDelegate}
   > Bind Event to Dialog Button Clicked Dispatcher(self:WBP UIKit Dialog Internal Object Reference=@n13, Delegate=Create Event()) {is=title K2Node_AddDelegate}
   > InitializeSelectedItems()
   > SetAllNavigationRules!(execute, self:Widget Object Reference, Rule:EUINavigationRule Enum=Stop, WidgetToFocus:Name=None) -> (then)
  ]}
 InitializeSelectedItems {nodes [
   InitializeSelectedItems() {is=title K2Node_FunctionEntry}
   > SaveCurrentFilters()
   > Set Saved Sort Selection(Saved Sort Selection=Get Saved Active Sort Item() -> (Saved Active Sort Item:Fort Sort Item VM Object Reference))
   > Set Saved Filter Selection(Saved Filter Selection=Get Saved Active Filter Item() -> (Saved Active Filter Item:Fort Filter Item VM Object Reference))
   > Set Saved Order by Selection(Saved Order by Selection:Pilgrim Song Catalog Item Secondary Filter Sort Order Item VM Object Reference=Get Saved Order by Item() -> (Saved Order by Item:Pilgrim Song Catalog Item Secondary Filter Sort Order Item VM Object Reference)) -> (Output_Get:Pilgrim Song Catalog Item Secondary Filter Sort Order Item VM Object Reference)
  ]}
 RestoreFiltersToSaved {nodes [
   RestoreFiltersToSaved() {is=title K2Node_FunctionEntry}
   > n1 = Set ActiveSortItem(ActiveSortItem=Get Saved Active Sort Item(), self=Get FortSortFilterVM())
   > HandleButtonClicked!(
    execute
    self:WBP Pilgrim Sort Filter Filters Section Object Reference=Get Sort_Section()
    Item VM:Fort Sort Filter Item VM Object Reference=@n1
   ) -> (then)
   > n2 = Set ActiveFilterItem(ActiveFilterItem=Get Saved Active Filter Item(), self=Get FortSortFilterVM())
   > HandleButtonClicked(self=Get Filter_Section() -> (Filter_Section:WBP Pilgrim Sort Filter Filters Section Object Reference), Item VM=@n2)
   > n3 = Set ActiveSortOrderItem(ActiveSortOrderItem=Get Saved Order by Item(), self=Get PilgrimSongCatalogVM())
   > HandleButtonClicked(self=Get OrderBy_Section() -> (OrderBy_Section:WBP Pilgrim Sort Filter Filters Section Object Reference), Item VM=@n3)
  ]}
 SaveCurrentFilters {nodes [
   SaveCurrentFilters() {is=title K2Node_FunctionEntry}
   > Set Saved Active Sort Item(Saved Active Sort Item:Fort Sort Item VM Object Reference=Get ActiveSortItem!(self:Fort Sort Filter VM Object Reference=Get FortSortFilterVM()) -> (ActiveSortItem:Fort Sort Item VM Object Reference)) -> (Output_Get:Fort Sort Item VM Object Reference)
   > Set Saved Active Filter Item(Saved Active Filter Item:Fort Filter Item VM Object Reference=Get ActiveFilterItem!(self:Fort Sort Filter VM Object Reference=Get FortSortFilterVM()) -> (ActiveFilterItem:Fort Filter Item VM Object Reference)) -> (Output_Get:Fort Filter Item VM Object Reference)
   > Set Saved Order by Item(Saved Order by Item:Pilgrim Song Catalog Item Secondary Filter Sort Order Item VM Object Reference=Get ActiveSortOrderItem!(self:Pilgrim Song Catalog VM Object Reference=Get PilgrimSongCatalogVM()) -> (ActiveSortOrderItem:Pilgrim Song Catalog Item Secondary Filter Sort Order Item VM Object Reference)) -> (Output_Get:Pilgrim Song Catalog Item Secondary Filter Sort Order Item VM Object Reference)
  ]}
 SetupForPilgrimInverse {nodes [
   n1 = SetupForPilgrimInverse!() -> (then, Index:bool=~) {is=title K2Node_FunctionEntry}
   n2 = Make S UI Bool Float!(Enable_3_#1:bool=true, Value_4_#2:double=1000) -> (S_UI_BoolFloat:S UI Bool Float Structure) {is=title K2Node_MakeStruct}
   n3 = Get WBP_UIKit_Dialog_Base()
   n4 = Get!(self:WBP UIKit Dialog Internal Object Reference=@n3) -> (
    DialogSIze_WidthOverride_107_#3:S UI Bool Float Structure
    DialogSIze_HeightOverride_108_#4:S UI Bool Float Structure
    DialogSIze_MinDesiredWidth_109_#5:S UI Bool Float Structure
    DialogSIze_MinDesiredHeight_110_#6:S UI Bool Float Structure
    DialogSIze_MaxDesiredWidth_111_#7:S UI Bool Float Structure
    DialogSIze_MaxDesiredHeight_112_#8:S UI Bool Float Structure
    DialogSIze_MinAspectRatio_113_#9:S UI Bool Float Structure
    DialogSIze_MaxAspectRatio_114_#10:S UI Bool Float Structure
   )
   > Set DialogSIze!(
    execute
    DialogSIze:S UI Size Box Structure=Make S UI Size Box!(
     WidthOverride_107_#3:S UI Bool Float Structure=Select(Option 0:S UI Bool Float Structure=@n4, Option 1:S UI Bool Float Structure=@n2, Index=@n1) -> (ReturnValue:S UI Bool Float Structure)
     HeightOverride_108_#4:S UI Bool Float Structure=@n4.DialogSIze_HeightOverride_108_#4
     MinDesiredWidth_109_#5:S UI Bool Float Structure=@n4.DialogSIze_MinDesiredWidth_109_#5
     MinDesiredHeight_110_#6:S UI Bool Float Structure=@n4.DialogSIze_MinDesiredHeight_110_#6
     MaxDesiredWidth_111_#7:S UI Bool Float Structure=@n4.DialogSIze_MaxDesiredWidth_111_#7
     MaxDesiredHeight_112_#8:S UI Bool Float Structure=@n4.DialogSIze_MaxDesiredHeight_112_#8
     MinAspectRatio_113_#9:S UI Bool Float Structure=@n4.DialogSIze_MinAspectRatio_113_#9
     MaxAspectRatio_114_#10:S UI Bool Float Structure=@n4.DialogSIze_MaxAspectRatio_114_#10
    ) -> (S_UI_SizeBox:S UI Size Box Structure) {is=title K2Node_MakeStruct}
    self:WBP UIKit Dialog Internal Object Reference=@n3
   ) -> (then, Output_Get:S UI Size Box Structure)
  ]}
 SortFilterAccepted {nodes [SortFilterAccepted() {is=title K2Node_FunctionEntry}]}
]
variables [
 PreviewPilgrimInverse {type=boolean}
 Saved Active Filter Item {type=FortFilterItemVM, default=null}
 Saved Active Sort Item {type=FortSortItemVM, default=null}
 Saved Filter Selection {type=FortFilterItemVM, default=null}
 Saved Order by Item {type=PilgrimSongCatalogItemSecondaryFilterSortOrderItemVM, default=null}
 Saved Order by Selection {type=PilgrimSongCatalogItemSecondaryFilterSortOrderItemVM, default=null}
 Saved Sort Selection {type=FortSortItemVM, default=null}
 SortFilterAccepted {}
]
widgets {
 info {
  inheritedWidgetCount=0
  namedSlotCount=1
  parentClass=&CommonActivatableWidget
  rootWidgetClass=/FortUILibrary/Systems/Dialog/WBP_UIKit_Dialog_Internal.WBP_UIKit_Dialog_Internal_C
  widgetCount=9
 }
 widgets [
  WBP_UIKit_Dialog_Base {
   widgetClassPath=/FortUILibrary/Systems/Dialog/WBP_UIKit_Dialog_Internal
   bIsVariable=true
   widget=WBP_UIKit_Dialog_Base
   properties {
    buttons {
     each Button Horizontal Fill=true
     each Button Min Width=0
     each Button Size=Medium
     focuse Button=0
     space Between Buttons=30
     cTA Buttons Types [Regular,Quiet]
     group Placement {
      horizontal Alignment=HAlign_Fill
      vertical Alignment=VAlign_Bottom
      grid {column=1, column Span=0, layer=0, row=2, row Span=0, nudge=<0 0>}
      padding=<ltrb 80 64 80 48>
     }
    }
    dialogSIze {
     height Override {enable=false, value=800}
     max Aspect Ratio {enable=false, value=0}
     max Desired Height {enable=true, value=800}
     max Desired Width {enable=false, value=0}
     min Aspect Ratio {enable=false, value=0}
     min Desired Height {enable=false, value=0}
     min Desired Width {enable=false, value=0}
     width Override {enable=true, value=850}
    }
    titleProperties {
     auto Wrap Text=true
     justification=Left
     line Height Percentage=1
     mobile Font Size Multiplier=1.3
     transform Policy=ToUpper
     wrap Text At=0
     color {colorUseRule=UseColor_Specified, specifiedColor=<rgba 1 1 1 1>}
     font {
      bForceMonospaced=false
      bMaterialIsStencil=false
      fontMaterial=null
      fontObject=&HeadingNow
      letterSpacing=1
      monospacedWidth=1
      size=24
      skewAmount=0
      typefaceFontName=86 Bold Italic
      outlineSettings {
       bApplyOutlineToDropShadows=false
       bMiteredCorners=false
       bSeparateFillAlpha=false
       outlineMaterial=null
       outlineSize=0
       outlineColor=<rgba 0 0 0 1>
      }
     }
     placement {
      horizontal Alignment=HAlign_Fill
      vertical Alignment=VAlign_Top
      grid {column=1, column Span=0, layer=0, row=0, row Span=0, nudge=<0 0>}
      padding=<ltrb 90 80 80 32>
     }
    }
   }
  }
  MainContent {widgetClassPath=&Overlay, namedSlotHost=WBP_UIKit_Dialog_Base, widget=MainContent}
  ScrollBox_1 {
   widgetClassPath=&ScrollBox
   parent=MainContent
   widget=ScrollBox_1
   properties {
    bAnimateWheelScrolling=true
    navigationScrollPadding=100
    scrollWhenFocusChanges=AnimatedScroll
    wheelScrollMultiplier=3
    scrollbarPadding=<ltrb 2 0 2 0>
    scrollbarThickness=<10 10>
    widgetBarStyle {
     thickness=10
     draggedThumbImage {
      bIsDynamicallyLoaded=false
      drawAs=Box
      imageType=FullColor
      mirroring=NoMirror
      resourceName=../../../Engine/Content/Slate/Common/Scrollbar_Thumb.png
      resourceObject=/FortUILibrary/Foundation/Art_ToolKit/Tools/M_UI_Circle_White
      tiling=NoTile
      imageSize=<10 10>
      margin=<ltrb 0.5 0.5 0.5 0.5>
      outlineSettings {
       bUseBrushTransparency=false
       roundingType=HalfHeightRadius
       width=0
       color {colorUseRule=UseColor_Specified, specifiedColor=<rgba 0 0 0 0>}
       cornerRadii=<0 0 0 0>
      }
      tintColor {colorUseRule=UseColor_Specified, specifiedColor=<rgba 1 1 1 1>}
      uVRegion {bIsValid=false, max=<0 0>, min=<0 0>}
     }
     horizontalBackgroundImage {
      bIsDynamicallyLoaded=false
      drawAs=NoDrawType
      imageType=NoImage
      mirroring=NoMirror
      resourceName=None
      resourceObject=null
      tiling=NoTile
      imageSize=<0 0>
      margin=<ltrb 0 0 0 0>
      outlineSettings {
       bUseBrushTransparency=false
       roundingType=HalfHeightRadius
       width=0
       color {colorUseRule=UseColor_Specified, specifiedColor=<rgba 0 0 0 0>}
       cornerRadii=<0 0 0 0>
      }
      tintColor {colorUseRule=UseColor_Specified, specifiedColor=<rgba 1 1 1 1>}
      uVRegion {bIsValid=false, max=<0 0>, min=<0 0>}
     }
     horizontalBottomSlotImage {
      bIsDynamicallyLoaded=false
      drawAs=Image
      imageType=FullColor
      mirroring=NoMirror
      resourceName=../../../Engine/Content/Slate/Common/Scrollbar_Background_Horizontal.png
      resourceObject=null
      tiling=NoTile
      imageSize=<8 8>
      margin=<ltrb 0 0 0 0>
      outlineSettings {
       bUseBrushTransparency=false
       roundingType=HalfHeightRadius
       width=0
       color {colorUseRule=UseColor_Specified, specifiedColor=<rgba 0 0 0 0>}
       cornerRadii=<0 0 0 0>
      }
      tintColor {colorUseRule=UseColor_Specified, specifiedColor=<rgba 1 1 1 1>}
      uVRegion {bIsValid=false, max=<0 0>, min=<0 0>}
     }
     horizontalTopSlotImage {
      bIsDynamicallyLoaded=false
      drawAs=Image
      imageType=FullColor
      mirroring=NoMirror
      resourceName=../../../Engine/Content/Slate/Common/Scrollbar_Background_Horizontal.png
      resourceObject=null
      tiling=NoTile
      imageSize=<8 8>
      margin=<ltrb 0 0 0 0>
      outlineSettings {
       bUseBrushTransparency=false
       roundingType=HalfHeightRadius
       width=0
       color {colorUseRule=UseColor_Specified, specifiedColor=<rgba 0 0 0 0>}
       cornerRadii=<0 0 0 0>
      }
      tintColor {colorUseRule=UseColor_Specified, specifiedColor=<rgba 1 1 1 1>}
      uVRegion {bIsValid=false, max=<0 0>, min=<0 0>}
     }
     hoveredThumbImage {
      bIsDynamicallyLoaded=false
      drawAs=Box
      imageType=FullColor
      mirroring=NoMirror
      resourceName=../../../Engine/Content/Slate/Common/Scrollbar_Thumb.png
      resourceObject=&M_UI_Circle_White
      tiling=NoTile
      imageSize=<10 10>
      margin=<ltrb 0.5 0.5 0.5 0.5>
      outlineSettings {
       bUseBrushTransparency=false
       roundingType=HalfHeightRadius
       width=0
       color {colorUseRule=UseColor_Specified, specifiedColor=<rgba 0 0 0 0>}
       cornerRadii=<0 0 0 0>
      }
      tintColor {colorUseRule=UseColor_Specified, specifiedColor=<rgba 1 1 1 1>}
      uVRegion {bIsValid=false, max=<0 0>, min=<0 0>}
     }
     normalThumbImage {
      bIsDynamicallyLoaded=false
      drawAs=Box
      imageType=FullColor
      mirroring=NoMirror
      resourceName=../../../Engine/Content/Slate/Common/Scrollbar_Thumb.png
      resourceObject=&M_UI_Circle_White
      tiling=NoTile
      imageSize=<10 10>
      margin=<ltrb 0.5 0.5 0.5 0.5>
      outlineSettings {
       bUseBrushTransparency=false
       roundingType=HalfHeightRadius
       width=0
       color {colorUseRule=UseColor_Specified, specifiedColor=<rgba 0 0 0 0>}
       cornerRadii=<1 1 1 1>
      }
      tintColor {colorUseRule=UseColor_Specified, specifiedColor=<rgba 1 1 1 1>}
      uVRegion {bIsValid=false, max=<0 0>, min=<0 0>}
     }
     verticalBackgroundImage {
      bIsDynamicallyLoaded=false
      drawAs=Box
      imageType=NoImage
      mirroring=NoMirror
      resourceName=None
      resourceObject=&M_UI_Circle_White
      tiling=NoTile
      imageSize=<10 10>
      margin=<ltrb 0.5 0.5 0.5 0.5>
      outlineSettings {
       bUseBrushTransparency=false
       roundingType=HalfHeightRadius
       width=9
       color {colorUseRule=UseColor_Specified, specifiedColor=<rgba 0 0 0 0>}
       cornerRadii=<1 1 1 1>
      }
      tintColor {colorUseRule=UseColor_Specified, specifiedColor=<rgba 1 1 1 0.2>}
      uVRegion {bIsValid=false, max=<0 0>, min=<0 0>}
     }
     verticalBottomSlotImage {
      bIsDynamicallyLoaded=false
      drawAs=NoDrawType
      imageType=FullColor
      mirroring=NoMirror
      resourceName=../../../Engine/Content/Slate/Common/Scrollbar_Background_Vertical.png
      resourceObject=null
      tiling=NoTile
      imageSize=<10 10>
      margin=<ltrb 0.5 0.5 0.5 0.5>
      outlineSettings {
       bUseBrushTransparency=false
       roundingType=HalfHeightRadius
       width=0
       color {colorUseRule=UseColor_Specified, specifiedColor=<rgba 0 0 0 0>}
       cornerRadii=<0 0 0 0>
      }
      tintColor {colorUseRule=UseColor_Specified, specifiedColor=<rgba 1 1 1 1>}
      uVRegion {bIsValid=false, max=<0 0>, min=<0 0>}
     }
     verticalTopSlotImage {
      bIsDynamicallyLoaded=false
      drawAs=NoDrawType
      imageType=FullColor
      mirroring=NoMirror
      resourceName=../../../Engine/Content/Slate/Common/Scrollbar_Background_Vertical.png
      resourceObject=null
      tiling=NoTile
      imageSize=<8 8>
      margin=<ltrb 0 0 0 0>
      outlineSettings {
       bUseBrushTransparency=false
       roundingType=HalfHeightRadius
       width=9
       color {colorUseRule=UseColor_Specified, specifiedColor=<rgba 0 0 0 0>}
       cornerRadii=<0.5 0.5 0.5 0.5>
      }
      tintColor {colorUseRule=UseColor_Specified, specifiedColor=<rgba 1 1 1 1>}
      uVRegion {bIsValid=false, max=<0 0>, min=<0 0>}
     }
    }
    widgetStyle {
     barThickness=10
     bottomShadowBrush {
      bIsDynamicallyLoaded=false
      drawAs=Box
      imageType=FullColor
      mirroring=NoMirror
      resourceName=../../../Engine/Content/Slate/Common/ScrollBoxShadowBottom.png
      resourceObject=null
      tiling=NoTile
      imageSize=<16 8>
      margin=<ltrb 0 0 0 0>
      outlineSettings {
       bUseBrushTransparency=false
       roundingType=HalfHeightRadius
       width=0
       color {colorUseRule=UseColor_Specified, specifiedColor=<rgba 0 0 0 0>}
       cornerRadii=<0 0 0 0>
      }
      tintColor {colorUseRule=UseColor_Specified, specifiedColor=<rgba 1 1 1 0>}
      uVRegion {bIsValid=false, max=<0 0>, min=<0 0>}
     }
     horizontalScrolledContentPadding=<ltrb 0 0 1 0>
     leftShadowBrush {
      bIsDynamicallyLoaded=false
      drawAs=Box
      imageType=FullColor
      mirroring=NoMirror
      resourceName=../../../Engine/Content/Slate/Common/ScrollBoxShadowLeft.png
      resourceObject=null
      tiling=NoTile
      imageSize=<8 16>
      margin=<ltrb 1 0.5 0 0.5>
      outlineSettings {
       bUseBrushTransparency=false
       roundingType=HalfHeightRadius
       width=0
       color {colorUseRule=UseColor_Specified, specifiedColor=<rgba 0 0 0 0>}
       cornerRadii=<0 0 0 0>
      }
      tintColor {colorUseRule=UseColor_Specified, specifiedColor=<rgba 1 1 1 1>}
      uVRegion {bIsValid=false, max=<0 0>, min=<0 0>}
     }
     rightShadowBrush {
      bIsDynamicallyLoaded=false
      drawAs=Box
      imageType=FullColor
      mirroring=NoMirror
      resourceName=../../../Engine/Content/Slate/Common/ScrollBoxShadowRight.png
      resourceObject=null
      tiling=NoTile
      imageSize=<8 16>
      margin=<ltrb 0 0.5 1 0.5>
      outlineSettings {
       bUseBrushTransparency=false
       roundingType=HalfHeightRadius
       width=0
       color {colorUseRule=UseColor_Specified, specifiedColor=<rgba 0 0 0 0>}
       cornerRadii=<0 0 0 0>
      }
      tintColor {colorUseRule=UseColor_Specified, specifiedColor=<rgba 1 1 1 0>}
      uVRegion {bIsValid=false, max=<0 0>, min=<0 0>}
     }
     topShadowBrush {
      bIsDynamicallyLoaded=false
      drawAs=NoDrawType
      imageType=FullColor
      mirroring=NoMirror
      resourceName=../../../Engine/Content/Slate/Common/ScrollBoxShadowTop.png
      resourceObject=null
      tiling=NoTile
      imageSize=<1 90>
      margin=<ltrb 0 0 0 0>
      outlineSettings {
       bUseBrushTransparency=false
       roundingType=HalfHeightRadius
       width=0
       color {colorUseRule=UseColor_Specified, specifiedColor=<rgba 0 0 0 0>}
       cornerRadii=<0 0 0 0>
      }
      tintColor {colorUseRule=UseColor_Specified, specifiedColor=<rgba 1 1 1 0>}
      uVRegion {bIsValid=false, max=<0 0>, min=<0 0>}
     }
     verticalScrolledContentPadding=<ltrb 0 0 0 1>
    }
   }
   slot {class=OverlaySlot, properties {horizontalAlignment=HAlign_Fill, verticalAlignment=VAlign_Fill, padding=<ltrb 80 135 74 168>}}
  }
  VerticalBox_0 {widgetClassPath=&VerticalBox, parent=ScrollBox_1, widget=VerticalBox_0, slot {class=ScrollBoxSlot}}
  Sort_Section {
   widgetClassPath=&/WBP_Pilgrim_SortFilter_FiltersSection
   bIsVariable=true
   parent=VerticalBox_0
   widget=Sort_Section
   properties {bIsFocusable=false, sectionLabel=Sort by}
   slot {class=VerticalBoxSlot}
  }
  OrderBy_Section {
   widgetClassPath=&WBP_Pilgrim_SortFilter_FiltersSection
   bIsVariable=true
   parent=VerticalBox_0
   widget=OrderBy_Section
   properties {bIsFocusable=false, sectionLabel=Order by}
   slot {class=VerticalBoxSlot}
  }
  Filter_Section {
   widgetClassPath=&WBP_Pilgrim_SortFilter_FiltersSection
   bIsVariable=true
   parent=VerticalBox_0
   widget=Filter_Section
   properties {bIsFocusable=false, sectionLabel=Filter by, visibility=Collapsed}
   slot {class=VerticalBoxSlot}
  }
  Secondary_Filter_Section {
   widgetClassPath=&/WBP_Pilgrim_SortFilter_OptionsSection
   bIsVariable=true
   parent=VerticalBox_0
   widget=Secondary_Filter_Section
   properties {sectionLabel=Filter by}
   slot {class=VerticalBoxSlot}
  }
  Option_Section {
   widgetClassPath=&WBP_Pilgrim_SortFilter_OptionsSection
   parent=VerticalBox_0
   widget=Option_Section
   properties {sectionLabel=Extra settings}
   slot {class=VerticalBoxSlot}
  }
 ]
}
