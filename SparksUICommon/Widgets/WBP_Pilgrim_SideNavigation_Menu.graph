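# Pilgrim side navigation menu, a child of the UIKit side navigation widget that adds focus-driven collapsing and no widgets of its own.
# On construct and input method change it sets EnableCollapsedState, which is off only for compact UI with touch input.
# Entering the focus path expands the menu and leaving it collapses it, gated by mobile-game and EnableCollapsedState checks.
# Focus events write to the shipping log, and focus loss only logs.
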WidgetBlueprint /SparksUICommon/Widgets/WBP_Pilgrim_SideNavigation_Menu
graphs [EventGraph {nodes [
   n1 = Event PreConstruct()
   > Parent: PreConstruct(IsDesignTime=@n1.IsDesignTime)
   Event Construct()
   > Parent: Construct()
   > Event On Input Method Changed!(
    execute
    self:WBP UIKit Tab Group Base Object Reference
    bNewInputType:ECommonInputType Enum=GetCurrentInputType(self=Get CommonInputSubsystem())
   ) -> (then)
   Event OnAddedToFocusPath!() -> (OutputDelegate, then, InFocusEvent:Focus Event Structure) {is=title K2Node_Event}
   > Branch(Condition=IsMobileGame())
   > FortShippingLog!(
    execute
    WorldContextObject:Object Reference=Self-Reference()
    InString:String=Pilgrim_SideNavigation_Menu: On Added to Focus Path
    bLogAsWarning:bool=false
   ) -> (then)
   > Branch(Condition=Get EnableCollapsedState() -> (EnableCollapsedState:bool))
   > SetIsCollapsed!(execute, self:WBP UIKit Side Navigation Base Object Reference, IsCollapsed:bool=false, SkipAnimation:bool=false) -> (then)
   Event OnRemovedFromFocusPath!() -> (OutputDelegate, then, InFocusEvent:Focus Event Structure) {is=title K2Node_Event}
   > Branch(Condition=IsMobileGame())
   > FortShippingLog(WorldContextObject=Self-Reference(), InString=Pilgrim_SideNavigation_Menu: On Removed from Focus Path, bLogAsWarning=false)
   > Branch(Condition=Get EnableCollapsedState())
   > SetIsCollapsed(IsCollapsed=true, SkipAnimation=false)
   Event OnFocusLost!() -> (OutputDelegate, then, InFocusEvent:Focus Event Structure) {is=title K2Node_Event}
   > Branch(Condition=IsMobileGame())
   > FortShippingLog(WorldContextObject=Self-Reference(), InString=Pilgrim_SideNavigation_Menu: On Focus Lost, bLogAsWarning=false)
   n2 = Event Event On Input Method Changed!() -> (OutputDelegate, then, bNewInputType:ECommonInputType Enum) {is=title K2Node_Event}
   > Parent: Event On Input Method Changed!(execute, bNewInputType:ECommonInputType Enum=@n2.bNewInputType) -> (then) {is=title K2Node_CallParentFunction}
   n3 = IsUsingCompactUI()
   n4 = IsUsingInputType!(InputType:ECommonInputType Enum=Touch, PlayerController:Player Controller Object Reference=GetOwningPlayer()) -> (ReturnValue:bool)
   > Set EnableCollapsedState(EnableCollapsedState:bool=NOT Boolean(A=AND Boolean(A=@n3, B=@n4))) -> (Output_Get:bool)
   Event Tick()
  ]}]
variables []
widgets {
 info {
  inheritedWidgetCount=0
  parentClass=/FortUILibrary/UIKit/Tabs/WBP_UIKit_SideNavigation.WBP_UIKit_SideNavigation_C
  rootWidgetClass=null
  widgetCount=0
 }
 widgets []
}
