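# Quickplay song select screen where players browse the song catalog, build a custom setlist and view leaderboards.
# Catalog selection drives the leaderboard, buttons and quick-access category tabs, which receive async-created intensity widgets.
# Event-router listeners refresh the setlist preview on changes and close the menu when quickplay leaves Pregame.
# It also opens a sort/filter modal, reports menu state to the server and swaps to an inverse layout for portrait touch.
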WidgetBlueprint /PilgrimQuickplay/Widgets/SetlistAndSongDisplay/Duplicated/WBP_PilgrimQuickplay_SongSelectMenu2
graphs [
 Add New Intensity Widget {nodes [
   Add New Intensity Widget!() -> (then, Index:int) {is=title K2Node_FunctionEntry}
   > Set Side Navigation Button(Side Navigation Button:WBP UIKit Button Generic Base Object Reference="Get (a copy)"(
     Array:Array of WBP UIKit Button Generic Base Object References=Get SideNavigationButtons() -> (SideNavigationButtons:Array of WBP UIKit Button Generic Base Object References)
     Dimension 1=Get Index() -> (Index:int=0)
    ) -> (Output:WBP UIKit Button Generic Base Object Reference)) -> (Output_Get:WBP UIKit Button Generic Base Object Reference)
   > Set Group Intensity VM(Group Intensity VM:Pilgrim Song Category Group Song Intensity VM Object Reference="Get (a copy)"(
     Array:Array of Pilgrim Song Category Group Song Intensity VM Object References=Get GroupIntensityVMs() -> (GroupIntensityVMs:Array of Pilgrim Song Category Group Song Intensity VM Object References)
     Dimension 1=Get Index() -> (Index=0)
    ) -> (Output:Pilgrim Song Category Group Song Intensity VM Object Reference)) -> (Output_Get:Pilgrim Song Category Group Song Intensity VM Object Reference)
   > Set Intensity Entry Widget(Intensity Entry Widget:WBP Pilgrim Quickplay Song Intensity Entry Object Reference="Get (a copy)"(
     Array:Array of WBP Pilgrim Quickplay Song Intensity Entry Object References=Get IntensityEntryWidgets() -> (IntensityEntryWidgets:Array of WBP Pilgrim Quickplay Song Intensity Entry Object References)
     Dimension 1=Get Index() -> (Index=0)
    ) -> (Output:WBP Pilgrim Quickplay Song Intensity Entry Object Reference)) -> (Output_Get:WBP Pilgrim Quickplay Song Intensity Entry Object Reference)
   n1 = Get PreviousSideNavButtonWidgets() -> (PreviousSideNavButtonWidgets:Array of Widget Object References)
   > n2 = Branch(Condition="Equal (Integer)"(A=Get Index() -> (Index=0), B=Length(TargetArray:Array of Widget Object References=@n1)))
   > n3 = Set Slot Main(Slot Main:Named Slot Object Reference=Get Slot_Main!(self:WBP UIKit Button Generic Base Object Reference=Get Side Navigation Button() -> (Side Navigation Button:WBP UIKit Button Generic Base Object Reference)) -> (Slot_Main:Named Slot Object Reference)) -> (Output_Get:Named Slot Object Reference)
   > Add(
    TargetArray:Array of Widget Object References=Get PreviousSideNavButtonWidgets()
    NewItem:"Widget Object Reference (by ref)"=GetContent!(self:Content Widget Object Reference=@n3) -> (ReturnValue:Widget Object Reference)
   ) {is=title K2Node_CallArrayFunction}
   > n4 = Switch on EPilgrimTrackType!(execute, Selection:EPilgrimTrackType Enum=Get Player Instrument Choice() -> (Player Instrument Choice:EPilgrimTrackType Enum=TrackGuitar)) -> (
    TrackGuitar:Exec
    TrackBass:Exec
    TrackVocals:Exec
    TrackDrum:Exec
    TrackPlasticGuitar:Exec
    TrackPlasticBass:Exec
    TrackPlasticDrum:Exec
    TrackProVocals:Exec
    TrackProCymbals:Exec
    TrackEvents:Exec
    TrackNone:Exec
   ) {is=title K2Node_SwitchEnum}
   n5 = Set Starting Instrument(execute=[@n4.TrackBass, @n4.TrackPlasticBass], Starting Instrument:ESparksInstrumentType Enum=Bass) -> (Output_Get:ESparksInstrumentType Enum=Guitar)
   Update Intensity for Instrument!(
    execute=[@n6.then, @n7.then, @n5.then, @n8.then]
    self:WBP Pilgrim Quickplay Song Intensity Entry Object Reference=Get Intensity Entry Widget() -> (Intensity Entry Widget:WBP Pilgrim Quickplay Song Intensity Entry Object Reference)
    Intensity:int=Get BarIntensity!(self:Pilgrim Song Category Group Song Intensity VM Object Reference=Get Group Intensity VM() -> (Group Intensity VM:Pilgrim Song Category Group Song Intensity VM Object Reference)) -> (BarIntensity:int=0)
    TrackType:EPilgrimTrackType Enum=Get Player Instrument Choice() -> (Player Instrument Choice=TrackGuitar)
    StartingInstrument:ESparksInstrumentType Enum=Get Starting Instrument() -> (Starting Instrument:ESparksInstrumentType Enum=Guitar)
    Is3D_deprecated:bool=false
    SongMetadata:Pilgrim Song Metadata Object Reference
   ) -> (then)
   > SetContent!(
    execute
    self:Content Widget Object Reference=Get Slot Main() -> (Slot Main:Named Slot Object Reference)
    Content:Widget Object Reference=Get Intensity Entry Widget()
   ) -> (then, ReturnValue:Panel Slot Object Reference)
   > SetText!(execute, self:WBP UIKit Button Generic Base Object Reference=Get Side Navigation Button(), Text:Text=GetEmptyText!() -> (ReturnValue:Text)) -> (then)
   n6 = Set Starting Instrument(execute=[@n4.TrackDrum, @n4.TrackPlasticDrum, @n4.TrackProCymbals], Starting Instrument=Drum) -> (Output_Get=Guitar)
   n7 = Set Starting Instrument(execute=[@n4.TrackGuitar, @n4.TrackPlasticGuitar], Starting Instrument=Guitar) -> (Output_Get=Guitar)
   n8 = Set Starting Instrument(execute=[@n4.TrackProVocals, @n4.TrackVocals], Starting Instrument=Vocals) -> (Output_Get=Guitar)
   n9 = Get PreviousSideNavButtonWidgets()
   PrintText!(
    execute=@n2.else
    InText=Format Text!(
     Format="INVTEXT(\"ERROR: New Intensity Widget [{index}] mismatch with {widget_length}.\")"
     index:int=Get Index() -> (Index=0)
     widget_length:int=Length(TargetArray=@n9)
    ) -> (Result)
    bPrintToScreen
    bPrintToLog
    TextColor=(R=0.8108,G=0.0024,B=0,A=1)
    Duration=10
    Key
   ) -> (then)
  ]}
 BP_GetDesiredFocusTarget {nodes [
   Get Desired Focus Target()
   > Return Node!(
    execute
    ReturnValue:Widget Object Reference=GetDesiredFocusTarget!(self:Common Activatable Widget Object Reference=Get WBP_PilgrimQuickplay_SongCatalog2() -> (WBP_PilgrimQuickplay_SongCatalog2:WBP Pilgrim Quickplay Song Catalog 2 Object Reference)) -> (ReturnValue:Widget Object Reference)
   )
  ]}
 Clear Intensity Widgets {nodes [
   Clear Intensity Widgets() {is=title K2Node_FunctionEntry}
   > n1 = For Each Loop(Array:Array of WBP UIKit Button Generic Base Object References=Get SideNavigationButtons()) -> (Array Element:WBP UIKit Button Generic Base Object Reference)
   > SetContent(
    self=Get Slot_Main(self=@n1)
    Content="Get (a copy)"(Array:Array of Widget Object References=Get PreviousSideNavButtonWidgets(), Dimension 1=@n1.Array Index) -> (Output:Widget Object Reference)
   )
   Clear(execute=@n1.Completed, TargetArray:Array of WBP UIKit Button Generic Base Object References=Get SideNavigationButtons())
   > Clear(TargetArray:Array of Pilgrim Song Category Group Song Intensity VM Object References=Get GroupIntensityVMs())
   > Clear(TargetArray:Array of WBP Pilgrim Quickplay Song Intensity Entry Object References=Get IntensityEntryWidgets())
   > Clear(TargetArray:Array of Widget Object References=Get PreviousSideNavButtonWidgets())
  ]}
 CreateSortFilterModal {nodes [
   CreateSortFilterModal() {is=title K2Node_FunctionEntry}
   > n1 = Create Widget!(
    execute:Exec
    Class:User Widget Class Reference=/PilgrimQuickplay/Widgets/SetlistAndSongDisplay/SortFilterModal/WBP_Pilgrim_SortFilter_Modal.WBP_Pilgrim_SortFilter_Modal_C
    OwningPlayer:Player Controller Object Reference=GetOwningPlayer()
    FortSortFilterVM:Fort Sort Filter VM Object Reference=Get FortSortFilterVM!(self:Fort Sort Filter Controls Object Reference=Get WBP_PilgrimSortFilter() -> (WBP_PilgrimSortFilter:WBP Pilgrim Sort Filter Object Reference)) -> (FortSortFilterVM:Fort Sort Filter VM Object Reference)
    PilgrimSongCatalogVM:Pilgrim Song Catalog VM Object Reference=Get PilgrimSongCatalogVM() -> (PilgrimSongCatalogVM:Pilgrim Song Catalog VM Object Reference)
   ) -> (then:Exec, ReturnValue:WBP Pilgrim Sort Filter Modal Object Reference)
   > Return Node(ReturnValue:WBP Pilgrim Sort Filter Modal Object Reference=@n1)
  ]}
 EventGraph {nodes [
   "On Item Deselected (WBP_PilgrimQuickplay_SongCatalog2)"() {is=title K2Node_ComponentBoundEvent}
   > SetIsEnabled(
    self=Get WBP_UIKit_Button_Regular_LEADERBOARD() -> (WBP_UIKit_Button_Regular_LEADERBOARD:WBP UIKit Button Regular Object Reference)
    bInIsEnabled=false
   )
   > SetIsEnabled(
    self=[
     Get WBP_UIKit_Button_Regular_SETLIST() -> (WBP_UIKit_Button_Regular_SETLIST:WBP UIKit Button Regular Object Reference)
     Get WBP_UIKit_Button_Loud_SETLIST() -> (WBP_UIKit_Button_Loud_SETLIST:WBP UIKit Button Loud Object Reference)
    ]
    bInIsEnabled=false
   )
   n1 = "On Item Selection Changed (WBP_PilgrimQuickplay_SongCatalog2)"!() -> (OutputDelegate, then, Item:Pilgrim Song Catalog Item VM Object Reference) {is=title K2Node_ComponentBoundEvent}
   > n2 = GetSongMetadata!(execute, self:Pilgrim Song Catalog Item VM Object Reference=@n1.Item) -> (then, ReturnValue:Pilgrim Song Metadata Object Reference)
   > OnMusicLibraryEntrySelected!(
    execute
    self:WBP Pilgrim Quickplay Leaderboard Mini Music Library Object Reference=Get Leaderboard() -> (Leaderboard:WBP Pilgrim Quickplay Leaderboard Mini Music Library Object Reference)
    Song:Pilgrim Song Metadata Object Reference=@n2
   ) -> (then)
   > Update Interaction Text!(execute, self, SongShortname:"Name (by ref)"=GetShortName!(self:Sparks Song Data Object Reference=@n2) -> (ReturnValue:Name=None)) -> (then)
   > SetIsEnabled(self=Get WBP_UIKit_Button_Regular_LEADERBOARD(), bInIsEnabled=true)
   > SetIsEnabled(self=[Get WBP_UIKit_Button_Loud_SETLIST(), Get WBP_UIKit_Button_Regular_SETLIST()], bInIsEnabled=true)
   > n3 = GetIsQuickAccessEnabled!(execute, self) -> (then, IsEnabled:bool)
   > Branch(Condition=@n3)
   > n4 = Cast To PilgrimSongCategoryGroupVM!(
    execute
    Object=GetCategoryGroupVM!(self:Pilgrim Song Catalog Item VM Object Reference=@n1.Item) -> (ReturnValue:Fort Locker Category Group VM Object Reference)
   ) -> (then, CastFailed, AsPilgrim Song Category Group VM:Pilgrim Song Category Group VM Object Reference) {is=title K2Node_DynamicCast}
   > SelectTabByID!(
    execute
    self:Common Tab List Widget Base Object Reference=Get WBP_Pilgrim_QuickAccessMenu() -> (WBP_Pilgrim_QuickAccessMenu:WBP Pilgrim Side Navigation Menu Object Reference)
    TabNameID:Name=GetUniqueName!(self:Fort Locker Category Group VM Object Reference=@n4) -> (ReturnValue:Name=None)
    bSuppressClickFeedback:bool=false
   ) -> (then, ReturnValue:bool)
   "Sort Filter Clicked (WBP_PilgrimSortFilter)"() {is=title K2Node_ComponentBoundEvent}
   > n5 = Sequence()
   > n6 = GetIsQuickAccessEnabled()
   > Branch(Condition=@n6)
   n7 = Get OwningPlayer() -> (OwningPlayer:Player Controller Object Reference)
   > SetIsCollapsed!(
    execute
    self:WBP UIKit Side Navigation Base Object Reference=Get WBP_Pilgrim_QuickAccessMenu()
    IsCollapsed:bool=NOT Boolean(A=IsPilgrimInverse!(OwningPlayerController:Player Controller Object Reference=@n7) -> (IsPilgrimInverse:bool))
    SkipAnimation:bool=false
   ) -> (then)
   n8 = CreateSortFilterModal!(execute=@n5.then_1, self) -> (then, ReturnValue:WBP Pilgrim Sort Filter Modal Object Reference) {is=title K2Node_CallFunction}
   > TryPushWidgetAsModalPanel!(execute, PushWidget:User Widget Object Reference=@n8) -> (then)
   > n9 = ActivateWidget(self=@n8)
   n10 = OnSortFilterModalCloseEvent() {is=title K2Node_CustomEvent}
   Bind Event to On Widget Deactivated(execute=@n9.then, self:Common Activatable Widget Object Reference=@n8, Delegate=@n10) {is=title K2Node_AddDelegate}
   > SetFocus(self=@n8)
   > n11 = SetVisibility(InVisibility=HitTestInvisible)
   n12 = OnSortFilterAccepted() {is=title K2Node_CustomEvent}
   Bind Event to Sort Filter Accepted(execute=@n11.then, self:WBP Pilgrim Sort Filter Modal Object Reference=@n8, Delegate=@n12) {is=title K2Node_AddDelegate}
   n13 = Remove Song From Catalogue Or Setlist!() -> (OutputDelegate, then, IsItemHovered:bool=~) {is=title K2Node_CustomEvent}
   > Unbind all Events from On Clicked!(execute:Exec, self:Common Button Base Object Reference=[Get WBP_UIKit_Button_Regular_SETLIST(), Get WBP_UIKit_Button_Loud_SETLIST()]) -> (then:Exec) {is=title K2Node_ClearDelegate}
   > n14 = Branch(Condition=@n13.IsItemHovered)
   n15 = OnButtonClickSetlist!() -> (OutputDelegate, then, Button:Common Button Base Object Reference) {is=title K2Node_CustomEvent}
   Assign On Clicked!(
    execute:Exec=@n14.then
    self:Common Button Base Object Reference=[Get WBP_UIKit_Button_Loud_SETLIST(), Get WBP_UIKit_Button_Regular_SETLIST()]
    Delegate:"Delegate (by ref)"=@n15
   ) -> (then:Exec) {is=title K2Node_AssignDelegate}
   n16 = ToggleSongInCatalog!() -> (OutputDelegate, then, Button:Common Button Base Object Reference) {is=title K2Node_CustomEvent}
   Bind Event to On Clicked(
    execute=[@n38.then, @n14.else]
    self:Common Button Base Object Reference=[Get WBP_UIKit_Button_Regular_SETLIST(), Get WBP_UIKit_Button_Loud_SETLIST()]
    Delegate=@n16
   ) {is=title K2Node_AddDelegate}
   n17 = OnSideNavigationInnerTabClicked!() -> (OutputDelegate, then, TabID:Name, InnerTabID:Name, TabViewModel:MVVM Base Viewmodel Object Reference) {is=title K2Node_CustomEvent}
   > UpdateSelectedCategoryViewModel!(execute, self, TabNameID:Name=@n17.TabID, TabVM:Object Reference=@n17.TabViewModel) -> (then)
   n18 = OnBackButtonClicked!() -> (OutputDelegate, then, Button:Common Button Base Object Reference) {is=title K2Node_CustomEvent}
   > DeactivateWidget()
   n19 = OnSideNavigationCollapseChanged!() -> (OutputDelegate, then, IsCollapsed:bool=~) {is=title K2Node_CustomEvent}
   > n20 = Branch(Condition=@n19.IsCollapsed)
   ScrollToWidget!(
    execute=@n20.else
    self:WBP UIKit Side Navigation Base Object Reference=Get WBP_Pilgrim_QuickAccessMenu()
    Widget:Widget Object Reference=GetQuickAccessFocusButton!(self) -> (Button:Common Button Base Object Reference)
    IsDelayed:bool=true
   ) -> (then)
   n21 = CreateSongIntensityWidget!() -> (
    OutputDelegate
    then
    Group Itensity VM:"Pilgrim Song Category Group Song Intensity VM Object Reference (by ref)"
    Side Navigation Button:WBP UIKit Button Generic Base Object Reference
   ) {is=title K2Node_CustomEvent}
   > Add(
    TargetArray:Array of Pilgrim Song Category Group Song Intensity VM Object References=Get GroupIntensityVMs()
    NewItem:"Pilgrim Song Category Group Song Intensity VM Object Reference (by ref)"=@n21.Group Itensity VM
   )
   > Add(
    TargetArray:Array of WBP UIKit Button Generic Base Object References=Get SideNavigationButtons()
    NewItem:"WBP UIKit Button Generic Base Object Reference (by ref)"=@n21.Side Navigation Button
   )
   > n22 = CreateWidgetAsync!(
    execute:Exec
    UserWidgetSoftClass:User Widget Soft Class Reference=/PilgrimQuickplay/Widgets/SetlistAndSongDisplay/WBP_PilgrimQuickplay_SongIntensityEntry.WBP_PilgrimQuickplay_SongIntensityEntry_C
    OwningPlayer:Player Controller Object Reference
    bSuspendInputUntilComplete:bool=false
   ) -> (then:Exec, OnComplete:Exec, OnFailure:Exec, UserWidget:User Widget Object Reference) {is=title K2Node_AsyncAction}
   n23 = Cast To WBP_PilgrimQuickplay_SongIntensityEntry!(execute=@n22.OnComplete, Object=@n22) -> (then, CastFailed, AsWBP Pilgrim Quickplay Song Intensity Entry:WBP Pilgrim Quickplay Song Intensity Entry Object Reference) {is=title K2Node_DynamicCast}
   > n24 = Add(
    TargetArray:Array of WBP Pilgrim Quickplay Song Intensity Entry Object References=Get IntensityEntryWidgets()
    NewItem:"WBP Pilgrim Quickplay Song Intensity Entry Object Reference (by ref)"=@n23
   )
   > Add New Intensity Widget!(execute, self, Index:int=@n24) -> (then) {is=title K2Node_CallFunction}
   StartListeningToEvents() {is=title K2Node_CustomEvent}
   > n25 = Get!(execute:Exec) -> (then:Exec, else:Exec, Event Router:Gameplay Event Router Component Object Reference)
   > n26 = StartListeningToEvent!(
    execute:Exec
    Target:Gameplay Event Router Component Object Reference=@n25
    EventType:Script Struct Object Reference=/Script/PilgrimCoreRuntime.PilgrimEvent_BandState_CustomSetlistChanged
    Context:Object Reference
    EventBubblingRule:EEventBubblingRule Enum
   ) -> (
    then:Exec
    AsyncTaskProxy:Async Action Start Listening To Event Object Reference
    OnEventReceived:Exec
    Payload:Pilgrim Event Band State Custom Setlist Changed Structure
   )
   > TrackEvent!(execute, self, NewItem:"Async Action Start Listening To Event Object Reference (by ref)"=@n26) -> (then)
   > n27 = StartListeningToEvent!(execute, Target=@n25, EventType=/Script/PilgrimQPSharedRuntime.PilgrimEvent_QuickplayState_Began, Context, EventBubblingRule) -> (then, AsyncTaskProxy, OnEventReceived, Payload_StateId:EPilgrimQuickplayState Enum)
   > TrackEvent(NewItem=@n27)
   OnQuickplayStateBegan!(execute=@n27.OnEventReceived, self, State:EPilgrimQuickplayState Enum=@n27.Payload_StateId) -> (then)
   OnSetlistChanged(execute=@n26.OnEventReceived)
   StopListeningToEvents() {is=title K2Node_CustomEvent}
   > n28 = For Each Loop(Array:Array of Async Action Start Listening To Event Object References=Get RouterEvents() -> (RouterEvents:Array of Async Action Start Listening To Event Object References)) -> (Array Element:Async Action Start Listening To Event Object Reference)
   Clear(execute=@n28.Completed, TargetArray:Array of Async Action Start Listening To Event Object References=Get RouterEvents())
   StopListeningToEvent(execute=@n28.LoopBody, self:Async Action Start Listening To Event Object Reference=@n28)
   OnSetlistChanged() {is=title K2Node_CustomEvent}
   > UpdateSongsChosenText()
   > UpdateButtonsContext()
   > UpdateSetlistPreview!(
    execute
    self
    CustomSetlist:Pilgrim Setlist Interface Interface=GetCustomSetlist!() -> (ReturnValue:Pilgrim Custom Setlist Object Reference)
    Setlist Preview:List View Object Reference=Get SetlistPreviewList!(self:WBP Pilgrim Quickplay Setlist Preview List Object Reference=Get WBP_PilgrimQuickplay_SetlistPreviewList() -> (WBP_PilgrimQuickplay_SetlistPreviewList:WBP Pilgrim Quickplay Setlist Preview List Object Reference)) -> (SetlistPreviewList:Common List View Object Reference)
   ) -> (then)
   > Update Interaction Text(SongShortname=GetSelectedSongShortName!(self:WBP Pilgrim Quickplay Song Catalog 2 Object Reference=Get WBP_PilgrimQuickplay_SongCatalog2()) -> (ShortName:Name=None))
   n29 = OnQuickplayStateBegan!() -> (OutputDelegate, then, State:EPilgrimQuickplayState Enum=Invalid) {is=title K2Node_CustomEvent}
   > Branch(Condition="Not Equal (Enum)"!(A:EPilgrimQuickplayState Enum=@n29.State, B:EPilgrimQuickplayState Enum=Pregame) -> (ReturnValue:bool))
   > DeactivateWidget()
   n30 = OnSideNavigationTabClicked!() -> (OutputDelegate, then, TabID:Name, IsTabContainerExpanded:bool=~, TabViewModel:MVVM Base Viewmodel Object Reference) {is=title K2Node_CustomEvent}
   > UpdateSelectedCategoryViewModel(TabNameID=@n30.TabID, TabVM=@n30.TabViewModel)
   > Fire Event Element Interacted(Element=Get WBP_Pilgrim_QuickAccessMenu())
   n31 = On Update Text button from Setlist!() -> (OutputDelegate, then, Song Meta Data:Pilgrim Song Metadata Object Reference) {is=title K2Node_CustomEvent}
   > n32 = GetSelectedSongShortName!(execute:Exec, self=Get WBP_PilgrimQuickplay_SongCatalog2()) -> (then:Exec, ShortName=None)
   > Branch(Condition="Equal (Name)"(A=@n32, B=GetShortName(self=@n31.Song Meta Data) -> (ReturnValue=None)))
   > Update Interaction Text(SongShortname=@n32)
   n33 = OnLeaderboardButtonClicked!() -> (OutputDelegate, then, Button:Common Button Base Object Reference) {is=title K2Node_CustomEvent}
   > Leaderboard Selected(self:WBP Pilgrim Quickplay Song Catalog 2 Object Reference=Get WBP_PilgrimQuickplay_SongCatalog2())
   RequestBeatenByScoreData() {is=title K2Node_CustomEvent}
   > Reroute Node()
   n34 = OnSideNavigationSelectionChanged!() -> (OutputDelegate, then, TabID:Name, InnerTabID:Name, TabViewModel:MVVM Base Viewmodel Object Reference) {is=title K2Node_CustomEvent}
   > UpdateSelectedCategoryViewModel(TabNameID=@n34.TabID, TabVM=@n34.TabViewModel)
   Event Construct()
   > n35 = Sequence!(execute) -> (then_0, then_1, then_2:Exec, then_3:Exec, then_4:Exec, then_5:Exec)
   > n36 = Set OwningPlayer(OwningPlayer:Player Controller Object Reference=GetOwningPlayer()) -> (Output_Get:Player Controller Object Reference)
   > Initialize Pilgrim Song Catalog Filter VM!(execute, self, In Owning Controller:Player Controller Object Reference=@n36) -> (then)
   > OnSetlistChanged() {is=title K2Node_CallFunction}
   > UpdateSideNavigationBindingState!(execute, self, ShouldAddBindings:bool=false) -> (then)
   > InitializeSetlistPreview()
   > RequestBeatenByScoreData() {is=title K2Node_CallFunction}
   SetPlayerLoadout(execute=@n35.then_1)
   > SetupQuickAccessBar()
   > SetNavigationRuleBase!(
    execute
    self:Widget Object Reference=Get EditableText_Search!(self:Fort Sort Filter Controls Object Reference=Get WBP_PilgrimSortFilter()) -> (EditableText_Search:Editable Text Object Reference)
    Direction:EUINavigation Enum=Right
    Rule:EUINavigationRule Enum=Stop
   ) -> (then)
   > SetupForPilgrimInverse!(execute, self, IsPilgrimInverse:bool=IsPilgrimInverse(OwningPlayerController=Get OwningPlayer())) -> (then)
   n37 = PilgrimWaitForEventRouter!(execute:Exec=@n35.then_5, ContextActor:Actor Object Reference=Get OwningPlayer()) -> (then:Exec, OnEventRouterAvailable:Exec, EventRouter:Gameplay Event Router Component Object Reference) {is=title K2Node_AsyncAction}
   Set Event Router(execute=@n37.OnEventRouterAvailable, Event Router:Gameplay Event Router Component Object Reference=@n37) -> (Output_Get:Gameplay Event Router Component Object Reference)
   > StartListeningToEvents() {is=title K2Node_CallFunction}
   Bind Event to On Remove Song from Setlist(
    execute=@n35.then_4
    self:WBP Pilgrim Quickplay Setlist Preview List Object Reference=Get WBP_PilgrimQuickplay_SetlistPreviewList()
    Delegate=@n31
   ) {is=title K2Node_AddDelegate}
   n38 = Bind Event to On Entry Hovered(
    execute=@n35.then_3
    self:WBP Pilgrim Quickplay Setlist Preview List Object Reference=Get WBP_PilgrimQuickplay_SetlistPreviewList()
    Delegate=@n13
   ) {is=title K2Node_AddDelegate}
   Assign On Clicked(execute=@n35.then_2, self=Get WBP_UIKit_Button_Regular_LEADERBOARD(), Delegate=@n33)
   > Assign On Clicked(
    self=[
     Get WBP_UIKit_Button_Loud_BACK_DONE() -> (WBP_UIKit_Button_Loud_BACK_DONE:WBP UIKit Button Loud Object Reference)
     Get WBP_UIKit_Button_Regular_BACK() -> (WBP_UIKit_Button_Regular_BACK:WBP UIKit Button Regular Object Reference)
    ]
    Delegate=@n18
   )
   Event On Deactivated()
   > Parent: On Deactivated() {is=title K2Node_CallParentFunction}
   > ClearQueuedLeaderboardRequests!(execute, ContextObject:Object Reference=Self-Reference()) -> (then)
   n39 = GetOwningPlayer()
   > StopCurrentPreview!(
    execute
    self:Sparks Song Previewer Component Object Reference=Get SongPreviewer!(self:BP Pilgrim Quickplay Playspace Object Reference=GetSparksMusicPlayspaceOfClass!(
      PlayspaceType:Sparks Music Playspace Class Reference=/PilgrimQuickplay/Playspaces/BP_PilgrimQuickplayPlayspace.BP_PilgrimQuickplayPlayspace_C
      Actor:Actor Object Reference=@n39
      bMatchChildTypes:bool=false
     ) -> (ReturnValue:BP Pilgrim Quickplay Playspace Object Reference)) -> (SongPreviewer:BPC Sparks Song Previewer Object Reference)
    bWithFadeout:bool=false
   ) -> (then)
   n40 = GetOwningPlayer()
   > ExecuteConsoleCommand!(execute, Command=log LogElectraPlayer log, SpecificPlayer=@n40) -> (then)
   > ExecuteConsoleCommand(Command=log LogElectraPlayerPlugin log, SpecificPlayer=@n40)
   > OnDeactivatePilgrimSongCatalogVM()
   > SendMenuStateToServer!(execute, self, IsOpen:bool=false) -> (then)
   > Broadcast Event!(
    execute
    self:Gameplay Event Router Component Object Reference=Get Event Router() -> (Event Router:Gameplay Event Router Component Object Reference)
    EventData:Pilgrim Event Playlist Display Closed Structure=Make Pilgrim Event Playlist Display Closed!(SongSelectionWidget_24_#1:Widget Object Reference=Self-Reference()) -> (PilgrimEvent_Playlist_Display_Closed:Pilgrim Event Playlist Display Closed Structure) {is=title K2Node_MakeStruct}
    EventContext:Object Reference
   ) -> (then, ReturnValue:bool)
   Event On Activated()
   > n41 = Sequence!(execute) -> (then_0, then_1)
   > Fire Event Main View Opened!(execute, View:Widget Object Reference=Self-Reference(), AdditionalInfo:Flow Analytics Additional Info Structure, bIsOverlay:bool=false) -> (then)
   n42 = GetOwningPlayer()
   > ExecuteConsoleCommand(Command=log LogElectraPlayer warning, SpecificPlayer=@n42)
   > ExecuteConsoleCommand(Command=log LogElectraPlayerPlugin warning, SpecificPlayer=@n42)
   > SendMenuStateToServer(IsOpen=true)
   n43 = Branch(execute=@n41.then_1, Condition=IsAtBackstageArea!(ContextActor:Actor Object Reference=Get OwningPlayer()) -> (ReturnValue:bool))
   > SetKeyboardFocus(self:Widget Object Reference=GetDesiredFocusTarget())
   DeactivateWidget(execute=@n43.else)
   Event PreConstruct()
   > SetupForPilgrimInverse(IsPilgrimInverse=Get PreviewPilgrimInverse() -> (PreviewPilgrimInverse:bool))
   Event Destruct()
   > UpdateSideNavigationBindingState(ShouldAddBindings=false)
   > Clear Intensity Widgets() {is=title K2Node_CallFunction}
   > StopListeningToEvents() {is=title K2Node_CallFunction}
   n44 = Find(
    TargetMap:Map of Gameplay Tag Structures to BP Song Entry Horizontal Item Obj Object References=Get SongListItemMap() -> (SongListItemMap:Map of Gameplay Tag Structures to BP Song Entry Horizontal Item Obj Object References)
    Key:"Gameplay Tag Structure (by ref)"
   ) -> (Value:BP Song Entry Horizontal Item Obj Object Reference)
   n45 = Branch(Condition=@n44.ReturnValue)
   > n46 = GetScoreWasBeatenText!(execute, self) -> (then, ScoreWasBeatenText:Text)
   > Set ScoreWasBeatenText!(execute, ScoreWasBeatenText:Text=@n46, self:BP Song Entry Horizontal Item Obj Object Reference=@n44) -> (then, Output_Get:Text)
   > Call On List Item Object Updated!(execute:Exec, self:BP Song Entry Horizontal Item Obj Object Reference=@n44, CantAddSongReason:EPilgrimCantAddSongReason Enum=None) -> (then:Exec) {is=title K2Node_CallDelegate}
   PrintString!(
    execute=@n45.else
    InString=Append!(
     A="Warning: Couldn't find song "
     B=GetDebugStringFromGameplayTag!(GameplayTag:Gameplay Tag Structure) -> (ReturnValue:String)
     C:String=" in music library item list"
    ) -> (ReturnValue)
    bPrintToScreen
    bPrintToLog
    TextColor=(R=0.1563,G=1,B=0,A=1)
    Duration=10
    Key
   ) -> (then)
   UpdateButtonsContext(execute=@n47.then)
   UpdateSongCatalogItems(execute=@n12.then, self:Pilgrim Song Catalog VM Object Reference=Get PilgrimSongCatalogVM())
   UpdateButtonsContext(execute=@n49.then)
   Unbind Event from Sort Filter Accepted(execute=@n48.then, self:WBP Pilgrim Sort Filter Modal Object Reference=@n8, Delegate=@n12) {is=title K2Node_RemoveDelegate}
   n47 = RemoveSongFromSetlist(execute=@n15.then, self:WBP Pilgrim Quickplay Setlist Preview List Object Reference=Get WBP_PilgrimQuickplay_SetlistPreviewList())
   SetVisibility(execute=@n10.then, InVisibility=Visible)
   > n48 = Unbind Event from On Widget Deactivated(self:Common Activatable Widget Object Reference, Delegate=@n10) {is=title K2Node_RemoveDelegate}
   n49 = ToggleItemSelection(execute=@n16.then, self:WBP Pilgrim Quickplay Song Catalog 2 Object Reference=Get WBP_PilgrimQuickplay_SongCatalog2())
  ]}
 FocusQuickAccess {nodes [
   FocusQuickAccess!() -> (then, Navigation:EUINavigation Enum) {is=title K2Node_FunctionEntry}
   > Return Node(ReturnValue:Widget Object Reference=GetQuickAccessFocusButton())
  ]}
 FocusSetList {nodes [
   FocusSetList!() -> (then, Navigation:EUINavigation Enum) {is=title K2Node_FunctionEntry}
   > Return Node(ReturnValue=GetDesiredFocusTarget(self=Get WBP_PilgrimQuickplay_SetlistPreviewList()))
  ]}
 FocusSongCatalog {nodes [
   FocusSongCatalog!() -> (then, Navigation:EUINavigation Enum) {is=title K2Node_FunctionEntry}
   > Return Node(ReturnValue=GetDesiredFocusTarget(self=Get WBP_PilgrimQuickplay_SongCatalog2()))
  ]}
 Get Is All Time Highscore {nodes [
   Get Is All Time Highscore() {is=title K2Node_FunctionEntry}
   n1 = Get PilgrimSongCatalogVM()
   > n2 = Is Valid(InputObject=@n1)
   n3 = Get ActiveShowScoreByItem!(self:Pilgrim Song Catalog VM Object Reference=@n1) -> (ActiveShowScoreByItem:Pilgrim Song Catalog Item Secondary Filter Show Score By Item VM Object Reference)
   > n4 = Is Valid(InputObject=@n3)
   Return Node(execute=[@n4.Is Not Valid, @n2.Is Not Valid], ReturnValue:bool=false)
   > Return Node(ReturnValue=Get IsAllTimeHighscore!(self:Pilgrim Song Catalog Item Secondary Filter Show Score By Item VM Object Reference=@n3) -> (IsAllTimeHighscore:bool))
  ]}
 GetIsQuickAccessEnabled {nodes [
   GetIsQuickAccessEnabled() {is=title K2Node_FunctionEntry}
   > n1 = GetConsoleVariableBoolValue!(execute, VariableName:String=Pilgrim.SidebarCompatibleWidget.bEnableQuickAccessBar) -> (then, ReturnValue:bool)
   > Return Node!(execute, IsEnabled:bool=@n1)
  ]}
 GetQuickAccessFocusButton {nodes [
   GetQuickAccessFocusButton() {is=title K2Node_FunctionEntry}
   n1 = Get WBP_Pilgrim_QuickAccessMenu()
   > Return Node!(
    execute
    Button:Common Button Base Object Reference=GetTabButtonBaseByID!(
     self:Common Tab List Widget Base Object Reference=@n1
     TabNameID:Name=GetSelectedTabId!(self:Common Tab List Widget Base Object Reference=@n1) -> (ReturnValue:Name=None)
    ) -> (ReturnValue:Common Button Base Object Reference)
   )
  ]}
 GetScoreWasBeatenText {nodes [
   GetScoreWasBeatenText() {is=title K2Node_FunctionEntry}
   > Return Node!(execute, ScoreWasBeatenText:Text=MakeLiteralText(Value="LOCTEXT(\"{PLAYER} and {NUM} others beat your score on {INSTRUMENTS}\")"))
  ]}
 Initialize Pilgrim Song Catalog Filter VM {nodes [
   n1 = Initialize Pilgrim Song Catalog Filter VM!() -> (then, In Owning Controller:Player Controller Object Reference) {is=title K2Node_FunctionEntry}
   n2 = MakeLiteralGameplayTagContainer!(Value:Gameplay Tag Container Structure) -> (ReturnValue:Gameplay Tag Container Structure)
   > Initialize!(
    execute
    self:Pilgrim Song Catalog VM Object Reference=Get PilgrimSongCatalogVM()
    InOwningController:Player Controller Object Reference=@n1
    InSortTags:"Gameplay Tag Container Structure (by ref)"=@n2
    InFilterTags:"Gameplay Tag Container Structure (by ref)"=@n2
   ) -> (then)
   > Start Analytics Tracking(self:BP Pilgrim Custom Setlist Editor Component Object Reference=GetComponentByClass!(
     self:Actor Object Reference=@n1
     ComponentClass:Actor Component Class Reference=/PilgrimCore/Player/Components/BP_PilgrimCustomSetlistEditorComponent.BP_PilgrimCustomSetlistEditorComponent_C
    ) -> (ReturnValue:BP Pilgrim Custom Setlist Editor Component Object Reference))
  ]}
 InitializeSetlistPreview {nodes [
   InitializeSetlistPreview() {is=title K2Node_FunctionEntry}
   > n1 = For Loop(FirstIndex=1, LastIndex=Get DefaultSetlistLength() -> (DefaultSetlistLength:int=0))
   > n2 = Construct Object from Class!(
    execute
    Class=/PilgrimQuickplay/Widgets/SetlistAndSongDisplay/BP_SongEntryHorizontal_ItemObj.BP_SongEntryHorizontal_ItemObj_C
    self
    SongMetadata:Pilgrim Song Metadata Object Reference
    DisplayStatus:ESongDisplayStatus Enum=NewEnumerator2
    SongAddedByPlayerState:Player State Object Reference
    ListNumber:int=@n1
    IsAllTimeHighscore:bool=Get Is All Time Highscore!() -> (ReturnValue:bool) {is=title K2Node_CallFunction}
   ) -> (then, ReturnValue:BP Song Entry Horizontal Item Obj Object Reference)
   n3 = Get SetlistPreviewList(self=Get WBP_PilgrimQuickplay_SetlistPreviewList())
   > AddItem!(execute, self:List View Object Reference=@n3, Item:Object Reference=@n2) -> (then)
   UpdateSetlistPreview(execute=@n1.Completed, CustomSetlist=GetCustomSetlist(), Setlist Preview=@n3)
  ]}
 OnDeactivatePilgrimSongCatalogVM {nodes [
   OnDeactivatePilgrimSongCatalogVM() {is=title K2Node_FunctionEntry}
   > Stop Analytics Tracking(self:BP Pilgrim Custom Setlist Editor Component Object Reference=GetComponentByClass(
     self=GetOwningPlayer()
     ComponentClass=&BP_PilgrimCustomSetlistEditorComponent_C
    ))
  ]}
 OnMenuClosed {nodes [OnMenuClosed() {is=title K2Node_FunctionEntry}]}
 PopulateCategories {nodes [
   PopulateCategories!() -> (then, CategoryGroupCollections:Array of Pilgrim Song Category Group Collection VM Object References) {is=title K2Node_FunctionEntry}
   > Clear Intensity Widgets()
   > ClearSideNavigationTabs(self:WBP UIKit Side Navigation Base Object Reference=Get WBP_Pilgrim_QuickAccessMenu())
   > n1 = GetSongCategoryGroups!(execute, self:Pilgrim Song Catalog VM Object Reference=Get PilgrimSongCatalogVM()) -> (then, ReturnValue:Array of Pilgrim Song Category Group VM Object References)
   > n2 = For Each Loop(Array:Array of Pilgrim Song Category Group VM Object References=@n1) -> (Array Element:Pilgrim Song Category Group VM Object Reference)
   > Is Valid(InputObject=@n2)
   n3 = GetTabId!(self:Pilgrim Song Category Group VM Object Reference=@n2) -> (ReturnValue:Name=None)
   > AddSideNavigationTab!(
    execute
    self:WBP UIKit Side Navigation Base Object Reference=Get WBP_Pilgrim_QuickAccessMenu()
    TabViewModel:MVVM Base Viewmodel Object Reference=@n2
    TabID:Name=@n3
    Index:int=-1
   ) -> (then)
   > n4 = Cast To WBP_UIKit_ButtonGeneric_Base!(execute, Object=GetTabButtonBaseByID(self=Get WBP_Pilgrim_QuickAccessMenu(), TabNameID=@n3)) -> (then, CastFailed, AsWBP UIKit Button Generic Base:WBP UIKit Button Generic Base Object Reference) {is=title K2Node_DynamicCast}
   > n5 = Cast To PilgrimSongCategoryGroup_SongIntensityVM!(execute, Object=@n2) -> (then, CastFailed, AsPilgrim Song Category Group Song Intensity VM:Pilgrim Song Category Group Song Intensity VM Object Reference) {is=title K2Node_DynamicCast}
   > CreateSongIntensityWidget!(
    execute
    self
    Group Itensity VM:"Pilgrim Song Category Group Song Intensity VM Object Reference (by ref)"=@n5
    Side Navigation Button:WBP UIKit Button Generic Base Object Reference=@n4
   ) -> (then) {is=title K2Node_CallFunction}
  ]}
 Refresh Focus {nodes [
   Refresh Focus() {is=title K2Node_FunctionEntry}
   > CenterWidget!(execute, ToCenter:Widget Object Reference=GetDesiredFocusTarget()) -> (then)
  ]}
 SendMenuStateToServer {nodes [
   n1 = SendMenuStateToServer!() -> (then, IsOpen:bool=~) {is=title K2Node_FunctionEntry}
   > n2 = GetPilgrimQuickplayPlayerStateComponent!(
    execute
    PlayerState:Player State Object Reference=Get PlayerState!(self:Controller Object Reference=GetOwningPlayer()) -> (PlayerState:Player State Object Reference)
   ) -> (then, ReturnValue:Pilgrim Quickplay Player State Component Object Reference)
   > Branch(Condition=IsValid(Object=@n2))
   > Server_SetSongSelectMenuStateChanged!(execute, self:Pilgrim Quickplay Player State Component Object Reference=@n2, HasMenuOpen:bool=@n1) -> (then)
  ]}
 SetPlayerLoadout {nodes [
   SetPlayerLoadout() {is=title K2Node_FunctionEntry}
   > n1 = Cast To FortPlayerState!(execute, Object=Get PlayerState(self=Get OwningPlayer())) -> (then, CastFailed, AsFort Player State:Fort Player State Object Reference) {is=title K2Node_DynamicCast}
   > SetAvatarIconImg!(
    execute
    self:WBP Pilgrim Quickplay Player Loadout Object Reference=Get PlayerLoadout() -> (PlayerLoadout:WBP Pilgrim Quickplay Player Loadout Object Reference)
    PlayerState:Fort Player State Object Reference=@n1
   ) -> (then)
  ]}
 SetupForPilgrimInverse {nodes [
   n1 = GetConsoleVariableBoolValue(VariableName=Fort.Rollback.ShouldForceBackstageToPortrait)
   n2 = GetFirstLocalFortPlayerController!(bIncludeNonSpectators:bool=true, bIncludeSpectators:bool=false) -> (ReturnValue:Fort Player Controller Object Reference)
   > Set IsPilgrimInverse_OldForPortrait(IsPilgrimInverse_OldForPortrait:bool=AND Boolean(
     A=@n1
     B=IsUsingInputType!(InputType:ECommonInputType Enum=Touch, PlayerController:Player Controller Object Reference=@n2) -> (ReturnValue:bool)
    )) -> (Output_Get:bool)
   > n3 = Sequence()
   > Set CachedSetlistPreviewList(CachedSetlistPreviewList:WBP Pilgrim Quickplay Setlist Preview List Object Reference=Select(
     Option 0:WBP Pilgrim Quickplay Setlist Preview List Object Reference=Get WBP_PilgrimQuickplay_SetlistPreviewList()
     Option 1:WBP Pilgrim Quickplay Setlist Preview List Object Reference=Get WBP_PilgrimQuickplay_SetlistPreviewList_Inverse() -> (WBP_PilgrimQuickplay_SetlistPreviewList_Inverse:WBP Pilgrim Quickplay Setlist Preview List Object Reference)
     Index=Get IsPilgrimInverse_OldForPortrait() -> (IsPilgrimInverse_OldForPortrait:bool)
    ) -> (ReturnValue:WBP Pilgrim Quickplay Setlist Preview List Object Reference)) -> (Output_Get:WBP Pilgrim Quickplay Setlist Preview List Object Reference)
   > Set CachedSetlistAddButton(CachedSetlistAddButton:WBP UIKit Button CTA Base Object Reference=Select(
     Option 0:WBP UIKit Button CTA Base Object Reference=Get WBP_UIKit_Button_Loud_SETLIST()
     Option 1:WBP UIKit Button CTA Base Object Reference=Get WBP_UIKit_Button_Loud_SETLIST_Inverse() -> (WBP_UIKit_Button_Loud_SETLIST_Inverse:WBP UIKit Button Loud Object Reference)
     Index=Get IsPilgrimInverse_OldForPortrait()
    ) -> (ReturnValue:WBP UIKit Button CTA Base Object Reference)) -> (Output_Get:WBP UIKit Button CTA Base Object Reference)
   > Set CachedLeaderboardsButton(CachedLeaderboardsButton:WBP UIKit Button CTA Base Object Reference=Select(
     Option 0=Get WBP_UIKit_Button_Regular_LEADERBOARD()
     Option 1=Get WBP_UIKit_Button_Regular_LEADERBOARD_Inverse() -> (WBP_UIKit_Button_Regular_LEADERBOARD_Inverse:WBP UIKit Button Regular Object Reference)
     Index=Get IsPilgrimInverse_OldForPortrait()
    )) -> (Output_Get:WBP UIKit Button CTA Base Object Reference)
   > Set CachedBackButton(CachedBackButton:WBP UIKit Button CTA Base Object Reference=Select(
     Option 0=Get WBP_UIKit_Button_Regular_BACK()
     Option 1=Get WBP_UIKit_Button_Regular_BACK_Inverse() -> (WBP_UIKit_Button_Regular_BACK_Inverse:WBP UIKit Button Regular Object Reference)
     Index=Get IsPilgrimInverse_OldForPortrait()
    )) -> (Output_Get:WBP UIKit Button CTA Base Object Reference)
   SetSize!(
    execute=@n3.then_1
    self:Horizontal Box Slot Object Reference=SlotAsHorizontalBoxSlot!(Widget:Widget Object Reference=Get LeaderboardsVerticalBox() -> (LeaderboardsVerticalBox:Vertical Box Object Reference)) -> (ReturnValue:Horizontal Box Slot Object Reference)
    InSize_Value:float=1
    InSize_SizeRule:ESlateSizeRule Enum=Select(Option 0:ESlateSizeRule Enum=Fill, Option 1:ESlateSizeRule Enum=Automatic, Index=Get IsPilgrimInverse_OldForPortrait()) -> (ReturnValue:ESlateSizeRule Enum)
   ) -> (then)
   > SetVisibility(
    self=Get LeaderboardsVerticalBox()
    InVisibility=Select(Option 0:ESlateVisibility Enum=SelfHitTestInvisible, Option 1:ESlateVisibility Enum=Collapsed, Index=Get IsPilgrimInverse_OldForPortrait()) -> (ReturnValue:ESlateVisibility Enum)
   )
   > SetVisibility(
    self=Get Inverse_DetailsPanel() -> (Inverse_DetailsPanel:Vertical Box Object Reference)
    InVisibility=Select(Option 0=Collapsed, Option 1=SelfHitTestInvisible, Index=Get IsPilgrimInverse_OldForPortrait())
   )
   > SetGridSlotPadding!(
    execute
    Widget:Widget Object Reference=Get VerticalBox_SongList() -> (VerticalBox_SongList:Vertical Box Object Reference)
    Padding_Left:float=0
    Padding_Top:float=0
    Padding_Right:float=Select(Option 0:double=96, Option 1:double=48, Index=Get IsPilgrimInverse_OldForPortrait()) -> (ReturnValue:double)
    Padding_Bottom:float=0
   ) -> (then)
   n4 = SetupForPilgrimInverse!() -> (then, IsPilgrimInverse:bool=~) {is=title K2Node_FunctionEntry}
   > SetMinDesiredWidth!(
    execute
    self:Size Box Object Reference=Get SZ_LeaderboardBtn() -> (SZ_LeaderboardBtn:Size Box Object Reference)
    InMinDesiredWidth:float=SelectFloat(A=440, B=0, bPickA=@n4)
   ) -> (then)
   > SetMaxDesiredWidth!(
    execute
    self:Size Box Object Reference=Get QuickAccessBarContainer() -> (QuickAccessBarContainer:Size Box Object Reference)
    InMaxDesiredWidth:float=SelectFloat(A=300, B=248, bPickA=@n4)
   ) -> (then)
  ]}
 SetupQuickAccessBar {nodes [
   SetupQuickAccessBar() {is=title K2Node_FunctionEntry}
   > n1 = GetIsQuickAccessEnabled() {is=title K2Node_CallFunction}
   > n2 = Branch(Condition=@n1)
   > SetVisibility(self=Get QuickAccessBarContainer(), InVisibility=SelfHitTestInvisible)
   > SetScrollbarPadding!(
    execute
    self:Scroll Box Object Reference=Get ScrollBox_Tabs!(self:WBP UIKit Tab Group Base Object Reference=Get WBP_Pilgrim_QuickAccessMenu()) -> (ScrollBox_Tabs:Scroll Box Object Reference)
    NewScrollbarPadding_Left:float=0
    NewScrollbarPadding_Top:float=12
    NewScrollbarPadding_Right:float=12
    NewScrollbarPadding_Bottom:float=12
   ) -> (then)
   n3 = GetOwningPlayer()
   > SetIsCollapsed(self=Get WBP_Pilgrim_QuickAccessMenu(), IsCollapsed=NOT Boolean(A=IsPilgrimInverse(OwningPlayerController=@n3)), SkipAnimation=true)
   > UpdateSideNavigationBindingState(ShouldAddBindings=true)
   > SetNavigationRuleCustom!(
    execute
    self:Widget Object Reference=Get WBP_PilgrimQuickplay_SongCatalog2()
    Direction:EUINavigation Enum=Left
    InCustomDelegate:Delegate=Create Event()
   ) -> (then)
   > SetNavigationRuleCustom(self=Get WBP_Pilgrim_QuickAccessMenu(), Direction=Right, InCustomDelegate=Create Event())
   SetVisibility(execute=@n2.else, self=Get QuickAccessBarContainer(), InVisibility=Collapsed)
  ]}
 TrackEvent {nodes [
   n1 = TrackEvent!() -> (then, NewItem:"Async Action Start Listening To Event Object Reference (by ref)") {is=title K2Node_FunctionEntry}
   > Add(
    TargetArray:Array of Async Action Start Listening To Event Object References=Get RouterEvents()
    NewItem:"Async Action Start Listening To Event Object Reference (by ref)"=@n1
   )
  ]}
 Update Interaction Text {nodes [
   n1 = Update Interaction Text!() -> (then, SongShortname:"Name (by ref)") {is=title K2Node_FunctionEntry}
   n2 = GetCustomSetlist()
   n3 = Get PlayerState(self=GetOwningPlayer())
   > SetText!(
    execute
    self:Fort CTAButton Object Reference=[Get WBP_UIKit_Button_Regular_SETLIST(), Get WBP_UIKit_Button_Loud_SETLIST()]
    ButtonText:"Text (by ref)"=Select(
     Option 0:"Text (by ref)"=MakeLiteralText(Value="LOCTEXT(\"Add Song\")")
     Option 1:"Text (by ref)"=MakeLiteralText(Value="LOCTEXT(\"Remove Song\")")
     Index=HasPlayerAddedSong!(self:Pilgrim Custom Setlist Object Reference=@n2, SongShortName:Name=@n1, Player:Player State Object Reference=@n3) -> (ReturnValue:bool)
    ) -> (ReturnValue:"Text (by ref)")
   ) -> (then)
  ]}
 UpdateButtonsContext {nodes [
   n1 = GetWidgetStyleTags!(execute, InWidget:Widget Object Reference=Get WBP_UIKit_Button_Loud_SETLIST()) -> (then, ReturnValue:Array of Gameplay Tag Structures)
   > n2 = Sequence()
   > n3 = For Each Loop(Array:Array of Gameplay Tag Structures=@n1) -> (Array Element:Gameplay Tag Structure)
   > PrintText(InText=Format Text!(Format="LOCTEXT(\"MURILOG - CACHED ADD BUTTON STYLE TAG: {tag}\")", tag:String=GetDebugStringFromGameplayTag(GameplayTag=@n3)) -> (Result))
   n4 = UpdateWidgetStyleTags!(
    execute=@n2.then_1
    InWidget:Widget Object Reference=Get WBP_UIKit_Button_Loud_SETLIST()
    InAddedTags:Array of Gameplay Tag Structures=Make Array("[0]"=(TagName="StyleSheet.Style.Regular"))
    InRemovedTags:Array of Gameplay Tag Structures=Make Array("[0]"=(TagName="StyleSheet.Style.Loud"))
   ) -> (then, ReturnValue:bool)
   > PrintString(InString=Append!(A="MURILOG UPDATED STYLES? ", B="To String (Boolean)"(InBool=@n4)) -> (ReturnValue))
   > n5 = GetWidgetStyleTags(InWidget=Get CachedBackButton() -> (CachedBackButton:WBP UIKit Button CTA Base Object Reference))
   > n6 = For Each Loop(Array=@n5)
   > PrintText(InText=Format Text(Format="LOCTEXT(\"MURILOG - CACHED BACK BUTTON STYLE TAG: {tag}\")", tag=GetDebugStringFromGameplayTag(GameplayTag=@n6)))
   UpdateButtonsContext() {is=title K2Node_FunctionEntry}
   n7 = GetCustomSetlist()
   n8 = "To Integer (Boolean)"!(InBool:bool="integer > integer"(A=GetNumSongs!(self:Pilgrim Setlist Interface Interface=@n7) -> (ReturnValue:int=0))) -> (ReturnValue:int=0)
   > SetActiveWidgetIndex(
    self:Common Visibility Switcher Object Reference=Get AddOrRemoveSong_ContextualSwitcher() -> (AddOrRemoveSong_ContextualSwitcher:Common Visibility Switcher Object Reference)
    Index=@n8
   )
   > SetActiveWidgetIndex(self=Get BackOrDone_ContextualSwitcher() -> (BackOrDone_ContextualSwitcher:Common Visibility Switcher Object Reference), Index=@n8)
  ]}
 UpdateSelectedCategoryViewModel {nodes [
   n1 = UpdateSelectedCategoryViewModel!() -> (then, TabNameID:Name, TabVM:Object Reference) {is=title K2Node_FunctionEntry}
   > n2 = Cast To PilgrimSongCategoryGroupVM(Object=@n1.TabVM)
   > Branch(Condition=HasUserFocusedDescendants!(self:Widget Object Reference=Get WBP_Pilgrim_QuickAccessMenu(), PlayerController:Player Controller Object Reference=GetOwningPlayer()) -> (ReturnValue:bool))
   n3 = Get CategoryItems!(self:Fort Locker Category Group VM Object Reference=@n2) -> (CategoryItems:Array of Fort Locker Category Item VM Object References)
   > NavigateToSongItem!(
    execute
    self:WBP Pilgrim Quickplay Song Catalog 2 Object Reference=Get WBP_PilgrimQuickplay_SongCatalog2()
    Item:Object Reference=Get EquippedItem!(self:Fort Locker Category Item VM Object Reference="Get (a copy)"(Array:Array of Fort Locker Category Item VM Object References=@n3) -> (Output:Fort Locker Category Item VM Object Reference)) -> (EquippedItem:Fort Item VM Object Reference)
   ) -> (then)
  ]}
 UpdateSetlistPreview {nodes [
   n1 = UpdateSetlistPreview!() -> (then, CustomSetlist:Pilgrim Setlist Interface Interface, Setlist Preview:List View Object Reference) {is=title K2Node_FunctionEntry}
   > n2 = Cast To PilgrimCustomSetlist!(execute, Object:Pilgrim Setlist Interface Interface=@n1) -> (then, CastFailed, AsPilgrim Custom Setlist:Pilgrim Custom Setlist Object Reference) {is=title K2Node_DynamicCast}
   > n3 = GetListItems!(execute, self:List View Object Reference=@n1.Setlist Preview) -> (then, ReturnValue:Array of Object References)
   > n4 = For Each Loop(Array:Array of Object References=@n3) -> (Array Element:Object Reference)
   > n5 = Cast To BP_SongEntryHorizontal_ItemObj!(execute, Object=@n4) -> (then, CastFailed, AsBP Song Entry Horizontal Item Obj:BP Song Entry Horizontal Item Obj Object Reference) {is=title K2Node_DynamicCast}
   n6 = GetSetlistEntryAtIndex!(self:Pilgrim Custom Setlist Object Reference=@n2, Index:int=@n4.Array Index) -> (OutSongShortName:Name=None, OutAddedByPlayer:Player State Object Reference)
   > n7 = Branch(Condition=NOT Boolean(A=IsValid(Object=@n6.OutAddedByPlayer)))
   > Call On Update Song and Player!(
    execute:Exec
    self:BP Song Entry Horizontal Item Obj Object Reference=@n5
    SongMetaData:Pilgrim Song Metadata Object Reference
    SongAddedByPlayer:Player State Object Reference
   ) -> (then:Exec) {is=title K2Node_CallDelegate}
   n8 = GetSongByShortName!(
    self:Pilgrim Song Catalog Object Reference=GetPilgrimSongCatalog!() -> (ReturnValue:Pilgrim Song Catalog Object Reference)
    ShortName:Name=@n6
   ) -> (ReturnValue:Pilgrim Song Metadata Object Reference)
   Set SongMetadata!(execute=@n7.else, SongMetadata:Pilgrim Song Metadata Object Reference=@n8, self:BP Song Entry Horizontal Item Obj Object Reference=@n5) -> (then, Output_Get:Pilgrim Song Metadata Object Reference)
   > Set SongAddedByPlayerState!(
    execute
    SongAddedByPlayerState:Player State Object Reference=@n6.OutAddedByPlayer
    self:BP Song Entry Horizontal Item Obj Object Reference=@n5
   ) -> (then, Output_Get:Player State Object Reference)
   > Call On Update Song and Player(self=@n5, SongMetaData=@n8, SongAddedByPlayer=@n6.OutAddedByPlayer)
   PrintString(execute=@n2.CastFailed, InString=WARNING: Song select menu can only display songs from a custom setlist)
   n9 = "To Text (String)"(InString=Get Title!(self:Sparks Song Data Object Reference) -> (Title:String))
   PrintString(
    InString=Append!(
     A="To String (Integer)"(InInt=@n4.Array Index)
     B=". "
     C:String="To String (Text)"(InText=@n9)
     D:String=" by "
     E:String=GetPlayerNameSafe!(Actor:Actor Object Reference) -> (ReturnValue:String)
    ) -> (ReturnValue)
    TextColor=(R=0,G=1,B=0.4933,A=1)
    Duration=8
   )
  ]}
 UpdateSideNavigationBindingState {nodes [
   n1 = UpdateSideNavigationBindingState!() -> (then, ShouldAddBindings:bool=~) {is=title K2Node_FunctionEntry}
   > Is Valid(InputObject=Get WBP_Pilgrim_QuickAccessMenu())
   > n2 = Branch(Condition=@n1)
   n3 = Create Event()
   > Bind Event to On Side Navigation Selection Changed(self:WBP UIKit Side Navigation Base Object Reference=Get WBP_Pilgrim_QuickAccessMenu(), Delegate=@n3) {is=title K2Node_AddDelegate}
   n4 = Create Event()
   > Bind Event to On Side Navigation Tab Clicked(self:WBP UIKit Side Navigation Base Object Reference=Get WBP_Pilgrim_QuickAccessMenu(), Delegate=@n4) {is=title K2Node_AddDelegate}
   n5 = Create Event()
   > Bind Event to On Side Navigation Inner Tab Clicked(self:WBP UIKit Side Navigation Base Object Reference=Get WBP_Pilgrim_QuickAccessMenu(), Delegate=@n5) {is=title K2Node_AddDelegate}
   n6 = Create Event()
   > Bind Event to On Side Navigation Collapsed State Changed(self:WBP UIKit Side Navigation Base Object Reference=Get WBP_Pilgrim_QuickAccessMenu(), Delegate=@n6) {is=title K2Node_AddDelegate}
   Unbind Event from On Side Navigation Selection Changed(execute=@n2.else, self:WBP UIKit Side Navigation Base Object Reference=Get WBP_Pilgrim_QuickAccessMenu(), Delegate=@n3) {is=title K2Node_RemoveDelegate}
   > Unbind Event from On Side Navigation Tab Clicked(self:WBP UIKit Side Navigation Base Object Reference=Get WBP_Pilgrim_QuickAccessMenu(), Delegate=@n4) {is=title K2Node_RemoveDelegate}
   > Unbind Event from On Side Navigation Inner Tab Clicked(self:WBP UIKit Side Navigation Base Object Reference=Get WBP_Pilgrim_QuickAccessMenu(), Delegate=@n5) {is=title K2Node_RemoveDelegate}
   > Unbind Event from On Side Navigation Collapsed State Changed(self:WBP UIKit Side Navigation Base Object Reference=Get WBP_Pilgrim_QuickAccessMenu(), Delegate=@n6) {is=title K2Node_RemoveDelegate}
  ]}
 UpdateSongsChosenText {nodes [
   UpdateSongsChosenText() {is=title K2Node_FunctionEntry}
   n1 = GetCustomSetlist()
   n2 = GetOwningPlayer()
   > "SetText (Text)"(
    self=[
     Get SongChosenHeader() -> (SongChosenHeader:Common Text Object Reference)
     Get SongChosenHeader_2() -> (SongChosenHeader_2:Common Text Object Reference)
    ]
    InText=Format Text!(
     Format=Get SongsChosenText() -> (SongsChosenText:Text)
     0:int=GetAdjustedNumAvailableSongsForPlayer!(self:Pilgrim Custom Setlist Object Reference=@n1, Player:Player Controller Object Reference=@n2) -> (ReturnValue:int=0)
    ) -> (Result)
   )
  ]}
]
variables [
 CachedBackButton {type=WBP_UIKit_ButtonCTA_Base_C, default=null}
 CachedLeaderboardsButton {type=WBP_UIKit_ButtonCTA_Base_C, default=null}
 CachedSetlistAddButton {type=WBP_UIKit_ButtonCTA_Base_C, default=null}
 CachedSetlistPreviewList {type=WBP_PilgrimQuickplay_SetlistPreviewList_C, default=null}
 DefaultSetlistLength {type=integer, default=4}
 Event Router {type=GameplayEventRouterComponent, default=null}
 GroupIntensityVMs {type="array<PilgrimSongCategoryGroup_SongIntensityVM>", default []}
 IntensityEntryWidgets {type="array<WBP_PilgrimQuickplay_SongIntensityEntry_C>", default []}
 OnMenuClosed {}
 OwningPlayer {type=PlayerController, default=null}
 Player Instrument Choice {type=EPilgrimTrackType, default=TrackGuitar}
 PreviewPilgrimInverse {type=boolean}
 PreviousSideNavButtonWidgets {type="array<Widget>", default []}
 RouterEvents {type="array<AsyncAction_StartListeningToEvent>", default []}
 SideNavigationButtons {type="array<WBP_UIKit_ButtonGeneric_Base_C>", default []}
 SongListItemMap {default {}}
 SongsChosenText {type=string, default="Songs you can add: {0}"}
]
widgets {
 info {
  inheritedWidgetCount=0
  parentClass=/Script/PilgrimQuickplayUI.PilgrimCommonActivatableWidget
  rootWidgetClass=&Overlay
  widgetCount=62
 }
 widgets [
  Overlay_20 {widgetClassPath=&Overlay, widget=Overlay_20}
  Image_BG {
   widgetClassPath=&Image
   parent=Overlay_20
   widget=Image_BG
   properties {brush {
     bIsDynamicallyLoaded=false
     drawAs=Image
     imageType=FullColor
     mirroring=NoMirror
     resourceName=None
     resourceObject=/SparksUICommon/Materials/MI_UI_Sparks_Fullscreen_Background_Triangles_LessPurple
     tiling=NoTile
     imageSize=<32 32>
     margin=<ltrb 0 0 0 0>
     outlineSettings {
      bUseBrushTransparency=false
      roundingType=HalfHeightRadius
      width=0
      color {colorUseRule=UseColor_Specified, specifiedColor=<rgba 0 0 0 0>}
      cornerRadii=<0 0 0 0>
     }
     tintColor {colorUseRule=UseColor_Specified, specifiedColor=<rgba 1 1 1 1>}
     uVRegion {bIsValid=false, max=<0 0>, min=<0 0>}
    }}
   slot {class=OverlaySlot, properties {horizontalAlignment=HAlign_Fill, verticalAlignment=VAlign_Fill}}
  }
  HorizontalBox-Backgrounds {
   widgetClassPath=&HorizontalBox
   parent=Overlay_20
   widget=HorizontalBox-Backgrounds
   slot {class=OverlaySlot, properties {horizontalAlignment=HAlign_Fill, verticalAlignment=VAlign_Fill}}
  }
  BG_left {
   widgetClassPath=&Border
   parent=HorizontalBox-Backgrounds
   widget=BG_left
   properties {
    background {
     bIsDynamicallyLoaded=false
     drawAs=Image
     imageType=FullColor
     mirroring=NoMirror
     resourceName=None
     resourceObject=/PilgrimQuickplay/Widgets/SetlistAndSongDisplay/Duplicated/Materials/MI_Leaderboards_BGwithFade_Inst_Inst_Inst
     tiling=NoTile
     imageSize=<0 0>
     margin=<ltrb 0 0 0 0>
     outlineSettings {
      bUseBrushTransparency=false
      roundingType=HalfHeightRadius
      width=0
      color {colorUseRule=UseColor_Specified, specifiedColor=<rgba 0 0 0 0>}
      cornerRadii=<0 0 0 0>
     }
     tintColor {colorUseRule=UseColor_Specified, specifiedColor=<rgba 1 1 1 1>}
     uVRegion {bIsValid=false, max=<0 0>, min=<0 0>}
    }
    brushColor=<rgba 1 1 1 0.6>
    padding=<ltrb 0 0 0 0>
   }
   slot {class=HorizontalBoxSlot, properties {size {sizeRule=Fill, value=0.8}}}
  }
  Spacer_92 {
   widgetClassPath=&Spacer
   parent=HorizontalBox-Backgrounds
   widget=Spacer_92
   slot {class=HorizontalBoxSlot, properties {size {sizeRule=Fill, value=0.5}}}
  }
  BG_right {
   widgetClassPath=&Border
   parent=HorizontalBox-Backgrounds
   widget=BG_right
   properties {
    visibility=Collapsed
    background {
     bIsDynamicallyLoaded=false
     drawAs=Image
     imageType=NoImage
     mirroring=NoMirror
     resourceName=None
     resourceObject=/PilgrimCore/Widgets/Materials/M_UI_Background_Triangles
     tiling=NoTile
     imageSize=<0 0>
     margin=<ltrb 0 0 0 0>
     outlineSettings {
      bUseBrushTransparency=false
      roundingType=HalfHeightRadius
      width=0
      color {colorUseRule=UseColor_Specified, specifiedColor=<rgba 0 0 0 0>}
      cornerRadii=<0 0 0 0>
     }
     tintColor {colorUseRule=UseColor_Specified, specifiedColor=<rgba 1 1 1 1>}
     uVRegion {bIsValid=false, max=<0 0>, min=<0 0>}
    }
    padding=<ltrb 0 0 0 0>
    renderTransformPivot=<1 1>
   }
   slot {class=HorizontalBoxSlot, properties {size {sizeRule=Fill, value=1}}}
  }
  WBP_CaptureForPostBufferUpdate {
   widgetClassPath=&WBP_CaptureForPostBufferUpdate
   bIsVariable=true
   parent=Overlay_20
   slot {class=OverlaySlot, properties {horizontalAlignment=HAlign_Fill, verticalAlignment=VAlign_Fill}}
  }
  SZ_TitlePortrait {
   widgetClassPath=&SafeZone
   bIsVariable=true
   parent=Overlay_20
   widget=SZ_TitlePortrait
   properties {padBottom=false}
   slot {class=OverlaySlot, properties {horizontalAlignment=HAlign_Fill, padding=<ltrb 48 24 48 48>}}
  }
  HorizontalBox_TitlePortrait {widgetClassPath=&HorizontalBox, parent=SZ_TitlePortrait, widget=HorizontalBox_TitlePortrait, slot {class=SafeZoneSlot}}
  Title_1 {
   widgetClassPath=&CommonTextBlock
   parent=HorizontalBox_TitlePortrait
   widget=Title_1
   properties {
    text=Music Library
    textTransformPolicy=ToUpper
    font {
     bForceMonospaced=false
     bMaterialIsStencil=false
     fontMaterial=null
     fontObject=&BurbankSmall
     letterSpacing=0
     monospacedWidth=1
     size=36
     skewAmount=0
     typefaceFontName=Black
     outlineSettings {
      bApplyOutlineToDropShadows=false
      bMiteredCorners=false
      bSeparateFillAlpha=false
      outlineMaterial=null
      outlineSize=0
      outlineColor=<rgba 0 0 0 1>
     }
    }
    shadowOffset=<0 0>
   }
   slot {class=HorizontalBoxSlot, properties {horizontalAlignment=HAlign_Center, verticalAlignment=VAlign_Bottom, padding=<ltrb 0 0 0 32>}}
  }
  Spacer_116 {
   widgetClassPath=&Spacer
   parent=HorizontalBox_TitlePortrait
   widget=Spacer_116
   slot {class=HorizontalBoxSlot, properties {size {sizeRule=Fill, value=1}}}
  }
  HeaderBox_Right {
   widgetClassPath=&Overlay
   parent=HorizontalBox_TitlePortrait
   widget=HeaderBox_Right
   slot {class=HorizontalBoxSlot, properties {horizontalAlignment=HAlign_Right, verticalAlignment=VAlign_Top}}
  }
  VerticalBox_1 {
   widgetClassPath=&VerticalBox
   parent=HeaderBox_Right
   widget=VerticalBox_1
   slot {class=OverlaySlot, properties {horizontalAlignment=HAlign_Fill, verticalAlignment=VAlign_Fill}}
  }
  ScaleBox_59 {
   widgetClassPath=&ScaleBox
   parent=VerticalBox_1
   widget=ScaleBox_59
   properties {stretch=UserSpecified, userSpecifiedScale=0.6}
   slot {class=VerticalBoxSlot, properties {horizontalAlignment=HAlign_Right}}
  }
  PlayerLoadout {
   widgetClassPath=/PilgrimQuickplay/Widgets/Results/WBP_Pilgrim_Quickplay_PlayerLoadout
   bIsVariable=true
   parent=ScaleBox_59
   widget=PlayerLoadout
   properties {hideDifficulty=true, hideName=true}
   slot {class=ScaleBoxSlot, properties {verticalAlignment=VAlign_Bottom}}
  }
  SZ_Content {
   widgetClassPath=&SafeZone
   bIsVariable=true
   parent=Overlay_20
   widget=SZ_Content
   slot {class=OverlaySlot, properties {horizontalAlignment=HAlign_Fill, verticalAlignment=VAlign_Fill, padding=<ltrb 48 96 48 48>}}
  }
  HorizontalBox-Content {widgetClassPath=&HorizontalBox, parent=SZ_Content, widget=HorizontalBox-Content, slot {class=SafeZoneSlot}}
  QuickAccessBarContainer {
   widgetClassPath=&SizeBox
   bIsVariable=true
   parent=HorizontalBox-Content
   widget=QuickAccessBarContainer
   properties {bOverride_MaxDesiredHeight=true, bOverride_MaxDesiredWidth=true, maxDesiredHeight=750, maxDesiredWidth=248, visibility=Collapsed}
   slot {class=HorizontalBoxSlot, properties {horizontalAlignment=HAlign_Left, verticalAlignment=VAlign_Top, padding=<ltrb 0 128 32 0>}}
  }
  WBP_Pilgrim_QuickAccessMenu {
   widgetClassPath=/SparksUICommon/Widgets/WBP_Pilgrim_SideNavigation_Menu
   bIsVariable=true
   parent=QuickAccessBarContainer
   widget=WBP_Pilgrim_QuickAccessMenu
   properties {scrollBarVisibility=Visible, tabGroupPadding=<ltrb 8 8 8 8>}
   slot {class=SizeBoxSlot}
  }
  GridPanel_0 {
   widgetClassPath=&GridPanel
   parent=HorizontalBox-Content
   widget=GridPanel_0
   properties {columnFill [3,2], rowFill [1]}
   slot {class=HorizontalBoxSlot, properties {padding=<ltrb 0 10 0 0>, size {sizeRule=Fill, value=1}}}
  }
  VerticalBox_SongList {
   widgetClassPath=&VerticalBox
   bIsVariable=true
   parent=GridPanel_0
   widget=VerticalBox_SongList
   slot {class=GridSlot, properties {padding=<ltrb 0 0 96 0>}}
  }
  HeaderBox_Left {
   widgetClassPath=&Overlay
   parent=VerticalBox_SongList
   widget=HeaderBox_Left
   slot {class=VerticalBoxSlot, properties {verticalAlignment=VAlign_Top}}
  }
  TitleSortBox_1 {
   widgetClassPath=&VerticalBox
   parent=HeaderBox_Left
   widget=TitleSortBox_1
   slot {class=OverlaySlot, properties {horizontalAlignment=HAlign_Fill, verticalAlignment=VAlign_Fill}}
  }
  HorizontalBox {
   widgetClassPath=&HorizontalBox
   parent=TitleSortBox_1
   widget=HorizontalBox
   slot {class=VerticalBoxSlot, properties {padding=<ltrb 0 0 0 24>, size {sizeRule=Fill, value=1}}}
  }
  WBP_PilgrimSortFilter {
   widgetClassPath=&/WBP_PilgrimSortFilter
   bIsVariable=true
   parent=HorizontalBox
   widget=WBP_PilgrimSortFilter
   properties {widgetBelowTextBox=&self:WidgetTree.WBP_PilgrimQuickplay_SongCatalog2}
   slot {class=HorizontalBoxSlot, properties {padding=<ltrb 0 0 50 0>, size {sizeRule=Fill, value=1}}}
  }
  Spacer_148 {
   widgetClassPath=&Spacer
   parent=HorizontalBox
   widget=Spacer_148
   properties {visibility=Collapsed}
   slot {class=HorizontalBoxSlot, properties {size {sizeRule=Fill, value=0.058}}}
  }
  WBP_PilgrimQuickplay_SongCatalog2 {
   widgetClassPath=&/WBP_PilgrimQuickplay_SongCatalog2
   bIsVariable=true
   parent=VerticalBox_SongList
   widget=WBP_PilgrimQuickplay_SongCatalog2
   properties {bIsModal=false}
   slot {class=VerticalBoxSlot, properties {size {sizeRule=Fill, value=1}}}
  }
  Inverse_DetailsPanel {
   widgetClassPath=&VerticalBox
   bIsVariable=true
   parent=VerticalBox_SongList
   widget=Inverse_DetailsPanel
   properties {bOverride_Cursor=true, visibility=Collapsed}
   slot {class=VerticalBoxSlot, properties {padding=<ltrb 0 12 0 0>}}
  }
  HorizontalBox_4 {widgetClassPath=&HorizontalBox, parent=Inverse_DetailsPanel, widget=HorizontalBox_4, slot {class=VerticalBoxSlot}}
  SetlistTitle_2 {
   widgetClassPath=&CommonTextBlock
   parent=HorizontalBox_4
   widget=SetlistTitle_2
   properties {
    justification=Center
    text=Setlist Selection
    textTransformPolicy=ToUpper
    colorAndOpacity {colorUseRule=UseColor_Specified, specifiedColor=<rgba 0.4125 0.5089 0.9647 1>}
    font {
     bForceMonospaced=false
     bMaterialIsStencil=false
     fontMaterial=null
     fontObject=&BurbankSmall
     letterSpacing=20
     monospacedWidth=1
     size=15
     skewAmount=0
     typefaceFontName=Black
     outlineSettings {
      bApplyOutlineToDropShadows=false
      bMiteredCorners=false
      bSeparateFillAlpha=false
      outlineMaterial=null
      outlineSize=0
      outlineColor=<rgba 0 0 0 1>
     }
    }
    shadowOffset=<0 0>
    strikeBrush {
     bIsDynamicallyLoaded=false
     drawAs=NoDrawType
     imageType=NoImage
     mirroring=NoMirror
     resourceName=None
     resourceObject=null
     tiling=NoTile
     imageSize=<32 32>
     margin=<ltrb 0 0 0 0>
     outlineSettings {
      bUseBrushTransparency=false
      roundingType=HalfHeightRadius
      width=0
      color {colorUseRule=UseColor_Specified, specifiedColor=<rgba 0 0 0 0>}
      cornerRadii=<0 0 0 0>
     }
     tintColor {colorUseRule=UseColor_Specified, specifiedColor=<rgba 1 1 1 1>}
     uVRegion {bIsValid=false, max=<0 0>, min=<0 0>}
    }
   }
   slot {class=HorizontalBoxSlot, properties {horizontalAlignment=HAlign_Left, padding=<ltrb 0 0 24 0>}}
  }
  SongChosenHeader_2 {
   widgetClassPath=&CommonTextBlock
   bIsVariable=true
   parent=HorizontalBox_4
   widget=SongChosenHeader_2
   properties {
    scrollStyle=&TextScrollStyle-Base_C
    text=Songs you can add: 0
    font {
     bForceMonospaced=false
     bMaterialIsStencil=false
     fontMaterial=null
     fontObject=&BurbankSmall
     letterSpacing=0
     monospacedWidth=1
     size=18
     skewAmount=0
     typefaceFontName=Black
     outlineSettings {
      bApplyOutlineToDropShadows=false
      bMiteredCorners=false
      bSeparateFillAlpha=false
      outlineMaterial=null
      outlineSize=0
      outlineColor=<rgba 0 0 0 1>
     }
    }
    shadowOffset=<0 0>
   }
   slot {class=HorizontalBoxSlot, properties {horizontalAlignment=HAlign_Right, verticalAlignment=VAlign_Bottom, size {sizeRule=Fill, value=1}}}
  }
  WBP_PilgrimQuickplay_SetlistPreviewList_Inverse {
   widgetClassPath=&/WBP_PilgrimQuickplay_SetlistPreviewList
   bIsVariable=true
   parent=Inverse_DetailsPanel
   widget=WBP_PilgrimQuickplay_SetlistPreviewList_Inverse
   properties {activatedVisibility=Visible, bAutoActivate=true, bSetVisibilityOnActivated=true, visibility=Visible}
   slot {class=VerticalBoxSlot}
  }
  Border_1 {
   widgetClassPath=&Border
   parent=Inverse_DetailsPanel
   widget=Border_1
   properties {
    background {
     bIsDynamicallyLoaded=false
     drawAs=Box
     imageType=NoImage
     mirroring=NoMirror
     resourceName=None
     resourceObject=null
     tiling=NoTile
     imageSize=<48 48>
     margin=<ltrb 0.5 0.5 0.5 0.5>
     outlineSettings {
      bUseBrushTransparency=false
      roundingType=HalfHeightRadius
      width=0
      color {colorUseRule=UseColor_Specified, specifiedColor=<rgba 0 0 0 0>}
      cornerRadii=<0 0 0 0>
     }
     tintColor {colorUseRule=UseColor_Specified, specifiedColor=<rgba 1 1 1 1>}
     uVRegion {bIsValid=false, max=<0 0>, min=<0 0>}
    }
    brushColor=<rgba 1 1 1 0>
    padding=<ltrb 0 0 0 0>
   }
   slot {class=VerticalBoxSlot}
  }
  Switcher_Actions_Inverse {
   widgetClassPath=&CommonWidgetSwitcherLegacy
   bIsVariable=true
   parent=Border_1
   widget=Switcher_Actions_Inverse
   slot {class=BorderSlot, properties {padding=<ltrb 0 0 0 0>}}
  }
  HorizontalBox_SongActions_Inverse {
   widgetClassPath=&HorizontalBox
   bIsVariable=true
   parent=Switcher_Actions_Inverse
   widget=HorizontalBox_SongActions_Inverse
   properties {visibility=Visible}
   slot {class=WidgetSwitcherSlot, properties {horizontalAlignment=HAlign_Right}}
  }
  WBP_UIKit_Button_Loud_SETLIST_Inverse {
   widgetClassPath=/FortUILibrary/UIKit/Buttons/WBP_UIKit_Button_Loud
   bIsVariable=true
   parent=HorizontalBox_SongActions_Inverse
   widget=WBP_UIKit_Button_Loud_SETLIST_Inverse
   properties {
    bIsFocusable=false
    bOverride_Cursor=true
    buttonSize=Small
    hasText=true
    holdCompletedSound=/Game/Sounds/Fort_UI_Sounds/FrontEnd/DefaultSounds/HoldableButtons/Holable_Quiet/MSS_UI_Button_Holdable_Quiet_Complete
    holdPressedSound=/Game/Sounds/Fort_UI_Sounds/FrontEnd/DefaultSounds/HoldableButtons/Holable_Quiet/MSS_UI_Button_Holdable_Quiet_Press
    holdReleasedSound=/Game/Sounds/Fort_UI_Sounds/FrontEnd/DefaultSounds/HoldableButtons/Holable_Quiet/MSS_UI_Button_Holdable_Quiet_Release
    holdSound=/Game/Sounds/Fort_UI_Sounds/FrontEnd/DefaultSounds/HoldableButtons/Holable_Quiet/MSS_UI_Button_Holdable_Quiet_Hold_lp
    minWidth=200
    text=Add to Setlist
    triggeringInputAction {dataTable=/PilgrimQuickplay/Widgets/InputActionDataTables/FM_PilgrimQuickplay_InputActionDataTable, rowName=AddToSetlist}
   }
   slot {class=HorizontalBoxSlot, properties {horizontalAlignment=HAlign_Right, padding=<ltrb 0 0 12 0>, size {sizeRule=Fill, value=1}}}
  }
  WBP_UIKit_Button_Regular_LEADERBOARD_Inverse {
   widgetClassPath=&WBP_UIKit_Button_Regular
   bIsVariable=true
   parent=HorizontalBox_SongActions_Inverse
   widget=WBP_UIKit_Button_Regular_LEADERBOARD_Inverse
   properties {
    bEnableOnClickedAnalytics=true
    bIsFocusable=false
    bOverride_Cursor=true
    buttonSize=Small
    hasText=true
    holdCompletedSound=&MSS_UI_Button_Holdable_Quiet_Complete
    holdPressedSound=&MSS_UI_Button_Holdable_Quiet_Press
    holdReleasedSound=&MSS_UI_Button_Holdable_Quiet_Release
    holdSound=&MSS_UI_Button_Holdable_Quiet_Hold_lp
    mobileMinHeight=0
    text=Leaderboard
    triggeringInputAction {dataTable=&FM_PilgrimQuickplay_InputActionDataTable, rowName=OpenLeaderboard}
   }
   slot {class=HorizontalBoxSlot, properties {horizontalAlignment=HAlign_Right, padding=<ltrb 0 0 12 0>}}
  }
  WBP_UIKit_Button_Regular_BACK_Inverse {
   widgetClassPath=&WBP_UIKit_Button_Regular
   bIsVariable=true
   parent=HorizontalBox_SongActions_Inverse
   widget=WBP_UIKit_Button_Regular_BACK_Inverse
   properties {
    bEnableOnClickedAnalytics=true
    bIsFocusable=false
    bOverride_Cursor=true
    buttonSize=Small
    hasText=true
    holdCompletedSound=&MSS_UI_Button_Holdable_Quiet_Complete
    holdPressedSound=&MSS_UI_Button_Holdable_Quiet_Press
    holdReleasedSound=&MSS_UI_Button_Holdable_Quiet_Release
    holdSound=&MSS_UI_Button_Holdable_Quiet_Hold_lp
    mobileMinHeight=0
    text=Back
    triggeringInputAction {dataTable=&InputActionDataTable, rowName=GenericBack}
   }
   slot {class=HorizontalBoxSlot, properties {horizontalAlignment=HAlign_Right}}
  }
  HorizontalBox_FilterActions_Inverse {
   widgetClassPath=&HorizontalBox
   bIsVariable=true
   parent=Switcher_Actions_Inverse
   widget=HorizontalBox_FilterActions_Inverse
   properties {visibility=Visible}
   slot {class=WidgetSwitcherSlot, properties {horizontalAlignment=HAlign_Right}}
  }
  WBP_UIKit_Button_Regular_FILTER_BACK_Inverse {
   widgetClassPath=&WBP_UIKit_Button_Regular
   bIsVariable=true
   parent=HorizontalBox_FilterActions_Inverse
   widget=WBP_UIKit_Button_Regular_FILTER_BACK_Inverse
   properties {
    bIsFocusable=false
    bOverride_Cursor=true
    buttonSize=Small
    hasText=true
    holdCompletedSound=&MSS_UI_Button_Holdable_Quiet_Complete
    holdPressedSound=&MSS_UI_Button_Holdable_Quiet_Press
    holdReleasedSound=&MSS_UI_Button_Holdable_Quiet_Release
    holdSound=&MSS_UI_Button_Holdable_Quiet_Hold_lp
    mobileMinHeight=0
    text=Back
    triggeringInputAction {dataTable=&InputActionDataTable, rowName=GenericBack}
   }
   slot {class=HorizontalBoxSlot}
  }
  LeaderboardsVerticalBox {
   widgetClassPath=&VerticalBox
   bIsVariable=true
   parent=GridPanel_0
   widget=LeaderboardsVerticalBox
   slot {class=GridSlot, properties {column=1, columnSpan=1}}
  }
  DetailsPanel {
   widgetClassPath=&VerticalBox
   parent=LeaderboardsVerticalBox
   widget=DetailsPanel
   slot {class=VerticalBoxSlot, properties {verticalAlignment=VAlign_Bottom, size {sizeRule=Fill, value=1}}}
  }
  Leaderboard {
   widgetClassPath=/PilgrimQuickplay/Widgets/Leaderboards/WBP_PilgrimQuickplay_Leaderboard_Mini_MusicLibrary
   bIsVariable=true
   parent=DetailsPanel
   widget=Leaderboard
   properties {bIsFocusable=false}
   slot {class=VerticalBoxSlot, properties {padding=<ltrb 0 0 0 48>, size {sizeRule=Fill, value=1}}}
  }
  VerticalBox_103 {
   widgetClassPath=&VerticalBox
   parent=DetailsPanel
   widget=VerticalBox_103
   slot {class=VerticalBoxSlot, properties {verticalAlignment=VAlign_Bottom, padding=<ltrb 0 0 0 48>}}
  }
  HorizontalBox_2 {
   widgetClassPath=&HorizontalBox
   parent=VerticalBox_103
   widget=HorizontalBox_2
   slot {class=VerticalBoxSlot, properties {verticalAlignment=VAlign_Center, padding=<ltrb 0 0 0 12>}}
  }
  SetlistTitle {
   widgetClassPath=&CommonTextBlock
   parent=HorizontalBox_2
   widget=SetlistTitle
   properties {
    justification=Center
    text=Setlist Selection
    textTransformPolicy=ToUpper
    colorAndOpacity {colorUseRule=UseColor_Specified, specifiedColor=<rgba 0.4125 0.5089 0.9647 1>}
    font {
     bForceMonospaced=false
     bMaterialIsStencil=false
     fontMaterial=null
     fontObject=&BurbankSmall
     letterSpacing=20
     monospacedWidth=1
     size=15
     skewAmount=0
     typefaceFontName=Black
     outlineSettings {
      bApplyOutlineToDropShadows=false
      bMiteredCorners=false
      bSeparateFillAlpha=false
      outlineMaterial=null
      outlineSize=0
      outlineColor=<rgba 0 0 0 1>
     }
    }
    shadowOffset=<0 0>
    strikeBrush {
     bIsDynamicallyLoaded=false
     drawAs=NoDrawType
     imageType=NoImage
     mirroring=NoMirror
     resourceName=None
     resourceObject=null
     tiling=NoTile
     imageSize=<32 32>
     margin=<ltrb 0 0 0 0>
     outlineSettings {
      bUseBrushTransparency=false
      roundingType=HalfHeightRadius
      width=0
      color {colorUseRule=UseColor_Specified, specifiedColor=<rgba 0 0 0 0>}
      cornerRadii=<0 0 0 0>
     }
     tintColor {colorUseRule=UseColor_Specified, specifiedColor=<rgba 1 1 1 1>}
     uVRegion {bIsValid=false, max=<0 0>, min=<0 0>}
    }
   }
   slot {class=HorizontalBoxSlot, properties {horizontalAlignment=HAlign_Left, padding=<ltrb 0 0 24 0>}}
  }
  SongChosenHeader {
   widgetClassPath=&CommonTextBlock
   bIsVariable=true
   parent=HorizontalBox_2
   widget=SongChosenHeader
   properties {
    scrollStyle=&TextScrollStyle-Base_C
    text=Songs you can add: 0
    font {
     bForceMonospaced=false
     bMaterialIsStencil=false
     fontMaterial=null
     fontObject=&BurbankSmall
     letterSpacing=0
     monospacedWidth=1
     size=18
     skewAmount=0
     typefaceFontName=Black
     outlineSettings {
      bApplyOutlineToDropShadows=false
      bMiteredCorners=false
      bSeparateFillAlpha=false
      outlineMaterial=null
      outlineSize=0
      outlineColor=<rgba 0 0 0 1>
     }
    }
    shadowOffset=<0 0>
   }
   slot {class=HorizontalBoxSlot, properties {horizontalAlignment=HAlign_Right, verticalAlignment=VAlign_Bottom, size {sizeRule=Fill, value=1}}}
  }
  WBP_PilgrimQuickplay_SetlistPreviewList {
   widgetClassPath=&WBP_PilgrimQuickplay_SetlistPreviewList
   bIsVariable=true
   parent=VerticalBox_103
   widget=WBP_PilgrimQuickplay_SetlistPreviewList
   properties {
    activatedVisibility=Visible
    bAutoActivate=true
    bSetVisibilityOnActivated=true
    navigation=&self:WidgetTree.WBP_PilgrimQuickplay_SetlistPreviewList.WidgetNavigation_0
    visibility=Visible
   }
   slot {class=VerticalBoxSlot}
  }
  Border_167 {
   widgetClassPath=&Border
   parent=DetailsPanel
   widget=Border_167
   properties {
    background {
     bIsDynamicallyLoaded=false
     drawAs=Box
     imageType=NoImage
     mirroring=NoMirror
     resourceName=None
     resourceObject=null
     tiling=NoTile
     imageSize=<48 48>
     margin=<ltrb 0.5 0.5 0.5 0.5>
     outlineSettings {
      bUseBrushTransparency=false
      roundingType=HalfHeightRadius
      width=0
      color {colorUseRule=UseColor_Specified, specifiedColor=<rgba 0 0 0 0>}
      cornerRadii=<0 0 0 0>
     }
     tintColor {colorUseRule=UseColor_Specified, specifiedColor=<rgba 1 1 1 1>}
     uVRegion {bIsValid=false, max=<0 0>, min=<0 0>}
    }
    brushColor=<rgba 1 1 1 0>
    padding=<ltrb 0 0 0 0>
   }
   slot {class=VerticalBoxSlot, properties {verticalAlignment=VAlign_Bottom, padding=<ltrb 0 10 0 30>}}
  }
  Switcher_Actions {
   widgetClassPath=&CommonWidgetSwitcherLegacy
   bIsVariable=true
   parent=Border_167
   widget=Switcher_Actions
   slot {class=BorderSlot, properties {padding=<ltrb 0 0 0 0>}}
  }
  HorizontalBox_SongActions {
   widgetClassPath=&HorizontalBox
   bIsVariable=true
   parent=Switcher_Actions
   widget=HorizontalBox_SongActions
   properties {visibility=Visible}
   slot {class=WidgetSwitcherSlot, properties {horizontalAlignment=HAlign_Right}}
  }
  SizeBox_0 {
   widgetClassPath=&SizeBox
   parent=HorizontalBox_SongActions
   widget=SizeBox_0
   properties {minDesiredWidth=305}
   slot {class=HorizontalBoxSlot, properties {padding=<ltrb 0 0 12 0>, size {sizeRule=Fill, value=1}}}
  }
  AddOrRemoveSong_ContextualSwitcher {
   widgetClassPath=&CommonVisibilitySwitcher
   bIsVariable=true
   parent=SizeBox_0
   widget=AddOrRemoveSong_ContextualSwitcher
   properties {bOverride_Cursor=true}
   slot {class=SizeBoxSlot}
  }
  WBP_UIKit_Button_Loud_SETLIST {
   widgetClassPath=&WBP_UIKit_Button_Loud
   bIsVariable=true
   parent=AddOrRemoveSong_ContextualSwitcher
   widget=WBP_UIKit_Button_Loud_SETLIST
   properties {
    bIsFocusable=false
    bOverride_Cursor=true
    buttonSize=Small
    hasText=true
    holdCompletedSound=&MSS_UI_Button_Holdable_Quiet_Complete
    holdPressedSound=&MSS_UI_Button_Holdable_Quiet_Press
    holdReleasedSound=&MSS_UI_Button_Holdable_Quiet_Release
    holdSound=&MSS_UI_Button_Holdable_Quiet_Hold_lp
    text=Add Song
    triggeringInputAction {dataTable=&FM_PilgrimQuickplay_InputActionDataTable, rowName=AddToSetlist}
   }
   slot {class=CommonVisibilitySwitcherSlot}
  }
  WBP_UIKit_Button_Regular_SETLIST {
   widgetClassPath=&WBP_UIKit_Button_Regular
   bIsVariable=true
   parent=AddOrRemoveSong_ContextualSwitcher
   widget=WBP_UIKit_Button_Regular_SETLIST
   properties {
    bEnableOnClickedAnalytics=true
    bIsFocusable=false
    bOverride_Cursor=true
    buttonSize=Small
    hasText=true
    holdCompletedSound=&MSS_UI_Button_Holdable_Quiet_Complete
    holdPressedSound=&MSS_UI_Button_Holdable_Quiet_Press
    holdReleasedSound=&MSS_UI_Button_Holdable_Quiet_Release
    holdSound=&MSS_UI_Button_Holdable_Quiet_Hold_lp
    mobileMinHeight=0
    text=Remove Song
    triggeringInputAction {dataTable=&FM_PilgrimQuickplay_InputActionDataTable, rowName=AddToSetlist}
   }
   slot {class=CommonVisibilitySwitcherSlot}
  }
  SZ_LeaderboardBtn {
   widgetClassPath=&SizeBox
   bIsVariable=true
   parent=HorizontalBox_SongActions
   widget=SZ_LeaderboardBtn
   properties {bOverride_MinDesiredWidth=true}
   slot {class=HorizontalBoxSlot, properties {padding=<ltrb 0 0 12 0>, size {sizeRule=Fill, value=1}}}
  }
  WBP_UIKit_Button_Regular_LEADERBOARD {
   widgetClassPath=&WBP_UIKit_Button_Regular
   bIsVariable=true
   parent=SZ_LeaderboardBtn
   widget=WBP_UIKit_Button_Regular_LEADERBOARD
   properties {
    bEnableOnClickedAnalytics=true
    bIsFocusable=false
    bOverride_Cursor=true
    buttonSize=Small
    hasText=true
    holdCompletedSound=&MSS_UI_Button_Holdable_Quiet_Complete
    holdPressedSound=&MSS_UI_Button_Holdable_Quiet_Press
    holdReleasedSound=&MSS_UI_Button_Holdable_Quiet_Release
    holdSound=&MSS_UI_Button_Holdable_Quiet_Hold_lp
    mobileMinHeight=0
    text=Leaderboard
    triggeringInputAction {dataTable=&FM_PilgrimQuickplay_InputActionDataTable, rowName=OpenLeaderboard}
   }
   slot {class=SizeBoxSlot}
  }
  BackOrDone_ContextualSwitcher {
   widgetClassPath=&CommonVisibilitySwitcher
   bIsVariable=true
   parent=HorizontalBox_SongActions
   widget=BackOrDone_ContextualSwitcher
   properties {bOverride_Cursor=true}
   slot {class=HorizontalBoxSlot}
  }
  WBP_UIKit_Button_Regular_BACK {
   widgetClassPath=&WBP_UIKit_Button_Regular
   bIsVariable=true
   parent=BackOrDone_ContextualSwitcher
   widget=WBP_UIKit_Button_Regular_BACK
   properties {
    bEnableOnClickedAnalytics=true
    bIsFocusable=false
    bOverride_Cursor=true
    buttonSize=Small
    hasText=true
    holdCompletedSound=&MSS_UI_Button_Holdable_Quiet_Complete
    holdPressedSound=&MSS_UI_Button_Holdable_Quiet_Press
    holdReleasedSound=&MSS_UI_Button_Holdable_Quiet_Release
    holdSound=&MSS_UI_Button_Holdable_Quiet_Hold_lp
    mobileMinHeight=0
    text=Back
    triggeringInputAction {dataTable=&InputActionDataTable, rowName=GenericBack}
   }
   slot {class=CommonVisibilitySwitcherSlot}
  }
  WBP_UIKit_Button_Loud_BACK_DONE {
   widgetClassPath=&WBP_UIKit_Button_Loud
   bIsVariable=true
   parent=BackOrDone_ContextualSwitcher
   widget=WBP_UIKit_Button_Loud_BACK_DONE
   properties {
    bIsFocusable=false
    bOverride_Cursor=true
    buttonSize=Small
    hasText=true
    holdCompletedSound=&MSS_UI_Button_Holdable_Quiet_Complete
    holdPressedSound=&MSS_UI_Button_Holdable_Quiet_Press
    holdReleasedSound=&MSS_UI_Button_Holdable_Quiet_Release
    holdSound=&MSS_UI_Button_Holdable_Quiet_Hold_lp
    minWidth=200
    text=Done
    triggeringInputAction {dataTable=&InputActionDataTable, rowName=GenericBack}
   }
   slot {class=CommonVisibilitySwitcherSlot}
  }
  HorizontalBox_FilterActions {
   widgetClassPath=&HorizontalBox
   bIsVariable=true
   parent=Switcher_Actions
   widget=HorizontalBox_FilterActions
   properties {visibility=Visible}
   slot {class=WidgetSwitcherSlot, properties {horizontalAlignment=HAlign_Right}}
  }
  WBP_UIKit_Button_Regular_FILTER_BACK {
   widgetClassPath=&WBP_UIKit_Button_Regular
   bIsVariable=true
   parent=HorizontalBox_FilterActions
   widget=WBP_UIKit_Button_Regular_FILTER_BACK
   properties {
    bIsFocusable=false
    bOverride_Cursor=true
    buttonSize=Small
    hasText=true
    holdCompletedSound=&MSS_UI_Button_Holdable_Quiet_Complete
    holdPressedSound=&MSS_UI_Button_Holdable_Quiet_Press
    holdReleasedSound=&MSS_UI_Button_Holdable_Quiet_Release
    holdSound=&MSS_UI_Button_Holdable_Quiet_Hold_lp
    mobileMinHeight=0
    text=Back
    triggeringInputAction {dataTable=&InputActionDataTable, rowName=GenericBack}
   }
   slot {class=HorizontalBoxSlot}
  }
 ]
}
