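# Activatable Pilgrim Quickplay menu titled "Spotlight Setlist", pairing a song catalog with a mini leaderboard and player avatar.
# Construct initializes the catalog view model, waits asynchronously for the event router, then binds Leaderboard and Back buttons and sets the avatar.
# Catalog selection updates the mini leaderboard and enables the Leaderboard button, which calls into the catalog; deselection disables it.
# Deactivation broadcasts a playlist-display-closed event and fades out the playspace's song preview.
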WidgetBlueprint /PilgrimQuickplay/Widgets/SetlistAndSongDisplay/Duplicated/WBP_PilgrimQuickplay_SongsCuratedMenu
graphs [
 EventGraph {nodes [
   n1 = "On Item Selection Changed (WBP_PilgrimQuickplay_SongCatalog2)"!() -> (OutputDelegate, then, Item:Pilgrim Song Catalog Item VM Object Reference) {is=title K2Node_ComponentBoundEvent}
   > OnMusicLibraryEntrySelected!(
    execute
    self:WBP Pilgrim Quickplay Leaderboard Mini Music Library Object Reference=Get Leaderboard() -> (Leaderboard:WBP Pilgrim Quickplay Leaderboard Mini Music Library Object Reference)
    Song:Pilgrim Song Metadata Object Reference=GetSongMetadata!(self:Pilgrim Song Catalog Item VM Object Reference=@n1.Item) -> (ReturnValue:Pilgrim Song Metadata Object Reference)
   ) -> (then)
   > SetIsEnabled(
    self=Get WBP_UIKit_Button_Quiet_Leaderboard() -> (WBP_UIKit_Button_Quiet_Leaderboard:WBP UIKit Button Quiet Object Reference)
    bInIsEnabled=true
   )
   "On Item Deselected (WBP_PilgrimQuickplay_SongCatalog2)"() {is=title K2Node_ComponentBoundEvent}
   > SetIsEnabled(self=Get WBP_UIKit_Button_Quiet_Leaderboard(), bInIsEnabled=false)
   n2 = OnBackButtonClicked!() -> (OutputDelegate, then, Button:Common Button Base Object Reference) {is=title K2Node_CustomEvent}
   > DeactivateWidget()
   n3 = OnLeaderboardButtonClicked!() -> (OutputDelegate, then, Button:Common Button Base Object Reference) {is=title K2Node_CustomEvent}
   > Leaderboard Selected(self:WBP Pilgrim Quickplay Song Catalog 2 Object Reference=Get WBP_PilgrimQuickplay_SongCatalog2() -> (WBP_PilgrimQuickplay_SongCatalog2:WBP Pilgrim Quickplay Song Catalog 2 Object Reference))
   Event On Activated()
   > Parent: On Activated() {is=title K2Node_CallParentFunction}
   > SetKeyboardFocus(self:Widget Object Reference=GetDesiredFocusTarget!(self:Common Activatable Widget Object Reference) -> (ReturnValue:Widget Object Reference))
   Event Construct()
   > Parent: Construct()
   > n4 = GetOwningPlayer!(execute:Exec, self) -> (then:Exec, ReturnValue)
   n5 = MakeLiteralGameplayTagContainer!(Value:Gameplay Tag Container Structure) -> (ReturnValue:Gameplay Tag Container Structure)
   > Initialize!(
    execute
    self:Pilgrim Song Catalog VM Object Reference=Get PilgrimSongCatalogVM() -> (PilgrimSongCatalogVM:Pilgrim Song Catalog VM Object Reference)
    InOwningController:Player Controller Object Reference=@n4
    InSortTags:"Gameplay Tag Container Structure (by ref)"=@n5
    InFilterTags:"Gameplay Tag Container Structure (by ref)"=@n5
   ) -> (then)
   > n6 = PilgrimWaitForEventRouter!(execute:Exec, ContextActor:Actor Object Reference=@n4) -> (then:Exec, OnEventRouterAvailable:Exec, EventRouter:Gameplay Event Router Component Object Reference) {is=title K2Node_AsyncAction}
   Set Event Router(execute=@n6.OnEventRouterAvailable, Event Router:Gameplay Event Router Component Object Reference=@n6) -> (Output_Get:Gameplay Event Router Component Object Reference)
   > Assign On Clicked!(execute:Exec, self:Common Button Base Object Reference=Get WBP_UIKit_Button_Quiet_Leaderboard(), Delegate:"Delegate (by ref)"=@n3) -> (then:Exec) {is=title K2Node_AssignDelegate}
   > Assign On Clicked(self=Get WBP_UIKit_Button_Quiet_Back() -> (WBP_UIKit_Button_Quiet_Back:WBP UIKit Button Quiet Object Reference), Delegate=@n2)
   > SetPlayerLoadout()
   Event On Deactivated()
   > Parent: On Deactivated() {is=title K2Node_CallParentFunction}
   > Broadcast Event!(
    execute
    self:Gameplay Event Router Component Object Reference=Get Event Router() -> (Event Router:Gameplay Event Router Component Object Reference)
    EventData:Pilgrim Event Playlist Display Closed Structure=Make Pilgrim Event Playlist Display Closed!(SongSelectionWidget_24_#1:Widget Object Reference=Self-Reference()) -> (PilgrimEvent_Playlist_Display_Closed:Pilgrim Event Playlist Display Closed Structure) {is=title K2Node_MakeStruct}
    EventContext:Object Reference
   ) -> (then, ReturnValue:bool)
   n7 = GetOwningPlayer!(self) -> (ReturnValue)
   > StopCurrentPreview!(
    execute
    self:Sparks Song Previewer Component Object Reference=Get SongPreviewer!(self:BP Pilgrim Quickplay Playspace Object Reference=GetSparksMusicPlayspaceOfClass!(
      PlayspaceType:Sparks Music Playspace Class Reference=/PilgrimQuickplay/Playspaces/BP_PilgrimQuickplayPlayspace.BP_PilgrimQuickplayPlayspace_C
      Actor:Actor Object Reference=@n7
      bMatchChildTypes:bool=false
     ) -> (ReturnValue:BP Pilgrim Quickplay Playspace Object Reference)) -> (SongPreviewer:BPC Sparks Song Previewer Object Reference)
    bWithFadeout:bool=true
   ) -> (then)
  ]}
 SetPlayerLoadout {nodes [
   SetPlayerLoadout() {is=title K2Node_FunctionEntry}
   > n1 = Cast To FortPlayerState!(execute, Object=Get PlayerState!(self:Controller Object Reference=GetOwningPlayer()) -> (PlayerState:Player State Object Reference)) -> (then, CastFailed, AsFort Player State:Fort Player State Object Reference) {is=title K2Node_DynamicCast}
   > SetAvatarIconImg!(
    execute
    self:WBP Pilgrim Quickplay Player Loadout Object Reference=Get PlayerLoadout() -> (PlayerLoadout:WBP Pilgrim Quickplay Player Loadout Object Reference)
    PlayerState:Fort Player State Object Reference=@n1
   ) -> (then)
  ]}
]
variables [Event Router {type=GameplayEventRouterComponent, default=null}]
widgets {
 info {inheritedWidgetCount=0, parentClass=&CommonActivatableWidget, rootWidgetClass=&Overlay, widgetCount=24}
 widgets [
  Overlay_17 {widgetClassPath=&Overlay, widget=Overlay_17}
  Image_Background {
   widgetClassPath=&Image
   parent=Overlay_17
   widget=Image_Background
   properties {brush {
     bIsDynamicallyLoaded=false
     drawAs=Image
     imageType=NoImage
     mirroring=NoMirror
     resourceName=None
     resourceObject=/SparksUICommon/Materials/MI_UI_Sparks_Fullscreen_Background_Triangles
     tiling=NoTile
     imageSize=<32 32>
     margin=<ltrb 0 0 0 0>
     outlineSettings {
      bUseBrushTransparency=false
      roundingType=HalfHeightRadius
      width=0
      color {colorUseRule=UseColor_Specified, specifiedColor=<rgba 0 0 0 0>}
      cornerRadii=<0 0 0 0>
     }
     tintColor {colorUseRule=UseColor_Specified, specifiedColor=<rgba 1 1 1 1>}
     uVRegion {bIsValid=false, max=<0 0>, min=<0 0>}
    }}
   slot {class=OverlaySlot, properties {horizontalAlignment=HAlign_Fill, verticalAlignment=VAlign_Fill}}
  }
  HorizontalBox-Backgrounds {
   widgetClassPath=&HorizontalBox
   parent=Overlay_17
   widget=HorizontalBox-Backgrounds
   slot {class=OverlaySlot, properties {horizontalAlignment=HAlign_Fill, verticalAlignment=VAlign_Fill}}
  }
  BG_left {
   widgetClassPath=&Border
   parent=HorizontalBox-Backgrounds
   widget=BG_left
   properties {
    background {
     bIsDynamicallyLoaded=false
     drawAs=Image
     imageType=FullColor
     mirroring=NoMirror
     resourceName=None
     resourceObject=/PilgrimQuickplay/Widgets/SetlistAndSongDisplay/Duplicated/Materials/MusicLibraryBGwithFade_Inst
     tiling=NoTile
     imageSize=<0 0>
     margin=<ltrb 0 0 0 0>
     outlineSettings {
      bUseBrushTransparency=false
      roundingType=HalfHeightRadius
      width=0
      color {colorUseRule=UseColor_Specified, specifiedColor=<rgba 0 0 0 0>}
      cornerRadii=<0 0 0 0>
     }
     tintColor {colorUseRule=UseColor_Specified, specifiedColor=<rgba 1 1 1 1>}
     uVRegion {bIsValid=false, max=<0 0>, min=<0 0>}
    }
    padding=<ltrb 0 0 0 0>
   }
   slot {class=HorizontalBoxSlot, properties {size {sizeRule=Fill, value=0.68}}}
  }
  Spacer_92 {
   widgetClassPath=&Spacer
   parent=HorizontalBox-Backgrounds
   widget=Spacer_92
   slot {class=HorizontalBoxSlot, properties {size {sizeRule=Fill, value=0.5}}}
  }
  SZ_TitlePortrait {
   widgetClassPath=&SafeZone
   parent=Overlay_17
   widget=SZ_TitlePortrait
   properties {padBottom=false}
   slot {class=OverlaySlot, properties {horizontalAlignment=HAlign_Fill, padding=<ltrb 48 24 48 48>}}
  }
  HorizontalBox_TitlePortrait {widgetClassPath=&HorizontalBox, parent=SZ_TitlePortrait, widget=HorizontalBox_TitlePortrait, slot {class=SafeZoneSlot}}
  TextBlock_Title {
   widgetClassPath=&CommonTextBlock
   bIsVariable=true
   parent=HorizontalBox_TitlePortrait
   widget=TextBlock_Title
   properties {
    text=Spotlight Setlist
    textTransformPolicy=ToUpper
    font {
     bForceMonospaced=false
     bMaterialIsStencil=false
     fontMaterial=null
     fontObject=null
     letterSpacing=0
     monospacedWidth=1
     size=24
     skewAmount=0
     typefaceFontName=None
     outlineSettings {
      bApplyOutlineToDropShadows=false
      bMiteredCorners=false
      bSeparateFillAlpha=false
      outlineMaterial=null
      outlineSize=0
      outlineColor=<rgba 0 0 0 1>
     }
    }
    shadowOffset=<0 0>
   }
   slot {class=HorizontalBoxSlot, properties {horizontalAlignment=HAlign_Center, verticalAlignment=VAlign_Bottom, padding=<ltrb 0 0 0 32>}}
  }
  Spacer_116 {
   widgetClassPath=&Spacer
   parent=HorizontalBox_TitlePortrait
   widget=Spacer_116
   slot {class=HorizontalBoxSlot, properties {size {sizeRule=Fill, value=1}}}
  }
  HeaderBox_Right {
   widgetClassPath=&Overlay
   parent=HorizontalBox_TitlePortrait
   widget=HeaderBox_Right
   slot {class=HorizontalBoxSlot, properties {horizontalAlignment=HAlign_Right, verticalAlignment=VAlign_Top}}
  }
  VerticalBox_1 {
   widgetClassPath=&VerticalBox
   parent=HeaderBox_Right
   widget=VerticalBox_1
   slot {class=OverlaySlot, properties {horizontalAlignment=HAlign_Fill, verticalAlignment=VAlign_Fill}}
  }
  ScaleBox_59 {
   widgetClassPath=&ScaleBox
   parent=VerticalBox_1
   widget=ScaleBox_59
   properties {stretch=UserSpecified, userSpecifiedScale=0.6}
   slot {class=VerticalBoxSlot, properties {horizontalAlignment=HAlign_Right}}
  }
  PlayerLoadout {
   widgetClassPath=/PilgrimQuickplay/Widgets/Results/WBP_Pilgrim_Quickplay_PlayerLoadout
   bIsVariable=true
   parent=ScaleBox_59
   widget=PlayerLoadout
   properties {hideDifficulty=true, hideName=true}
   slot {class=ScaleBoxSlot, properties {verticalAlignment=VAlign_Bottom}}
  }
  SafeZone_Content {
   widgetClassPath=&SafeZone
   parent=Overlay_17
   widget=SafeZone_Content
   slot {class=OverlaySlot, properties {horizontalAlignment=HAlign_Fill, verticalAlignment=VAlign_Fill, padding=<ltrb 48 96 48 48>}}
  }
  HorizontalBox_0 {widgetClassPath=&HorizontalBox, parent=SafeZone_Content, widget=HorizontalBox_0, slot {class=SafeZoneSlot}}
  WBP_PilgrimQuickplay_SongCatalog2 {
   widgetClassPath=&/WBP_PilgrimQuickplay_SongCatalog2
   bIsVariable=true
   parent=HorizontalBox_0
   widget=WBP_PilgrimQuickplay_SongCatalog2
   properties {bIsModal=false}
   slot {class=HorizontalBoxSlot, properties {padding=<ltrb 0 128 48 0>, size {sizeRule=Fill, value=1}}}
  }
  LeaderboardsVerticalBox {
   widgetClassPath=&VerticalBox
   bIsVariable=true
   parent=HorizontalBox_0
   widget=LeaderboardsVerticalBox
   slot {class=HorizontalBoxSlot, properties {padding=<ltrb 48 86 0 0>, size {sizeRule=Fill, value=1}}}
  }
  DetailsPanel {
   widgetClassPath=&VerticalBox
   parent=LeaderboardsVerticalBox
   widget=DetailsPanel
   slot {class=VerticalBoxSlot, properties {size {sizeRule=Fill, value=1}}}
  }
  Leaderboard {
   widgetClassPath=/PilgrimQuickplay/Widgets/Leaderboards/WBP_PilgrimQuickplay_Leaderboard_Mini_MusicLibrary
   bIsVariable=true
   parent=DetailsPanel
   widget=Leaderboard
   properties {bIsFocusable=false}
   slot {class=VerticalBoxSlot, properties {padding=<ltrb 0 0 0 48>, size {sizeRule=Fill, value=1}}}
  }
  Overlay_Actions {
   widgetClassPath=&Overlay
   parent=DetailsPanel
   widget=Overlay_Actions
   properties {bOverride_Cursor=true}
   slot {class=VerticalBoxSlot}
  }
  WBP_CaptureForPostBufferUpdate {
   widgetClassPath=&WBP_CaptureForPostBufferUpdate
   bIsVariable=true
   parent=Overlay_Actions
   slot {class=OverlaySlot, properties {horizontalAlignment=HAlign_Fill, verticalAlignment=VAlign_Fill}}
  }
  HorizontalBox_SongActions {
   widgetClassPath=&HorizontalBox
   parent=Overlay_Actions
   widget=HorizontalBox_SongActions
   properties {visibility=Visible}
   slot {class=OverlaySlot, properties {horizontalAlignment=HAlign_Right}}
  }
  WBP_UIKit_Button_Quiet_Leaderboard {
   widgetClassPath=/FortUILibrary/UIKit/Buttons/WBP_UIKit_Button_Quiet
   bIsVariable=true
   parent=HorizontalBox_SongActions
   widget=WBP_UIKit_Button_Quiet_Leaderboard
   properties {
    bIsFocusable=false
    text=Leaderboard
    triggeringInputAction {dataTable=/PilgrimQuickplay/Widgets/InputActionDataTables/FM_PilgrimQuickplay_InputActionDataTable, rowName=OpenLeaderboard}
   }
   slot {class=HorizontalBoxSlot, properties {horizontalAlignment=HAlign_Right, padding=<ltrb 0 0 12 0>}}
  }
  WBP_UIKit_Button_Quiet_Back {
   widgetClassPath=&WBP_UIKit_Button_Quiet
   bIsVariable=true
   parent=HorizontalBox_SongActions
   widget=WBP_UIKit_Button_Quiet_Back
   properties {bIsFocusable=false, text=Back, triggeringInputAction {dataTable=&InputActionDataTable, rowName=GenericBack}}
   slot {class=HorizontalBoxSlot, properties {horizontalAlignment=HAlign_Right}}
  }
 ]
}
